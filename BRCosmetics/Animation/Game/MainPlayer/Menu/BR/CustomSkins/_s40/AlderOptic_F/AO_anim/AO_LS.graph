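# Level Sequence that authors a menu animation for the AlderOptic_F custom skin on a spawned FortPlayerMannequin.
# A Control Rig track on Fortnite_Biped_CtrlRig carries the hand-keyed or baked motion over about 440 frames, while the skeletal animation track is empty.
# FK/IK switches and foot-roll controls shape the limbs, with left-arm stretch and right-arm FK local space toggling on for a stretch late in the shot.
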LevelSequence /BRCosmetics/Animation/Game/MainPlayer/Menu/BR/CustomSkins/_s40/AlderOptic_F/AO_anim/AO_LS
bindings [
 f {
  id=#1
  possessed_class=FortPlayerMannequin
  tracks [
   {type=MovieSceneSpawnTrack, sections [{type=MovieSceneSpawnSection, channels [None {type=MovieSceneScriptingBoolChannel, default=true}]}]}
   {
    type=MovieScene3DTransformTrack
    sections [{
      type=MovieScene3DTransformSection
      channels [
       Weight {type=MovieSceneScriptingFloatChannel, default=1}
       Location.X {type=MovieSceneScriptingDoubleChannel, default=0}
       Location.Y {type=MovieSceneScriptingDoubleChannel, default=0}
       Location.Z {type=MovieSceneScriptingDoubleChannel, default=0}
       Rotation.X {type=MovieSceneScriptingDoubleChannel, default=0}
       Rotation.Y {type=MovieSceneScriptingDoubleChannel, default=0}
       Rotation.Z {type=MovieSceneScriptingDoubleChannel, default=0}
       Scale.X {type=MovieSceneScriptingDoubleChannel, default=1}
       Scale.Y {type=MovieSceneScriptingDoubleChannel, default=1}
       Scale.Z {type=MovieSceneScriptingDoubleChannel, default=1}
      ]
     }]
   }
   {type=MovieSceneSkeletalAnimationTrack, sections []}
   {
    type=MovieSceneControlRigParameterTrack
    display_name=Fortnite_Biped_CtrlRig
    sections [
     {
      type=MovieSceneControlRigParameterSection
      channels [
       leg_l_stretch_switch {type=MovieSceneScriptingBoolChannel, default=true, const [true,145]}
       leg_r_stretch_switch {type=MovieSceneScriptingBoolChannel, default=true, const [true,64]}
       arm_l_fk_local {type=MovieSceneScriptingBoolChannel, default=true, const [true,64]}
       Auto Interpolate_3 {type=MovieSceneScriptingBoolChannel, default=false, const [false,94]}
       Auto Interpolate {type=MovieSceneScriptingBoolChannel, default=false, const [false,94]}
       arm_l_stretch_switch {
        type=MovieSceneScriptingBoolChannel
        default=true
        keys [[0,false],[1,false],[2,false],[3,false],[4,false],[5,false],[6,false],[7,false],[8,false],[9,false],[10,false],[11,false],[12,false],[13,false],[14,false],[15,false],[16,false],[17,false],[18,false],[19,false],[20,false],[21,false],[22,false],[23,false],[24,false],[25,false],[26,false],[27,false],[28,false],[29,false],[30,false],[31,false],[32,false],[33,false],[34,false],[35,false],[36,false],[37,false],[38,false],[39,false],[40,false],[41,false],[42,false],[43,false],[44,false],[45,false],[46,false],[47,false],[48,false],[49,false],[50,false],[51,false],[52,false],[53,false],[54,false],[55,false],[56,false],[57,false],[58,false],[59,false],[60,false],[61,false],[62,false],[63,false],[64,false],[65,false],[66,false],[67,false],[68,false],[69,false],[70,false],[71,false],[72,false],[73,false],[74,false],[75,false],[76,false],[77,false],[78,false],[79,false],[80,false],[81,false],[82,false],[83,false],[84,false],[85,false],[86,false],[87,false],[88,false],[89,false],[90,false],[91,false],[92,false],[93,false],[94,false],[95,false],[96,false],[97,false],[98,false],[99,false],[100,false],[101,false],[102,false],[103,false],[104,false],[105,false],[106,false],[107,false],[108,false],[109,false],[110,false],[111,false],[112,false],[113,false],[114,false],[115,false],[116,false],[117,false],[118,false],[119,false],[120,false],[121,false],[122,false],[123,false],[124,false],[125,false],[126,false],[127,false],[128,false],[129,false],[130,false],[131,false],[132,false],[133,false],[134,false],[135,false],[136,false],[137,false],[138,false],[139,false],[140,false],[141,false],[142,false],[143,false],[144,false],[145,false],[146,false],[147,false],[148,false],[149,false],[150,false],[151,false],[152,false],[153,false],[154,false],[155,false],[156,false],[157,false],[158,false],[159,false],[160,false],[161,false],[162,false],[163,false],[164,false],[165,false],[166,false],[167,false],[168,false],[169,false],[170,false],[171,false],[172,false],[173,false],[174,false],[175,false],[176,false],[177,false],[178,false],[179,false],[180,false],[181,false],[182,false],[183,false],[184,false],[185,false],[186,false],[187,false],[188,false],[189,false],[190,false],[191,false],[192,false],[193,false],[194,false],[195,false],[196,false],[197,false],[198,false],[199,false],[200,false],[201,false],[202,false],[203,false],[204,false],[205,false],[206,false],[207,false],[208,false],[209,false],[210,false],[211,false],[212,false],[213,false],[214,false],[215,false],[216,false],[217,false],[218,false],[219,false],[220,false],[221,false],[222,false],[223,false],[224,false],[225,false],[226,false],[227,false],[228,false],[229,false],[230,false],[231,false],[232,false],[233,false],[234,false],[235,false],[236,false],[237,false],[238,false],[239,false],[240,false],[241,false],[242,false],[243,false],[244,false],[245,false],[246,false],[247,false],[248,false],[249,false],[250,false],[251,false],[252,false],[253,false],[254,false],[255,false],[256,false],[257,false],[258,false],[259,false],[260,false],[261,false],[262,false],[263,false],[264,false],[265,false],[266,false],[267,false],[268,true],[269,true],[270,true],[271,true],[272,true],[273,true],[274,true],[275,true],[276,true],[277,true],[278,true],[279,true],[280,true],[281,true],[282,true],[283,true],[284,true],[285,true],[286,true],[287,true],[288,true],[289,true],[290,true],[291,true],[292,true],[293,true],[294,true],[295,true],[296,true],[297,true],[298,true],[299,true],[300,true],[301,true],[302,true],[303,true],[304,true],[305,true],[306,true],[307,true],[308,true],[309,true],[310,true],[311,true],[312,true],[313,true],[314,true],[315,true],[316,true],[317,true],[318,true],[319,true],[320,true],[321,true],[322,true],[323,true],[324,true],[325,true],[326,true],[327,true],[328,true],[329,true],[330,true],[331,true],[332,true],[333,true],[334,true],[335,true],[336,true],[338,true],[340,true],[341,true],[342,true],[343,true],[344,true],[345,true],[346,true],[347,true],[348,true],[349,true],[350,true],[351,true],[352,true],[353,true],[354,true],[355,true],[356,true],[357,true],[358,true],[359,false],[360,false],[361,false],[362,false],[363,false],[364,false],[365,false],[366,false],[367,false],[368,false],[369,false],[370,false],[371,false],[372,false],[373,false],[374,false],[375,false],[376,false],[377,false],[378,false],[379,false],[380,false],[381,false],[382,false],[383,false],[384,false],[385,false],[386,false],[387,false],[388,false],[389,false],[390,false],[391,false],[392,false],[393,false],[394,false],[395,false],[396,false],[397,false],[398,false],[399,false],[400,false],[401,false],[402,false],[403,false],[404,false],[405,false],[406,false],[407,false],[408,false],[409,false],[410,false],[411,false],[412,false],[413,false],[414,false],[415,false],[416,false],[417,false],[418,false],[419,false],[420,false],[421,false],[422,false],[423,false],[424,false],[425,false],[426,false],[427,false],[428,false],[429,false],[430,false],[431,false],[432,false],[433,false],[434,false],[435,false],[436,false],[437,false],[438,false],[439,false],[440,false]]
       }
       Keep Aligned {type=MovieSceneScriptingBoolChannel, default=false, const [false,439]}
       arm_r_fk_local {
        type=MovieSceneScriptingBoolChannel
        default=true
        keys [[0,false],[2,false],[10,false],[20,false],[30,false],[40,false],[50,false],[60,false],[70,false],[80,false],[90,false],[100,false],[110,false],[120,false],[130,false],[136,false],[140,false],[150,false],[160,false],[170,false],[180,false],[190,false],[200,false],[210,false],[220,false],[230,false],[240,false],[250,false],[255,false],[256,false],[257,false],[258,false],[259,false],[260,false],[261,false],[262,false],[263,false],[264,false],[265,false],[266,false],[267,false],[268,false],[269,false],[270,true],[271,true],[272,true],[273,true],[274,true],[275,true],[276,true],[277,true],[278,true],[279,true],[280,true],[281,true],[282,true],[283,true],[284,true],[285,true],[286,true],[287,true],[288,true],[289,true],[290,true],[291,true],[292,true],[299,true],[309,true],[327,true],[328,true],[329,true],[330,true],[331,true],[332,true],[333,true],[334,true],[335,true],[336,true],[337,true],[338,true],[339,true],[340,true],[341,true],[342,true],[343,true],[344,true],[345,true],[346,true],[347,true],[348,true],[349,true],[350,true],[351,true],[352,true],[353,true],[354,true],[355,true],[356,true],[357,true],[358,true],[359,true],[360,true],[361,true],[362,true],[363,false],[364,false],[365,false],[366,false],[367,false],[368,false],[369,false],[370,false],[371,false],[372,false],[373,false],[374,false],[375,false],[376,false],[377,false],[378,false],[381,false],[384,false],[387,false],[391,false],[396,false],[399,false],[402,false],[405,false],[408,false],[411,false],[414,false],[416,false],[418,false],[420,false],[423,false],[426,false],[429,false],[432,false],[435,false],[437,false],[439,false],[440,false]]
       }
       Auto Interpolate_4 {type=MovieSceneScriptingBoolChannel, default=false, const [false,62]}
       Auto Interpolate_2 {type=MovieSceneScriptingBoolChannel, default=false, const [false,62]}
       arm_r_stretch_switch {type=MovieSceneScriptingBoolChannel, default=false, const [false,439]}
       Keep Aligned_2 {type=MovieSceneScriptingBoolChannel, default=false, const [false,439]}
       head_local {type=MovieSceneScriptingBoolChannel, default=true, const [true,64]}
       Stretch {type=MovieSceneScriptingBoolChannel, default=false, const [false,64]}
       head_stretch {type=MovieSceneScriptingBoolChannel, default=false, const [false,64]}
       spine_stretch_switch {type=MovieSceneScriptingBoolChannel, default=false, const [false,63]}
       arm_l_fk_ik_switch {type=MovieSceneScriptingBoolChannel, default=true, const [true,65]}
       leg_l_fk_ik_switch {type=MovieSceneScriptingBoolChannel, default=true, const [true,64]}
       arm_r_fk_ik_switch {type=MovieSceneScriptingBoolChannel, default=true, const [false,29]}
       leg_r_fk_ik_switch {type=MovieSceneScriptingBoolChannel, default=true, const [true,64]}
       ShowBodyControls {type=MovieSceneScriptingBoolChannel, default=true, const [true,64]}
       spine_fk_ik_switch {type=MovieSceneScriptingBoolChannel, default=false, const [false,65]}
       neck_fk_ik_switch {type=MovieSceneScriptingBoolChannel, default=false, const [false,64]}
       space {type=MovieSceneScriptingByteChannel, default=2, const [2,64]}
       global_ctrl.Location.X {type=MovieSceneScriptingFloatChannel, default=0, thinned_from=64, keys [[0,0],[439,0]]}
       global_ctrl.Location.Y {type=MovieSceneScriptingFloatChannel, default=0, thinned_from=64, keys [[0,0],[439,0]]}
       global_ctrl.Location.Z {type=MovieSceneScriptingFloatChannel, default=0, thinned_from=64, keys [[0,0],[439,0]]}
       global_ctrl.Rotation.X {type=MovieSceneScriptingFloatChannel, default=0, thinned_from=64, keys [[0,0],[439,0]]}
       global_ctrl.Rotation.Y {type=MovieSceneScriptingFloatChannel, default=0, thinned_from=64, keys [[0,0],[439,0]]}
       global_ctrl.Rotation.Z {type=MovieSceneScriptingFloatChannel, default=0, thinned_from=64, keys [[0,0],[439,0]]}
       root_ctrl.Location.X {type=MovieSceneScriptingFloatChannel, default=0, thinned_from=64, keys [[0,0],[439,0]]}
       root_ctrl.Location.Y {type=MovieSceneScriptingFloatChannel, default=0, thinned_from=64, keys [[0,0],[439,0]]}
       root_ctrl.Location.Z {type=MovieSceneScriptingFloatChannel, default=0, thinned_from=64, keys [[0,0],[439,0]]}
       root_ctrl.Rotation.X {type=MovieSceneScriptingFloatChannel, default=0, thinned_from=64, keys [[0,0],[439,0]]}
       root_ctrl.Rotation.Y {type=MovieSceneScriptingFloatChannel, default=0, thinned_from=64, keys [[0,0],[439,0]]}
       root_ctrl.Rotation.Z {type=MovieSceneScriptingFloatChannel, default=0, thinned_from=64, keys [[0,0],[439,0]]}
       root_ctrl.Scale.X {type=MovieSceneScriptingFloatChannel, default=1, thinned_from=64, keys [[0,1],[439,1]]}
       root_ctrl.Scale.Y {type=MovieSceneScriptingFloatChannel, default=1, thinned_from=64, keys [[0,1],[439,1]]}
       root_ctrl.Scale.Z {type=MovieSceneScriptingFloatChannel, default=1, thinned_from=64, keys [[0,1],[439,1]]}
       body_offset_ctrl.Location.X {type=MovieSceneScriptingFloatChannel, default=0, thinned_from=64, keys [[0,0],[439,0]]}
       body_offset_ctrl.Location.Y {type=MovieSceneScriptingFloatChannel, default=0, thinned_from=64, keys [[0,0],[439,0]]}
       body_offset_ctrl.Location.Z {type=MovieSceneScriptingFloatChannel, default=0, thinned_from=64, keys [[0,0],[439,0]]}
       body_offset_ctrl.Rotation.X {type=MovieSceneScriptingFloatChannel, default=0, thinned_from=64, keys [[0,0],[439,0]]}
       body_offset_ctrl.Rotation.Y {type=MovieSceneScriptingFloatChannel, default=0, thinned_from=64, keys [[0,0],[439,0]]}
       body_offset_ctrl.Rotation.Z {type=MovieSceneScriptingFloatChannel, default=0, thinned_from=64, keys [[0,0],[439,0]]}
       body_offset_ctrl.Scale.X {type=MovieSceneScriptingFloatChannel, default=1, thinned_from=64, keys [[0,1],[439,1]]}
       body_offset_ctrl.Scale.Y {type=MovieSceneScriptingFloatChannel, default=1, thinned_from=64, keys [[0,1],[439,1]]}
       body_offset_ctrl.Scale.Z {type=MovieSceneScriptingFloatChannel, default=1, thinned_from=64, keys [[0,1],[439,1]]}
       thigh_l_fk_ctrl.Location.X {type=MovieSceneScriptingFloatChannel, default=0, thinned_from=64, keys [[0,0],[439,0]]}
       thigh_l_fk_ctrl.Location.Y {type=MovieSceneScriptingFloatChannel, default=0, thinned_from=64, keys [[0,0],[439,0]]}
       thigh_l_fk_ctrl.Location.Z {type=MovieSceneScriptingFloatChannel, default=0, thinned_from=64, keys [[0,0],[439,0]]}
       thigh_l_fk_ctrl.Rotation.X {
        type=MovieSceneScriptingFloatChannel
        default=42.9692
        thinned_from=64
        keys [[0,59.9342],[15,59.9342],[22,60.2529],[45,60.3505],[51,59.9],[56,59.7258],[62,59.1019],[66,55.7313],[70,54.1317],[83,53.2779],[94,53.1226],[101,53.1598],[107,53.4205],[111,54.2181],[115,55.2352],[121,55.535],[130,55.635],[138,55.8344],[143,56.4328],[148,58.8263],[153,42.4378],[160,36.2922],[172,55.4883],[187,61.887],[194,60.0569],[199,54.5665],[204,32.6048],[215,32.6048],[223,36.074],[233,46.4817],[241,49.9626],[246,51.123],[251,50.2691],[257,47.7075],[267,37.461],[275,49.6693],[279,46.1887],[283,44.6224],[287,44.2308],[292,44.1329],[337,44.1003],[345,44.0067],[351,43.7261],[356,42.6035],[366,42.7416],[371,48.9039],[375,54.8536],[378,58.029],[382,59.2198],[387,59.7556],[409,60.0875],[439,59.9342]]
       }
       thigh_l_fk_ctrl.Rotation.Y {
        type=MovieSceneScriptingFloatChannel
        default=8.0284
        thinned_from=64
        keys [[0,12.1109],[15,12.1109],[22,12.1694],[45,12.1872],[51,12.1],[56,12.0034],[62,11.6797],[66,12.4289],[70,12.2764],[75,12.4931],[83,12.6544],[94,12.7069],[101,12.7184],[107,12.7993],[111,13.0468],[115,11.9604],[121,12.1442],[130,12.2054],[138,12.1507],[143,11.9866],[148,11.33],[153,14.4516],[160,15.6222],[172,12.9571],[187,12.0688],[194,12.3348],[199,13.1327],[204,16.3246],[215,16.3246],[223,14.4249],[233,8.726],[241,6.9954],[246,6.4185],[251,6.5955],[257,7.1263],[267,9.2494],[275,9.6179],[279,7.9835],[283,7.2481],[287,7.0642],[292,7.0182],[337,7.0029],[345,7.1082],[351,7.4243],[356,8.6884],[361,10.2061],[366,10.7119],[371,11.0289],[375,11.335],[378,11.82],[382,12.0018],[387,12.0836],[409,12.1483],[439,12.1109]]
       }
       thigh_l_fk_ctrl.Rotation.Z {
        type=MovieSceneScriptingFloatChannel
        default=5.9679
        thinned_from=64
        keys [[0,7.3236],[15,7.3236],[22,7.4096],[30,7.4078],[45,7.4358],[51,7.3],[56,7.2635],[62,7.0828],[66,5.9964],[70,5.5224],[75,5.4121],[83,5.2648],[94,5.2168],[101,5.2256],[107,5.2871],[111,5.4754],[115,5.7505],[121,5.8323],[130,5.8596],[138,5.9162],[143,6.0861],[148,6.7656],[153,2.9766],[160,1.5557],[172,6.0694],[187,7.5739],[194,7.1445],[199,5.8562],[204,0.7032],[215,0.7032],[223,1.3759],[233,3.394],[241,4.0702],[246,4.2956],[257,4.3104],[267,4.3551],[275,5.3543],[279,5.7035],[283,5.8606],[287,5.8999],[345,5.9175],[351,5.9312],[356,5.9857],[361,5.4397],[366,5.2577],[371,6.102],[375,6.9171],[378,7.1712],[382,7.2664],[387,7.3093],[395,7.3183],[409,7.3922],[439,7.3236]]
       }
       thigh_l_fk_ctrl.Scale.X {type=MovieSceneScriptingFloatChannel, default=1, thinned_from=64, keys [[0,1],[439,1]]}
       thigh_l_fk_ctrl.Scale.Y {type=MovieSceneScriptingFloatChannel, default=1, thinned_from=64, keys [[0,1],[439,1]]}
       thigh_l_fk_ctrl.Scale.Z {type=MovieSceneScriptingFloatChannel, default=1, thinned_from=64, keys [[0,1],[439,1]]}
       calf_l_fk_ctrl.Location.X {
        type=MovieSceneScriptingFloatChannel
        default=-0.4533
        thinned_from=64
        keys [[0,-1.1428],[15,-1.1428],[22,-1.132],[30,-1.1322],[45,-1.1287],[51,-1.1],[56,-1.1456],[62,-1.1543],[66,-1.0796],[70,-1.0794],[75,-1.0687],[83,-1.0591],[101,-1.0549],[107,-1.047],[111,-1.023],[115,-1.0336],[121,-1.0171],[130,-1.0116],[143,-1.0123],[148,-1.0144],[153,-0.3804],[160,-0.1426],[172,-0.7377],[187,-0.9361],[194,-0.8775],[199,-0.702],[204,0],[215,0],[223,-0.1874],[233,-0.7495],[241,-0.9282],[246,-0.9878],[251,-0.9811],[257,-0.961],[267,-0.8806],[271,-0.5993],[275,-0.3214],[279,-0.3862],[283,-0.4153],[287,-0.4226],[292,-0.4244],[337,-0.425],[345,-0.4279],[351,-0.4367],[356,-0.4715],[361,-0.5607],[366,-0.5905],[371,-0.6376],[375,-0.6831],[378,-0.9704],[382,-1.0782],[387,-1.1267],[395,-1.1368],[409,-1.1346],[439,-1.1428]]
       }
       calf_l_fk_ctrl.Location.Y {
        type=MovieSceneScriptingFloatChannel
        default=0.0594
        thinned_from=64
        keys [[0,0.1497],[15,0.1497],[22,0.1483],[30,0.1483],[45,0.1479],[51,0.1],[56,0.1501],[62,0.1512],[66,0.1414],[70,0.1414],[75,0.14],[83,0.1387],[101,0.1382],[107,0.1372],[111,0.134],[115,0.1354],[121,0.1332],[130,0.1325],[143,0.1326],[148,0.1329],[153,0.0498],[160,0.0187],[172,0.0966],[187,0.1226],[194,0.115],[199,0.092],[204,0],[215,0],[223,0.0245],[233,0.0982],[241,0.1216],[246,0.1294],[251,0.1285],[257,0.1259],[267,0.1154],[271,0.0785],[275,0.0421],[279,0.0506],[283,0.0544],[287,0.0554],[292,0.0556],[337,0.0557],[345,0.0561],[351,0.0572],[356,0.0618],[361,0.0735],[366,0.0774],[371,0.0835],[375,0.0895],[378,0.1271],[382,0.1412],[387,0.1476],[395,0.1489],[409,0.1486],[439,0.1497]]
       }
       calf_l_fk_ctrl.Location.Z {
        type=MovieSceneScriptingFloatChannel
        default=-0.0084
        thinned_from=64
        keys [[0,-0.0212],[15,-0.0212],[22,-0.021],[45,-0.021],[51,0],[56,-0.0213],[62,-0.0214],[66,-0.0201],[83,-0.0197],[107,-0.0194],[111,-0.019],[115,-0.0192],[121,-0.0189],[148,-0.0188],[153,-0.0071],[160,-0.0026],[172,-0.0137],[187,-0.0174],[194,-0.0163],[199,-0.013],[204,0],[215,0],[223,-0.0035],[233,-0.0139],[241,-0.0172],[246,-0.0183],[251,-0.0182],[257,-0.0178],[267,-0.0164],[275,-0.006],[279,-0.0072],[283,-0.0077],[345,-0.0079],[351,-0.0081],[356,-0.0088],[361,-0.0104],[366,-0.011],[375,-0.0127],[378,-0.018],[382,-0.02],[387,-0.0209],[395,-0.0211],[439,-0.0212]]
       }
       calf_l_fk_ctrl.Rotation.X {
        type=MovieSceneScriptingFloatChannel
        default=-0.0692
        thinned_from=64
        keys [[0,-0.0692],[45,-0.0692],[51,-0.1],[56,-0.0692],[148,-0.0692],[153,0.012],[160,0.0425],[172,-0.0413],[187,-0.0692],[194,-0.0611],[199,-0.0367],[204,0.0607],[215,0.0607],[223,0.0364],[233,-0.0367],[241,-0.0611],[246,-0.0692],[439,-0.0692]]
       }
       calf_l_fk_ctrl.Rotation.Y {type=MovieSceneScriptingFloatChannel, default=1.0551, thinned_from=64, keys [[0,1.0551],[45,1.0551],[51,1.1],[56,1.0551],[439,1.0551]]}
       calf_l_fk_ctrl.Rotation.Z {
        type=MovieSceneScriptingFloatChannel
        default=-7.4634
        thinned_from=64
        keys [[0,-7.4634],[45,-7.4634],[51,-7.5],[56,-7.4634],[148,-7.4634],[153,-3.0605],[160,-1.4094],[172,-5.9499],[187,-7.4634],[194,-7.0231],[199,-5.7022],[204,-0.4187],[215,-0.4187],[223,-1.7396],[233,-5.7022],[241,-7.0231],[246,-7.4634],[439,-7.4634]]
       }
       calf_l_fk_ctrl.Scale.X {type=MovieSceneScriptingFloatChannel, default=1, thinned_from=64, keys [[0,1],[439,1]]}
       calf_l_fk_ctrl.Scale.Y {type=MovieSceneScriptingFloatChannel, default=1, thinned_from=64, keys [[0,1],[439,1]]}
       calf_l_fk_ctrl.Scale.Z {type=MovieSceneScriptingFloatChannel, default=1, thinned_from=64, keys [[0,1],[439,1]]}
       foot_l_fk_ctrl.Location.X {
        type=MovieSceneScriptingFloatChannel
        default=-0.5031
        thinned_from=64
        keys [[0,-1.2685],[15,-1.2685],[22,-1.2565],[30,-1.2567],[45,-1.2528],[51,-1.3],[56,-1.2716],[62,-1.2812],[66,-1.1983],[70,-1.198],[75,-1.1862],[83,-1.1756],[101,-1.1708],[107,-1.1621],[111,-1.1355],[115,-1.1472],[121,-1.129],[130,-1.1229],[143,-1.1236],[148,-1.1259],[153,-0.4222],[160,-0.1583],[172,-0.8188],[187,-1.039],[194,-0.974],[199,-0.7792],[204,0],[215,0],[223,-0.208],[233,-0.8319],[241,-1.0303],[246,-1.0964],[251,-1.0889],[257,-1.0666],[267,-0.9774],[271,-0.6652],[275,-0.3567],[279,-0.4286],[283,-0.461],[287,-0.4691],[292,-0.4711],[337,-0.4718],[345,-0.475],[351,-0.4847],[356,-0.5233],[361,-0.6224],[366,-0.6554],[371,-0.7077],[375,-0.7582],[378,-1.0771],[382,-1.1967],[387,-1.2505],[395,-1.2617],[409,-1.2593],[439,-1.2685]]
       }
       foot_l_fk_ctrl.Location.Y {type=MovieSceneScriptingFloatChannel, default=0.0627, thinned_from=64, keys [[0,0.0627],[45,0.0627],[51,0.1],[56,0.0627],[439,0.0627]]}
       foot_l_fk_ctrl.Location.Z {type=MovieSceneScriptingFloatChannel, default=-0.8453, thinned_from=64, keys [[0,-0.8453],[45,-0.8453],[51,-0.8],[56,-0.8453],[439,-0.8453]]}
       foot_l_fk_ctrl.Rotation.X {
        type=MovieSceneScriptingFloatChannel
        default=-2.1582
        thinned_from=64
        keys [[0,-3.9014],[45,-3.8945],[62,-3.8921],[66,-3.8184],[70,-3.7657],[111,-3.7533],[115,-3.6968],[138,-3.6981],[143,-3.6894],[148,-3.6546],[153,-2.9278],[160,-2.6552],[172,-3.4023],[187,-3.6513],[194,-3.5788],[199,-3.3614],[204,-2.4917],[215,-2.4917],[223,-2.6013],[233,-2.93],[241,-3.0452],[246,-3.0835],[251,-3.1067],[257,-3.1761],[267,-3.4539],[271,-2.658],[275,-1.8713],[279,-1.7155],[283,-1.6454],[287,-1.6278],[292,-1.6234],[337,-1.622],[345,-1.6764],[351,-1.8398],[356,-2.4933],[361,-3.1564],[366,-3.3774],[371,-3.4589],[375,-3.5376],[378,-3.765],[382,-3.8502],[387,-3.8886],[409,-3.9034],[439,-3.9014]]
       }
       foot_l_fk_ctrl.Rotation.Y {
        type=MovieSceneScriptingFloatChannel
        default=0.155
        thinned_from=64
        keys [[0,-7.5233],[45,-7.5263],[51,-7.5],[56,-7.5114],[62,-7.4755],[66,-7.1355],[70,-6.9531],[83,-6.8921],[101,-6.8823],[107,-6.8885],[111,-6.9074],[115,-6.8541],[121,-6.871],[138,-6.881],[143,-6.8944],[148,-6.948],[153,-4.9494],[160,-4.1999],[172,-6.3443],[187,-7.0592],[194,-6.8523],[199,-6.2319],[204,-3.7502],[215,-3.7502],[223,-4.0419],[233,-4.917],[241,-5.2206],[246,-5.3218],[251,-5.2984],[257,-5.2284],[267,-4.9482],[271,-1.9728],[275,0.9681],[279,1.4834],[283,1.7153],[287,1.7733],[292,1.7878],[309,1.7922],[337,1.7926],[345,1.6244],[351,1.1198],[356,-0.8987],[361,-3.842],[366,-4.8231],[371,-5.5977],[375,-6.3457],[378,-7.0817],[382,-7.3577],[387,-7.4819],[409,-7.5398],[439,-7.5233]]
       }
       foot_l_fk_ctrl.Rotation.Z {
        type=MovieSceneScriptingFloatChannel
        default=19.4615
        thinned_from=64
        keys [[0,15.3756],[15,15.3756],[45,15.3409],[51,15.4],[62,15.3911],[66,15.4685],[70,15.4476],[83,15.4937],[107,15.4872],[111,15.4465],[115,15.2426],[138,15.2152],[143,15.1568],[148,14.9231],[153,11.4116],[160,10.0947],[172,13.6027],[187,14.772],[194,14.4303],[199,13.4051],[204,9.3046],[215,9.3046],[223,10.1943],[233,12.8635],[241,13.7682],[246,14.0697],[251,14.2362],[257,14.7358],[267,16.7339],[275,19.3338],[279,19.8084],[283,20.022],[287,20.0754],[337,20.0932],[345,20.0283],[351,19.8336],[356,19.055],[361,17.1418],[366,16.504],[371,15.7969],[375,15.1141],[378,15.2776],[382,15.3388],[395,15.3722],[439,15.3756]]
       }
       foot_l_fk_ctrl.Scale.X {type=MovieSceneScriptingFloatChannel, default=1, thinned_from=64, keys [[0,1],[439,1]]}
       foot_l_fk_ctrl.Scale.Y {type=MovieSceneScriptingFloatChannel, default=1, thinned_from=64, keys [[0,1],[439,1]]}
       foot_l_fk_ctrl.Scale.Z {type=MovieSceneScriptingFloatChannel, default=1, thinned_from=64, keys [[0,1],[439,1]]}
       ball_l_fk_ctrl.Location.X {type=MovieSceneScriptingFloatChannel, default=0, thinned_from=64, keys [[0,0],[439,0]]}
       ball_l_fk_ctrl.Location.Y {type=MovieSceneScriptingFloatChannel, default=0, thinned_from=64, keys [[0,0],[439,0]]}
       ball_l_fk_ctrl.Location.Z {type=MovieSceneScriptingFloatChannel, default=0, thinned_from=64, keys [[0,0],[439,0]]}
       ball_l_fk_ctrl.Rotation.X {type=MovieSceneScriptingFloatChannel, default=0, thinned_from=64, keys [[0,0],[439,0]]}
       ball_l_fk_ctrl.Rotation.Y {type=MovieSceneScriptingFloatChannel, default=0, thinned_from=64, keys [[0,0],[439,0]]}
       ball_l_fk_ctrl.Rotation.Z {type=MovieSceneScriptingFloatChannel, default=0, thinned_from=64, keys [[0,0],[439,0]]}
       ball_l_fk_ctrl.Scale.X {type=MovieSceneScriptingFloatChannel, default=1, thinned_from=64, keys [[0,1],[439,1]]}
       ball_l_fk_ctrl.Scale.Y {type=MovieSceneScriptingFloatChannel, default=1, thinned_from=64, keys [[0,1],[439,1]]}
       ball_l_fk_ctrl.Scale.Z {type=MovieSceneScriptingFloatChannel, default=1, thinned_from=64, keys [[0,1],[439,1]]}
       foot_l_ik_ctrl.Location.X {
        type=MovieSceneScriptingFloatChannel
        default=4.724
        thinned_from=140
        keys [[0,4.9],[277,4.9],[279,4.75],[285,1.75],[346,1.75],[347,2],[350,3.5],[351,3.875],[352,4.175],[353,4.425],[354,4.625],[355,4.775],[356,4.875],[357,4.9],[439,4.9]]
       }
       foot_l_ik_ctrl.Location.Y {type=MovieSceneScriptingFloatChannel, default=11.784, thinned_from=145, keys [[0,11.784],[439,11.784]]}
       foot_l_ik_ctrl.Location.Z {type=MovieSceneScriptingFloatChannel, default=0, thinned_from=145, keys [[0,0],[439,0]]}
       foot_l_ik_ctrl.Rotation.X {type=MovieSceneScriptingFloatChannel, default=0, const [0,145]}
       foot_l_ik_ctrl.Rotation.Y {type=MovieSceneScriptingFloatChannel, default=0, const [0,145]}
       foot_l_ik_ctrl.Rotation.Z {
        type=MovieSceneScriptingFloatChannel
        default=-43.75
        thinned_from=134
        keys [[0,-38.3023],[265,-38.3023],[276,-38.25],[279,-41.25],[285,-60],[288,-60.5],[345,-60.5],[346,-60],[347,-57.5],[350,-46.25],[351,-43.75],[353,-40],[356,-38.3023],[439,-38.3023]]
       }
       foot_l_ik_ctrl.Scale.X {type=MovieSceneScriptingFloatChannel, default=1, thinned_from=145, keys [[0,1],[439,1]]}
       foot_l_ik_ctrl.Scale.Y {type=MovieSceneScriptingFloatChannel, default=1, thinned_from=145, keys [[0,1],[439,1]]}
       foot_l_ik_ctrl.Scale.Z {type=MovieSceneScriptingFloatChannel, default=1, thinned_from=145, keys [[0,1],[439,1]]}
       ball_l_ik_ctrl.Location.X {type=MovieSceneScriptingFloatChannel, default=0, thinned_from=145, keys [[0,0],[439,0]]}
       ball_l_ik_ctrl.Location.Y {type=MovieSceneScriptingFloatChannel, default=0, thinned_from=145, keys [[0,0],[439,0]]}
       ball_l_ik_ctrl.Location.Z {type=MovieSceneScriptingFloatChannel, default=0, thinned_from=145, keys [[0,0],[439,0]]}
       ball_l_ik_ctrl.Rotation.X {type=MovieSceneScriptingFloatChannel, default=0, thinned_from=145, keys [[0,0],[439,0]]}
       ball_l_ik_ctrl.Rotation.Y {type=MovieSceneScriptingFloatChannel, default=0, thinned_from=145, keys [[0,0],[439,0]]}
       ball_l_ik_ctrl.Rotation.Z {type=MovieSceneScriptingFloatChannel, default=0, thinned_from=145, keys [[0,0],[439,0]]}
       ball_l_ik_ctrl.Scale.X {type=MovieSceneScriptingFloatChannel, default=1, thinned_from=145, keys [[0,1],[439,1]]}
       ball_l_ik_ctrl.Scale.Y {type=MovieSceneScriptingFloatChannel, default=1, thinned_from=145, keys [[0,1],[439,1]]}
       ball_l_ik_ctrl.Scale.Z {type=MovieSceneScriptingFloatChannel, default=1, thinned_from=145, keys [[0,1],[439,1]]}
       foot_bk1_l_ctrl.Location.X {type=MovieSceneScriptingFloatChannel, default=0, thinned_from=145, keys [[0,0],[439,0]]}
       foot_bk1_l_ctrl.Location.Y {type=MovieSceneScriptingFloatChannel, default=0, thinned_from=145, keys [[0,0],[439,0]]}
       foot_bk1_l_ctrl.Location.Z {type=MovieSceneScriptingFloatChannel, default=0, thinned_from=145, keys [[0,0],[439,0]]}
       foot_bk1_l_ctrl.Rotation.X {type=MovieSceneScriptingFloatChannel, default=0, thinned_from=145, keys [[0,0],[439,0]]}
       foot_bk1_l_ctrl.Rotation.Y {type=MovieSceneScriptingFloatChannel, default=0, thinned_from=145, keys [[0,0],[439,0]]}
       foot_bk1_l_ctrl.Rotation.Z {
        type=MovieSceneScriptingFloatChannel
        default=-4
        thinned_from=137
        keys [[0,0],[260,0],[261,-0.125],[262,-0.375],[263,-0.75],[265,-1.75],[270,-3.25],[279,-3.875],[282,-3.875],[289,0],[342,0],[343,-0.2001],[344,-0.7003],[345,-1.3506],[346,-2.2101],[347,-2.9469],[348,-3.438],[350,-3.9292],[351,-4],[352,-3.5],[353,-2.5],[354,-1.4064],[355,-0.5],[356,-0.125],[357,0],[439,0]]
       }
       foot_bk1_l_ctrl.Scale.X {type=MovieSceneScriptingFloatChannel, default=1, thinned_from=145, keys [[0,1],[439,1]]}
       foot_bk1_l_ctrl.Scale.Y {type=MovieSceneScriptingFloatChannel, default=1, thinned_from=145, keys [[0,1],[439,1]]}
       foot_bk1_l_ctrl.Scale.Z {type=MovieSceneScriptingFloatChannel, default=1, thinned_from=145, keys [[0,1],[439,1]]}
       foot_roll_l_ctrl.Location.X {type=MovieSceneScriptingFloatChannel, default=0, thinned_from=145, keys [[0,0],[439,0]]}
       foot_roll_l_ctrl.Location.Y {type=MovieSceneScriptingFloatChannel, default=0, thinned_from=145, keys [[0,0],[439,0]]}
       foot_roll_l_ctrl.Location.Z {type=MovieSceneScriptingFloatChannel, default=0, thinned_from=145, keys [[0,0],[439,0]]}
       foot_roll_l_ctrl.Rotation.X {type=MovieSceneScriptingFloatChannel, default=0, thinned_from=145, keys [[0,0],[439,0]]}
       foot_roll_l_ctrl.Rotation.Y {type=MovieSceneScriptingFloatChannel, default=0, thinned_from=145, keys [[0,0],[439,0]]}
       foot_roll_l_ctrl.Rotation.Z {type=MovieSceneScriptingFloatChannel, default=0, thinned_from=145, keys [[0,0],[439,0]]}
       foot_roll_l_ctrl.Scale.X {type=MovieSceneScriptingFloatChannel, default=1, thinned_from=145, keys [[0,1],[439,1]]}
       foot_roll_l_ctrl.Scale.Y {type=MovieSceneScriptingFloatChannel, default=1, thinned_from=145, keys [[0,1],[439,1]]}
       foot_roll_l_ctrl.Scale.Z {type=MovieSceneScriptingFloatChannel, default=1, thinned_from=145, keys [[0,1],[439,1]]}
       roll_blend_l_ctrl {type=MovieSceneScriptingFloatChannel, default=1, thinned_from=145, keys [[0,1],[439,1]]}
       tip_l_ctrl.Location.X {type=MovieSceneScriptingFloatChannel, default=0, thinned_from=145, keys [[0,0],[439,0]]}
       tip_l_ctrl.Location.Y {type=MovieSceneScriptingFloatChannel, default=0, thinned_from=145, keys [[0,0],[439,0]]}
       tip_l_ctrl.Location.Z {type=MovieSceneScriptingFloatChannel, default=0, thinned_from=145, keys [[0,0],[439,0]]}
       tip_l_ctrl.Rotation.X {type=MovieSceneScriptingFloatChannel, default=0, thinned_from=145, keys [[0,0],[439,0]]}
       tip_l_ctrl.Rotation.Y {type=MovieSceneScriptingFloatChannel, default=0, thinned_from=145, keys [[0,0],[439,0]]}
       tip_l_ctrl.Rotation.Z {type=MovieSceneScriptingFloatChannel, default=0, thinned_from=145, keys [[0,0],[439,0]]}
       tip_l_ctrl.Scale.X {type=MovieSceneScriptingFloatChannel, default=1, thinned_from=145, keys [[0,1],[439,1]]}
       tip_l_ctrl.Scale.Y {type=MovieSceneScriptingFloatChannel, default=1, thinned_from=145, keys [[0,1],[439,1]]}
       tip_l_ctrl.Scale.Z {type=MovieSceneScriptingFloatChannel, default=1, thinned_from=145, keys [[0,1],[439,1]]}
       heel_l_ctrl.Location.X {type=MovieSceneScriptingFloatChannel, default=0, thinned_from=145, keys [[0,0],[439,0]]}
       heel_l_ctrl.Location.Y {type=MovieSceneScriptingFloatChannel, default=0, thinned_from=145, keys [[0,0],[439,0]]}
       heel_l_ctrl.Location.Z {type=MovieSceneScriptingFloatChannel, default=0, thinned_from=145, keys [[0,0],[439,0]]}
       heel_l_ctrl.Rotation.X {type=MovieSceneScriptingFloatChannel, default=0, thinned_from=145, keys [[0,0],[439,0]]}
       heel_l_ctrl.Rotation.Y {type=MovieSceneScriptingFloatChannel, default=0, thinned_from=145, keys [[0,0],[439,0]]}
       heel_l_ctrl.Rotation.Z {type=MovieSceneScriptingFloatChannel, default=0, thinned_from=145, keys [[0,0],[439,0]]}
       heel_l_ctrl.Scale.X {type=MovieSceneScriptingFloatChannel, default=1, thinned_from=145, keys [[0,1],[439,1]]}
       heel_l_ctrl.Scale.Y {type=MovieSceneScriptingFloatChannel, default=1, thinned_from=145, keys [[0,1],[439,1]]}
       heel_l_ctrl.Scale.Z {type=MovieSceneScriptingFloatChannel, default=1, thinned_from=145, keys [[0,1],[439,1]]}
       leg_l_softness {
        type=MovieSceneScriptingFloatChannel
        default=0
        thinned_from=159
        keys [[0,0],[256,0],[257,0.0125],[258,0.05],[260,0.15],[261,0.195],[262,0.235],[263,0.265],[264,0.285],[265,0.295],[266,0.3],[268,0.3075],[269,0.3098],[271,0.3119],[272,0.375],[273,0.55],[275,0.95],[276,1.05],[277,1.1],[278,1.125],[279,1.1375],[281,1.1375],[282,1.125],[283,1.0875],[285,0.9631],[287,0.8467],[288,0.7873],[289,0.7222],[290,0.6665],[291,0.6033],[292,0.5444],[293,0.48],[294,0.4255],[295,0.3755],[296,0.3349],[297,0.3011],[314,0],[439,0]]
       }
       leg_l_pv_ik_ctrl.Location.X {type=MovieSceneScriptingFloatChannel, default=0, thinned_from=64, keys [[0,0],[439,0]]}
       leg_l_pv_ik_ctrl.Location.Y {type=MovieSceneScriptingFloatChannel, default=0, thinned_from=64, keys [[0,0],[439,0]]}
       leg_l_pv_ik_ctrl.Location.Z {type=MovieSceneScriptingFloatChannel, default=0, thinned_from=64, keys [[0,0],[439,0]]}
       leg_l_pv_ik_ctrl.Rotation.X {type=MovieSceneScriptingFloatChannel, default=0, thinned_from=64, keys [[0,0],[439,0]]}
       leg_l_pv_ik_ctrl.Rotation.Y {type=MovieSceneScriptingFloatChannel, default=0, thinned_from=64, keys [[0,0],[439,0]]}
       leg_l_pv_ik_ctrl.Rotation.Z {type=MovieSceneScriptingFloatChannel, default=0, thinned_from=64, keys [[0,0],[439,0]]}
       leg_l_pv_ik_ctrl.Scale.X {type=MovieSceneScriptingFloatChannel, default=1, thinned_from=64, keys [[0,1],[439,1]]}
       leg_l_pv_ik_ctrl.Scale.Y {type=MovieSceneScriptingFloatChannel, default=1, thinned_from=64, keys [[0,1],[439,1]]}
       leg_l_pv_ik_ctrl.Scale.Z {type=MovieSceneScriptingFloatChannel, default=1, thinned_from=64, keys [[0,1],[439,1]]}
       thigh_r_fk_ctrl.Location.X {type=MovieSceneScriptingFloatChannel, default=0, thinned_from=64, keys [[0,0],[439,0]]}
       thigh_r_fk_ctrl.Location.Y {type=MovieSceneScriptingFloatChannel, default=0, thinned_from=64, keys [[0,0],[439,0]]}
       thigh_r_fk_ctrl.Location.Z {type=MovieSceneScriptingFloatChannel, default=0, thinned_from=64, keys [[0,0],[439,0]]}
       thigh_r_fk_ctrl.Rotation.X {
        type=MovieSceneScriptingFloatChannel
        default=5.807
        thinned_from=64
        keys [[0,-4.4148],[15,-4.4148],[22,-4.7258],[30,-4.7195],[45,-4.821],[51,-4.4],[56,-4.2115],[62,-3.6013],[66,-0.4375],[70,1.0608],[75,1.4002],[83,1.843],[94,1.9872],[101,1.952],[107,1.7055],[111,0.9511],[115,0.015],[121,-0.2802],[130,-0.3786],[138,-0.5694],[143,-1.1417],[148,-3.4309],[153,18.7617],[160,27.0839],[172,1.9773],[187,-6.3915],[194,-3.9872],[199,3.2257],[204,32.0772],[215,32.0772],[223,26.8311],[233,11.093],[241,5.8248],[246,4.0687],[251,3.563],[257,2.0458],[262,-0.9709],[267,-4.0231],[271,0.1576],[275,4.2897],[279,7.7527],[283,9.3111],[287,9.7007],[292,9.7981],[299,9.8224],[337,9.8306],[345,9.4173],[351,8.1774],[356,3.2178],[361,-6.1938],[366,-9.3311],[371,-4.5339],[375,0.0978],[378,-2.7226],[382,-3.7802],[387,-4.2561],[395,-4.3553],[409,-4.5831],[439,-4.4148]]
       }
       thigh_r_fk_ctrl.Rotation.Y {
        type=MovieSceneScriptingFloatChannel
        default=-13.7092
        thinned_from=64
        keys [[0,-12.1768],[15,-12.1768],[22,-12.2215],[45,-12.2352],[51,-12.2],[56,-12.0791],[62,-11.7864],[66,-12.6734],[70,-12.5782],[75,-12.7988],[83,-12.9829],[94,-13.0429],[101,-13.0527],[107,-13.1214],[111,-13.3317],[115,-12.2125],[121,-12.3706],[130,-12.4233],[138,-12.3605],[143,-12.172],[148,-11.4181],[153,-15.7767],[160,-17.4112],[172,-13.3813],[187,-12.038],[194,-12.4351],[199,-13.6265],[204,-18.3919],[215,-18.3919],[223,-16.2198],[233,-9.7035],[241,-7.6961],[246,-7.0269],[251,-7.2982],[257,-8.1118],[267,-11.3665],[275,-15.4673],[279,-14.6092],[283,-14.2231],[287,-14.1265],[337,-14.0944],[345,-14.0548],[351,-13.9361],[356,-13.4613],[361,-12.3375],[366,-11.9629],[371,-11.7257],[375,-11.4966],[378,-11.9217],[382,-12.0811],[387,-12.1529],[409,-12.224],[439,-12.1768]]
       }
       thigh_r_fk_ctrl.Rotation.Z {
        type=MovieSceneScriptingFloatChannel
        default=18.8606
        thinned_from=64
        keys [[0,9.3708],[15,9.3708],[22,9.3422],[45,9.3333],[51,9.4],[56,9.3826],[62,9.4172],[66,9.5426],[70,9.6157],[83,9.6685],[101,9.6731],[107,9.6467],[111,9.5658],[115,9.4377],[138,9.394],[143,9.3422],[148,9.1353],[153,9.2064],[160,9.233],[172,8.9485],[187,8.8536],[194,8.8783],[199,8.9525],[204,9.249],[223,9.2418],[246,9.2148],[251,9.5226],[257,10.446],[267,14.1397],[275,17.9072],[279,19.0819],[283,19.6105],[287,19.7427],[337,19.7867],[345,19.6916],[351,19.4062],[356,18.2647],[361,14.5538],[366,13.3168],[371,11.9939],[375,10.7167],[378,9.8755],[382,9.5601],[387,9.4181],[395,9.3886],[439,9.3708]]
       }
       thigh_r_fk_ctrl.Scale.X {type=MovieSceneScriptingFloatChannel, default=1, thinned_from=64, keys [[0,1],[439,1]]}
       thigh_r_fk_ctrl.Scale.Y {type=MovieSceneScriptingFloatChannel, default=1, thinned_from=64, keys [[0,1],[439,1]]}
       thigh_r_fk_ctrl.Scale.Z {type=MovieSceneScriptingFloatChannel, default=1, thinned_from=64, keys [[0,1],[439,1]]}
       calf_r_fk_ctrl.Location.X {
        type=MovieSceneScriptingFloatChannel
        default=-0.3065
        thinned_from=64
        keys [[0,-0.9551],[15,-0.9551],[22,-0.9577],[45,-0.9585],[51,-1],[56,-0.944],[62,-0.9106],[66,-0.9873],[70,-1.0006],[75,-1.0214],[83,-1.0362],[94,-1.0411],[101,-1.0417],[107,-1.0462],[111,-1.0601],[115,-0.9636],[121,-0.9705],[130,-0.9728],[138,-0.9681],[143,-0.9541],[148,-0.8982],[153,-0.9698],[160,-0.9966],[172,-0.9533],[187,-0.9388],[194,-0.9434],[199,-0.9573],[204,-1.0127],[215,-1.0127],[223,-0.9386],[233,-0.7162],[241,-0.6526],[246,-0.6314],[251,-0.6262],[257,-0.6105],[267,-0.5478],[275,-0.5224],[279,-0.3481],[283,-0.2697],[287,-0.2501],[292,-0.2452],[299,-0.244],[337,-0.2436],[345,-0.2498],[351,-0.2685],[356,-0.3431],[361,-0.4073],[366,-0.4286],[371,-0.5248],[375,-0.6177],[378,-0.8286],[382,-0.9077],[387,-0.9433],[395,-0.9507],[439,-0.9551]]
       }
       calf_r_fk_ctrl.Location.Y {
        type=MovieSceneScriptingFloatChannel
        default=0.0409
        thinned_from=64
        keys [[0,0.1276],[15,0.1276],[22,0.1279],[45,0.128],[51,0.1],[56,0.1261],[62,0.1216],[66,0.1318],[70,0.1336],[75,0.1364],[83,0.1384],[94,0.139],[101,0.1391],[107,0.1397],[111,0.1416],[115,0.1287],[121,0.1296],[130,0.1299],[138,0.1293],[143,0.1274],[148,0.12],[153,0.1295],[160,0.1331],[172,0.1273],[187,0.1254],[194,0.126],[199,0.1278],[204,0.1352],[215,0.1352],[223,0.1253],[233,0.0957],[241,0.0872],[246,0.0843],[251,0.0836],[257,0.0815],[267,0.0732],[275,0.0698],[279,0.0465],[283,0.036],[287,0.0334],[292,0.0327],[337,0.0325],[345,0.0334],[351,0.0359],[356,0.0458],[361,0.0544],[366,0.0572],[371,0.0701],[375,0.0825],[378,0.1107],[382,0.1212],[387,0.126],[395,0.127],[439,0.1276]]
       }
       calf_r_fk_ctrl.Location.Z {
        type=MovieSceneScriptingFloatChannel
        default=-0.0056
        thinned_from=64
        keys [[0,-0.0174],[45,-0.0175],[51,0],[56,-0.0172],[62,-0.0166],[66,-0.018],[75,-0.0186],[83,-0.0189],[107,-0.0191],[111,-0.0194],[115,-0.0176],[130,-0.0178],[138,-0.0177],[143,-0.0174],[148,-0.0164],[153,-0.0177],[160,-0.0182],[172,-0.0174],[187,-0.0171],[194,-0.0172],[199,-0.0175],[204,-0.0185],[215,-0.0185],[223,-0.0171],[233,-0.0131],[241,-0.0119],[246,-0.0115],[257,-0.0111],[267,-0.01],[275,-0.0095],[279,-0.0064],[283,-0.0049],[287,-0.0046],[337,-0.0044],[345,-0.0046],[351,-0.0049],[361,-0.0074],[366,-0.0078],[371,-0.0096],[375,-0.0113],[378,-0.0151],[382,-0.0166],[387,-0.0172],[439,-0.0174]]
       }
       calf_r_fk_ctrl.Rotation.X {type=MovieSceneScriptingFloatChannel, default=-0.0638, thinned_from=64, keys [[0,-0.0638],[45,-0.0638],[51,-0.1],[56,-0.0638],[439,-0.0638]]}
       calf_r_fk_ctrl.Rotation.Y {type=MovieSceneScriptingFloatChannel, default=1.0367, thinned_from=64, keys [[0,1.0367],[45,1.0367],[51,1],[56,1.0367],[439,1.0367]]}
       calf_r_fk_ctrl.Rotation.Z {type=MovieSceneScriptingFloatChannel, default=-7.6066, thinned_from=64, keys [[0,-7.6066],[439,-7.6066]]}
       calf_r_fk_ctrl.Scale.X {type=MovieSceneScriptingFloatChannel, default=1, thinned_from=64, keys [[0,1],[439,1]]}
       calf_r_fk_ctrl.Scale.Y {type=MovieSceneScriptingFloatChannel, default=1, thinned_from=64, keys [[0,1],[439,1]]}
       calf_r_fk_ctrl.Scale.Z {type=MovieSceneScriptingFloatChannel, default=1, thinned_from=64, keys [[0,1],[439,1]]}
       foot_r_fk_ctrl.Location.X {
        type=MovieSceneScriptingFloatChannel
        default=-0.3403
        thinned_from=64
        keys [[0,-1.0605],[15,-1.0605],[22,-1.0633],[45,-1.0642],[51,-1.1],[56,-1.0481],[62,-1.0111],[66,-1.0962],[70,-1.1109],[75,-1.134],[83,-1.1505],[94,-1.1559],[101,-1.1566],[107,-1.1616],[111,-1.177],[115,-1.0698],[121,-1.0775],[130,-1.0801],[138,-1.0749],[143,-1.0594],[148,-0.9973],[153,-1.0767],[160,-1.1065],[172,-1.0584],[187,-1.0424],[194,-1.0475],[199,-1.0629],[204,-1.1244],[215,-1.1244],[223,-1.0421],[233,-0.7952],[241,-0.7246],[246,-0.701],[251,-0.6952],[257,-0.6778],[267,-0.6082],[275,-0.5801],[279,-0.3865],[283,-0.2994],[287,-0.2777],[292,-0.2722],[299,-0.2709],[337,-0.2704],[345,-0.2773],[351,-0.2981],[356,-0.381],[361,-0.4522],[366,-0.4759],[371,-0.5827],[375,-0.6859],[378,-0.92],[382,-1.0078],[387,-1.0473],[395,-1.0555],[439,-1.0605]]
       }
       foot_r_fk_ctrl.Location.Y {type=MovieSceneScriptingFloatChannel, default=-0.0543, thinned_from=64, keys [[0,-0.0543],[45,-0.0543],[51,-0.1],[56,-0.0543],[439,-0.0543]]}
       foot_r_fk_ctrl.Location.Z {type=MovieSceneScriptingFloatChannel, default=-0.8609, thinned_from=64, keys [[0,-0.8609],[45,-0.8609],[51,-0.9],[56,-0.8609],[439,-0.8609]]}
       foot_r_fk_ctrl.Rotation.X {
        type=MovieSceneScriptingFloatChannel
        default=-0.5315
        thinned_from=64
        keys [[0,-0.6843],[45,-0.6845],[51,-0.7],[56,-0.6849],[62,-0.6864],[66,-0.6971],[70,-0.7042],[83,-0.7063],[107,-0.7059],[111,-0.7042],[115,-0.7075],[143,-0.7075],[148,-0.7099],[153,-0.8051],[160,-0.8408],[172,-0.7446],[187,-0.7125],[194,-0.7219],[199,-0.75],[204,-0.8622],[215,-0.8622],[223,-0.8442],[233,-0.7902],[241,-0.7717],[246,-0.7655],[251,-0.7533],[257,-0.7168],[267,-0.5707],[275,-0.5301],[279,-0.525],[283,-0.5227],[337,-0.522],[345,-0.5228],[351,-0.5253],[356,-0.5354],[361,-0.5386],[366,-0.5397],[371,-0.6314],[375,-0.7198],[378,-0.6977],[382,-0.6893],[387,-0.6856],[409,-0.6835],[439,-0.6843]]
       }
       foot_r_fk_ctrl.Rotation.Y {
        type=MovieSceneScriptingFloatChannel
        default=8.3075
        thinned_from=64
        keys [[0,5.3149],[15,5.3149],[22,5.2858],[45,5.2768],[56,5.3172],[62,5.3232],[70,5.0203],[75,5.0052],[107,4.9971],[111,5.0017],[115,4.8332],[130,4.8302],[138,4.8168],[143,4.7765],[148,4.6154],[153,3.1089],[160,2.5439],[172,3.9905],[187,4.4727],[194,4.331],[199,3.9058],[204,2.2049],[215,2.2049],[223,2.3564],[233,2.8107],[241,2.9785],[246,3.0345],[251,3.3242],[257,4.1934],[262,5.9216],[267,7.6702],[275,8.4994],[279,8.5499],[287,8.5783],[337,8.5802],[345,8.5521],[351,8.4681],[356,8.132],[361,7.4726],[366,7.2527],[371,5.7927],[375,4.3829],[378,4.9654],[382,5.1839],[387,5.2822],[395,5.3027],[409,5.3228],[439,5.3149]]
       }
       foot_r_fk_ctrl.Rotation.Z {
        type=MovieSceneScriptingFloatChannel
        default=2.6657
        thinned_from=64
        keys [[0,2.1622],[15,2.1622],[22,2.207],[30,2.2061],[45,2.2207],[51,2.2],[62,2.057],[66,1.786],[70,1.6667],[83,1.5962],[94,1.5823],[101,1.5867],[107,1.6172],[111,1.7105],[115,1.8431],[121,1.8811],[130,1.8937],[138,1.9174],[143,1.9883],[148,2.2721],[153,3.3794],[160,3.7947],[172,2.9496],[187,2.6679],[194,2.7539],[199,3.0119],[204,4.0438],[215,4.0438],[223,3.6331],[233,2.401],[241,1.9908],[246,1.854],[251,1.895],[257,2.0178],[267,2.5093],[275,2.5457],[279,2.58],[283,2.5954],[337,2.6005],[345,2.6072],[351,2.6273],[356,2.7077],[361,3.222],[366,3.3934],[371,2.7746],[375,2.1772],[378,2.1678],[387,2.1627],[395,2.1624],[409,2.1906],[439,2.1622]]
       }
       foot_r_fk_ctrl.Scale.X {type=MovieSceneScriptingFloatChannel, default=1, thinned_from=64, keys [[0,1],[439,1]]}
       foot_r_fk_ctrl.Scale.Y {type=MovieSceneScriptingFloatChannel, default=1, thinned_from=64, keys [[0,1],[439,1]]}
       foot_r_fk_ctrl.Scale.Z {type=MovieSceneScriptingFloatChannel, default=1, thinned_from=64, keys [[0,1],[439,1]]}
       ball_r_fk_ctrl.Location.X {type=MovieSceneScriptingFloatChannel, default=0, thinned_from=64, keys [[0,0],[439,0]]}
       ball_r_fk_ctrl.Location.Y {type=MovieSceneScriptingFloatChannel, default=0, thinned_from=64, keys [[0,0],[439,0]]}
       ball_r_fk_ctrl.Location.Z {type=MovieSceneScriptingFloatChannel, default=0, thinned_from=64, keys [[0,0],[439,0]]}
       ball_r_fk_ctrl.Rotation.X {type=MovieSceneScriptingFloatChannel, default=0, thinned_from=64, keys [[0,0],[439,0]]}
       ball_r_fk_ctrl.Rotation.Y {type=MovieSceneScriptingFloatChannel, default=0, thinned_from=64, keys [[0,0],[439,0]]}
       ball_r_fk_ctrl.Rotation.Z {type=MovieSceneScriptingFloatChannel, default=0, thinned_from=64, keys [[0,0],[439,0]]}
       ball_r_fk_ctrl.Scale.X {type=MovieSceneScriptingFloatChannel, default=1, thinned_from=64, keys [[0,1],[439,1]]}
       ball_r_fk_ctrl.Scale.Y {type=MovieSceneScriptingFloatChannel, default=1, thinned_from=64, keys [[0,1],[439,1]]}
       ball_r_fk_ctrl.Scale.Z {type=MovieSceneScriptingFloatChannel, default=1, thinned_from=64, keys [[0,1],[439,1]]}
       foot_r_ik_ctrl.Location.X {type=MovieSceneScriptingFloatChannel, default=-4.2937, const [-4.2937,64]}
       foot_r_ik_ctrl.Location.Y {
        type=MovieSceneScriptingFloatChannel
        default=-1.8006
        thinned_from=64
        keys [[0,-1.8187],[66,-1.8187],[70,-1.8047],[75,-1.8012],[392,-1.8006],[406,-1.8084],[439,-1.8187]]
       }
       foot_r_ik_ctrl.Location.Z {type=MovieSceneScriptingFloatChannel, default=0, thinned_from=64, keys [[0,0],[439,0]]}
       foot_r_ik_ctrl.Rotation.X {
        type=MovieSceneScriptingFloatChannel
        default=2.4996
        thinned_from=64
        keys [[0,2.4895],[45,2.4895],[51,2.5],[56,2.4895],[66,2.4895],[70,2.4974],[392,2.4996],[439,2.4895]]
       }
       foot_r_ik_ctrl.Rotation.Y {
        type=MovieSceneScriptingFloatChannel
        default=-2.0996
        thinned_from=64
        keys [[0,-2.0871],[45,-2.0871],[51,-2.1],[56,-2.0871],[66,-2.0871],[70,-2.0968],[392,-2.0996],[439,-2.0871]]
       }
       foot_r_ik_ctrl.Rotation.Z {
        type=MovieSceneScriptingFloatChannel
        default=-2.75
        thinned_from=64
        keys [[0,-14.9285],[251,-14.9285],[257,-9.75],[262,-7.25],[358,-7.25],[363,-12.4464],[368,-14.9285],[439,-14.9285]]
       }
       foot_r_ik_ctrl.Scale.X {type=MovieSceneScriptingFloatChannel, default=1, thinned_from=64, keys [[0,1],[439,1]]}
       foot_r_ik_ctrl.Scale.Y {type=MovieSceneScriptingFloatChannel, default=1, thinned_from=64, keys [[0,1],[439,1]]}
       foot_r_ik_ctrl.Scale.Z {type=MovieSceneScriptingFloatChannel, default=1, thinned_from=64, keys [[0,1],[439,1]]}
       ball_r_ik_ctrl.Location.X {type=MovieSceneScriptingFloatChannel, default=0, thinned_from=64, keys [[0,0],[439,0]]}
       ball_r_ik_ctrl.Location.Y {type=MovieSceneScriptingFloatChannel, default=0, thinned_from=64, keys [[0,0],[439,0]]}
       ball_r_ik_ctrl.Location.Z {type=MovieSceneScriptingFloatChannel, default=0, thinned_from=64, keys [[0,0],[439,0]]}
       ball_r_ik_ctrl.Rotation.X {type=MovieSceneScriptingFloatChannel, default=0, thinned_from=64, keys [[0,0],[439,0]]}
       ball_r_ik_ctrl.Rotation.Y {type=MovieSceneScriptingFloatChannel, default=0, thinned_from=64, keys [[0,0],[439,0]]}
       ball_r_ik_ctrl.Rotation.Z {type=MovieSceneScriptingFloatChannel, default=0, thinned_from=64, keys [[0,0],[439,0]]}
       ball_r_ik_ctrl.Scale.X {type=MovieSceneScriptingFloatChannel, default=1, thinned_from=64, keys [[0,1],[439,1]]}
       ball_r_ik_ctrl.Scale.Y {type=MovieSceneScriptingFloatChannel, default=1, thinned_from=64, keys [[0,1],[439,1]]}
       ball_r_ik_ctrl.Scale.Z {type=MovieSceneScriptingFloatChannel, default=1, thinned_from=64, keys [[0,1],[439,1]]}
       foot_bk1_r_ctrl.Location.X {type=MovieSceneScriptingFloatChannel, default=0, thinned_from=64, keys [[0,0],[439,0]]}
       foot_bk1_r_ctrl.Location.Y {type=MovieSceneScriptingFloatChannel, default=0, thinned_from=64, keys [[0,0],[439,0]]}
       foot_bk1_r_ctrl.Location.Z {type=MovieSceneScriptingFloatChannel, default=0, thinned_from=64, keys [[0,0],[439,0]]}
       foot_bk1_r_ctrl.Rotation.X {type=MovieSceneScriptingFloatChannel, default=0, thinned_from=64, keys [[0,0],[439,0]]}
       foot_bk1_r_ctrl.Rotation.Y {type=MovieSceneScriptingFloatChannel, default=0, thinned_from=64, keys [[0,0],[439,0]]}
       foot_bk1_r_ctrl.Rotation.Z {type=MovieSceneScriptingFloatChannel, default=-0.0067, thinned_from=64, keys [[0,0],[348,0],[353,-0.0104],[406,-0.0104],[439,0]]}
       foot_bk1_r_ctrl.Scale.X {type=MovieSceneScriptingFloatChannel, default=1, thinned_from=64, keys [[0,1],[439,1]]}
       foot_bk1_r_ctrl.Scale.Y {type=MovieSceneScriptingFloatChannel, default=1, thinned_from=64, keys [[0,1],[439,1]]}
       foot_bk1_r_ctrl.Scale.Z {type=MovieSceneScriptingFloatChannel, default=1, thinned_from=64, keys [[0,1],[439,1]]}
       foot_roll_r_ctrl.Location.X {type=MovieSceneScriptingFloatChannel, default=0, thinned_from=64, keys [[0,0],[439,0]]}
       foot_roll_r_ctrl.Location.Y {type=MovieSceneScriptingFloatChannel, default=0, thinned_from=64, keys [[0,0],[439,0]]}
       foot_roll_r_ctrl.Location.Z {type=MovieSceneScriptingFloatChannel, default=0, thinned_from=64, keys [[0,0],[439,0]]}
       foot_roll_r_ctrl.Rotation.X {type=MovieSceneScriptingFloatChannel, default=0, thinned_from=64, keys [[0,0],[439,0]]}
       foot_roll_r_ctrl.Rotation.Y {type=MovieSceneScriptingFloatChannel, default=0, thinned_from=64, keys [[0,0],[439,0]]}
       foot_roll_r_ctrl.Rotation.Z {type=MovieSceneScriptingFloatChannel, default=0, thinned_from=64, keys [[0,0],[439,0]]}
       foot_roll_r_ctrl.Scale.X {type=MovieSceneScriptingFloatChannel, default=1, thinned_from=64, keys [[0,1],[439,1]]}
       foot_roll_r_ctrl.Scale.Y {type=MovieSceneScriptingFloatChannel, default=1, thinned_from=64, keys [[0,1],[439,1]]}
       foot_roll_r_ctrl.Scale.Z {type=MovieSceneScriptingFloatChannel, default=1, thinned_from=64, keys [[0,1],[439,1]]}
       roll_blend_r_ctrl {type=MovieSceneScriptingFloatChannel, default=1, thinned_from=64, keys [[0,1],[439,1]]}
       tip_r_ctrl.Location.X {type=MovieSceneScriptingFloatChannel, default=0, thinned_from=64, keys [[0,0],[439,0]]}
       tip_r_ctrl.Location.Y {type=MovieSceneScriptingFloatChannel, default=0, thinned_from=64, keys [[0,0],[439,0]]}
       tip_r_ctrl.Location.Z {type=MovieSceneScriptingFloatChannel, default=0, thinned_from=64, keys [[0,0],[439,0]]}
       tip_r_ctrl.Rotation.X {type=MovieSceneScriptingFloatChannel, default=0, thinned_from=64, keys [[0,0],[439,0]]}
       tip_r_ctrl.Rotation.Y {type=MovieSceneScriptingFloatChannel, default=0, thinned_from=64, keys [[0,0],[439,0]]}
       tip_r_ctrl.Rotation.Z {type=MovieSceneScriptingFloatChannel, default=0, thinned_from=64, keys [[0,0],[439,0]]}
       tip_r_ctrl.Scale.X {type=MovieSceneScriptingFloatChannel, default=1, thinned_from=64, keys [[0,1],[439,1]]}
       tip_r_ctrl.Scale.Y {type=MovieSceneScriptingFloatChannel, default=1, thinned_from=64, keys [[0,1],[439,1]]}
       tip_r_ctrl.Scale.Z {type=MovieSceneScriptingFloatChannel, default=1, thinned_from=64, keys [[0,1],[439,1]]}
       heel_r_ctrl.Location.X {type=MovieSceneScriptingFloatChannel, default=0, thinned_from=64, keys [[0,0],[439,0]]}
       heel_r_ctrl.Location.Y {type=MovieSceneScriptingFloatChannel, default=0, thinned_from=64, keys [[0,0],[439,0]]}
       heel_r_ctrl.Location.Z {type=MovieSceneScriptingFloatChannel, default=0, thinned_from=64, keys [[0,0],[439,0]]}
       heel_r_ctrl.Rotation.X {type=MovieSceneScriptingFloatChannel, default=0, thinned_from=64, keys [[0,0],[439,0]]}
       heel_r_ctrl.Rotation.Y {type=MovieSceneScriptingFloatChannel, default=0, thinned_from=64, keys [[0,0],[439,0]]}
       heel_r_ctrl.Rotation.Z {
        type=MovieSceneScriptingFloatChannel
        default=0
        thinned_from=64
        keys [[0,0],[246,0],[251,0.55],[257,1.275],[262,0.425],[267,0],[353,0],[363,0.3],[368,0.35],[375,0],[439,0]]
       }
       heel_r_ctrl.Scale.X {type=MovieSceneScriptingFloatChannel, default=1, thinned_from=64, keys [[0,1],[439,1]]}
       heel_r_ctrl.Scale.Y {type=MovieSceneScriptingFloatChannel, default=1, thinned_from=64, keys [[0,1],[439,1]]}
       heel_r_ctrl.Scale.Z {type=MovieSceneScriptingFloatChannel, default=1, thinned_from=64, keys [[0,1],[439,1]]}
       leg_r_softness {type=MovieSceneScriptingFloatChannel, default=0, thinned_from=64, keys [[0,0],[439,0]]}
       leg_r_pv_ik_ctrl.Location.X {type=MovieSceneScriptingFloatChannel, default=0, thinned_from=64, keys [[0,0],[439,0]]}
       leg_r_pv_ik_ctrl.Location.Y {type=MovieSceneScriptingFloatChannel, default=0, thinned_from=64, keys [[0,0],[439,0]]}
       leg_r_pv_ik_ctrl.Location.Z {type=MovieSceneScriptingFloatChannel, default=0, thinned_from=64, keys [[0,0],[439,0]]}
       leg_r_pv_ik_ctrl.Rotation.X {type=MovieSceneScriptingFloatChannel, default=0, thinned_from=64, keys [[0,0],[439,0]]}
       leg_r_pv_ik_ctrl.Rotation.Y {type=MovieSceneScriptingFloatChannel, default=0, thinned_from=64, keys [[0,0],[439,0]]}
       leg_r_pv_ik_ctrl.Rotation.Z {type=MovieSceneScriptingFloatChannel, default=0, thinned_from=64, keys [[0,0],[439,0]]}
       leg_r_pv_ik_ctrl.Scale.X {type=MovieSceneScriptingFloatChannel, default=1, thinned_from=64, keys [[0,1],[439,1]]}
       leg_r_pv_ik_ctrl.Scale.Y {type=MovieSceneScriptingFloatChannel, default=1, thinned_from=64, keys [[0,1],[439,1]]}
       leg_r_pv_ik_ctrl.Scale.Z {type=MovieSceneScriptingFloatChannel, default=1, thinned_from=64, keys [[0,1],[439,1]]}
       upperarm_l_fk_ctrl.Location.X {type=MovieSceneScriptingFloatChannel, default=0, const [0,65]}
       upperarm_l_fk_ctrl.Location.Y {type=MovieSceneScriptingFloatChannel, default=0, const [0,65]}
       upperarm_l_fk_ctrl.Location.Z {type=MovieSceneScriptingFloatChannel, default=0, const [0,65]}
       upperarm_l_fk_ctrl.Rotation.X {
        type=MovieSceneScriptingFloatChannel
        default=63.7057
        thinned_from=65
        keys [[0,10.479],[251,10.479],[257,21.4555],[267,29.5033],[271,42.8317],[275,56.0052],[279,77.7538],[283,87.5406],[287,89.9874],[292,90.599],[309,90.7873],[337,90.8029],[345,88.2598],[351,80.6306],[356,50.1136],[361,30.3942],[366,23.821],[371,20.9969],[375,18.2701],[378,17.4749],[382,17.1768],[387,17.0426],[395,17.0146],[409,33.6898],[439,16.9979],[440,33.7099]]
       }
       upperarm_l_fk_ctrl.Rotation.Y {
        type=MovieSceneScriptingFloatChannel
        default=5.9874
        thinned_from=65
        keys [[0,-6.326],[251,-6.326],[257,-9.075],[262,-3.3811],[267,2.3797],[271,13.8273],[275,25.1418],[279,12.4713],[283,6.7696],[287,5.3442],[292,4.9878],[299,4.8987],[309,4.8781],[337,4.869],[345,4.9615],[351,5.239],[356,6.3489],[361,-0.484],[366,-2.7616],[371,-7.4852],[375,-12.0458],[378,-13.1361],[382,-13.545],[387,-13.7289],[395,-13.7673],[409,-0.5334],[439,-13.7903],[440,-0.5291]]
       }
       upperarm_l_fk_ctrl.Rotation.Z {
        type=MovieSceneScriptingFloatChannel
        default=49.9173
        thinned_from=65
        keys [[0,-33.5615],[251,-33.5615],[257,-36.207],[267,-29.8592],[271,-0.9118],[275,27.699],[279,46.0379],[283,54.2904],[287,56.3535],[292,56.8693],[299,56.9982],[337,57.0412],[345,56.3095],[351,54.1142],[356,45.333],[361,-13.6693],[366,-33.3367],[371,-35.5691],[375,-37.7245],[378,-38.2692],[382,-38.4735],[387,-38.5654],[395,-38.5846],[409,-32.3324],[439,-38.5961],[440,-32.2835]]
       }
       upperarm_l_fk_ctrl.Scale.X {type=MovieSceneScriptingFloatChannel, default=1, const [1,65]}
       upperarm_l_fk_ctrl.Scale.Y {type=MovieSceneScriptingFloatChannel, default=1, const [1,65]}
       upperarm_l_fk_ctrl.Scale.Z {type=MovieSceneScriptingFloatChannel, default=1, const [1,65]}
       lowerarm_l_fk_ctrl.Location.X {
        type=MovieSceneScriptingFloatChannel
        default=0.0842
        thinned_from=65
        keys [[0,0],[275,0],[279,0.2154],[283,0.3124],[287,0.3366],[292,0.3427],[299,0.3442],[337,0.3447],[345,0.3231],[351,0.2585],[356,0],[440,0]]
       }
       lowerarm_l_fk_ctrl.Location.Y {
        type=MovieSceneScriptingFloatChannel
        default=-0.044
        thinned_from=65
        keys [[0,0],[275,0],[279,-0.1127],[283,-0.1634],[287,-0.1761],[292,-0.1792],[299,-0.18],[337,-0.1803],[345,-0.169],[351,-0.1352],[356,0],[440,0]]
       }
       lowerarm_l_fk_ctrl.Location.Z {
        type=MovieSceneScriptingFloatChannel
        default=0.0007
        thinned_from=65
        keys [[0,0],[275,0],[279,0.0019],[283,0.0027],[292,0.003],[337,0.003],[345,0.0028],[351,0.0023],[356,0],[440,0]]
       }
       lowerarm_l_fk_ctrl.Rotation.X {
        type=MovieSceneScriptingFloatChannel
        default=0.0631
        thinned_from=65
        keys [[0,0.102],[251,0.102],[257,0.4497],[267,0.4592],[275,0.4327],[279,0.0892],[283,-0.0654],[287,-0.1041],[292,-0.1137],[299,-0.1161],[309,-0.1167],[337,-0.1169],[345,-0.0989],[351,-0.0446],[356,0.1723],[361,0.3864],[366,0.4577],[375,0.4538],[378,0.4467],[382,0.444],[387,0.4428],[440,0.4424]]
       }
       lowerarm_l_fk_ctrl.Rotation.Y {
        type=MovieSceneScriptingFloatChannel
        default=0.4335
        thinned_from=65
        keys [[0,0.4483],[251,0.4483],[257,0.0877],[262,0.0553],[267,0.0226],[271,0.0894],[275,0.1555],[279,0.3362],[283,0.4175],[287,0.4379],[292,0.443],[299,0.4442],[337,0.4446],[345,0.4435],[351,0.44],[356,0.4263],[361,0.1392],[366,0.0436],[371,0.059],[375,0.0738],[378,0.1059],[382,0.118],[387,0.1234],[395,0.1245],[409,0.1245],[440,0.1252]]
       }
       lowerarm_l_fk_ctrl.Rotation.Z {
        type=MovieSceneScriptingFloatChannel
        default=-5.2543
        thinned_from=66
        keys [[0,-0.0544],[251,-0.0544],[257,66.0975],[267,74.3106],[275,57.3492],[279,4.2482],[283,-19.6472],[287,-25.6211],[292,-27.1145],[299,-27.4879],[309,-27.5741],[337,-27.6123],[345,-25.3157],[351,-18.4259],[356,9.1334],[361,56.0475],[366,71.6856],[371,69.7498],[375,67.8807],[378,63.7828],[382,62.246],[387,61.5545],[395,61.4104],[440,61.324],[440.0013,0]]
       }
       lowerarm_l_fk_ctrl.Scale.X {type=MovieSceneScriptingFloatChannel, default=1, const [1,65]}
       lowerarm_l_fk_ctrl.Scale.Y {type=MovieSceneScriptingFloatChannel, default=1, const [1,65]}
       lowerarm_l_fk_ctrl.Scale.Z {type=MovieSceneScriptingFloatChannel, default=1, const [1,65]}
       hand_l_fk_ctrl.Location.X {
        type=MovieSceneScriptingFloatChannel
        default=0.0816
        thinned_from=65
        keys [[0,0],[275,0],[279,0.2088],[283,0.3028],[287,0.3263],[292,0.3322],[299,0.3337],[337,0.3341],[345,0.3133],[351,0.2506],[356,0],[440,0]]
       }
       hand_l_fk_ctrl.Location.Y {type=MovieSceneScriptingFloatChannel, default=-0.1788, thinned_from=65, keys [[0,-0.1788],[440,-0.1788]]}
       hand_l_fk_ctrl.Location.Z {type=MovieSceneScriptingFloatChannel, default=0.0249, thinned_from=65, keys [[0,0.0249],[440,0.0249]]}
       hand_l_fk_ctrl.Rotation.X {
        type=MovieSceneScriptingFloatChannel
        default=101.522
        thinned_from=65
        keys [[0,4.1325],[251,4.1325],[257,8.4026],[262,2.0846],[267,-4.3078],[271,-1.3561],[275,1.5613],[279,21.1913],[283,30.0249],[287,32.2332],[292,32.7853],[299,32.9234],[337,32.9694],[345,38.6391],[351,55.6482],[356,123.6848],[361,32.4636],[366,2.0565],[371,6.1557],[375,10.1135],[378,12.2627],[382,13.0687],[387,13.4313],[395,13.5069],[409,-1.5922],[439,13.5522],[440,-1.6202]]
       }
       hand_l_fk_ctrl.Rotation.Y {
        type=MovieSceneScriptingFloatChannel
        default=40.9406
        thinned_from=65
        keys [[0,-1.4416],[251,-1.4416],[257,-23.4565],[262,-28.9069],[267,-34.4215],[271,1.1769],[275,36.3613],[279,43.1282],[283,46.1733],[287,46.9346],[292,47.1249],[337,47.1883],[345,46.5466],[351,44.6213],[356,36.9201],[361,-2.0978],[366,-15.1038],[371,-18.1401],[375,-21.0717],[378,-19.1167],[382,-18.3836],[387,-18.0537],[395,-17.985],[409,-14.4792],[439,-17.9437],[440,-14.4359]]
       }
       hand_l_fk_ctrl.Rotation.Z {
        type=MovieSceneScriptingFloatChannel
        default=38.8736
        thinned_from=65
        keys [[0,-0.055],[251,-0.055],[257,15.5519],[262,12.0392],[267,8.4852],[271,-0.4414],[275,-9.2642],[279,-38.0472],[283,-50.9996],[287,-54.2377],[292,-55.0472],[299,-55.2496],[337,-55.317],[345,-47.5269],[351,-24.1564],[356,69.3253],[361,26.1304],[366,11.7321],[371,15.4447],[375,19.0294],[378,19.6021],[382,19.8168],[387,19.9135],[395,19.9336],[409,6.9805],[439,19.9457],[440,6.9699]]
       }
       hand_l_fk_ctrl.Scale.X {type=MovieSceneScriptingFloatChannel, default=1, const [1,65]}
       hand_l_fk_ctrl.Scale.Y {type=MovieSceneScriptingFloatChannel, default=1, const [1,65]}
       hand_l_fk_ctrl.Scale.Z {type=MovieSceneScriptingFloatChannel, default=1, const [1,65]}
       thumb_01_l_ctrl.Location.X {
        type=MovieSceneScriptingFloatChannel
        default=0
        thinned_from=94
        keys [[0,0.3381],[264,0.3381],[265,0.2602],[266,0.1316],[267,0.0262],[268,-0.015],[269,-0.0413],[270,-0.0673],[271,0],[359,0],[360,-0.0898],[361,0.1354],[362,0.2621],[363,0.3191],[364,0.3381],[440,0.3381]]
       }
       thumb_01_l_ctrl.Location.Y {
        type=MovieSceneScriptingFloatChannel
        default=0
        thinned_from=94
        keys [[0,-0.6458],[264,-0.6458],[265,-0.5608],[266,-0.413],[267,-0.2738],[268,-0.2012],[269,-0.096],[270,0.008],[271,0],[359,0],[360,0.0107],[361,-0.3348],[362,-0.5292],[363,-0.6166],[364,-0.6458],[440,-0.6458]]
       }
       thumb_01_l_ctrl.Location.Z {
        type=MovieSceneScriptingFloatChannel
        default=0
        thinned_from=94
        keys [[0,1.0905],[264,1.0905],[265,0.9877],[266,0.8005],[267,0.605],[268,0.4863],[269,0.2735],[270,0.0632],[271,0],[359,0],[360,0.0844],[361,0.6139],[362,0.9118],[363,1.0458],[364,1.0905],[440,1.0905]]
       }
       thumb_01_l_ctrl.Rotation.X {
        type=MovieSceneScriptingFloatChannel
        default=-8.7595
        thinned_from=87
        keys [[0,-26.3065],[264,-26.3065],[265,-27.4331],[266,-28.6655],[267,-28.2081],[268,-26.5479],[269,-20.6892],[270,-14.8987],[279,-8.6086],[283,-5.0333],[287,-4.1395],[292,-3.916],[299,-3.8601],[309,-3.8472],[352,-3.8415],[354,-4.8569],[355,-5.9487],[356,-8.0293],[357,-10.6338],[360,-18.2678],[361,-22.4987],[362,-24.8786],[363,-25.9496],[364,-26.3065],[440,-26.3065]]
       }
       thumb_01_l_ctrl.Rotation.Y {
        type=MovieSceneScriptingFloatChannel
        default=33.2518
        thinned_from=90
        keys [[0,21.1474],[264,21.1474],[265,21.838],[266,23.1851],[267,24.7794],[268,25.2774],[270,25.1032],[271,23.53],[273,18.561],[274,21.5174],[279,31.4739],[283,35.5122],[287,36.5218],[292,36.7742],[309,36.8519],[352,36.8583],[355,34.9569],[356,31.3493],[360,27.6993],[361,24.2509],[362,22.3112],[363,21.4384],[364,21.1474],[440,21.1474]]
       }
       thumb_01_l_ctrl.Rotation.Z {
        type=MovieSceneScriptingFloatChannel
        default=1.5328
        thinned_from=94
        keys [[0,5],[264,5],[265,5.7517],[266,6.7922],[267,7.1784],[268,6.8579],[269,5.1264],[270,3.415],[271,1.0743],[272,1.1745],[273,1.4349],[274,1.856],[279,2.557],[283,2.8094],[287,2.8725],[292,2.8883],[352,2.8936],[354,2.6855],[355,2.4684],[356,2.0797],[357,1.9813],[358,1.6929],[359,1.6948],[360,3.3571],[361,4.2218],[362,4.7082],[363,2.5],[365,1.5],[366,1.9375],[367,2.7031],[368,3.643],[369,4.5719],[370,4.8435],[371,5],[440,5]]
       }
       thumb_01_l_ctrl.Scale.X {
        type=MovieSceneScriptingFloatChannel
        default=1
        thinned_from=94
        keys [[0,0.9426],[264,0.9426],[265,0.944],[266,0.9478],[268,0.9601],[270,0.9892],[271,1],[359,1],[360,0.9856],[361,0.9629],[362,0.9502],[363,0.9445],[364,0.9426],[440,0.9426]]
       }
       thumb_01_l_ctrl.Scale.Y {
        type=MovieSceneScriptingFloatChannel
        default=1
        thinned_from=94
        keys [[0,0.9426],[264,0.9426],[265,0.944],[266,0.9478],[268,0.9601],[270,0.9892],[271,1],[359,1],[360,0.9856],[361,0.9629],[362,0.9502],[363,0.9445],[364,0.9426],[440,0.9426]]
       }
       thumb_01_l_ctrl.Scale.Z {
        type=MovieSceneScriptingFloatChannel
        default=1
        thinned_from=94
        keys [[0,0.9426],[264,0.9426],[265,0.944],[266,0.9478],[268,0.9601],[269,0.9598],[270,0.9892],[271,1],[359,1],[360,0.9856],[361,0.9629],[362,0.9502],[363,0.9445],[364,0.9426],[440,0.9426]]
       }
       thumb_02_l_ctrl.Location.X {
        type=MovieSceneScriptingFloatChannel
        default=0
        thinned_from=94
        keys [[0,-0.4396],[264,-0.4396],[265,-0.3846],[266,-0.2885],[267,-0.1965],[268,-0.1474],[269,-0.0733],[270,0],[360,0],[361,-0.2314],[362,-0.3615],[363,-0.42],[364,-0.4396],[440,-0.4396]]
       }
       thumb_02_l_ctrl.Location.Y {
        type=MovieSceneScriptingFloatChannel
        default=0
        thinned_from=94
        keys [[0,0.0898],[264,0.0898],[265,0.0785],[266,0.0589],[267,0.0401],[268,0.0301],[270,0],[360,0],[361,0.0472],[362,0.0738],[363,0.0858],[364,0.0898],[440,0.0898]]
       }
       thumb_02_l_ctrl.Location.Z {
        type=MovieSceneScriptingFloatChannel
        default=0
        thinned_from=94
        keys [[0,0.1821],[264,0.1821],[265,0.1593],[266,0.1195],[267,0.0814],[268,0.061],[269,0.0303],[270,0],[360,0],[361,0.0958],[362,0.1497],[363,0.174],[364,0.1821],[440,0.1821]]
       }
       thumb_02_l_ctrl.Rotation.X {
        type=MovieSceneScriptingFloatChannel
        default=0
        thinned_from=94
        keys [[0,-8.1362],[264,-8.1362],[265,-7.3733],[266,-5.982],[267,-4.5272],[268,-3.7054],[269,-2.1446],[270,-0.602],[271,0],[359,0],[360,-0.6391],[361,-4.585],[362,-6.8045],[363,-7.8033],[364,-8.1362],[440,-8.1362]]
       }
       thumb_02_l_ctrl.Rotation.Y {
        type=MovieSceneScriptingFloatChannel
        default=0
        thinned_from=94
        keys [[0,-7.0264],[264,-7.0264],[265,-5.5392],[266,-3.0714],[267,-1.0101],[268,-0.0147],[269,0.7181],[270,1.4423],[271,0],[359,0],[360,1.5313],[361,-2.9727],[362,-5.5063],[363,-6.6464],[364,-7.0264],[440,-7.0264]]
       }
       thumb_02_l_ctrl.Rotation.Z {
        type=MovieSceneScriptingFloatChannel
        default=25.1558
        thinned_from=94
        keys [[0,33.5435],[264,33.5435],[265,30.8353],[266,25.7505],[267,20.1244],[268,16.8459],[269,9.9286],[270,3.0917],[271,-0.6041],[272,1.0242],[273,5.2577],[274,12.1048],[279,23.5011],[283,27.6054],[287,28.6314],[292,28.8879],[299,28.9521],[352,28.9734],[354,28.2084],[355,27.3915],[357,24.3021],[358,19.7446],[359,18.3669],[360,22.9602],[361,28.5303],[362,31.6635],[363,30.6464],[364,30.5435],[365,30.0435],[366,30.481],[367,31.2466],[369,33.1153],[370,33.387],[371,33.5435],[440,33.5435]]
       }
       thumb_02_l_ctrl.Scale.X {
        type=MovieSceneScriptingFloatChannel
        default=1
        thinned_from=94
        keys [[0,0.9004],[264,0.9004],[265,0.9129],[267,0.9555],[268,0.9666],[270,1],[360,1],[361,0.9476],[362,0.9181],[363,0.9048],[364,0.9004],[440,0.9004]]
       }
       thumb_02_l_ctrl.Scale.Y {type=MovieSceneScriptingFloatChannel, default=1, const [1,94]}
       thumb_02_l_ctrl.Scale.Z {type=MovieSceneScriptingFloatChannel, default=1, thinned_from=94, keys [[0,1],[268,1],[269,0.9846],[270,1],[440,1]]}
       thumb_03_l_ctrl.Location.X {type=MovieSceneScriptingFloatChannel, default=0, const [0,94]}
       thumb_03_l_ctrl.Location.Y {type=MovieSceneScriptingFloatChannel, default=0, const [0,94]}
       thumb_03_l_ctrl.Location.Z {type=MovieSceneScriptingFloatChannel, default=0, const [0,94]}
       thumb_03_l_ctrl.Rotation.X {type=MovieSceneScriptingFloatChannel, default=0, thinned_from=94, keys [[0,0],[440,0]]}
       thumb_03_l_ctrl.Rotation.Y {
        type=MovieSceneScriptingFloatChannel
        default=0
        thinned_from=94
        keys [[0,-9.2373],[264,-9.2373],[265,-8.0826],[266,-6.062],[267,-4.1299],[268,-3.0975],[269,-1.5397],[270,0],[360,0],[361,-4.8617],[362,-7.5965],[363,-8.8271],[364,-9.2373],[440,-9.2373]]
       }
       thumb_03_l_ctrl.Rotation.Z {
        type=MovieSceneScriptingFloatChannel
        default=36.0779
        thinned_from=94
        keys [[0,-19.5],[264,-19.5],[265,-18.6006],[266,-15.6775],[267,-9.8968],[268,-5.7454],[269,8.2159],[270,22.0149],[271,36.4513],[272,36.3768],[273,36.183],[274,35.8698],[279,35.3483],[283,35.1606],[352,35.098],[354,35.3003],[355,35.5178],[357,36.2964],[358,37.3632],[359,37.6688],[360,22.1195],[361,0.2145],[362,-12.1071],[363,-20.0788],[364,-22.5],[365,-23],[366,-22.5625],[367,-21.7969],[369,-19.9281],[370,-19.6565],[371,-19.5],[440,-19.5]]
       }
       thumb_03_l_ctrl.Scale.X {
        type=MovieSceneScriptingFloatChannel
        default=1
        thinned_from=94
        keys [[0,0.9207],[264,0.9207],[265,0.9121],[266,0.9012],[267,0.8999],[268,0.9023],[270,0.9562],[271,1],[359,1],[360,0.9535],[361,0.9363],[362,0.9265],[363,0.9222],[440,0.9207]]
       }
       thumb_03_l_ctrl.Scale.Y {
        type=MovieSceneScriptingFloatChannel
        default=1
        thinned_from=94
        keys [[0,0.7912],[264,0.7912],[265,0.7927],[266,0.8007],[267,0.8205],[268,0.8354],[270,0.9417],[271,1],[359,1],[360,0.9381],[361,0.8608],[362,0.8173],[363,0.7977],[364,0.7912],[440,0.7912]]
       }
       thumb_03_l_ctrl.Scale.Z {
        type=MovieSceneScriptingFloatChannel
        default=1
        thinned_from=94
        keys [[0,0.7389],[264,0.7389],[265,0.7469],[266,0.7664],[267,0.7971],[268,0.8178],[269,0.8656],[270,0.9417],[271,1],[359,1],[360,0.9381],[361,0.8332],[362,0.7743],[363,0.7477],[364,0.7389],[440,0.7389]]
       }
       index_metacarpal_l_ctrl.Location.X {type=MovieSceneScriptingFloatChannel, default=0, const [0,94]}
       index_metacarpal_l_ctrl.Location.Y {type=MovieSceneScriptingFloatChannel, default=0, const [0,94]}
       index_metacarpal_l_ctrl.Location.Z {type=MovieSceneScriptingFloatChannel, default=0, const [0,94]}
       index_metacarpal_l_ctrl.Rotation.X {
        type=MovieSceneScriptingFloatChannel
        default=0.005
        thinned_from=94
        keys [[0,0],[262,0],[263,0.001],[264,0.0031],[265,0.0044],[266,0.0053],[267,0.0076],[268,0.0116],[270,0.0201],[271,0.0263],[272,0.0249],[273,0.0211],[274,0.015],[279,0.0049],[283,0.0012],[287,0.0003],[292,0.0001],[352,0],[354,0.0011],[355,0.0023],[356,0.0046],[357,0.0065],[358,0.0123],[359,0.0139],[360,0.0108],[361,0.003],[363,0.001],[364,0.0004],[371,0],[440,0]]
       }
       index_metacarpal_l_ctrl.Rotation.Y {
        type=MovieSceneScriptingFloatChannel
        default=-4.9899
        thinned_from=94
        keys [[0,3.5],[262,3.5],[263,3.3904],[264,3.1491],[265,3.0036],[266,2.8998],[267,2.6341],[268,2.173],[269,1.689],[270,1.2106],[271,0.5],[272,-2.3444],[273,-9.7398],[274,-15.1889],[275,-17],[279,-10.2728],[283,-7.2455],[287,-6.4887],[292,-6.2995],[299,-6.2522],[352,-6.2364],[354,-5.9676],[355,-5.6748],[356,-5.1027],[357,-4.6166],[358,-3.1915],[359,-2.789],[360,-0.6319],[361,2.3413],[362,2.766],[363,3.1266],[364,3.36],[365,3.4488],[366,3.4814],[367,3.4943],[440,3.5]]
       }
       index_metacarpal_l_ctrl.Rotation.Z {
        type=MovieSceneScriptingFloatChannel
        default=8.5913
        thinned_from=94
        keys [[0,0.5],[262,0.5],[263,0.1912],[264,-0.4883],[265,-0.8983],[266,-1.1907],[267,-1.9388],[268,-3.2376],[269,-4.601],[270,-5.9486],[271,-7.95],[272,-6.1625],[273,-1.5149],[274,6.0018],[275,8.5],[279,8.9632],[283,9.1716],[287,9.2237],[352,9.2411],[354,8.9143],[355,8.5305],[356,7.6735],[357,6.5289],[358,3.1741],[359,2.0766],[360,1.2536],[361,0.7113],[362,0.6339],[363,0.5681],[364,0.5255],[365,0.5093],[366,0.5034],[367,0.501],[368,0.5],[440,0.5]]
       }
       index_metacarpal_l_ctrl.Scale.X {type=MovieSceneScriptingFloatChannel, default=1, const [1,94]}
       index_metacarpal_l_ctrl.Scale.Y {type=MovieSceneScriptingFloatChannel, default=1, const [1,94]}
       index_metacarpal_l_ctrl.Scale.Z {type=MovieSceneScriptingFloatChannel, default=1, thinned_from=94, keys [[0,1],[268,1],[269,0.9925],[270,1],[440,1]]}
       index_01_l_ctrl.Location.X {type=MovieSceneScriptingFloatChannel, default=0, const [0,94]}
       index_01_l_ctrl.Location.Y {type=MovieSceneScriptingFloatChannel, default=0, const [0,94]}
       index_01_l_ctrl.Location.Z {type=MovieSceneScriptingFloatChannel, default=0, const [0,94]}
       index_01_l_ctrl.Rotation.X {
        type=MovieSceneScriptingFloatChannel
        default=2.0391
        thinned_from=94
        keys [[0,8.5],[262,8.5],[263,8.1872],[264,7.4992],[265,7.084],[266,6.7879],[267,6.0303],[268,4.715],[269,3.3343],[270,1.9697],[271,-0.0571],[272,-0.1767],[273,-0.4879],[274,-0.9911],[279,-1.8287],[283,-2.1304],[287,-2.2058],[292,-2.2247],[299,-2.2294],[352,-2.231],[354,-1.9764],[355,-1.7118],[356,-1.2428],[357,-1.0025],[358,-0.2981],[359,-0.1561],[360,-0.0779],[361,3.9744],[362,-7.6416],[363,0.547],[364,0.0496],[365,4.2801],[366,8.4615],[368,8.5],[440,8.5]]
       }
       index_01_l_ctrl.Rotation.Y {
        type=MovieSceneScriptingFloatChannel
        default=-9.092
        thinned_from=94
        keys [[0,0],[262,0],[263,-0.1643],[264,-0.5259],[265,-0.7441],[266,-0.8996],[267,-1.2977],[268,-1.9888],[269,-2.7143],[270,-3.4314],[271,-4.4964],[272,-4.8008],[273,-5.5924],[274,-6.8726],[279,-9.0034],[283,-9.7707],[287,-9.9626],[292,-10.0105],[352,-10.0265],[354,-9.8725],[355,-9.7042],[357,-9.036],[358,-8.0476],[359,-7.7483],[360,-5.7179],[361,-1.6034],[362,-1.0157],[363,-0.5167],[364,-0.1938],[365,-0.1092],[366,-0.0257],[367,-0.0079],[368,0],[440,0]]
       }
       index_01_l_ctrl.Rotation.Z {
        type=MovieSceneScriptingFloatChannel
        default=-3.463
        thinned_from=94
        keys [[0,-56.5],[262,-56.5],[263,-54.2317],[264,-49.2413],[265,-46.2302],[266,-44.0825],[267,-38.5878],[268,-29.0489],[269,-19.0351],[270,-9.1377],[271,5.5616],[272,9.85],[273,21],[274,18],[275,12.929],[279,23.5899],[283,28.3873],[287,29.5867],[292,29.8865],[299,29.9615],[352,29.9864],[354,29.2798],[355,28.4664],[356,26.7088],[357,24.3568],[358,17.4632],[359,33],[360,12.3801],[361,-37.1841],[362,-44.2647],[363,-50.275],[364,-54.1656],[366,-56.1907],[367,-56.405],[440,-56.5]]
       }
       index_01_l_ctrl.Scale.X {
        type=MovieSceneScriptingFloatChannel
        default=1
        thinned_from=94
        keys [[0,1.0131],[263,1.0127],[270,1.0031],[271,1],[360,1],[361,1.0095],[371,1.0131],[440,1.0131]]
       }
       index_01_l_ctrl.Scale.Y {type=MovieSceneScriptingFloatChannel, default=1, const [1,94]}
       index_01_l_ctrl.Scale.Z {type=MovieSceneScriptingFloatChannel, default=1, thinned_from=94, keys [[0,1],[268,1],[269,0.9849],[270,1],[440,1]]}
       index_02_l_ctrl.Location.X {type=MovieSceneScriptingFloatChannel, default=0, const [0,94]}
       index_02_l_ctrl.Location.Y {type=MovieSceneScriptingFloatChannel, default=0, const [0,94]}
       index_02_l_ctrl.Location.Z {type=MovieSceneScriptingFloatChannel, default=0, const [0,94]}
       index_02_l_ctrl.Rotation.X {type=MovieSceneScriptingFloatChannel, default=0, thinned_from=94, keys [[0,0],[440,0]]}
       index_02_l_ctrl.Rotation.Y {type=MovieSceneScriptingFloatChannel, default=0, const [0,94]}
       index_02_l_ctrl.Rotation.Z {
        type=MovieSceneScriptingFloatChannel
        default=31.2886
        thinned_from=94
        keys [[0,-57],[262,-57],[264,-58.5],[265,-54.6786],[266,-51.953],[267,-44.9796],[268,-32.8739],[269,-14.656],[270,3.3501],[271,30.0923],[272,30.7943],[273,32.6194],[274,35.5713],[279,40.4843],[283,42.2537],[287,42.696],[292,42.8066],[352,42.8435],[354,40.7635],[355,38.5393],[356,34.3472],[357,31.896],[358,24.7114],[359,23.0816],[360,24.0464],[361,-47.5],[362,-52],[363,-54.4561],[364,-56.0461],[365,-56.6513],[366,-56.8736],[440,-57]]
       }
       index_02_l_ctrl.Scale.X {type=MovieSceneScriptingFloatChannel, default=1, const [1,94]}
       index_02_l_ctrl.Scale.Y {type=MovieSceneScriptingFloatChannel, default=1, const [1,94]}
       index_02_l_ctrl.Scale.Z {type=MovieSceneScriptingFloatChannel, default=1, thinned_from=94, keys [[0,1],[268,1],[269,0.9849],[270,1],[440,1]]}
       index_03_l_ctrl.Location.X {type=MovieSceneScriptingFloatChannel, default=0, const [0,94]}
       index_03_l_ctrl.Location.Y {type=MovieSceneScriptingFloatChannel, default=0, const [0,94]}
       index_03_l_ctrl.Location.Z {type=MovieSceneScriptingFloatChannel, default=0, const [0,94]}
       index_03_l_ctrl.Rotation.X {type=MovieSceneScriptingFloatChannel, default=0, thinned_from=94, keys [[0,0],[440,0]]}
       index_03_l_ctrl.Rotation.Y {type=MovieSceneScriptingFloatChannel, default=0, thinned_from=94, keys [[0,0],[440,0]]}
       index_03_l_ctrl.Rotation.Z {
        type=MovieSceneScriptingFloatChannel
        default=37.1103
        thinned_from=94
        keys [[0,-23],[262,-23],[263,-25.2524],[264,-30.2076],[265,-34.3008],[266,-37.2203],[267,-44.6895],[268,-57.6562],[269,-33.4664],[270,-9.5579],[271,25.9506],[272,26.6841],[273,28.5912],[274,31.6757],[279,36.8096],[283,38.6585],[287,39.1207],[292,39.2363],[352,39.2748],[354,38.808],[355,38.2997],[356,37.3063],[357,36.4621],[358,33.9877],[359,33.2887],[360,22.6378],[361,17],[362,-35.5],[363,-29.614],[364,-25.4803],[365,-23.9065],[366,-23.3287],[367,-23.1009],[368,-23],[440,-23]]
       }
       index_03_l_ctrl.Scale.X {type=MovieSceneScriptingFloatChannel, default=1, const [1,94]}
       index_03_l_ctrl.Scale.Y {type=MovieSceneScriptingFloatChannel, default=1, const [1,94]}
       index_03_l_ctrl.Scale.Z {type=MovieSceneScriptingFloatChannel, default=1, thinned_from=94, keys [[0,1],[268,1],[269,0.9848],[270,1],[440,1]]}
       middle_metacarpal_l_ctrl.Location.X {type=MovieSceneScriptingFloatChannel, default=0, const [0,94]}
       middle_metacarpal_l_ctrl.Location.Y {type=MovieSceneScriptingFloatChannel, default=0, const [0,94]}
       middle_metacarpal_l_ctrl.Location.Z {type=MovieSceneScriptingFloatChannel, default=0, const [0,94]}
       middle_metacarpal_l_ctrl.Rotation.X {
        type=MovieSceneScriptingFloatChannel
        default=-0.3684
        thinned_from=94
        keys [[0,0],[262,0],[263,-0.0042],[264,-0.0135],[265,-0.0191],[266,-0.0231],[267,-0.0333],[268,-0.0511],[269,-0.0697],[270,-0.0881],[271,-0.1154],[272,-0.1327],[273,-0.1774],[274,-0.2498],[279,-0.3703],[283,-0.4136],[287,-0.4245],[292,-0.4272],[352,-0.4281],[354,-0.4152],[355,-0.4012],[356,-0.3738],[357,-0.3505],[358,-0.2823],[359,-0.2631],[360,-0.1916],[361,-0.0537],[362,-0.034],[363,-0.0173],[364,-0.0065],[365,-0.0024],[366,-0.0009],[367,-0.0003],[368,0],[440,0]]
       }
       middle_metacarpal_l_ctrl.Rotation.Y {
        type=MovieSceneScriptingFloatChannel
        default=-7.6486
        thinned_from=94
        keys [[0,0],[262,0],[263,-0.0121],[264,-0.0387],[265,-0.0547],[266,-0.0662],[267,-0.0954],[268,-0.1463],[269,-0.1996],[270,-0.2524],[271,-0.3307],[272,-0.9316],[273,-2.4939],[274,-4.4471],[279,-7.698],[283,-8.8688],[287,-9.1615],[292,-9.2347],[309,-9.2572],[352,-9.2591],[354,-8.9118],[355,-8.5336],[356,-7.7944],[357,-7.1663],[358,-5.3253],[359,-4.8052],[360,-3.4543],[361,-0.9687],[362,-0.6136],[363,-0.3122],[364,-0.1171],[365,-0.0428],[366,-0.0155],[367,-0.0048],[368,0],[440,0]]
       }
       middle_metacarpal_l_ctrl.Rotation.Z {
        type=MovieSceneScriptingFloatChannel
        default=27.937
        thinned_from=94
        keys [[0,0],[262,0],[263,-0.0381],[264,-0.122],[265,-0.1726],[266,-0.2087],[267,-0.301],[268,-0.4613],[269,-0.6296],[270,-0.796],[271,-1.043],[272,0.9804],[273,6.2413],[274,14.7499],[279,28.9117],[283,34.0119],[287,35.287],[292,35.6058],[299,35.6855],[352,35.712],[354,34.0635],[355,32.2756],[356,28.8087],[357,25.9685],[358,17.6437],[359,15.3301],[360,10.8411],[361,3.0401],[362,1.9257],[363,0.9797],[364,0.3674],[365,0.1343],[366,0.0487],[367,0.0149],[368,0],[440,0]]
       }
       middle_metacarpal_l_ctrl.Scale.X {type=MovieSceneScriptingFloatChannel, default=1, const [1,94]}
       middle_metacarpal_l_ctrl.Scale.Y {type=MovieSceneScriptingFloatChannel, default=1, const [1,94]}
       middle_metacarpal_l_ctrl.Scale.Z {type=MovieSceneScriptingFloatChannel, default=1, thinned_from=94, keys [[0,1],[268,1],[269,0.9925],[270,1],[440,1]]}
       middle_01_l_ctrl.Location.X {type=MovieSceneScriptingFloatChannel, default=0, thinned_from=94, keys [[0,0],[440,0]]}
       middle_01_l_ctrl.Location.Y {type=MovieSceneScriptingFloatChannel, default=0, thinned_from=94, keys [[0,0],[440,0]]}
       middle_01_l_ctrl.Location.Z {type=MovieSceneScriptingFloatChannel, default=0, thinned_from=94, keys [[0,0],[440,0]]}
       middle_01_l_ctrl.Rotation.X {
        type=MovieSceneScriptingFloatChannel
        default=-0.2678
        thinned_from=94
        keys [[0,0],[262,0],[263,-0.0513],[264,-0.1641],[265,-0.2321],[266,-0.2807],[267,-0.4049],[268,-0.6205],[269,-0.8469],[270,-1.0706],[271,-1.4029],[272,-1.3257],[273,-1.1249],[274,-0.8001],[279,-0.2596],[283,-0.0649],[287,-0.0162],[292,-0.0041],[299,-0.001],[309,-0.0003],[352,0],[354,-0.0578],[355,-0.1206],[356,-0.2435],[357,-0.348],[358,-0.6541],[359,-0.7406],[360,-0.578],[361,-0.1621],[362,-0.1027],[363,-0.0522],[364,-0.0196],[365,-0.0072],[366,-0.0026],[367,-0.0008],[368,0],[440,0]]
       }
       middle_01_l_ctrl.Rotation.Y {
        type=MovieSceneScriptingFloatChannel
        default=-4.223
        thinned_from=94
        keys [[0,-6],[262,-6],[263,-6.2352],[264,-6.7526],[265,-7.0648],[266,-7.2875],[267,-7.8572],[268,-8.8463],[270,-10.9108],[271,-12.4349],[272,-10],[273,-13.8295],[274,-15.4586],[275,-16],[279,-8.004],[283,-4.4059],[287,-3.5063],[292,-3.2814],[299,-3.2252],[309,-3.2122],[352,-3.2065],[354,-3.4257],[355,-3.6644],[356,-4.131],[357,-4.5275],[358,-5.6896],[359,-6.0178],[360,-4.5946],[361,-5.6059],[362,-5.7503],[363,-5.873],[364,-5.9524],[365,-5.9826],[371,-6],[440,-6]]
       }
       middle_01_l_ctrl.Rotation.Z {
        type=MovieSceneScriptingFloatChannel
        default=13.5746
        thinned_from=94
        keys [[0,-62.5],[262,-62.5],[263,-60.7025],[264,-56.7482],[265,-54.3621],[266,-52.6603],[267,-48.3062],[268,-40.7475],[269,-32.8124],[270,-24.9697],[271,-13.3218],[272,-11.5657],[273,-7],[274,0.977],[275,3.5],[279,16.8206],[283,22.8149],[287,24.3134],[292,24.6881],[299,24.7817],[352,24.813],[354,22.0942],[355,19.055],[356,12.8142],[357,10.0001],[358,1.7519],[359,-0.8539],[360,-16.1896],[361,-49.5133],[362,-54.2738],[363,-58.3147],[364,-60.9305],[365,-61.9264],[366,-62.292],[368,-62.5],[440,-62.5]]
       }
       middle_01_l_ctrl.Scale.X {
        type=MovieSceneScriptingFloatChannel
        default=1
        thinned_from=94
        keys [[0,0.8301],[262,0.8301],[263,0.8363],[264,0.85],[265,0.8582],[266,0.8641],[267,0.8791],[270,0.9598],[271,1],[360,1],[361,0.8777],[362,0.8603],[363,0.8454],[364,0.8358],[365,0.8322],[371,0.8301],[440,0.8301]]
       }
       middle_01_l_ctrl.Scale.Y {type=MovieSceneScriptingFloatChannel, default=1, const [1,94]}
       middle_01_l_ctrl.Scale.Z {type=MovieSceneScriptingFloatChannel, default=1, thinned_from=94, keys [[0,1],[268,1],[269,0.9849],[270,1],[440,1]]}
       middle_02_l_ctrl.Location.X {type=MovieSceneScriptingFloatChannel, default=0, thinned_from=94, keys [[0,0],[440,0]]}
       middle_02_l_ctrl.Location.Y {type=MovieSceneScriptingFloatChannel, default=0, thinned_from=94, keys [[0,0],[440,0]]}
       middle_02_l_ctrl.Location.Z {type=MovieSceneScriptingFloatChannel, default=0, thinned_from=94, keys [[0,0],[440,0]]}
       middle_02_l_ctrl.Rotation.X {type=MovieSceneScriptingFloatChannel, default=0, thinned_from=94, keys [[0,0],[440,0]]}
       middle_02_l_ctrl.Rotation.Y {type=MovieSceneScriptingFloatChannel, default=0, thinned_from=94, keys [[0,0],[440,0]]}
       middle_02_l_ctrl.Rotation.Z {
        type=MovieSceneScriptingFloatChannel
        default=40.7734
        thinned_from=94
        keys [[0,-59.5],[262,-59.5],[264,-61.8232],[265,-58.3432],[266,-55.8612],[267,-49.5109],[268,-38.4869],[269,-21.4046],[270,-4.5209],[271,20.5544],[272,22.655],[273,28.1164],[274,36.9492],[279,51.6508],[283,56.9454],[287,58.269],[292,58.5999],[352,58.7102],[354,54.8446],[355,50.698],[356,42.8342],[357,38.1293],[358,24.3394],[359,21.1553],[360,1.384],[361,-47],[362,-48.6851],[363,-53.9977],[364,-57.4366],[365,-58.7458],[366,-59.2266],[368,-59.5],[440,-59.5]]
       }
       middle_02_l_ctrl.Scale.X {type=MovieSceneScriptingFloatChannel, default=1, const [1,94]}
       middle_02_l_ctrl.Scale.Y {type=MovieSceneScriptingFloatChannel, default=1, const [1,94]}
       middle_02_l_ctrl.Scale.Z {type=MovieSceneScriptingFloatChannel, default=1, thinned_from=94, keys [[0,1],[268,1],[269,0.9849],[270,1],[440,1]]}
       middle_03_l_ctrl.Location.X {type=MovieSceneScriptingFloatChannel, default=0, thinned_from=94, keys [[0,0],[440,0]]}
       middle_03_l_ctrl.Location.Y {type=MovieSceneScriptingFloatChannel, default=0, thinned_from=94, keys [[0,0],[440,0]]}
       middle_03_l_ctrl.Location.Z {type=MovieSceneScriptingFloatChannel, default=0, thinned_from=94, keys [[0,0],[440,0]]}
       middle_03_l_ctrl.Rotation.X {type=MovieSceneScriptingFloatChannel, default=0, thinned_from=94, keys [[0,0],[440,0]]}
       middle_03_l_ctrl.Rotation.Y {type=MovieSceneScriptingFloatChannel, default=0, thinned_from=94, keys [[0,0],[440,0]]}
       middle_03_l_ctrl.Rotation.Z {
        type=MovieSceneScriptingFloatChannel
        default=35.8297
        thinned_from=94
        keys [[0,-26.5],[262,-26.5],[263,-28.5955],[264,-33.2055],[265,-37.145],[266,-39.9549],[267,-47.1437],[268,-59.6236],[269,-34.9686],[270,-10.6002],[271,25.5912],[272,26.262],[273,28.0062],[274,30.8271],[279,35.5223],[283,37.2132],[287,37.6359],[292,37.7416],[352,37.7768],[354,37.3569],[355,36.8996],[356,36.006],[357,35.2465],[358,33.0206],[359,32.3917],[360,17.4104],[361,-14.1863],[362,-18.7001],[363,-22.5316],[364,-25.0119],[365,-25.9561],[366,-26.3028],[367,-26.4395],[368,-26.5],[440,-26.5]]
       }
       middle_03_l_ctrl.Scale.X {type=MovieSceneScriptingFloatChannel, default=1, const [1,94]}
       middle_03_l_ctrl.Scale.Y {type=MovieSceneScriptingFloatChannel, default=1, const [1,94]}
       middle_03_l_ctrl.Scale.Z {type=MovieSceneScriptingFloatChannel, default=1, thinned_from=94, keys [[0,1],[268,1],[269,0.9848],[270,1],[440,1]]}
       ring_metacarpal_l_ctrl.Location.X {type=MovieSceneScriptingFloatChannel, default=0, const [0,94]}
       ring_metacarpal_l_ctrl.Location.Y {type=MovieSceneScriptingFloatChannel, default=0, const [0,94]}
       ring_metacarpal_l_ctrl.Location.Z {type=MovieSceneScriptingFloatChannel, default=0, const [0,94]}
       ring_metacarpal_l_ctrl.Rotation.X {
        type=MovieSceneScriptingFloatChannel
        default=-1.5701
        thinned_from=94
        keys [[0,0],[262,0],[263,0.0006],[264,0.0018],[265,0.0026],[266,0.0031],[267,0.0045],[270,0.0118],[271,0.0155],[272,0.0147],[273,0.0124],[274,0.0088],[279,0.0029],[283,0.0007],[287,0.0002],[352,0],[354,-0.2677],[355,-0.5451],[356,-1.0333],[357,-1.2381],[358,-1.8384],[359,-1.9324],[360,-1.4722],[361,-0.4129],[362,-0.2615],[363,-0.1331],[364,-0.0499],[365,-0.0182],[366,-0.0066],[367,-0.002],[368,0],[440,0]]
       }
       ring_metacarpal_l_ctrl.Rotation.Y {
        type=MovieSceneScriptingFloatChannel
        default=-4.3978
        thinned_from=94
        keys [[0,-3.5],[262,-3.5],[263,-3.5914],[264,-3.7924],[265,-3.9137],[266,-4.0002],[267,-4.2216],[268,-4.6058],[270,-5.4079],[271,-6],[272,-5.7219],[273,-4.9989],[274,-5.375],[275,-5.5],[279,-4.4324],[283,-3.952],[287,-3.8318],[292,-3.8018],[309,-3.7926],[352,-3.7918],[354,-3.9315],[355,-4.0785],[356,-4.3457],[357,-4.3927],[358,-4.5305],[359,-4.5046],[360,-8],[361,-4.7619],[362,-4.2993],[363,-3.9067],[364,-3.6525],[365,-3.5557],[366,-3.5202],[367,-3.5062],[440,-3.5]]
       }
       ring_metacarpal_l_ctrl.Rotation.Z {
        type=MovieSceneScriptingFloatChannel
        default=18.4529
        thinned_from=94
        keys [[0,-1],[262,-1],[263,-0.8718],[264,-0.5896],[265,-0.4194],[266,-0.298],[267,0.0127],[268,0.5519],[269,1.118],[270,1.6776],[271,2.5086],[272,3.5934],[273,6.414],[274,10.9757],[279,18.5683],[283,21.3027],[287,21.9863],[292,22.1572],[309,22.2098],[352,22.2141],[354,21.4029],[355,20.5196],[356,18.7934],[357,17.3265],[358,13.0267],[359,11.812],[360,8.6345],[361,1.7018],[362,0.7114],[363,-0.1293],[364,-0.6735],[365,-0.8807],[366,-0.9567],[367,-0.9867],[368,-1],[440,-1]]
       }
       ring_metacarpal_l_ctrl.Scale.X {type=MovieSceneScriptingFloatChannel, default=1, const [1,94]}
       ring_metacarpal_l_ctrl.Scale.Y {type=MovieSceneScriptingFloatChannel, default=1, const [1,94]}
       ring_metacarpal_l_ctrl.Scale.Z {type=MovieSceneScriptingFloatChannel, default=1, thinned_from=94, keys [[0,1],[268,1],[269,0.9925],[270,1],[440,1]]}
       ring_01_l_ctrl.Location.X {type=MovieSceneScriptingFloatChannel, default=0, const [0,94]}
       ring_01_l_ctrl.Location.Y {type=MovieSceneScriptingFloatChannel, default=0, const [0,94]}
       ring_01_l_ctrl.Location.Z {type=MovieSceneScriptingFloatChannel, default=0, const [0,94]}
       ring_01_l_ctrl.Rotation.X {
        type=MovieSceneScriptingFloatChannel
        default=-1.4465
        thinned_from=94
        keys [[0,-0.3564],[262,-0.3564],[263,-0.6204],[264,-1.2011],[265,-1.5515],[266,-1.8014],[267,-2.4408],[268,-3.5508],[269,-4.7161],[270,-5.8678],[271,-7.5783],[272,-7.1611],[273,-6.0764],[274,-4.322],[279,-1.4021],[283,-0.3505],[287,-0.0876],[292,-0.0219],[299,-0.0055],[309,-0.0017],[327,-0.0004],[337,0],[352,0],[354,-0.312],[355,-0.6517],[356,-1.3155],[357,-1.8797],[358,-3.5333],[359,-4.0004],[360,-14.5],[361,-4.3227],[362,-2.8688],[363,-1.6347],[364,-0.8358],[365,-0.5316],[366,-0.42],[367,-0.3759],[368,-0.3564],[440,-0.3564]]
       }
       ring_01_l_ctrl.Rotation.Y {
        type=MovieSceneScriptingFloatChannel
        default=-1.1632
        thinned_from=94
        keys [[0,-6],[262,-6],[263,-5.9615],[264,-5.8769],[265,-5.8258],[266,-5.7893],[267,-5.6961],[270,-5.1965],[271,-4.9472],[272,-5.0616],[273,-5.359],[274,-5.8401],[275,-6],[279,-2.6452],[283,-1.1356],[287,-0.7582],[292,-0.6638],[299,-0.6402],[309,-0.6348],[352,-0.6324],[354,-0.7468],[355,-0.8715],[356,-1.1151],[357,-1.3221],[358,-1.9289],[359,-2.1003],[360,-1.6192],[361,-4.7715],[362,-5.2218],[363,-5.6041],[364,-5.8515],[365,-5.9457],[366,-5.9803],[368,-6],[440,-6]]
       }
       ring_01_l_ctrl.Rotation.Z {
        type=MovieSceneScriptingFloatChannel
        default=20.169
        thinned_from=94
        keys [[0,-59.6733],[262,-59.6733],[263,-58.0331],[264,-54.4249],[265,-52.2476],[266,-50.6948],[267,-46.7218],[268,-39.8247],[269,-32.5842],[270,-25.4279],[271,-14.7995],[272,-12.5937],[273,-6.8585],[274,2.4171],[275,5.5],[279,23.7942],[283,32.0267],[287,34.0848],[292,34.5993],[299,34.7279],[352,34.7708],[354,31.7163],[355,28.4274],[356,22.1414],[357,17.7787],[358,4.9914],[359,1.6375],[360,-16.4157],[361,-47.5426],[362,-51.9894],[363,-55.7639],[364,-58.2073],[365,-59.1374],[366,-59.479],[368,-59.6733],[440,-59.6733]]
       }
       ring_01_l_ctrl.Scale.X {
        type=MovieSceneScriptingFloatChannel
        default=1
        thinned_from=94
        keys [[0,0.7763],[262,0.7763],[263,0.7845],[264,0.8024],[265,0.8133],[266,0.821],[267,0.8408],[270,0.947],[271,1],[360,1],[361,0.839],[362,0.816],[363,0.7965],[364,0.7839],[365,0.779],[371,0.7763],[440,0.7763]]
       }
       ring_01_l_ctrl.Scale.Y {type=MovieSceneScriptingFloatChannel, default=1, const [1,94]}
       ring_01_l_ctrl.Scale.Z {type=MovieSceneScriptingFloatChannel, default=1, thinned_from=94, keys [[0,1],[268,1],[269,0.9849],[270,1],[440,1]]}
       ring_02_l_ctrl.Location.X {type=MovieSceneScriptingFloatChannel, default=0, const [0,94]}
       ring_02_l_ctrl.Location.Y {type=MovieSceneScriptingFloatChannel, default=0, const [0,94]}
       ring_02_l_ctrl.Location.Z {type=MovieSceneScriptingFloatChannel, default=0, const [0,94]}
       ring_02_l_ctrl.Rotation.X {type=MovieSceneScriptingFloatChannel, default=0, thinned_from=94, keys [[0,0],[440,0]]}
       ring_02_l_ctrl.Rotation.Y {type=MovieSceneScriptingFloatChannel, default=0, thinned_from=94, keys [[0,0],[440,0]]}
       ring_02_l_ctrl.Rotation.Z {
        type=MovieSceneScriptingFloatChannel
        default=29.8193
        thinned_from=94
        keys [[0,-57.5711],[262,-57.5711],[264,-61.4873],[265,-58.6682],[266,-56.6576],[267,-51.5133],[268,-42.5829],[269,-27.6984],[270,-12.9869],[271,8.8623],[272,11.0593],[273,16.7716],[274,26.0102],[279,41.3871],[283,46.9249],[287,48.3093],[292,48.6555],[309,48.762],[352,48.7708],[354,44.7265],[355,40.3909],[356,32.1791],[357,27.2621],[358,12.85],[359,9.5201],[360,-11.3613],[361,-44.6126],[362,-49.3628],[363,-53.3949],[364,-56.005],[365,-56.9987],[366,-57.3635],[368,-57.5711],[440,-57.5711]]
       }
       ring_02_l_ctrl.Scale.X {type=MovieSceneScriptingFloatChannel, default=1, const [1,94]}
       ring_02_l_ctrl.Scale.Y {type=MovieSceneScriptingFloatChannel, default=1, const [1,94]}
       ring_02_l_ctrl.Scale.Z {type=MovieSceneScriptingFloatChannel, default=1, thinned_from=94, keys [[0,1],[268,1],[269,0.9849],[270,1],[440,1]]}
       ring_03_l_ctrl.Location.X {type=MovieSceneScriptingFloatChannel, default=0, const [0,94]}
       ring_03_l_ctrl.Location.Y {type=MovieSceneScriptingFloatChannel, default=0, const [0,94]}
       ring_03_l_ctrl.Location.Z {type=MovieSceneScriptingFloatChannel, default=0, const [0,94]}
       ring_03_l_ctrl.Rotation.X {type=MovieSceneScriptingFloatChannel, default=0, thinned_from=94, keys [[0,0],[440,0]]}
       ring_03_l_ctrl.Rotation.Y {type=MovieSceneScriptingFloatChannel, default=0, thinned_from=94, keys [[0,0],[440,0]]}
       ring_03_l_ctrl.Rotation.Z {
        type=MovieSceneScriptingFloatChannel
        default=27.2775
        thinned_from=94
        keys [[0,-24],[262,-24],[263,-26.3403],[264,-31.4889],[265,-35.6682],[266,-38.6491],[267,-46.2755],[268,-59.5149],[269,-35.5858],[270,-11.9348],[271,23.191],[272,23.4433],[273,24.0992],[274,25.16],[279,26.9256],[283,27.5614],[287,27.7204],[352,27.7734],[354,27.6664],[355,27.55],[357,27.129],[358,26.5621],[359,26.4019],[360,9.6879],[361,-14.553],[362,-18.016],[363,-20.9555],[364,-22.8583],[365,-23.5827],[366,-23.8487],[367,-23.9535],[440,-24]]
       }
       ring_03_l_ctrl.Scale.X {type=MovieSceneScriptingFloatChannel, default=1, const [1,94]}
       ring_03_l_ctrl.Scale.Y {type=MovieSceneScriptingFloatChannel, default=1, const [1,94]}
       ring_03_l_ctrl.Scale.Z {type=MovieSceneScriptingFloatChannel, default=1, thinned_from=94, keys [[0,1],[268,1],[269,0.9848],[270,1],[440,1]]}
       pinky_metacarpal_l_ctrl.Location.X {type=MovieSceneScriptingFloatChannel, default=0, const [0,94]}
       pinky_metacarpal_l_ctrl.Location.Y {type=MovieSceneScriptingFloatChannel, default=0, const [0,94]}
       pinky_metacarpal_l_ctrl.Location.Z {type=MovieSceneScriptingFloatChannel, default=0, const [0,94]}
       pinky_metacarpal_l_ctrl.Rotation.X {
        type=MovieSceneScriptingFloatChannel
        default=0.1465
        thinned_from=94
        keys [[0,0],[262,0],[263,-0.0073],[264,-0.0233],[265,-0.0329],[266,-0.0398],[267,-0.0574],[268,-0.0879],[269,-0.12],[270,-0.1517],[271,-0.1988],[272,-0.1753],[273,-0.1142],[274,-0.0154],[279,0.149],[283,0.2082],[287,0.223],[292,0.2267],[299,0.2276],[352,0.2279],[354,0.2104],[355,0.1912],[356,0.1539],[357,0.1221],[358,0.029],[359,0.0027],[360,-0.0052],[361,-0.0015],[363,-0.0005],[365,-0.0001],[440,0]]
       }
       pinky_metacarpal_l_ctrl.Rotation.Y {
        type=MovieSceneScriptingFloatChannel
        default=-2.1973
        thinned_from=94
        keys [[0,-3],[262,-3],[263,-3.0548],[264,-3.1754],[265,-3.2482],[266,-3.3001],[267,-3.4329],[268,-3.6635],[269,-5.5],[270,-6],[271,-10],[272,-8.5],[273,-7.5044],[274,-6],[275,-5.5],[279,-1.8165],[283,-0.1589],[287,0.2555],[292,0.3591],[299,0.385],[309,0.3909],[337,0.3936],[352,0.3936],[354,-0.4375],[355,-1.3253],[356,-2.9948],[357,-3.7489],[358,-5.9592],[359,-6.3445],[360,-13],[361,-5.8043],[362,-4.7763],[363,-3.9037],[364,-3.3389],[365,-3.1239],[366,-3.0449],[367,-3.0138],[368,-3],[440,-3]]
       }
       pinky_metacarpal_l_ctrl.Rotation.Z {
        type=MovieSceneScriptingFloatChannel
        default=2.8157
        thinned_from=94
        keys [[0,-5.5],[262,-5.5],[263,-5.457],[264,-5.3625],[265,-5.3054],[266,-5.2647],[267,-5.1606],[270,-4.6027],[271,-4.3242],[272,-3.8384],[273,-2.5754],[274,-0.5326],[275,0.1463],[279,2.8674],[283,4.0918],[287,4.398],[292,4.4745],[299,4.4936],[352,4.5],[354,4.1367],[355,3.7412],[356,2.9682],[357,2.3113],[358,0.3858],[359,-0.1581],[360,-2.0277],[361,-4.5263],[362,-4.8832],[363,-5.1862],[364,-5.3823],[365,-5.457],[366,-5.4844],[371,-5.5],[440,-5.5]]
       }
       pinky_metacarpal_l_ctrl.Scale.X {type=MovieSceneScriptingFloatChannel, default=1, const [1,94]}
       pinky_metacarpal_l_ctrl.Scale.Y {type=MovieSceneScriptingFloatChannel, default=1, const [1,94]}
       pinky_metacarpal_l_ctrl.Scale.Z {type=MovieSceneScriptingFloatChannel, default=1, thinned_from=94, keys [[0,1],[268,1],[269,0.9925],[270,1],[440,1]]}
       pinky_01_l_ctrl.Location.X {type=MovieSceneScriptingFloatChannel, default=0, const [0,94]}
       pinky_01_l_ctrl.Location.Y {type=MovieSceneScriptingFloatChannel, default=0, const [0,94]}
       pinky_01_l_ctrl.Location.Z {type=MovieSceneScriptingFloatChannel, default=0, const [0,94]}
       pinky_01_l_ctrl.Rotation.X {
        type=MovieSceneScriptingFloatChannel
        default=-3.533
        thinned_from=94
        keys [[0,-9.5],[262,-9.5],[263,-9.3758],[264,-9.1025],[265,-8.9376],[266,-8.82],[267,-8.5191],[268,-7.9967],[270,-6.9062],[271,-6.1013],[272,-5.7654],[273,-4.8921],[274,-3.4797],[279,-1.1288],[283,-0.2822],[287,-0.0706],[292,-0.0176],[299,-0.0044],[309,-0.0014],[327,-0.0003],[337,0],[352,0],[354,-2.2717],[355,-4.6899],[356,-9.2039],[357,-10.77],[358,-15.3603],[359,-15.8404],[360,-18.5],[361,-12.0238],[362,-11.0987],[363,-10.3134],[364,-9.805],[365,-9.6115],[366,-9.5404],[368,-9.5],[440,-9.5]]
       }
       pinky_01_l_ctrl.Rotation.Y {
        type=MovieSceneScriptingFloatChannel
        default=1.6964
        thinned_from=94
        keys [[0,-24.5],[262,-24.5],[263,-23.8361],[264,-22.3756],[265,-21.8468],[266,-21.4695],[267,-20.5044],[268,-18.829],[269,-15.2146],[270,-11.6423],[271,-6.3367],[272,-5.7902],[273,-4.3691],[274,-2.0708],[279,1.7545],[283,3.1321],[287,3.4766],[292,3.5627],[299,3.5842],[352,3.5914],[354,3.1827],[355,2.7376],[356,1.8679],[357,1.1288],[358,-1.0375],[359,-1.6495],[360,-2.6597],[361,-18.3754],[362,-20.6205],[363,-22.5262],[364,-23.7598],[365,-24.2295],[366,-24.4019],[368,-24.5],[440,-24.5]]
       }
       pinky_01_l_ctrl.Rotation.Z {
        type=MovieSceneScriptingFloatChannel
        default=25.8628
        thinned_from=94
        keys [[0,-60.5],[262,-60.5],[263,-58.2883],[264,-53.4225],[265,-50.4865],[266,-48.3924],[267,-43.0348],[268,-33.734],[269,-23.9701],[270,-14.3198],[271,0.0126],[272,2.1974],[273,7.8777],[274,17.0647],[279,32.3556],[283,37.8624],[287,39.2391],[292,39.5833],[299,39.6694],[352,39.698],[354,36.8438],[355,33.7787],[356,27.953],[357,24.0698],[358,12.6881],[359,9.763],[360,-10.8446],[361,-46.5752],[362,-51.6796],[363,-56.0124],[364,-58.8172],[365,-59.8849],[366,-60.277],[368,-60.5],[440,-60.5]]
       }
       pinky_01_l_ctrl.Scale.X {type=MovieSceneScriptingFloatChannel, default=1, const [1,94]}
       pinky_01_l_ctrl.Scale.Y {type=MovieSceneScriptingFloatChannel, default=1, const [1,94]}
       pinky_01_l_ctrl.Scale.Z {type=MovieSceneScriptingFloatChannel, default=1, thinned_from=94, keys [[0,1],[268,1],[269,0.9849],[270,1],[440,1]]}
       pinky_02_l_ctrl.Location.X {type=MovieSceneScriptingFloatChannel, default=0, const [0,94]}
       pinky_02_l_ctrl.Location.Y {type=MovieSceneScriptingFloatChannel, default=0, const [0,94]}
       pinky_02_l_ctrl.Location.Z {type=MovieSceneScriptingFloatChannel, default=0, const [0,94]}
       pinky_02_l_ctrl.Rotation.X {type=MovieSceneScriptingFloatChannel, default=0, thinned_from=94, keys [[0,0],[440,0]]}
       pinky_02_l_ctrl.Rotation.Y {type=MovieSceneScriptingFloatChannel, default=0, thinned_from=94, keys [[0,0],[440,0]]}
       pinky_02_l_ctrl.Rotation.Z {
        type=MovieSceneScriptingFloatChannel
        default=31.5895
        thinned_from=94
        keys [[0,-64.5],[262,-64.5],[264,-67.7243],[265,-64.6181],[266,-62.4027],[267,-56.7347],[268,-46.8949],[269,-31.0557],[270,-15.4007],[271,7.8497],[272,10.2855],[273,16.6184],[274,26.8609],[279,43.9085],[283,50.048],[287,51.5829],[292,51.9666],[309,52.0847],[352,52.0945],[354,47.7588],[355,43.1119],[356,34.3143],[357,28.9878],[358,13.3756],[359,9.7385],[360,-17.2879],[361,-51.2604],[362,-56.1136],[363,-60.2333],[364,-62.9],[365,-63.9152],[366,-64.288],[368,-64.5],[440,-64.5]]
       }
       pinky_02_l_ctrl.Scale.X {type=MovieSceneScriptingFloatChannel, default=1, const [1,94]}
       pinky_02_l_ctrl.Scale.Y {type=MovieSceneScriptingFloatChannel, default=1, const [1,94]}
       pinky_02_l_ctrl.Scale.Z {type=MovieSceneScriptingFloatChannel, default=1, thinned_from=94, keys [[0,1],[268,1],[269,0.9849],[270,1],[440,1]]}
       pinky_03_l_ctrl.Location.X {type=MovieSceneScriptingFloatChannel, default=0, const [0,94]}
       pinky_03_l_ctrl.Location.Y {type=MovieSceneScriptingFloatChannel, default=0, const [0,94]}
       pinky_03_l_ctrl.Location.Z {type=MovieSceneScriptingFloatChannel, default=0, const [0,94]}
       pinky_03_l_ctrl.Rotation.X {type=MovieSceneScriptingFloatChannel, default=0, thinned_from=94, keys [[0,0],[440,0]]}
       pinky_03_l_ctrl.Rotation.Y {type=MovieSceneScriptingFloatChannel, default=0, thinned_from=94, keys [[0,0],[440,0]]}
       pinky_03_l_ctrl.Rotation.Z {
        type=MovieSceneScriptingFloatChannel
        default=28.6796
        thinned_from=94
        keys [[0,-5],[262,-5],[263,-8.5428],[264,-16.3369],[265,-21.6939],[266,-25.5147],[267,-35.2902],[268,-52.2604],[269,-31.8965],[270,-11.7695],[271,18.1227],[272,18.8152],[273,20.6157],[274,23.5277],[279,28.3744],[283,30.1199],[287,30.5563],[292,30.6654],[352,30.7018],[354,30.2656],[355,29.7907],[356,28.8626],[357,28.0739],[358,25.7621],[359,25.109],[360,7.357],[361,-1.5348],[362,-2.805],[363,-3.8833],[364,-4.5812],[365,-4.8469],[366,-4.9445],[367,-4.983],[368,-5],[440,-5]]
       }
       pinky_03_l_ctrl.Scale.X {type=MovieSceneScriptingFloatChannel, default=1, const [1,94]}
       pinky_03_l_ctrl.Scale.Y {type=MovieSceneScriptingFloatChannel, default=1, const [1,94]}
       pinky_03_l_ctrl.Scale.Z {type=MovieSceneScriptingFloatChannel, default=1, thinned_from=94, keys [[0,1],[268,1],[269,0.9848],[270,1],[440,1]]}
       hand_l_ik_ctrl.Location.X {
        type=MovieSceneScriptingFloatChannel
        default=21.1065
        thinned_from=441
        keys [[-0.0013,-17.5117],[0,16.6627],[268,16.6734],[269,16.7055],[270,16.8268],[271,16.8823],[273,15.9481],[274,15.8276],[275,14.9467],[276,14.8276],[278,15.6444],[279,15.8503],[280,15.9725],[281,15.9546],[283,16.5026],[284,17.2336],[285,18.2085],[286,19.3006],[287,20.673],[288,21.7519],[289,22.2506],[290,22.5352],[291,22.6386],[340,22.3947],[341,22.4523],[344,22.7339],[345,22.7665],[346,22.7523],[349,22.2791],[350,22.3403],[351,22.051],[353,21.1914],[355,19.9364],[356,19.245],[357,18.0862],[358,16.6563],[359,15.4357],[360,15.0056],[361,14.9543],[362,14.9988],[363,15.4981],[365,16.2388],[366,16.4925],[367,16.5989],[440,16.6627]]
       }
       hand_l_ik_ctrl.Location.Y {
        type=MovieSceneScriptingFloatChannel
        default=18.778
        thinned_from=441
        keys [[-0.0013,21.3289],[0,-1.8176],[267,-1.8176],[268,-1.1783],[269,0.7396],[270,5.0269],[271,10.7021],[272,16.776],[273,23.1409],[274,24.7653],[275,26.1967],[276,27.4531],[277,29.0444],[278,29.6763],[280,29.8052],[281,30.7978],[282,32.6246],[283,34.6983],[284,36.3981],[285,38.3584],[286,40.1824],[287,41.7785],[288,42.4574],[289,42.5228],[292,42.2907],[340,41.8175],[349,40.9556],[350,40.3688],[351,39.2376],[353,36.3222],[355,32.478],[356,30.3314],[357,27.5879],[358,24.5773],[359,19.7503],[360,13.9262],[360.9987,7.3061],[362,3.6362],[363,1.1543],[364,0.0964],[365,-0.7785],[366,-1.4004],[367,-1.6611],[368,-1.7589],[369,-1.7992],[370,-1.8176],[440,-1.8176]]
       }
       hand_l_ik_ctrl.Location.Z {
        type=MovieSceneScriptingFloatChannel
        default=9.4859
        thinned_from=441
        keys [[-0.0013,13.6473],[0,-0.526],[267,-0.526],[268,-0.2414],[269,0.6125],[270,2.4446],[271,3.9068],[272,5.2311],[273,6.6188],[274,8.4205],[275,10.1471],[276,12.6085],[277,14.4799],[278,17.2816],[279,18.3838],[280,18.8897],[281,18.6172],[282,17.7501],[283,17.0646],[284,16.6847],[285,16.0963],[287,14.729],[288,14.2398],[289,14.0067],[290,13.8586],[292,13.7442],[298,13.5526],[335,13.4559],[338,13.4923],[340,13.5858],[341,13.6863],[344,14.1265],[345,14.3854],[346,14.7067],[347,15.3858],[348,16.4137],[349,17.1749],[350,17.6857],[351,17.6739],[353,17.3586],[355,14.6],[356,12.7462],[357,10.5864],[358,8.6328],[359,6.8303],[360,4.9354],[361,3.1057],[362,2.0992],[363,1.738],[364,1.0146],[365,0.3104],[366,-0.1902],[367,-0.4001],[368,-0.4788],[369,-0.5112],[370,-0.526],[440,-0.526]]
       }
       hand_l_ik_ctrl.Rotation.X {
        type=MovieSceneScriptingFloatChannel
        default=70.7546
        thinned_from=441
        keys [[-0.0013,58.3547],[0,28.6966],[267,28.6966],[268,28.7742],[269,29.0069],[270,29.5848],[271,32.2316],[272,37.1514],[273,42.3069],[275,58.2],[276,68.5387],[277,85.6275],[278,103.7692],[279,111.5471],[280,116.1146],[281,116.3719],[282,117.6886],[283,119.9736],[284,123.1704],[285,128.0923],[286,134.97],[287,145.4142],[288,154.6895],[289,159.2686],[290,162.0077],[291,163.1505],[292,163.4907],[340,163.4055],[343,161.9406],[344,161.1044],[346,157.7115],[347,154.1184],[348,147.2987],[349,137.057],[350,130.1173],[351,125.5597],[353,117.4497],[355,106.2642],[356,107.0708],[357,94.8475],[358,74.4602],[359,53.3447],[360,33.0007],[361,30.9647],[362,29.3839],[366,28.7796],[440,28.6966]]
       }
       hand_l_ik_ctrl.Rotation.Y {
        type=MovieSceneScriptingFloatChannel
        default=20.0261
        thinned_from=441
        keys [[-0.0013,70.0208],[0,5.4028],[267,5.4028],[268,3.6338],[269,-1.673],[270,-12.9254],[271,-15.6664],[272,-13.3656],[273,-10.9546],[274,-1.049],[275,8.973],[276,7.6937],[277,7.7584],[278,3.0564],[279,0.0361],[280,-1.8957],[281,-2.6743],[282,-3.8128],[283,-5.4838],[284,-8.6719],[285,-12.759],[286,-16.5156],[287,-18.6735],[288,-17.2294],[289,-15.0154],[290,-12.969],[291,-11.8563],[292,-11.4495],[304,-11.2536],[315,-11.2321],[338,-11.3694],[340,-11.517],[341,-11.9378],[342,-12.7008],[343,-13.4109],[344,-14.3707],[345,-16.0957],[346,-18.0415],[347,-21.8133],[348,-28.1852],[349,-35.8196],[350,-38.6611],[351,-38.5872],[353,-37.4245],[355,-21.1954],[356,-10.6579],[357,-6.0976],[358,-10.0198],[359,-14.0821],[360,-22.1558],[361,-15.4731],[362,-12.863],[363,-9.542],[364,-4.723],[365,-0.0945],[366,3.1956],[367,4.5751],[368,5.0924],[369,5.3057],[370,5.4028],[440,5.4028]]
       }
       hand_l_ik_ctrl.Rotation.Z {
        type=MovieSceneScriptingFloatChannel
        default=82.4187
        thinned_from=441
        keys [[-0.0013,67.8071],[0,96.126],[267,96.126],[268,96.8595],[269,99.0603],[270,104.2058],[271,110.5431],[272,117.7003],[273,125.2004],[275,148.8761],[276,155.4503],[277,163.9109],[278,168.7968],[279,169.3451],[280,169.1841],[283,171.9635],[284,171.7794],[285,170.1049],[286,166.0606],[287,157.6219],[288,148.2013],[289,142.6427],[290,138.8224],[291,137.0321],[292,136.4404],[341,136.2587],[344,135.8349],[346,136.1087],[347,136.9203],[348,138.8391],[349,142.7555],[350,144.2667],[351,143.3053],[353,139.632],[355,146.4133],[356,140.3753],[357,133.3952],[359,121.5732],[360,112.4351],[361,100.3913],[362,95.686],[440,96.126]]
       }
       hand_l_ik_ctrl.Scale.X {type=MovieSceneScriptingFloatChannel, default=1, thinned_from=441, keys [[-0.0013,0.9606],[440,0.9606]]}
       hand_l_ik_ctrl.Scale.Y {type=MovieSceneScriptingFloatChannel, default=1, thinned_from=441, keys [[-0.0013,0.9606],[440,0.9606]]}
       hand_l_ik_ctrl.Scale.Z {type=MovieSceneScriptingFloatChannel, default=1, thinned_from=441, keys [[-0.0013,0.9606],[440,0.9606]]}
       Softness {
        type=MovieSceneScriptingFloatChannel
        default=0.0092
        thinned_from=118
        keys [[0,0],[287,0],[288,0.2107],[289,0.3037],[290,0.3483],[291,0.3683],[292,0.375],[338,0.375],[340,0.3729],[341,0.3669],[342,0.356],[343,0.3457],[344,0.3315],[345,0.3157],[347,0.2812],[348,0.2482],[349,0.1927],[350,0.1274],[351,0.0845],[352,0.0513],[353,0.0184],[355,0.0184],[356,0.0093],[357,0.0046],[359,0],[440,0]]
       }
       hand_l_ikorient_ctrl.Rotation.Roll {
        type=MovieSceneScriptingFloatChannel
        default=0
        thinned_from=64
        keys [[0,0],[267,0],[271,19.1435],[275,38.0644],[279,14.2741],[283,3.5685],[287,0.8921],[292,0.223],[299,0.0558],[309,0.0171],[327,0.0043],[337,0],[439,0]]
       }
       hand_l_ikorient_ctrl.Rotation.Pitch {
        type=MovieSceneScriptingFloatChannel
        default=0
        thinned_from=64
        keys [[0,0],[267,0],[271,-3.2884],[275,-6.5385],[279,-2.452],[283,-0.613],[287,-0.1532],[292,-0.0383],[299,-0.0096],[309,-0.0029],[327,-0.0007],[337,0],[439,0]]
       }
       hand_l_ikorient_ctrl.Rotation.Yaw {
        type=MovieSceneScriptingFloatChannel
        default=0
        thinned_from=64
        keys [[0,0],[267,0],[271,53.8779],[275,107.1294],[279,40.1735],[283,10.0434],[287,2.5108],[292,0.6277],[299,0.1569],[309,0.0482],[327,0.012],[337,0],[439,0]]
       }
       arm_l_pv_ik_ctrl.Location.X {
        type=MovieSceneScriptingFloatChannel
        default=0.5
        thinned_from=109
        keys [[0,27.8842],[250,27.8842],[265,10.5574],[266,10.6438],[267,10.9329],[268,11.4685],[269,12.3385],[270,14.1349],[271,16.9619],[272,22.155],[288,22.155],[289,11.5781],[290,0.5605],[291,-4.3974],[292,-6.05],[338,-6.05],[340,-6.0132],[341,-5.9093],[342,-5.7478],[343,-5.5383],[344,-5.2904],[345,-5.0138],[346,-4.7179],[347,-4.4125],[348,-3.8354],[349,-2.866],[350,-1.726],[351,-0.7792],[352,0.1788],[353,0.5],[354,0.869],[355,2.4922],[356,5.1974],[357,16.5908],[358,28.5757],[359,25.4923],[360,22.5569],[361,19.0853],[362,18.1607],[364,21.0689],[365,23.9771],[367,27.8842],[440,27.8842]]
       }
       arm_l_pv_ik_ctrl.Location.Y {
        type=MovieSceneScriptingFloatChannel
        default=23.0668
        thinned_from=113
        keys [[0,10],[267,10],[268,10.4365],[269,10.9118],[270,11.7575],[271,12.9736],[272,11.6554],[288,11.6554],[289,22.1851],[290,33.3917],[291,38.53],[292,40.2427],[338,40.2487],[340,40.2063],[342,39.9005],[345,39.0548],[347,38.362],[348,37.8691],[349,37.1305],[350,36.1953],[352,33.932],[353,32.7021],[354,30.5086],[355,27.718],[356,23.0668],[357,10],[359,10],[360,11.5609],[361,13.2851],[362,11.991],[364,9.4754],[365,8.5428],[367,10],[440,10]]
       }
       arm_l_pv_ik_ctrl.Location.Z {
        type=MovieSceneScriptingFloatChannel
        default=16.0412
        thinned_from=113
        keys [[0,0],[271,0],[272,3.7136],[273,6.8173],[274,9.749],[275,13.361],[276,14.6813],[277,15.297],[278,15.6084],[279,15.7953],[280,16.6573],[281,17.2181],[282,17.9791],[283,18.9137],[284,19.6926],[285,21.1367],[286,22.0573],[287,23.7443],[288,26.071],[289,29.0269],[290,32.2898],[291,33.8316],[292,34.3455],[338,34.3506],[340,34.3144],[342,34.0534],[345,33.3317],[347,32.7404],[348,32.3197],[349,31.6893],[350,30.8912],[352,28.9596],[354,26.8086],[355,23.9204],[356,20.5766],[357,17.0344],[358,7.1038],[359,5.3085],[360,5.1458],[361,-0.0974],[362,-0.0243],[364,-0.0137],[365,-0.0031],[367,0],[440,0]]
       }
       arm_l_pv_ik_ctrl.Rotation.X {
        type=MovieSceneScriptingFloatChannel
        default=16.7059
        thinned_from=113
        keys [[0,0],[288,0],[289,13.2439],[290,27.2514],[291,33.6394],[292,35.7687],[338,35.774],[340,35.7363],[342,35.4645],[345,34.7128],[347,34.0971],[348,33.659],[349,33.0025],[350,32.1713],[352,30.1596],[353,29.0664],[354,26.5376],[355,22.8507],[356,16.7059],[357,0],[440,0]]
       }
       arm_l_pv_ik_ctrl.Rotation.Y {
        type=MovieSceneScriptingFloatChannel
        default=20.4183
        thinned_from=113
        keys [[0,0],[288,0],[289,16.187],[290,33.3071],[291,41.1147],[292,43.7172],[338,43.7237],[340,43.6776],[342,43.3454],[345,42.4267],[347,41.6741],[348,41.1386],[349,40.3362],[350,39.3203],[352,36.8616],[353,35.5255],[354,32.4348],[355,27.9286],[356,20.4183],[357,0],[440,0]]
       }
       arm_l_pv_ik_ctrl.Rotation.Z {
        type=MovieSceneScriptingFloatChannel
        default=40.4496
        thinned_from=113
        keys [[0,0],[288,0],[289,32.0671],[290,65.983],[291,81.4502],[292,86.6059],[338,86.6187],[340,86.5275],[342,85.8693],[345,84.0493],[347,82.5584],[348,81.4975],[349,79.908],[350,77.8955],[352,73.0246],[353,70.3777],[354,64.2548],[355,55.3278],[356,40.4496],[357,0],[440,0]]
       }
       arm_l_pv_ik_ctrl.Scale.X {type=MovieSceneScriptingFloatChannel, default=1, thinned_from=113, keys [[0,1],[440,1]]}
       arm_l_pv_ik_ctrl.Scale.Y {type=MovieSceneScriptingFloatChannel, default=1, thinned_from=113, keys [[0,1],[440,1]]}
       arm_l_pv_ik_ctrl.Scale.Z {type=MovieSceneScriptingFloatChannel, default=1, thinned_from=113, keys [[0,1],[440,1]]}
       upperarm_r_fk_ctrl.Location.X {
        type=MovieSceneScriptingFloatChannel
        default=0.0054
        thinned_from=142
        keys [[0,0],[274,0],[275,0.0012],[276,0.0039],[277,0.0334],[278,0.0624],[279,0.075],[280,0.0825],[281,0.0851],[282,0.0955],[283,0.1097],[284,0.1326],[285,0.1652],[286,0.2047],[287,0.2577],[288,0.3047],[289,0.3308],[290,0.3486],[291,0.3569],[292,0.3597],[339,0.3598],[340,0.3591],[341,0.3577],[342,0.352],[343,0.3432],[344,0.3316],[345,0.318],[346,0.3028],[348,0.2698],[349,0.2381],[350,0.1849],[351,0.1223],[352,0.0706],[353,0.0176],[355,0.0176],[356,0.0132],[357,0.005],[358,0],[440,0]]
       }
       upperarm_r_fk_ctrl.Location.Y {
        type=MovieSceneScriptingFloatChannel
        default=0.0005
        thinned_from=142
        keys [[0,0],[274,0],[276,0.0004],[278,0.0062],[279,0.0075],[280,0.0082],[281,0.0086],[282,0.0095],[283,0.011],[284,0.0133],[285,0.0165],[286,0.0205],[287,0.0258],[288,0.0305],[289,0.0331],[290,0.0348],[291,0.0357],[292,0.0359],[340,0.0359],[341,0.0358],[342,0.0352],[343,0.0343],[344,0.0331],[346,0.0303],[348,0.027],[349,0.0238],[350,0.0185],[351,0.0122],[353,0.0018],[355,0.0018],[356,0.0013],[357,0.0005],[358,0],[440,0]]
       }
       upperarm_r_fk_ctrl.Location.Z {
        type=MovieSceneScriptingFloatChannel
        default=0.0138
        thinned_from=142
        keys [[0,0],[274,0],[275,0.0031],[276,0.01],[277,0.085],[278,0.1588],[279,0.1908],[280,0.21],[281,0.2166],[282,0.2429],[283,0.2791],[284,0.3374],[285,0.4202],[286,0.5208],[287,0.6557],[288,0.7753],[289,0.8416],[290,0.8868],[291,0.9081],[292,0.9151],[339,0.9152],[340,0.9136],[341,0.9101],[342,0.8956],[343,0.873],[344,0.8437],[345,0.8091],[346,0.7704],[348,0.6864],[349,0.6058],[350,0.4703],[351,0.311],[352,0.1795],[353,0.0449],[355,0.0449],[356,0.0337],[357,0.0126],[358,0],[440,0]]
       }
       upperarm_r_fk_ctrl.Rotation.X {
        type=MovieSceneScriptingFloatChannel
        default=52.9144
        thinned_from=135
        keys [[0,-8.5],[190,-8.5],[200,-8.912],[210,-9.975],[220,-11.1985],[230,-11.6104],[240,-11.6104],[250,-11.2191],[255,-10.6713],[257,-10.3618],[260,-9.8183],[266,-1],[269,13.1313],[270,29.4994],[271,35.1192],[272,40.2046],[273,45.1235],[274,46.7534],[275,47.4905],[276,47.4189],[278,48.6407],[279,48.9032],[280,49.0607],[281,48.375],[283,48.6873],[284,49.3511],[285,50.3318],[286,51.5393],[287,53.1633],[288,54.6058],[289,55.406],[290,55.952],[291,56.209],[292,56.2946],[341,56.3283],[343,56.6223],[348,58.1014],[349,58.7086],[350,59.7023],[351,60.8608],[353,62.7863],[355,62.7863],[356,56.2863],[357,46.6799],[358,36.7864],[359,27.5214],[360,20.2522],[361,14.0687],[362,12.8957],[363,-11.1019],[364,-12.5173],[365,-15.0155],[366,-18.7752],[367,-25.4685],[368,-32.3951],[369,-34.9516],[370,-33.815],[371,-31.1602],[372,-27.3081],[373,-22.6557],[374,-17.7033],[375,-12.6175],[376,-9.5989],[377,-8.5],[440,-8.5]]
       }
       upperarm_r_fk_ctrl.Rotation.Y {
        type=MovieSceneScriptingFloatChannel
        default=-3.01
        thinned_from=196
        keys [[0,-12.5],[20,-12.3649],[30,-12.3907],[60,-12.6118],[70,-12.6338],[80,-12.5339],[100,-12.4895],[110,-12.446],[126,-12.3005],[133,-12.4576],[143,-12.7685],[159,-12.9859],[170,-13.0289],[180,-13.0289],[190,-12.5],[200,-9.5412],[210,-4.6581],[220,-7.8137],[221,-8.1055],[222,-8.3622],[223,-8.5326],[225,-8.7398],[226,-8.8013],[229,-8.882],[232,-8.9103],[234,-8.889],[237,-8.773],[241,-8.5066],[246,-8.1145],[247,-8.0151],[249,-7.7683],[252,-7.3485],[253,-7.2653],[254,-7.2319],[255,-7.2422],[256,-7.284],[257,-7.3547],[258,-7.4416],[259,-7.556],[260,-7.7341],[261,-7.9698],[262,-8.2609],[264,-8.936],[265,-9.0094],[266,-8.8599],[267,-8.2427],[268,-5.861],[269,-2.31],[270,-13.7179],[271,-11.58],[272,-10.0438],[273,-7.9821],[274,-6.2482],[275,-5.6768],[276,-4.8799],[277,-4.1658],[278,-3.4628],[279,-3.1582],[280,-2.9755],[281,-3.0869],[282,-3.6039],[283,-4.4088],[284,-6.3786],[285,-9.2926],[286,-12.8821],[287,-17.71],[288,-21.999],[289,-24.3782],[290,-26.0016],[291,-26.7655],[292,-27.0201],[340,-27.0174],[341,-26.8822],[342,-26.5195],[343,-25.9567],[344,-25.2271],[345,-24.3641],[346,-23.401],[348,-21.3081],[349,-19.957],[350,-18.1077],[351,-15.8597],[352,-13.2441],[353,-10.5666],[354,-7.238],[355,-3.75],[356,-3.8538],[357,-3.9658],[358,-4.1653],[359,-4.4791],[360,-5.0934],[361,-6.8021],[362,-9.5723],[363,0.4592],[364,-2.8996],[365,-5.8903],[366,-8.2192],[367,-9.8445],[368,-10.8798],[369,-11.4886],[370,-11.7062],[371,-11.6175],[372,-11.2419],[374,-9.9926],[375,-9.5171],[376,-9.3765],[377,-9.5523],[378,-9.9483],[380,-10.9907],[381,-11.4771],[382,-11.9343],[383,-12.342],[384,-12.778],[385,-13.1252],[386,-13.3705],[387,-13.4867],[388,-13.5238],[399,-13.5238],[400,-13.4671],[404,-12.9814],[405,-12.8941],[413,-12.5251],[440,-12.4969]]
       }
       upperarm_r_fk_ctrl.Rotation.Z {
        type=MovieSceneScriptingFloatChannel
        default=54.9868
        thinned_from=142
        keys [[0,9],[10,9.3964],[20,9.9039],[28,9.6583],[34,9.2302],[47,8.1068],[56,7.7791],[68,7.6169],[78,9.2632],[90,9.8565],[101,10.3215],[103,10.4775],[107,10.9271],[110,11.1519],[127,6.2131],[128,6.0311],[129,5.9004],[130,5.8],[131,5.725],[132,5.6859],[133,5.6715],[134,5.6841],[135,5.7189],[136,5.7721],[138,5.9168],[142,6.246],[146,6.5234],[148,6.6382],[150,6.7241],[152,6.7717],[153,6.7781],[166,5.8526],[176,5.3023],[187,4.8903],[194,4.5903],[200,4.4746],[205,5.35],[212,7.9662],[218,9.9987],[224,10.8557],[237,9.8455],[250,8.878],[262,8.3736],[263,8.4258],[264,8.7907],[265,9.6253],[266,11.0711],[267,13.2152],[268,18.8576],[269,25.6288],[270,12.8185],[271,19.8545],[272,25.1149],[274,32.9256],[275,35.6915],[278,44.7384],[279,47.5084],[280,49.1703],[281,49.6798],[282,51.616],[283,54.1271],[284,57.1355],[285,61.247],[286,66.1729],[287,72.7507],[288,78.575],[289,81.7993],[290,83.9957],[291,85.0279],[292,85.3719],[341,85.3032],[343,84.6468],[346,82.8318],[348,81.345],[349,80.3653],[350,78.9965],[351,77.3163],[353,73.3335],[354,69.8746],[355,66.25],[356,60.1096],[357,51.8958],[358,41.6885],[359,32.0254],[360,21.6388],[361,11.0517],[362,2.195],[363,-0.5504],[364,-1.9877],[365,-3.1518],[366,-4.225],[367,-4.923],[368,-5.2709],[369,-5.36],[370,-5.1547],[371,-4.634],[372,-4.0058],[373,-3.0475],[374,-1.6853],[379,7.8218],[382,10.8047],[385,11.708],[407,8.0594],[440,9.0003]]
       }
       upperarm_r_fk_ctrl.Scale.X {type=MovieSceneScriptingFloatChannel, default=1, const [1,142]}
       upperarm_r_fk_ctrl.Scale.Y {type=MovieSceneScriptingFloatChannel, default=1, const [1,142]}
       upperarm_r_fk_ctrl.Scale.Z {type=MovieSceneScriptingFloatChannel, default=1, const [1,142]}
       lowerarm_r_fk_ctrl.Location.X {type=MovieSceneScriptingFloatChannel, default=0, thinned_from=130, keys [[0,0],[440,0]]}
       lowerarm_r_fk_ctrl.Location.Y {type=MovieSceneScriptingFloatChannel, default=0, thinned_from=130, keys [[0,0],[440,0]]}
       lowerarm_r_fk_ctrl.Location.Z {type=MovieSceneScriptingFloatChannel, default=0, thinned_from=130, keys [[0,0],[440,0]]}
       lowerarm_r_fk_ctrl.Rotation.X {
        type=MovieSceneScriptingFloatChannel
        default=0.3666
        thinned_from=130
        keys [[0,0],[339,0],[340,0.0005],[341,0.0015],[342,0.0059],[343,0.0127],[344,0.0215],[345,0.0319],[346,0.0435],[347,0.056],[348,0.0688],[349,0.0877],[350,0.1168],[351,0.153],[352,0.1936],[353,0.2351],[355,0.2351],[356,0.2372],[357,0.2409],[358,0.2432],[359,0.2145],[360,0.1515],[361,0.0314],[362,0.0118],[363,0.0031],[364,0.0008],[366,0.0002],[440,0]]
       }
       lowerarm_r_fk_ctrl.Rotation.Y {
        type=MovieSceneScriptingFloatChannel
        default=0.2775
        thinned_from=130
        keys [[0,0],[339,0],[340,0.0006],[341,0.0021],[342,0.0079],[343,0.017],[344,0.0288],[345,0.0427],[346,0.0583],[347,0.0749],[348,0.0921],[349,0.1245],[350,0.179],[351,0.2431],[352,0.296],[353,0.3502],[355,0.3502],[356,0.3087],[357,0.2479],[358,0.1844],[359,0.1565],[360,0.1148],[361,0.0238],[362,0.0089],[363,0.0024],[364,0.0006],[366,0.0002],[440,0]]
       }
       lowerarm_r_fk_ctrl.Rotation.Z {
        type=MovieSceneScriptingFloatChannel
        default=39.9981
        thinned_from=105
        keys [[0,-5.5],[10,-5.3708],[20,-5.217],[28,-5.1742],[37,-5.334],[47,-5.7683],[54,-6.3148],[67,-7.1542],[77,-5.932],[87,-4.8808],[97,-4.4487],[110,-4.3047],[116,-4.6721],[127,-5.8119],[136,-6.5369],[144,-5.4792],[153,-4.1846],[159,-5.3061],[168,-6.6796],[181,-7.4653],[189,-7.6744],[197,-8.4881],[203,-9.6652],[209,-8.1],[219,-3.5386],[226,-1.5809],[231,-1.2167],[247,-4.5295],[262,-5.8942],[263,-5.7848],[264,-5.5443],[265,-5.2195],[266,-4.828],[267,-4.3797],[268,-3.8811],[269,-1.5],[270,8.5476],[271,22.8293],[272,39.3789],[273,51.2283],[274,56.5372],[275,58.8348],[276,58.915],[278,59.8657],[280,60.1925],[281,59.1849],[282,55.2499],[283,50.1315],[284,43.2593],[285,33.6758],[286,22.1092],[287,6.6337],[288,-7.081],[289,-14.6777],[290,-19.8549],[291,-22.2886],[292,-23.0998],[327,-23.0998],[339,-23.134],[340,-23.0515],[341,-22.8699],[342,-22.1235],[343,-20.9637],[344,-19.4595],[345,-17.6797],[346,-15.6931],[347,-13.5688],[348,-11.3756],[349,-8.0163],[350,-2.7814],[351,3.6782],[352,10.592],[353,17.6694],[354,26.132],[355,35],[356,38.2944],[357,41.0851],[358,41.7221],[359,43.4646],[360,47.1157],[361,48.432],[362,48.4813],[363,46.1639],[364,40.3563],[365,30.3351],[366,17.6411],[367,7.1581],[368,-1.1369],[369,-5.6183],[370,-8.1629],[371,-9.3012],[372,-10.066],[373,-10.5317],[374,-10.6901],[377,-9.2055],[385,-1.2065],[391,0.7065],[407,-4.7607],[414,-6.1736],[428,-5.9869],[440,-5.4975]]
       }
       lowerarm_r_fk_ctrl.Scale.X {type=MovieSceneScriptingFloatChannel, default=1, const [1,130]}
       lowerarm_r_fk_ctrl.Scale.Y {type=MovieSceneScriptingFloatChannel, default=1, const [1,130]}
       lowerarm_r_fk_ctrl.Scale.Z {type=MovieSceneScriptingFloatChannel, default=1, const [1,130]}
       hand_r_fk_ctrl.Location.X {type=MovieSceneScriptingFloatChannel, default=0, thinned_from=141, keys [[0,0],[440,0]]}
       hand_r_fk_ctrl.Location.Y {
        type=MovieSceneScriptingFloatChannel
        default=-0.1788
        thinned_from=141
        keys [[0,0],[260,0],[261,-0.0005],[262,-0.0016],[263,-0.003],[264,-0.0048],[266,-0.0089],[268,-0.0137],[269,-0.0177],[270,-0.0257],[271,-0.0431],[272,-0.0873],[273,-0.1788],[274,-0.1788],[275,-0.1782],[276,-0.1768],[278,-0.1466],[279,-0.1401],[280,-0.1362],[281,-0.1346],[282,-0.1282],[283,-0.1192],[284,-0.1074],[285,-0.0916],[286,-0.0728],[287,-0.0479],[288,-0.0258],[289,-0.0136],[290,-0.0053],[291,-0.0014],[292,-0.0001],[345,0],[346,-0.0093],[347,-0.0231],[348,-0.0447],[349,-0.0604],[350,-0.0869],[351,-0.118],[352,-0.1437],[353,-0.17],[355,-0.1635],[356,-0.1572],[357,-0.1425],[358,-0.1187],[359,-0.1019],[360,-0.0739],[361,-0.0153],[362,-0.0119],[363,-0.009],[364,-0.0004],[440,0]]
       }
       hand_r_fk_ctrl.Location.Z {
        type=MovieSceneScriptingFloatChannel
        default=0.0249
        thinned_from=141
        keys [[0,0],[261,0.0001],[268,0.0019],[269,0.0025],[270,0.0036],[271,0.006],[272,0.0122],[273,0.0249],[276,0.0246],[278,0.0204],[279,0.0195],[280,0.019],[281,0.0187],[282,0.0178],[283,0.0166],[284,0.0149],[285,0.0127],[286,0.0101],[287,0.0067],[288,0.0036],[289,0.0019],[290,0.0007],[291,0.0002],[292,0],[345,0],[346,0.0039],[348,0.0062],[349,0.0084],[350,0.0121],[351,0.0164],[353,0.0237],[355,0.0228],[356,0.0219],[357,0.0198],[358,0.0165],[359,0.0142],[360,0.0103],[361,0.0021],[363,0.0013],[364,0.0001],[440,0]]
       }
       hand_r_fk_ctrl.Rotation.X {
        type=MovieSceneScriptingFloatChannel
        default=8.7367
        thinned_from=141
        keys [[0,25],[266,25],[267,23.0444],[268,16.8981],[269,11.781],[270,7.3612],[271,3.8903],[272,0.3829],[273,-2.2488],[274,-4.1176],[275,-5.1796],[276,-2.5574],[277,11.642],[278,25.6199],[279,31.6747],[280,35.3076],[281,35.734],[282,37.4781],[283,39.9644],[284,43.4846],[285,48.2038],[286,53.817],[287,61.2984],[288,67.9169],[289,71.5788],[290,74.0722],[291,75.2435],[292,75.6339],[340,75.6023],[341,75.4859],[342,75.0074],[343,74.2638],[344,73.2993],[345,72.1582],[346,70],[347,67.5],[348,62.5],[349,53.75],[350,42.5],[351,35],[353,22.5],[354,17.3152],[355,11.8821],[356,8.8004],[357,7.3048],[358,5.2017],[359,2.6258],[360,0.9765],[361,4.7595],[362,12.3164],[363,20.9963],[364,25],[440,25]]
       }
       hand_r_fk_ctrl.Rotation.Y {
        type=MovieSceneScriptingFloatChannel
        default=32.5298
        thinned_from=141
        keys [[0,-5],[260,-5],[261,-4.9538],[262,-4.8523],[263,-4.7159],[264,-4.5488],[265,-4.3621],[266,-4.1597],[267,-3.944],[268,-3.7166],[269,-3.3381],[270,-2.581],[271,-0.9471],[272,3.2017],[273,11.7964],[274,20.884],[275,29.0512],[276,33.3348],[278,42.0209],[279,43.8874],[280,45.0073],[281,45.1646],[282,45.8039],[283,46.706],[284,47.9796],[285,49.6869],[286,51.7175],[287,54.4239],[288,56.8181],[289,58.1428],[290,59.0448],[291,59.4685],[292,59.6097],[340,59.5729],[341,59.4748],[342,59.0718],[343,58.4455],[344,57.6332],[346,55.5994],[348,53.268],[349,51.5031],[350,48.7842],[351,45.4039],[352,41.6609],[353,37.8292],[355,36.0247],[356,34.2412],[357,28.9454],[358,23.4771],[359,18.1704],[360,12.7357],[361,19],[362,13.561],[363,8.992],[364,-4.7149],[365,-4.8533],[366,-4.9262],[367,-4.9616],[378,-5],[440,-5]]
       }
       hand_r_fk_ctrl.Rotation.Z {
        type=MovieSceneScriptingFloatChannel
        default=-4.7127
        thinned_from=141
        keys [[0,5],[260,5],[261,4.9264],[262,4.7646],[263,4.5473],[264,4.2811],[265,3.9834],[266,3.661],[267,3.3173],[268,2.9549],[269,2.3517],[270,1.1453],[271,-1.4583],[272,-8.0695],[273,-21.765],[274,-25.7055],[275,-16.9703],[276,-16.8275],[278,-14.8455],[279,-14.4196],[280,-14.164],[281,-14.0297],[282,-13.4885],[283,-12.7452],[284,-11.7973],[285,-10.5435],[286,-9.0598],[287,-7.0851],[288,-5.3392],[289,-4.3737],[290,-3.7164],[291,-3.4077],[292,-3.3048],[339,-3.2993],[340,-3.2904],[341,-3.2717],[342,-3.1951],[343,-3.0766],[344,-2.9231],[345,-2.7417],[346,-2.5393],[347,-2.3229],[348,-2.0995],[349,-1.6769],[350,-0.9667],[351,-0.1313],[352,0.5585],[353,1.2646],[354,1.4289],[355,1.6011],[356,1.9338],[357,3.0888],[358,3.9415],[359,4.1037],[360,4.3407],[361,15.5],[362,13.1364],[363,11.1703],[364,5.2721],[365,5.14],[366,5.0704],[367,5.0367],[378,5],[440,5]]
       }
       hand_r_fk_ctrl.Scale.X {
        type=MovieSceneScriptingFloatChannel
        default=1
        thinned_from=141
        keys [[0,0.96],[269,0.964],[270,0.9658],[271,0.9697],[272,0.9795],[273,1],[274,0.96],[353,0.96],[355,0.9633],[356,0.9666],[357,0.9791],[358,0.9865],[359,0.9834],[360,0.9765],[361,0.9634],[440,0.96]]
       }
       hand_r_fk_ctrl.Scale.Y {
        type=MovieSceneScriptingFloatChannel
        default=1
        thinned_from=141
        keys [[0,0.96],[269,0.964],[270,0.9658],[271,0.9697],[272,0.9795],[273,1],[274,0.96],[353,0.96],[355,0.9633],[356,0.9666],[357,0.9791],[358,0.9865],[359,0.9834],[360,0.9765],[361,0.9634],[440,0.96]]
       }
       hand_r_fk_ctrl.Scale.Z {
        type=MovieSceneScriptingFloatChannel
        default=1
        thinned_from=141
        keys [[0,0.96],[269,0.964],[270,0.9658],[271,0.9697],[272,0.9795],[273,1],[274,0.96],[353,0.96],[355,0.9633],[356,0.9666],[357,0.9791],[358,0.9865],[359,0.9834],[360,0.9765],[361,0.9634],[440,0.96]]
       }
       thumb_01_r_ctrl.Location.X {type=MovieSceneScriptingFloatChannel, default=0, const [0,62]}
       thumb_01_r_ctrl.Location.Y {type=MovieSceneScriptingFloatChannel, default=0, const [0,62]}
       thumb_01_r_ctrl.Location.Z {type=MovieSceneScriptingFloatChannel, default=0, const [0,62]}
       thumb_01_r_ctrl.Rotation.X {
        type=MovieSceneScriptingFloatChannel
        default=-22.9239
        thinned_from=62
        keys [[0,18.5],[272,18.5],[273,-30.0572],[275,-26.4822],[279,-21.7674],[283,-19.9211],[287,-19.4595],[292,-19.3441],[348,-19.3057],[350,-19.4244],[352,-19.6841],[353,-20.2051],[355,-22.8023],[356,-24.0667],[358,-24.2518],[359,-24.2518],[360,-12.4118],[361,-4.6839],[362,1.1121],[364,3.6587],[367,10.5789],[370,15.4458],[374,18.5],[439,18.5]]
       }
       thumb_01_r_ctrl.Rotation.Y {
        type=MovieSceneScriptingFloatChannel
        default=26.7504
        thinned_from=62
        keys [[0,20],[272,20],[273,22.8588],[275,22.8635],[279,24.0312],[283,25.6202],[287,26.0175],[292,26.1168],[350,26.1736],[352,26.2216],[353,26.3066],[356,26.8498],[359,26.8659],[360,24.8794],[361,23.6595],[362,22.7446],[364,22.3427],[367,21.2503],[370,20.4821],[374,20],[439,20]]
       }
       thumb_01_r_ctrl.Rotation.Z {
        type=MovieSceneScriptingFloatChannel
        default=11.0489
        thinned_from=67
        keys [[0,0],[2,0],[12,0.1176],[22,0.6002],[32,0.9679],[42,1.0904],[52,0.6815],[62,0.2556],[72,-0.493],[82,-1.2938],[92,-0.6545],[102,-0.1636],[112,0.3978],[122,1.6309],[132,2.3838],[142,1.4899],[152,0.3725],[162,-0.8331],[172,-1.3757],[182,-1.5565],[192,-1.7713],[202,-2.2335],[212,-0.7902],[220,2.5595],[230,1.6213],[238,0.476],[248,0],[270,0],[273,7.7549],[275,6.4032],[279,4.6205],[283,3.9224],[287,3.7479],[292,3.7043],[299,3.6934],[348,3.6897],[349,3.8525],[350,3.9501],[351,4.2755],[352,4.501],[353,5.554],[354,8.0884],[355,10.5482],[356,12.9512],[358,13.2594],[359,13.2594],[360,9.5026],[361,7.1269],[362,5.3452],[364,6.8534],[367,8.4727],[370,10],[374,7.2807],[381,-1.9384],[384,-3.6104],[402,0.5304],[406,0.7259],[424,0],[439,0]]
       }
       thumb_01_r_ctrl.Scale.X {type=MovieSceneScriptingFloatChannel, default=1, const [1,62]}
       thumb_01_r_ctrl.Scale.Y {type=MovieSceneScriptingFloatChannel, default=1, const [1,62]}
       thumb_01_r_ctrl.Scale.Z {type=MovieSceneScriptingFloatChannel, default=1, const [1,62]}
       thumb_02_r_ctrl.Location.X {type=MovieSceneScriptingFloatChannel, default=0, const [0,62]}
       thumb_02_r_ctrl.Location.Y {type=MovieSceneScriptingFloatChannel, default=0, const [0,62]}
       thumb_02_r_ctrl.Location.Z {type=MovieSceneScriptingFloatChannel, default=0, const [0,62]}
       thumb_02_r_ctrl.Rotation.X {type=MovieSceneScriptingFloatChannel, default=0, const [0,62]}
       thumb_02_r_ctrl.Rotation.Y {
        type=MovieSceneScriptingFloatChannel
        default=6.5994
        thinned_from=62
        keys [[0,0],[348,0],[349,0.1483],[350,0.2373],[351,0.5339],[352,0.736],[353,1.679],[354,3.928],[355,6.0985],[356,8.2189],[358,8.4813],[359,8.4813],[360,5.9889],[361,4.4917],[362,3.3687],[364,2.8754],[367,1.5346],[370,0.5917],[374,0],[439,0]]
       }
       thumb_02_r_ctrl.Rotation.Z {
        type=MovieSceneScriptingFloatChannel
        default=21.178
        thinned_from=67
        keys [[0,0],[2,0],[12,0.1176],[22,0.6002],[32,0.9679],[42,1.0904],[52,0.6815],[62,0.2556],[72,-0.493],[82,-1.2938],[92,-0.6545],[102,-0.1636],[112,0.3978],[122,1.6309],[132,2.3838],[142,1.4899],[152,0.3725],[162,-0.8331],[172,-1.3757],[182,-1.5565],[192,-1.7713],[202,-2.2335],[212,-0.7902],[220,2.5595],[230,1.6213],[238,0.476],[248,0],[270,0],[273,-4.8395],[275,5.3022],[279,23.6697],[283,27.6475],[287,28.642],[292,28.8906],[299,28.9527],[348,28.9734],[349,28.8135],[350,28.7176],[351,28.3977],[352,28.1582],[353,27.0356],[355,21.44],[356,18.7159],[358,18.317],[359,18.317],[360,12.7799],[361,9.585],[362,7.1887],[364,10.0427],[370,10],[374,7.2807],[381,-1.9384],[384,-3.6104],[402,0.5304],[406,0.7259],[424,0],[439,0]]
       }
       thumb_02_r_ctrl.Scale.X {type=MovieSceneScriptingFloatChannel, default=1, const [1,62]}
       thumb_02_r_ctrl.Scale.Y {type=MovieSceneScriptingFloatChannel, default=1, const [1,62]}
       thumb_02_r_ctrl.Scale.Z {type=MovieSceneScriptingFloatChannel, default=1, const [1,62]}
       thumb_03_r_ctrl.Location.X {type=MovieSceneScriptingFloatChannel, default=0, const [0,62]}
       thumb_03_r_ctrl.Location.Y {type=MovieSceneScriptingFloatChannel, default=0, const [0,62]}
       thumb_03_r_ctrl.Location.Z {type=MovieSceneScriptingFloatChannel, default=0, const [0,62]}
       thumb_03_r_ctrl.Rotation.X {type=MovieSceneScriptingFloatChannel, default=0, const [0,62]}
       thumb_03_r_ctrl.Rotation.Y {type=MovieSceneScriptingFloatChannel, default=0, const [0,62]}
       thumb_03_r_ctrl.Rotation.Z {
        type=MovieSceneScriptingFloatChannel
        default=36.4941
        thinned_from=67
        keys [[0,0],[2,0],[12,0.1176],[22,0.6002],[32,0.9679],[42,1.0904],[52,0.6815],[62,0.2556],[72,-0.493],[82,-1.2938],[92,-0.6545],[102,-0.1636],[112,0.3978],[122,1.6309],[132,2.3838],[142,1.4899],[152,0.3725],[162,-0.8331],[172,-1.3757],[182,-1.5565],[192,-1.7713],[202,-2.2335],[212,-0.7902],[220,2.5595],[230,1.6213],[238,0.476],[248,0],[270,0],[273,28.5975],[275,29.6577],[279,36.0478],[283,35.3354],[287,35.1573],[350,35.1438],[352,35.244],[353,35.445],[356,36.935],[359,37.0065],[360,26.6803],[361,20.0102],[362,15.0077],[364,14.5077],[367,12.188],[370,10],[374,7.2807],[381,-1.9384],[384,-3.6104],[402,0.5304],[406,0.7259],[424,0],[439,0]]
       }
       thumb_03_r_ctrl.Scale.X {type=MovieSceneScriptingFloatChannel, default=1, const [1,62]}
       thumb_03_r_ctrl.Scale.Y {type=MovieSceneScriptingFloatChannel, default=1, const [1,62]}
       thumb_03_r_ctrl.Scale.Z {type=MovieSceneScriptingFloatChannel, default=1, const [1,62]}
       index_metacarpal_r_ctrl.Location.X {type=MovieSceneScriptingFloatChannel, default=0, const [0,62]}
       index_metacarpal_r_ctrl.Location.Y {type=MovieSceneScriptingFloatChannel, default=0, const [0,62]}
       index_metacarpal_r_ctrl.Location.Z {type=MovieSceneScriptingFloatChannel, default=0, const [0,62]}
       index_metacarpal_r_ctrl.Rotation.X {
        type=MovieSceneScriptingFloatChannel
        default=0.1653
        thinned_from=62
        keys [[0,0],[272,0],[273,0.0982],[275,0.1318],[279,0.1762],[283,0.1935],[287,0.1979],[292,0.199],[348,0.1993],[350,0.1982],[352,0.1958],[353,0.1909],[355,0.1664],[356,0.1545],[358,0.1528],[359,0.1528],[360,0.1081],[361,0.0811],[362,0.0608],[364,0.0519],[367,0.0277],[370,0.0107],[374,0],[439,0]]
       }
       index_metacarpal_r_ctrl.Rotation.Y {
        type=MovieSceneScriptingFloatChannel
        default=-8.4202
        thinned_from=63
        keys [[0,2],[272,2],[273,-0.027],[274,0.5],[275,-6.8201],[279,-11.0087],[283,-12.6489],[287,-13.0589],[292,-13.1614],[309,-13.193],[348,-13.1956],[349,-14.4388],[350,-15.1848],[351,-17.6712],[352,-16.9197],[353,-15.5528],[354,-14.1482],[355,-13.1856],[356,-12.4662],[357,-11.47],[358,-10.0906],[359,-7.8706],[360,-5.4537],[361,-2.6377],[362,-1.4783],[364,-0.9688],[367,0.4155],[370,1.389],[374,2],[439,2]]
       }
       index_metacarpal_r_ctrl.Rotation.Z {
        type=MovieSceneScriptingFloatChannel
        default=4.1272
        thinned_from=67
        keys [[0,0],[2,0],[12,0.1176],[22,0.6002],[32,0.9679],[42,1.0904],[52,0.6815],[62,0.2556],[72,-0.493],[82,-1.2938],[92,-0.6545],[102,-0.1636],[112,0.3978],[122,1.6309],[132,2.3838],[142,1.4899],[152,0.3725],[162,-0.8331],[172,-1.3757],[182,-1.5565],[192,-1.7713],[202,-2.2335],[212,-0.7902],[220,2.5595],[230,1.6213],[238,0.476],[248,0],[270,0],[273,-9.056],[275,-4.2161],[279,2.1672],[283,4.6668],[287,5.2917],[292,5.4479],[299,5.487],[348,5.5],[349,7.8438],[350,9.25],[351,13.9375],[352,12.8304],[353,11.4308],[354,8.4965],[355,5.9213],[356,3.4056],[358,3.2939],[359,3.2939],[360,2.0618],[361,1.5463],[362,1.1597],[364,1.9506],[367,6.093],[370,10],[374,7.2807],[381,-1.9384],[384,-3.6104],[402,0.5304],[406,0.7259],[424,0],[439,0]]
       }
       index_metacarpal_r_ctrl.Scale.X {type=MovieSceneScriptingFloatChannel, default=1, const [1,62]}
       index_metacarpal_r_ctrl.Scale.Y {type=MovieSceneScriptingFloatChannel, default=1, const [1,62]}
       index_metacarpal_r_ctrl.Scale.Z {type=MovieSceneScriptingFloatChannel, default=1, const [1,62]}
       index_01_r_ctrl.Location.X {type=MovieSceneScriptingFloatChannel, default=0, const [0,62]}
       index_01_r_ctrl.Location.Y {type=MovieSceneScriptingFloatChannel, default=0, const [0,62]}
       index_01_r_ctrl.Location.Z {type=MovieSceneScriptingFloatChannel, default=0, const [0,62]}
       index_01_r_ctrl.Rotation.X {
        type=MovieSceneScriptingFloatChannel
        default=-1.6762
        thinned_from=62
        keys [[0,0],[272,0],[273,-0.492],[275,-1.0855],[279,-1.8682],[283,-2.1747],[287,-2.2513],[292,-2.2705],[299,-2.2752],[348,-2.2768],[349,-2.2645],[350,-2.2571],[351,-2.2325],[352,-2.214],[353,-2.1275],[355,-1.6964],[356,-1.6964],[357,-1.5782],[358,-1.4557],[359,-1.4557],[360,-1.0411],[361,-0.7808],[362,-0.5856],[364,-0.4998],[367,-0.2668],[370,-0.1029],[374,0],[439,0]]
       }
       index_01_r_ctrl.Rotation.Y {
        type=MovieSceneScriptingFloatChannel
        default=-8.1076
        thinned_from=62
        keys [[0,0],[272,0],[273,-7.2137],[275,-9.576],[279,-10.6271],[283,-9.7843],[287,-9.5736],[292,-9.5209],[348,-9.5034],[350,-9.4576],[352,-9.3574],[353,-9.1564],[355,-8.1546],[356,-8.1546],[358,-7.5954],[359,-7.5954],[360,-5.447],[361,-4.0852],[362,-3.0639],[364,-2.6152],[367,-1.3958],[370,-0.5382],[374,0],[439,0]]
       }
       index_01_r_ctrl.Rotation.Z {
        type=MovieSceneScriptingFloatChannel
        default=23.711
        thinned_from=68
        keys [[0,6],[2,6],[12,6.1176],[22,6.6002],[32,6.9679],[42,7.0904],[52,6.6815],[62,6.2556],[72,5.507],[82,4.7062],[92,5.3455],[102,5.8364],[112,6.3978],[122,7.6309],[132,8.3838],[142,7.4899],[152,6.3725],[162,5.1669],[172,4.6243],[182,4.4435],[192,4.2287],[202,3.7665],[212,5.2098],[220,8.5595],[230,7.6213],[238,6.476],[248,6],[270,6],[273,-17.1529],[274,-21],[275,19.1592],[279,29.5719],[283,34.783],[287,36.0858],[292,36.4115],[299,36.4929],[348,36.5201],[349,36.1738],[350,35.966],[351,35.2734],[352,34.8842],[353,33.0874],[354,29.2977],[355,25.9445],[356,25.9445],[357,19.0946],[358,12],[359,12],[360,10.6172],[361,9.4629],[362,8.5972],[364,9.254],[367,12.7256],[370,16],[374,13.2807],[381,4.0616],[384,2.3896],[402,6.5304],[406,6.7259],[424,6],[439,6]]
       }
       index_01_r_ctrl.Scale.X {type=MovieSceneScriptingFloatChannel, default=1, const [1,62]}
       index_01_r_ctrl.Scale.Y {type=MovieSceneScriptingFloatChannel, default=1, const [1,62]}
       index_01_r_ctrl.Scale.Z {type=MovieSceneScriptingFloatChannel, default=1, const [1,62]}
       index_02_r_ctrl.Location.X {type=MovieSceneScriptingFloatChannel, default=0, const [0,62]}
       index_02_r_ctrl.Location.Y {type=MovieSceneScriptingFloatChannel, default=0, const [0,62]}
       index_02_r_ctrl.Location.Z {type=MovieSceneScriptingFloatChannel, default=0, const [0,62]}
       index_02_r_ctrl.Rotation.X {type=MovieSceneScriptingFloatChannel, default=0, const [0,62]}
       index_02_r_ctrl.Rotation.Y {type=MovieSceneScriptingFloatChannel, default=0, const [0,62]}
       index_02_r_ctrl.Rotation.Z {
        type=MovieSceneScriptingFloatChannel
        default=34.9382
        thinned_from=67
        keys [[0,19.5],[2,19.5],[12,19.6176],[22,20.1002],[32,20.4679],[42,20.5904],[62,19.7556],[72,19.007],[82,18.2062],[92,18.8456],[102,19.3364],[112,19.8978],[122,21.1309],[132,21.8838],[142,20.9899],[152,19.8725],[162,18.6669],[172,18.1243],[192,17.7287],[202,17.2665],[212,18.7098],[220,22.0595],[230,21.1213],[238,19.976],[248,19.5],[270,19.5],[273,21.444],[275,23.6295],[279,40.8514],[283,42.3455],[287,42.719],[292,42.8123],[348,42.8435],[349,42.604],[350,42.4604],[351,41.9815],[352,41.7427],[353,40.6484],[354,38.6234],[356,35.416],[357,32.9903],[358,30.478],[359,30.478],[360,27.4107],[361,25.433],[362,23.9498],[364,24.4339],[367,27.041],[370,29.5],[374,26.7807],[381,17.5616],[384,15.8896],[402,20.0304],[406,20.2259],[424,19.5],[439,19.5]]
       }
       index_02_r_ctrl.Scale.X {type=MovieSceneScriptingFloatChannel, default=1, const [1,62]}
       index_02_r_ctrl.Scale.Y {type=MovieSceneScriptingFloatChannel, default=1, const [1,62]}
       index_02_r_ctrl.Scale.Z {type=MovieSceneScriptingFloatChannel, default=1, const [1,62]}
       index_03_r_ctrl.Location.X {type=MovieSceneScriptingFloatChannel, default=0, const [0,62]}
       index_03_r_ctrl.Location.Y {type=MovieSceneScriptingFloatChannel, default=0, const [0,62]}
       index_03_r_ctrl.Location.Z {type=MovieSceneScriptingFloatChannel, default=0, const [0,62]}
       index_03_r_ctrl.Rotation.X {type=MovieSceneScriptingFloatChannel, default=0, const [0,62]}
       index_03_r_ctrl.Rotation.Y {type=MovieSceneScriptingFloatChannel, default=0, const [0,62]}
       index_03_r_ctrl.Rotation.Z {
        type=MovieSceneScriptingFloatChannel
        default=36.191
        thinned_from=67
        keys [[0,21.5],[12,21.6176],[22,22.1002],[32,22.4679],[42,22.5904],[62,21.7556],[72,21.007],[82,20.2062],[92,20.8456],[102,21.3364],[112,21.8978],[122,23.1309],[132,23.8838],[142,22.9899],[152,21.8725],[162,20.6669],[172,20.1243],[192,19.7287],[202,19.2665],[212,20.7098],[220,24.0595],[230,23.1213],[238,21.976],[248,21.5],[270,21.5],[275,33.1583],[279,37.1767],[283,38.7503],[287,39.1437],[292,39.242],[348,39.2748],[350,39.1736],[352,38.9523],[353,38.5082],[356,35.2171],[358,35.0593],[359,35.0593],[360,31.089],[361,28.6917],[362,26.8938],[364,27.2899],[367,29.4565],[370,31.5],[374,28.7807],[381,19.5616],[384,17.8896],[402,22.0304],[406,22.2259],[424,21.5],[439,21.5]]
       }
       index_03_r_ctrl.Scale.X {type=MovieSceneScriptingFloatChannel, default=1, const [1,62]}
       index_03_r_ctrl.Scale.Y {type=MovieSceneScriptingFloatChannel, default=1, const [1,62]}
       index_03_r_ctrl.Scale.Z {type=MovieSceneScriptingFloatChannel, default=1, const [1,62]}
       middle_metacarpal_r_ctrl.Location.X {type=MovieSceneScriptingFloatChannel, default=0, const [0,62]}
       middle_metacarpal_r_ctrl.Location.Y {type=MovieSceneScriptingFloatChannel, default=0, const [0,62]}
       middle_metacarpal_r_ctrl.Location.Z {type=MovieSceneScriptingFloatChannel, default=0, const [0,62]}
       middle_metacarpal_r_ctrl.Rotation.X {
        type=MovieSceneScriptingFloatChannel
        default=-0.2315
        thinned_from=62
        keys [[0,0],[272,0],[273,-0.1212],[275,-0.1764],[279,-0.2493],[283,-0.2779],[287,-0.285],[292,-0.2868],[348,-0.2874],[350,-0.2856],[351,-0.2833],[352,-0.2816],[353,-0.2735],[355,-0.2333],[356,-0.2138],[358,-0.2109],[359,-0.2109],[360,-0.1488],[361,-0.1116],[362,-0.0837],[364,-0.0714],[367,-0.0381],[370,-0.0147],[374,0],[439,0]]
       }
       middle_metacarpal_r_ctrl.Rotation.Y {
        type=MovieSceneScriptingFloatChannel
        default=-1.4266
        thinned_from=63
        keys [[0,0],[272,0],[273,1.0375],[274,1],[275,-0.5892],[279,-1.6988],[283,-2.1333],[287,-2.242],[292,-2.2691],[299,-2.2759],[348,-2.2782],[349,-2.2607],[350,-2.2502],[351,-2.2153],[352,-2.1891],[353,-2.0665],[355,-1.4553],[356,-1.1577],[358,-1.1141],[359,-1.1141],[360,-0.7575],[361,-0.5681],[362,-0.4261],[364,0.062],[367,1.3883],[370,3],[374,0],[439,0]]
       }
       middle_metacarpal_r_ctrl.Rotation.Z {
        type=MovieSceneScriptingFloatChannel
        default=5.7115
        thinned_from=67
        keys [[0,0],[2,0],[12,0.1176],[22,0.6002],[32,0.9679],[42,1.0904],[52,0.6815],[62,0.2556],[72,-0.493],[82,-1.2938],[92,-0.6545],[102,-0.1636],[112,0.3978],[122,1.6309],[132,2.3838],[142,1.4899],[152,0.3725],[162,-0.8331],[172,-1.3757],[182,-1.5565],[192,-1.7713],[202,-2.2335],[212,-0.7902],[220,2.5595],[230,1.6213],[238,0.476],[248,0],[270,0],[273,-3.7178],[275,1.0078],[279,7.2403],[283,9.6808],[287,10.2909],[292,10.4435],[299,10.4816],[348,10.4943],[349,10.3962],[350,10.3373],[351,10.1411],[352,9.9941],[353,9.3054],[355,5.8722],[356,4.2009],[358,3.9561],[359,3.9561],[360,2.6003],[361,1.9502],[362,1.4627],[364,2.2253],[367,6.2263],[370,10],[374,7.2807],[381,-1.9384],[384,-3.6104],[402,0.5304],[406,0.7259],[424,0],[439,0]]
       }
       middle_metacarpal_r_ctrl.Scale.X {type=MovieSceneScriptingFloatChannel, default=1, const [1,62]}
       middle_metacarpal_r_ctrl.Scale.Y {type=MovieSceneScriptingFloatChannel, default=1, const [1,62]}
       middle_metacarpal_r_ctrl.Scale.Z {type=MovieSceneScriptingFloatChannel, default=1, const [1,62]}
       middle_01_r_ctrl.Location.X {type=MovieSceneScriptingFloatChannel, default=0, const [0,62]}
       middle_01_r_ctrl.Location.Y {type=MovieSceneScriptingFloatChannel, default=0, const [0,62]}
       middle_01_r_ctrl.Location.Z {type=MovieSceneScriptingFloatChannel, default=0, const [0,62]}
       middle_01_r_ctrl.Rotation.X {
        type=MovieSceneScriptingFloatChannel
        default=-0.7875
        thinned_from=62
        keys [[0,-9.5],[272,-9.5],[273,-1.3311],[275,-1.0587],[279,-0.6994],[283,-0.5587],[287,-0.5235],[292,-0.5147],[299,-0.5125],[348,-0.5118],[349,-0.5175],[350,-0.5208],[351,-0.5322],[352,-0.5406],[353,-0.5803],[354,-0.6796],[355,-0.7771],[356,-0.7771],[357,-0.8319],[358,-0.8887],[359,-0.8887],[360,-3.3156],[361,-4.8617],[362,-6.0213],[364,-6.5308],[367,-7.9153],[370,-8.889],[374,-9.5],[439,-9.5]]
       }
       middle_01_r_ctrl.Rotation.Y {
        type=MovieSceneScriptingFloatChannel
        default=-10.8211
        thinned_from=62
        keys [[0,-3],[272,-3],[273,-10.509],[275,-10.6654],[279,-10.8717],[283,-10.9525],[352,-10.9629],[353,-10.9401],[354,-10.8831],[355,-7.5],[356,-5.5],[357,-8.0853],[358,-10.763],[359,-10.763],[360,-8.5768],[361,-7.1826],[362,-6.137],[364,-5.6775],[367,-4.429],[370,-3.551],[374,-3],[439,-3]]
       }
       middle_01_r_ctrl.Rotation.Z {
        type=MovieSceneScriptingFloatChannel
        default=18.9787
        thinned_from=68
        keys [[0,3],[2,3],[12,3.1176],[22,3.6002],[32,3.9679],[42,4.0904],[52,3.6815],[62,3.2556],[72,2.507],[82,1.7062],[92,2.3455],[102,2.8364],[112,3.3978],[122,4.6309],[132,5.3838],[142,4.4899],[152,3.3725],[162,2.1669],[172,1.6243],[182,1.4435],[192,1.2287],[202,0.7665],[212,2.2098],[220,5.5595],[230,4.6213],[238,3.476],[248,3],[270,3],[273,-28.058],[274,-27],[275,8.4756],[279,31.8969],[283,39.0478],[287,40.8355],[292,41.2824],[299,41.3942],[348,41.4314],[349,40.9542],[350,40.668],[351,39.7137],[352,39.0247],[353,35.8009],[354,27.8787],[355,34],[356,34],[357,14.5965],[358,-5.5],[359,-5.5],[360,-2.8047],[361,-1.3535],[362,-0.2651],[364,0.9381],[367,7.1454],[370,13],[374,10.2807],[381,1.0616],[384,-0.6104],[402,3.5304],[406,3.7259],[424,3],[439,3]]
       }
       middle_01_r_ctrl.Scale.X {type=MovieSceneScriptingFloatChannel, default=1, const [1,62]}
       middle_01_r_ctrl.Scale.Y {type=MovieSceneScriptingFloatChannel, default=1, const [1,62]}
       middle_01_r_ctrl.Scale.Z {type=MovieSceneScriptingFloatChannel, default=1, const [1,62]}
       middle_02_r_ctrl.Location.X {type=MovieSceneScriptingFloatChannel, default=0, const [0,62]}
       middle_02_r_ctrl.Location.Y {type=MovieSceneScriptingFloatChannel, default=0, const [0,62]}
       middle_02_r_ctrl.Location.Z {type=MovieSceneScriptingFloatChannel, default=0, const [0,62]}
       middle_02_r_ctrl.Rotation.X {type=MovieSceneScriptingFloatChannel, default=0, const [0,62]}
       middle_02_r_ctrl.Rotation.Y {type=MovieSceneScriptingFloatChannel, default=0, const [0,62]}
       middle_02_r_ctrl.Rotation.Z {
        type=MovieSceneScriptingFloatChannel
        default=32.7186
        thinned_from=67
        keys [[0,-17],[2,-17],[12,-16.8824],[22,-16.3998],[32,-16.0321],[42,-15.9096],[62,-16.7444],[72,-17.493],[82,-18.2938],[92,-17.6544],[102,-17.1636],[112,-16.6022],[122,-15.3691],[132,-14.6162],[142,-15.5101],[152,-16.6275],[162,-17.8331],[172,-18.3757],[192,-18.7713],[202,-19.2335],[212,-17.7902],[220,-14.4405],[230,-15.3787],[238,-16.524],[248,-17],[270,-17],[273,12.7208],[275,27.6486],[279,42.709],[283,46.5858],[287,47.555],[292,47.7972],[348,47.878],[349,47.3355],[350,47.01],[351,45.9251],[352,45.4714],[353,43.4205],[354,41.0024],[356,38.7218],[357,36.6479],[358,34.5],[359,34.5],[360,19.7659],[361,10.5744],[362,3.6808],[364,5.2649],[367,-1.0469],[370,-7],[374,-9.7193],[381,-18.9384],[384,-20.6104],[402,-16.4696],[406,-16.2741],[424,-17],[439,-17]]
       }
       middle_02_r_ctrl.Scale.X {type=MovieSceneScriptingFloatChannel, default=1, const [1,62]}
       middle_02_r_ctrl.Scale.Y {type=MovieSceneScriptingFloatChannel, default=1, const [1,62]}
       middle_02_r_ctrl.Scale.Z {type=MovieSceneScriptingFloatChannel, default=1, const [1,62]}
       middle_03_r_ctrl.Location.X {type=MovieSceneScriptingFloatChannel, default=0, const [0,62]}
       middle_03_r_ctrl.Location.Y {type=MovieSceneScriptingFloatChannel, default=0, const [0,62]}
       middle_03_r_ctrl.Location.Z {type=MovieSceneScriptingFloatChannel, default=0, const [0,62]}
       middle_03_r_ctrl.Rotation.X {type=MovieSceneScriptingFloatChannel, default=0, const [0,62]}
       middle_03_r_ctrl.Rotation.Y {type=MovieSceneScriptingFloatChannel, default=0, const [0,62]}
       middle_03_r_ctrl.Rotation.Z {
        type=MovieSceneScriptingFloatChannel
        default=33.8368
        thinned_from=67
        keys [[0,-39.5],[12,-39.3824],[22,-38.8998],[32,-38.5321],[42,-38.4096],[62,-39.2444],[82,-40.7938],[92,-40.1544],[112,-39.1022],[122,-37.8691],[132,-37.1162],[142,-38.0101],[152,-39.1275],[162,-40.3331],[172,-40.8757],[192,-41.2713],[202,-41.7335],[212,-40.2902],[220,-36.9405],[230,-37.8787],[238,-39.024],[248,-39.5],[270,-39.5],[273,27.3707],[275,31.4073],[279,34.6264],[283,35.887],[287,36.2021],[292,36.2809],[348,36.3071],[350,36.2261],[352,36.0488],[353,35.6931],[356,33.0565],[358,32.9301],[359,32.9301],[360,12.4336],[361,-0.5498],[362,-10.2874],[364,-9.4985],[367,-19.7917],[370,-29.5],[374,-32.2193],[381,-41.4384],[384,-43.1104],[402,-38.9696],[406,-38.7741],[424,-39.5],[439,-39.5]]
       }
       middle_03_r_ctrl.Scale.X {type=MovieSceneScriptingFloatChannel, default=1, const [1,62]}
       middle_03_r_ctrl.Scale.Y {type=MovieSceneScriptingFloatChannel, default=1, const [1,62]}
       middle_03_r_ctrl.Scale.Z {type=MovieSceneScriptingFloatChannel, default=1, const [1,62]}
       ring_metacarpal_r_ctrl.Location.X {type=MovieSceneScriptingFloatChannel, default=0, const [0,62]}
       ring_metacarpal_r_ctrl.Location.Y {type=MovieSceneScriptingFloatChannel, default=0, const [0,62]}
       ring_metacarpal_r_ctrl.Location.Z {type=MovieSceneScriptingFloatChannel, default=0, const [0,62]}
       ring_metacarpal_r_ctrl.Rotation.X {
        type=MovieSceneScriptingFloatChannel
        default=0.0042
        thinned_from=62
        keys [[0,0],[272,0],[273,0.0125],[275,0.0084],[279,0.0029],[283,0.0007],[287,0.0002],[350,0.0001],[352,0.0004],[353,0.001],[356,0.0055],[359,0.0058],[360,0.0044],[361,0.0033],[362,0.0025],[364,0.0021],[367,0.0011],[370,0.0004],[374,0],[439,0]]
       }
       ring_metacarpal_r_ctrl.Rotation.Y {
        type=MovieSceneScriptingFloatChannel
        default=-0.2842
        thinned_from=62
        keys [[0,-6.5],[272,-6.5],[273,-0.8446],[275,-0.5638],[279,-0.1934],[283,-0.0483],[287,-0.0121],[292,-0.003],[299,-0.0008],[309,-0.0002],[348,0],[349,-0.0058],[350,-0.0093],[351,-0.021],[352,-0.0297],[353,-0.0707],[354,-0.173],[355,-0.2747],[356,-0.374],[358,-0.3886],[359,-0.3886],[360,-2.1219],[361,-3.2164],[362,-4.0373],[364,-4.398],[367,-5.3781],[370,-6.0674],[374,-6.5],[439,-6.5]]
       }
       ring_metacarpal_r_ctrl.Rotation.Z {
        type=MovieSceneScriptingFloatChannel
        default=-2.6817
        thinned_from=67
        keys [[0,0],[2,0],[12,0.1176],[22,0.6002],[32,0.9679],[42,1.0904],[52,0.6815],[62,0.2556],[72,-0.493],[82,-1.2938],[92,-0.6545],[102,-0.1636],[112,0.3978],[122,1.6309],[132,2.3838],[142,1.4899],[152,0.3725],[162,-0.8331],[172,-1.3757],[182,-1.5565],[192,-1.7713],[202,-2.2335],[212,-0.7902],[220,2.5595],[230,1.6213],[238,0.476],[248,0],[270,0],[273,-2.5242],[275,-3.4103],[279,-2.8343],[283,-3.0778],[287,-3.1386],[292,-3.1539],[348,-3.1589],[350,-3.1433],[352,-3.109],[353,-3.0403],[355,-2.6978],[356,-2.531],[358,-2.5066],[359,-2.5066],[360,-1.7773],[361,-1.333],[362,-0.9998],[364,-0.0077],[367,5.1425],[370,10],[374,7.2807],[381,-1.9384],[384,-3.6104],[402,0.5304],[406,0.7259],[424,0],[439,0]]
       }
       ring_metacarpal_r_ctrl.Scale.X {type=MovieSceneScriptingFloatChannel, default=1, const [1,62]}
       ring_metacarpal_r_ctrl.Scale.Y {type=MovieSceneScriptingFloatChannel, default=1, const [1,62]}
       ring_metacarpal_r_ctrl.Scale.Z {type=MovieSceneScriptingFloatChannel, default=1, const [1,62]}
       ring_01_r_ctrl.Location.X {type=MovieSceneScriptingFloatChannel, default=0, const [0,62]}
       ring_01_r_ctrl.Location.Y {type=MovieSceneScriptingFloatChannel, default=0, const [0,62]}
       ring_01_r_ctrl.Location.Z {type=MovieSceneScriptingFloatChannel, default=0, const [0,62]}
       ring_01_r_ctrl.Rotation.X {
        type=MovieSceneScriptingFloatChannel
        default=-1.774
        thinned_from=62
        keys [[0,-21.3989],[272,-21.3989],[273,-5.9842],[275,-3.8742],[279,-1.0914],[283,-0.0017],[287,0.2707],[292,0.3388],[299,0.3559],[309,0.3598],[337,0.3615],[348,0.3615],[349,0.3177],[350,0.2914],[351,0.2038],[352,0.1382],[353,-0.1693],[354,-0.9383],[355,-1.7022],[356,-2.4485],[358,-2.5578],[359,-2.5578],[360,-7.9339],[361,-11.3001],[362,-13.8248],[364,-14.9341],[367,-17.9485],[370,-20.0685],[374,-21.3989],[439,-21.3989]]
       }
       ring_01_r_ctrl.Rotation.Y {
        type=MovieSceneScriptingFloatChannel
        default=-7.4099
        thinned_from=62
        keys [[0,10.8238],[272,10.8238],[273,-5.9721],[275,-6.0845],[279,-6.7749],[283,-7.5736],[287,-7.7732],[292,-7.8231],[348,-7.8398],[352,-7.7948],[353,-7.7329],[356,-7.2741],[358,-7.2521],[359,-7.2521],[360,-2.1527],[361,1.0914],[362,3.5245],[364,4.5936],[367,7.4986],[370,9.5417],[374,10.8238],[439,10.8238]]
       }
       ring_01_r_ctrl.Rotation.Z {
        type=MovieSceneScriptingFloatChannel
        default=17.8405
        thinned_from=67
        keys [[0,0.5],[2,0.5],[12,0.6176],[22,1.1002],[32,1.4679],[42,1.5904],[52,1.1815],[62,0.7556],[72,0.007],[82,-0.7938],[92,-0.1545],[102,0.3364],[112,0.8978],[122,2.1309],[132,2.8838],[142,1.9899],[152,0.8725],[162,-0.3331],[172,-0.8757],[182,-1.0565],[192,-1.2713],[202,-1.7335],[212,-0.2902],[220,3.0595],[230,2.1213],[238,0.976],[248,0.5],[270,0.5],[273,-16.1177],[275,1.2866],[279,29.4076],[283,36.8212],[287,38.6746],[292,39.1379],[299,39.2538],[348,39.2924],[349,38.8387],[350,38.5666],[351,37.6592],[352,37.0009],[353,33.9195],[354,26.3285],[355,18.8543],[356,11.553],[357,8.5964],[358,5.5342],[359,5.5342],[360,3.7912],[361,2.9684],[362,2.3513],[364,3.0777],[367,6.8974],[370,10.5],[374,7.7807],[381,-1.4384],[384,-3.1104],[402,1.0304],[406,1.2259],[424,0.5],[439,0.5]]
       }
       ring_01_r_ctrl.Scale.X {type=MovieSceneScriptingFloatChannel, default=1, const [1,62]}
       ring_01_r_ctrl.Scale.Y {type=MovieSceneScriptingFloatChannel, default=1, const [1,62]}
       ring_01_r_ctrl.Scale.Z {type=MovieSceneScriptingFloatChannel, default=1, const [1,62]}
       ring_02_r_ctrl.Location.X {type=MovieSceneScriptingFloatChannel, default=0, const [0,62]}
       ring_02_r_ctrl.Location.Y {type=MovieSceneScriptingFloatChannel, default=0, const [0,62]}
       ring_02_r_ctrl.Location.Z {type=MovieSceneScriptingFloatChannel, default=0, const [0,62]}
       ring_02_r_ctrl.Rotation.X {type=MovieSceneScriptingFloatChannel, default=0, const [0,62]}
       ring_02_r_ctrl.Rotation.Y {type=MovieSceneScriptingFloatChannel, default=0, const [0,62]}
       ring_02_r_ctrl.Rotation.Z {
        type=MovieSceneScriptingFloatChannel
        default=28.8073
        thinned_from=67
        keys [[0,-39],[12,-38.8824],[22,-38.3998],[32,-38.0321],[42,-37.9096],[62,-38.7444],[82,-40.2938],[92,-39.6544],[112,-38.6022],[122,-37.3691],[132,-36.6162],[142,-37.5101],[152,-38.6275],[162,-39.8331],[172,-40.3757],[192,-40.7713],[202,-41.2335],[212,-39.7902],[220,-36.4405],[230,-37.3787],[238,-38.524],[248,-39],[270,-39],[273,10.3629],[275,22.752],[279,45.3663],[283,50.6236],[287,51.9379],[292,52.2665],[348,52.376],[349,51.8665],[350,51.5608],[351,50.5418],[352,49.819],[353,46.4399],[354,38.209],[355,30.1605],[356,22.2981],[357,19.3238],[358,16.2433],[359,16.2433],[360,0.6044],[361,-9.2967],[362,-16.7225],[364,-15.2039],[367,-22.3037],[370,-29],[374,-31.7193],[381,-40.9384],[384,-42.6104],[402,-38.4696],[406,-38.2741],[424,-39],[439,-39]]
       }
       ring_02_r_ctrl.Scale.X {type=MovieSceneScriptingFloatChannel, default=1, const [1,62]}
       ring_02_r_ctrl.Scale.Y {type=MovieSceneScriptingFloatChannel, default=1, const [1,62]}
       ring_02_r_ctrl.Scale.Z {type=MovieSceneScriptingFloatChannel, default=1, const [1,62]}
       ring_03_r_ctrl.Location.X {type=MovieSceneScriptingFloatChannel, default=0, const [0,62]}
       ring_03_r_ctrl.Location.Y {type=MovieSceneScriptingFloatChannel, default=0, const [0,62]}
       ring_03_r_ctrl.Location.Z {type=MovieSceneScriptingFloatChannel, default=0, const [0,62]}
       ring_03_r_ctrl.Rotation.X {type=MovieSceneScriptingFloatChannel, default=0, const [0,62]}
       ring_03_r_ctrl.Rotation.Y {type=MovieSceneScriptingFloatChannel, default=0, const [0,62]}
       ring_03_r_ctrl.Rotation.Z {
        type=MovieSceneScriptingFloatChannel
        default=29.9269
        thinned_from=67
        keys [[0,-61],[12,-60.8824],[32,-60.0321],[42,-59.9096],[62,-60.7444],[82,-62.2938],[92,-61.6545],[112,-60.6022],[122,-59.3691],[132,-58.6162],[142,-59.5101],[162,-61.8331],[172,-62.3757],[192,-62.7713],[202,-63.2335],[212,-61.7902],[220,-58.4405],[230,-59.3787],[238,-60.524],[248,-61],[270,-61],[273,25.4692],[275,28.4993],[279,30.3909],[283,31.1317],[287,31.3169],[350,31.3309],[352,31.2268],[353,31.0177],[356,29.4684],[359,29.3941],[360,3.897],[361,-12.3273],[362,-24.4955],[364,-24.3029],[367,-38.0418],[370,-51],[374,-53.7193],[381,-62.9384],[384,-64.6104],[402,-60.4696],[406,-60.2741],[424,-61],[439,-61]]
       }
       ring_03_r_ctrl.Scale.X {type=MovieSceneScriptingFloatChannel, default=1, const [1,62]}
       ring_03_r_ctrl.Scale.Y {type=MovieSceneScriptingFloatChannel, default=1, const [1,62]}
       ring_03_r_ctrl.Scale.Z {type=MovieSceneScriptingFloatChannel, default=1, const [1,62]}
       pinky_metacarpal_r_ctrl.Location.X {type=MovieSceneScriptingFloatChannel, default=0, const [0,62]}
       pinky_metacarpal_r_ctrl.Location.Y {type=MovieSceneScriptingFloatChannel, default=0, const [0,62]}
       pinky_metacarpal_r_ctrl.Location.Z {type=MovieSceneScriptingFloatChannel, default=0, const [0,62]}
       pinky_metacarpal_r_ctrl.Rotation.X {
        type=MovieSceneScriptingFloatChannel
        default=-0.0689
        thinned_from=62
        keys [[0,0],[272,0],[273,-0.2048],[275,-0.1367],[279,-0.0469],[283,-0.0117],[287,-0.0029],[292,-0.0007],[299,-0.0002],[348,0],[349,-0.0014],[350,-0.0023],[351,-0.0051],[352,-0.0072],[353,-0.0171],[355,-0.0666],[356,-0.0907],[358,-0.0942],[359,-0.0942],[360,-0.0712],[361,-0.0534],[362,-0.0401],[364,-0.0342],[367,-0.0183],[370,-0.007],[374,0],[439,0]]
       }
       pinky_metacarpal_r_ctrl.Rotation.Y {
        type=MovieSceneScriptingFloatChannel
        default=-2.7944
        thinned_from=62
        keys [[0,-6],[272,-6],[273,-4.542],[275,-4.5402],[279,-1.6489],[283,-2.1659],[287,-2.2952],[292,-2.3275],[299,-2.3356],[348,-2.3383],[350,-2.3532],[352,-2.386],[353,-2.4517],[355,-2.7791],[356,-2.9385],[358,-2.9619],[359,-2.9619],[360,-3.8395],[361,-4.3796],[362,-4.7847],[364,-4.9627],[367,-5.4464],[370,-5.7865],[374,-6],[439,-6]]
       }
       pinky_metacarpal_r_ctrl.Rotation.Z {
        type=MovieSceneScriptingFloatChannel
        default=-2.0692
        thinned_from=67
        keys [[0,0],[2,0],[12,0.1176],[22,0.6002],[32,0.9679],[42,1.0904],[52,0.6815],[62,0.2556],[72,-0.493],[82,-1.2938],[92,-0.6545],[102,-0.1636],[112,0.3978],[122,1.6309],[132,2.3838],[142,1.4899],[152,0.3725],[162,-0.8331],[172,-1.3757],[182,-1.5565],[192,-1.7713],[202,-2.2335],[212,-0.7902],[220,2.5595],[230,1.6213],[238,0.476],[248,0],[270,0],[273,-8.1312],[275,-7.1361],[279,-1.4078],[283,-0.352],[287,-0.088],[292,-0.022],[299,-0.0055],[309,-0.0017],[327,-0.0004],[337,0],[348,0],[349,-0.0425],[350,-0.0679],[351,-0.1528],[352,-0.2164],[353,-0.5144],[354,-1.2595],[355,-1.9997],[356,-2.7227],[358,-2.8286],[359,-2.8286],[360,-2.1383],[361,-1.6037],[362,-1.2028],[364,-0.1918],[367,5.0531],[370,10],[374,7.2807],[381,-1.9384],[384,-3.6104],[402,0.5304],[406,0.7259],[424,0],[439,0]]
       }
       pinky_metacarpal_r_ctrl.Scale.X {type=MovieSceneScriptingFloatChannel, default=1, const [1,62]}
       pinky_metacarpal_r_ctrl.Scale.Y {type=MovieSceneScriptingFloatChannel, default=1, const [1,62]}
       pinky_metacarpal_r_ctrl.Scale.Z {type=MovieSceneScriptingFloatChannel, default=1, const [1,62]}
       pinky_01_r_ctrl.Location.X {type=MovieSceneScriptingFloatChannel, default=0, thinned_from=62, keys [[0,0],[439,0]]}
       pinky_01_r_ctrl.Location.Y {type=MovieSceneScriptingFloatChannel, default=0, thinned_from=62, keys [[0,0],[439,0]]}
       pinky_01_r_ctrl.Location.Z {type=MovieSceneScriptingFloatChannel, default=0, thinned_from=62, keys [[0,0],[439,0]]}
       pinky_01_r_ctrl.Rotation.X {
        type=MovieSceneScriptingFloatChannel
        default=-1.9469
        thinned_from=62
        keys [[0,-19.5247],[272,-19.5247],[273,-5.0701],[275,-3.5049],[279,-1.4405],[283,-0.6321],[287,-0.43],[292,-0.3795],[299,-0.3669],[309,-0.364],[348,-0.3627],[349,-0.3952],[350,-0.4147],[351,-0.4797],[352,-0.5283],[353,-0.7565],[354,-1.3269],[355,-1.8936],[356,-2.4472],[358,-2.5283],[359,-2.5283],[360,-7.3657],[361,-10.4054],[362,-12.6853],[364,-13.687],[367,-16.409],[370,-18.3234],[374,-19.5247],[439,-19.5247]]
       }
       pinky_01_r_ctrl.Rotation.Y {
        type=MovieSceneScriptingFloatChannel
        default=1.0199
        thinned_from=62
        keys [[0,-6.2887],[272,-6.2887],[273,-4.3687],[275,-1.6681],[279,1.8936],[283,3.2883],[287,3.637],[292,3.7241],[299,3.7459],[348,3.7532],[349,3.6971],[350,3.6635],[351,3.5513],[352,3.4674],[353,3.0738],[354,2.0895],[355,1.1118],[356,0.1567],[357,0.0879],[358,0.0168],[359,7.5],[360,2.8218],[361,0.5442],[362,-1.1641],[364,-1.9146],[367,-3.9542],[370,-5.3886],[374,-6.2887],[439,-6.2887]]
       }
       pinky_01_r_ctrl.Rotation.Z {
        type=MovieSceneScriptingFloatChannel
        default=1.2024
        thinned_from=68
        keys [[0,-1.3361],[2,-1.3361],[12,-1.2185],[22,-0.7359],[32,-0.3683],[42,-0.2457],[52,-0.6546],[62,-1.0806],[72,-1.8291],[82,-2.6299],[92,-1.9906],[102,-1.4997],[112,-0.9383],[122,0.2948],[132,1.0477],[142,0.1537],[152,-0.9637],[162,-2.1693],[172,-2.7118],[182,-2.8926],[192,-3.1074],[202,-3.5697],[212,-2.1263],[220,1.2234],[230,0.2852],[238,-0.8602],[248,-1.3361],[270,-1.3361],[273,-21.7634],[274,-16.5],[275,-9.8702],[279,5.8866],[283,9.0494],[287,9.8401],[292,10.0378],[299,10.0872],[348,10.1037],[349,9.9158],[350,9.803],[351,9.4272],[352,9.154],[353,7.8752],[354,4.7221],[355,1.6159],[356,-1.4185],[357,-4.0832],[358,-6.8432],[359,-6.8432],[360,-5.3696],[361,-4.3612],[362,-3.6049],[364,-2.4945],[367,3.2478],[370,8.6639],[374,5.9446],[381,-3.2745],[384,-4.9466],[402,-0.8057],[406,-0.6102],[424,-1.3361],[439,-1.3361]]
       }
       pinky_01_r_ctrl.Scale.X {type=MovieSceneScriptingFloatChannel, default=1, const [1,62]}
       pinky_01_r_ctrl.Scale.Y {type=MovieSceneScriptingFloatChannel, default=1, const [1,62]}
       pinky_01_r_ctrl.Scale.Z {type=MovieSceneScriptingFloatChannel, default=1, const [1,62]}
       pinky_02_r_ctrl.Location.X {type=MovieSceneScriptingFloatChannel, default=0, thinned_from=62, keys [[0,0],[439,0]]}
       pinky_02_r_ctrl.Location.Y {type=MovieSceneScriptingFloatChannel, default=0, thinned_from=62, keys [[0,0],[439,0]]}
       pinky_02_r_ctrl.Location.Z {type=MovieSceneScriptingFloatChannel, default=0, thinned_from=62, keys [[0,0],[439,0]]}
       pinky_02_r_ctrl.Rotation.X {
        type=MovieSceneScriptingFloatChannel
        default=0
        thinned_from=62
        keys [[0,-19.9993],[272,-19.9993],[273,0],[359,0],[360,-5.6248],[361,-9.2184],[362,-11.9136],[364,-13.0978],[367,-16.3158],[370,-18.579],[374,-19.9993],[439,-19.9993]]
       }
       pinky_02_r_ctrl.Rotation.Y {type=MovieSceneScriptingFloatChannel, default=0, thinned_from=62, keys [[0,0],[439,0]]}
       pinky_02_r_ctrl.Rotation.Z {
        type=MovieSceneScriptingFloatChannel
        default=34.7873
        thinned_from=67
        keys [[0,-59.5],[12,-59.3824],[32,-58.5321],[42,-58.4096],[62,-59.2444],[82,-60.7938],[92,-60.1544],[112,-59.1022],[122,-57.8691],[132,-57.1162],[142,-58.0101],[162,-60.3331],[172,-60.8757],[192,-61.2713],[202,-61.7335],[212,-60.2902],[220,-56.9405],[230,-57.8787],[238,-59.024],[248,-59.5],[270,-59.5],[273,14.2295],[275,25.5061],[279,45.8766],[283,50.846],[287,52.0884],[292,52.399],[348,52.5025],[349,52.1235],[350,51.896],[351,51.1379],[352,50.5944],[353,48.0524],[354,41.8272],[355,35.7201],[356,29.7542],[357,26.8974],[358,23.9386],[359,23.9386],[360,0.3292],[361,-14.6281],[362,-25.8461],[364,-23.5742],[367,-36.9161],[370,-49.5],[374,-52.2193],[381,-61.4384],[384,-63.1104],[402,-58.9696],[406,-58.7741],[424,-59.5],[439,-59.5]]
       }
       pinky_02_r_ctrl.Scale.X {type=MovieSceneScriptingFloatChannel, default=1, const [1,62]}
       pinky_02_r_ctrl.Scale.Y {type=MovieSceneScriptingFloatChannel, default=1, const [1,62]}
       pinky_02_r_ctrl.Scale.Z {type=MovieSceneScriptingFloatChannel, default=1, const [1,62]}
       pinky_03_r_ctrl.Location.X {type=MovieSceneScriptingFloatChannel, default=0, thinned_from=62, keys [[0,0],[439,0]]}
       pinky_03_r_ctrl.Location.Y {type=MovieSceneScriptingFloatChannel, default=0, thinned_from=62, keys [[0,0],[439,0]]}
       pinky_03_r_ctrl.Location.Z {type=MovieSceneScriptingFloatChannel, default=0, thinned_from=62, keys [[0,0],[439,0]]}
       pinky_03_r_ctrl.Rotation.X {type=MovieSceneScriptingFloatChannel, default=0, const [0,62]}
       pinky_03_r_ctrl.Rotation.Y {type=MovieSceneScriptingFloatChannel, default=0, const [0,62]}
       pinky_03_r_ctrl.Rotation.Z {
        type=MovieSceneScriptingFloatChannel
        default=22.6225
        thinned_from=67
        keys [[0,-59.5],[12,-59.3824],[32,-58.5321],[42,-58.4096],[62,-59.2444],[82,-60.7938],[92,-60.1544],[112,-59.1022],[122,-57.8691],[132,-57.1162],[142,-58.0101],[162,-60.3331],[172,-60.8757],[192,-61.2713],[202,-61.7335],[212,-60.2902],[220,-56.9405],[230,-57.8787],[238,-59.024],[248,-59.5],[270,-59.5],[273,18.0167],[275,21.121],[279,23.1105],[283,23.8896],[287,24.0844],[299,24.1452],[348,24.1493],[350,24.0992],[352,23.9896],[353,23.7698],[356,22.1403],[358,22.0622],[359,22.0622],[360,-0.9548],[361,-15.5911],[362,-26.5683],[364,-24.2292],[367,-37.2341],[370,-49.5],[374,-52.2193],[381,-61.4384],[384,-63.1104],[402,-58.9696],[406,-58.7741],[424,-59.5],[439,-59.5]]
       }
       pinky_03_r_ctrl.Scale.X {type=MovieSceneScriptingFloatChannel, default=1, const [1,62]}
       pinky_03_r_ctrl.Scale.Y {type=MovieSceneScriptingFloatChannel, default=1, const [1,62]}
       pinky_03_r_ctrl.Scale.Z {type=MovieSceneScriptingFloatChannel, default=1, const [1,62]}
       hand_r_ik_ctrl.Location.X {
        type=MovieSceneScriptingFloatChannel
        default=-9.0396
        thinned_from=441
        keys [[0,-14.6657],[1,-7.575],[11,-7.5644],[12,-7.5491],[13,-7.5554],[14,-7.5342],[15,-7.5441],[16,-7.5205],[18,-7.5119],[19,-7.4798],[20,-7.4842],[23,-7.4281],[24,-7.4221],[25,-7.4004],[26,-7.4053],[27,-7.3839],[29,-7.3769],[31,-7.3565],[32,-7.358],[34,-7.3379],[42,-7.1927],[47,-7.0517],[54,-6.7993],[56,-6.748],[57,-6.7485],[60,-6.795],[61,-6.8426],[62,-6.9346],[63,-7.0532],[64,-7.2465],[65,-7.53],[66,-7.8378],[67,-8.1138],[68,-8.4367],[69,-8.8753],[70,-9.3479],[71,-9.7599],[74,-10.8876],[75,-11.2291],[76,-11.5104],[77,-11.7433],[79,-12.0572],[82,-12.4507],[87,-12.921],[93,-13.3077],[97,-13.4636],[102,-13.6082],[109,-13.8607],[111,-13.9007],[112,-13.8938],[113,-13.7473],[114,-13.4028],[116,-12.4794],[117,-11.9783],[118,-11.3836],[120,-10.0905],[121,-9.4749],[122,-8.9224],[123,-8.4569],[125,-7.6446],[126,-7.2881],[127,-6.9981],[128,-6.7903],[130,-6.498],[131,-6.3656],[132,-6.2567],[133,-6.1738],[134,-6.1217],[135,-6.098],[136,-6.1042],[137,-6.1359],[138,-6.1971],[141,-6.4459],[144,-6.6557],[145,-6.6991],[146,-6.711],[147,-6.698],[148,-6.6664],[149,-6.6188],[150,-6.539],[151,-6.4202],[152,-6.2752],[154,-5.9416],[155,-5.7121],[156,-5.4338],[158,-4.82],[159,-4.5362],[160,-4.2843],[161,-4.0708],[163,-3.6956],[164,-3.5227],[165,-3.3588],[166,-3.2151],[167,-3.0811],[169,-2.8515],[171,-2.6457],[172,-2.5527],[174,-2.3906],[175,-2.3187],[177,-2.1875],[180,-2.014],[182,-1.9169],[184,-1.8246],[186,-1.7422],[187,-1.712],[188,-1.6989],[190,-1.6999],[194,-1.7196],[195,-1.7347],[196,-1.7822],[197,-1.867],[198,-1.9844],[199,-2.1249],[200,-2.3155],[201,-2.5697],[202,-2.9267],[203,-3.3986],[204,-3.935],[205,-4.5059],[206,-5.1338],[207,-5.8302],[208,-6.553],[210,-7.9456],[212,-9.3783],[213,-10.0745],[214,-10.7344],[215,-11.3275],[216,-11.862],[217,-12.3487],[218,-12.7664],[220,-13.4706],[221,-13.7917],[222,-14.0658],[223,-14.2875],[224,-14.4444],[226,-14.6373],[239,-15.3435],[242,-15.451],[243,-15.4406],[244,-15.3712],[245,-15.2505],[246,-15.0652],[249,-14.3336],[251,-13.9115],[252,-13.7668],[253,-13.6897],[254,-13.7161],[255,-13.8576],[256,-14.109],[257,-14.4657],[258,-14.9172],[259,-15.5852],[260,-16.5502],[261,-17.8024],[262,-19.1131],[263,-19.238],[264,-17.791],[265,-15.6888],[266,-13.7365],[267,-12.6839],[268,-34.7002],[269,-41.7687],[270,-50.305],[271,-54.3308],[272,-54.0285],[273,-52.1203],[274,-51.5303],[275,-51.1069],[276,-51.4678],[277,-51.3674],[278,-51.052],[279,-50.5557],[280,-49.7608],[281,-49.109],[282,-48.9433],[283,-49.0395],[284,-49.3213],[286,-50.6921],[287,-51.5455],[288,-51.9348],[289,-51.7855],[291,-51.1908],[292,-50.9772],[340,-50.6679],[348,-51.1649],[349,-50.971],[350,-50.4285],[352,-48.8694],[355,-46.0656],[356,-22.1288],[357,-19.1694],[358,-17.6671],[359,-16.8678],[360,-16.5718],[361,-16.7858],[362,-16.4487],[363,-16.6321],[364,-17.458],[365,-18.0786],[366,-18.4121],[367,-18.5257],[368,-18.5498],[369,-18.4957],[370,-18.3759],[374,-17.6799],[375,-17.5623],[376,-17.5114],[439,-17.2445],[440,-20.0579]]
       }
       hand_r_ik_ctrl.Location.Y {
        type=MovieSceneScriptingFloatChannel
        default=33.9829
        thinned_from=441
        keys [[0,9.4146],[1,18.3428],[52,18.1],[56,18.2076],[60,18.6106],[63,19.0128],[66,19.5728],[68,19.7747],[70,19.8771],[104,19.8319],[108,19.7254],[111,19.5057],[114,19.1755],[116,19.0686],[141,18.4922],[144,18.3493],[150,17.6997],[153,17.3308],[160,16.6768],[172,16.043],[181,15.8048],[186,15.819],[188,15.9059],[190,16.0534],[192,16.2892],[194,16.6352],[195,16.8814],[196,17.2036],[197,17.564],[199,18.3711],[200,18.8485],[201,19.4293],[202,20.1046],[205,22.2827],[207,23.8705],[208,24.5501],[209,25.0441],[210,25.3743],[211,25.6163],[212,25.7856],[213,25.8722],[214,25.8798],[224,25.3033],[244,25.3998],[249,25.7411],[258,26.6862],[262,27.4761],[263,27.5838],[264,27.513],[265,27.235],[266,26.8163],[267,26.4692],[268,26.4714],[269,23.9747],[270,17.6521],[271,9.9623],[272,2.1676],[273,-2.7069],[274,-5.9008],[275,-7.9918],[276,-9.8271],[277,-11.5506],[278,-13.1593],[279,-14.6989],[280,-16.0488],[281,-17.2035],[282,-18.32],[283,-19.5384],[284,-20.5299],[285,-21.6097],[286,-22.8408],[287,-24.389],[288,-26.0396],[289,-27.3017],[290,-28.2876],[291,-28.884],[292,-29.1619],[297,-29.5644],[332,-29.4964],[338,-29.2596],[340,-29.0607],[342,-28.7357],[345,-28.0551],[346,-27.775],[347,-27.3943],[348,-26.9037],[349,-26.3353],[350,-25.6886],[351,-24.9221],[352,-23.9652],[353,-22.8773],[354,-21.2057],[355,-19.4539],[356,17.8628],[357,15.5265],[358,12.9499],[359,11.1186],[360,10.1636],[361,9.6621],[362,9.5757],[363,9.3224],[364,8.6631],[365,8.0865],[366,7.7514],[367,7.4615],[368,7.1311],[369,6.6972],[370,6.206],[371,5.7023],[372,5.2339],[373,4.8081],[374,4.4226],[375,4.1162],[376,3.9284],[378,3.6996],[379,3.5967],[380,3.512],[381,3.4514],[382,3.4195],[383,3.408],[384,3.4048],[385,3.4103],[387,3.4446],[390,3.5079],[392,3.534],[439,3.6662],[440,1.2189]]
       }
       hand_r_ik_ctrl.Location.Z {
        type=MovieSceneScriptingFloatChannel
        default=7.2794
        thinned_from=441
        keys [[0,26.275],[1,12.863],[11,12.8182],[14,12.8353],[15,12.8223],[29,12.9566],[49,13.0636],[54,13.1099],[56,13.0929],[57,13.1043],[58,13.1551],[60,13.3572],[61,13.5373],[62,13.7991],[64,14.3952],[69,15.6379],[70,15.8471],[71,15.9923],[76,16.4718],[84,16.7246],[97,16.7585],[105,16.5421],[108,16.3817],[109,16.2961],[110,16.1661],[111,16.0013],[112,15.8029],[113,15.5279],[116,14.416],[119,13.4033],[120,13.1062],[121,12.8366],[122,12.6356],[124,12.3322],[126,12.0871],[127,11.986],[129,11.8798],[132,11.829],[134,11.8458],[136,11.9265],[143,12.3986],[147,12.5612],[152,12.5835],[154,12.52],[159,12.1187],[162,11.9207],[163,11.8787],[166,11.706],[173,11.4664],[181,11.2439],[184,11.1123],[189,10.8168],[190,10.7909],[191,10.7391],[192,10.6465],[193,10.5159],[195,10.2052],[196,10.0728],[197,9.9978],[198,9.9437],[200,9.9102],[201,9.9369],[202,9.9985],[203,10.0981],[204,10.2321],[205,10.4222],[206,10.805],[207,11.4295],[208,12.2012],[213,16.2889],[214,17.0436],[215,17.6857],[216,18.1807],[218,18.9651],[219,19.3231],[220,19.6285],[222,20.1429],[223,20.3466],[227,20.9525],[234,21.7099],[246,22.8458],[250,23.39],[256,24.5836],[260,25.6641],[262,26.3381],[263,26.8973],[265,28.7589],[266,29.5639],[267,30.1058],[268,29.8384],[269,31.6417],[270,34.9255],[271,39.0916],[272,42.8792],[273,44.3316],[274,44.4913],[275,44.3798],[278,42.5829],[279,42.0421],[280,41.2721],[281,40.1579],[282,39.4072],[283,38.7916],[284,38.7283],[285,39.0781],[287,39.9822],[288,39.7116],[289,38.8627],[290,37.8659],[291,36.9528],[292,36.2458],[293,35.7999],[296,35.0162],[298,34.6346],[300,34.4267],[338,34.3129],[346,34.9211],[347,35.05],[348,35.2543],[349,35.6015],[350,36.0449],[351,36.4055],[353,36.6343],[355,35.7386],[356,44.4473],[357,41.7012],[358,36.7446],[359,30.3246],[361,16.3312],[362,11.526],[363,8.3694],[364,7.3828],[366,6.4521],[368,5.774],[369,5.4616],[370,5.2065],[371,5.0374],[372,4.9469],[373,4.8968],[374,4.873],[375,4.8707],[376,4.8949],[377,4.9313],[378,4.9495],[380,4.9668],[381,4.9836],[382,5.014],[385,5.1354],[386,5.1569],[395,5.1957],[408,5.3041],[439,5.3361],[440,8.3265]]
       }
       hand_r_ik_ctrl.Rotation.X {
        type=MovieSceneScriptingFloatChannel
        default=72.4001
        thinned_from=441
        keys [[0,-56.1836],[1,-90.6673],[58,-90.0164],[61,-89.4309],[64,-88.3566],[68,-85.7315],[70,-84.158],[71,-83.7342],[75,-83.3357],[106,-84.8885],[109,-85.3433],[112,-86.6924],[113,-87.4192],[115,-89.2355],[116,-89.7384],[119,-89.9365],[133,-88.4661],[148,-89.5396],[159,-92.0998],[194,-91.79],[205,-93.9732],[209,-95.6312],[218,-101.9061],[225,-104.7295],[229,-105.376],[243,-105.0191],[250,-103.0444],[258,-98.4979],[259,-97.6962],[260,-96.37],[261,-94.3531],[262,-91.6502],[263,-87.966],[264,-82.9709],[265,-77.0827],[266,-70.7208],[267,-64.0324],[268,-71.5197],[269,-61.5591],[270,-54.1618],[271,-54.5986],[272,-69.7257],[273,-89.59],[274,-110.4901],[275,-129.8568],[276,-131.3603],[277,-122.7107],[278,-113.562],[279,-109.6477],[280,-107.1398],[281,-105.5885],[282,-102.9454],[283,-99.4178],[284,-95.569],[285,-90.7397],[286,-85.1009],[288,-69.6008],[289,-64.3451],[290,-60.1549],[291,-57.6793],[292,-56.4328],[293,-55.9247],[297,-54.7615],[307,-53.4415],[338,-52.715],[339,-52.7705],[341,-53.1266],[342,-53.7303],[343,-54.6093],[344,-55.7216],[345,-57.0255],[346,-59.4243],[347,-62.3161],[348,-67.8617],[349,-78.0463],[350,-92.0183],[351,-102.9621],[352,-112.2534],[353,-122.2844],[354,-133.9143],[355,-146.1013],[356,148.131],[358,148.1832],[359,135.3656],[360,116.9588],[361,82.8003],[362,57.431],[363,41.9471],[364,30.6215],[365,16.9965],[366,13.4938],[367,10.2097],[368,7.1886],[369,4.5267],[370,2.2997],[371,0.5681],[372,-0.8755],[373,-2.2373],[374,-3.4642],[375,-4.5081],[376,-5.4182],[377,-6.1361],[378,-6.4618],[379,-6.5219],[380,-6.5619],[381,-6.5623],[382,-6.476],[383,-6.3089],[385,-5.9206],[386,-5.7837],[389,-5.4991],[391,-5.3202],[393,-5.1659],[394,-5.1057],[400,-4.8364],[403,-4.7214],[406,-4.6426],[408,-4.6113],[415,-4.572],[439,-4.4904],[440,-90.2435]]
       }
       hand_r_ik_ctrl.Rotation.Y {
        type=MovieSceneScriptingFloatChannel
        default=-16.8198
        thinned_from=441
        keys [[0,-43.1113],[1,-33.92],[55,-34.0878],[60,-33.8819],[62,-33.7398],[63,-33.6068],[65,-33.1395],[66,-33.0286],[69,-32.955],[71,-32.9738],[75,-32.8751],[107,-32.6333],[113,-32.9407],[114,-32.8015],[121,-30.4401],[124,-29.8141],[127,-29.38],[130,-29.2129],[133,-29.2274],[136,-29.4496],[139,-29.8623],[141,-30.3344],[143,-30.9103],[144,-31.2516],[147,-32.5928],[148,-32.9659],[150,-33.5089],[152,-33.8578],[154,-33.9079],[166,-32.8762],[184,-32.069],[187,-31.7217],[190,-31.1635],[191,-30.9077],[192,-30.5756],[194,-29.7635],[195,-29.2808],[197,-28.0701],[202,-24.6528],[203,-24.074],[204,-23.6238],[205,-23.3321],[206,-23.348],[207,-23.7154],[208,-24.2843],[209,-24.9077],[214,-27.8098],[215,-28.2791],[216,-28.6186],[219,-29.328],[222,-29.782],[231,-30.4044],[248,-30.796],[251,-30.6041],[253,-30.6184],[254,-30.5768],[257,-30.1305],[258,-29.7366],[259,-28.9986],[260,-28.0176],[261,-26.9381],[262,-25.7851],[263,-24.9562],[264,-24.6323],[265,-24.4541],[266,-24.1826],[267,-23.7131],[268,-20.7962],[269,-22.9609],[270,-30.4077],[271,-41.685],[272,-51.6055],[273,-46.8556],[274,-42.7862],[275,-46.0887],[277,-40.8643],[278,-37.578],[279,-35.3181],[280,-33.637],[281,-32.9917],[282,-33.6814],[283,-34.6483],[284,-36.1943],[285,-39.2792],[286,-43.5398],[287,-49.7396],[288,-55.2664],[289,-58.023],[290,-59.7304],[291,-60.2782],[292,-60.2164],[294,-59.7475],[340,-58.1767],[343,-58.4594],[346,-59.1135],[348,-59.7654],[349,-60.2043],[353,-62.821],[354,-59.0093],[355,-55.0151],[356,-1.7967],[357,20],[358,37.5165],[360,74.6412],[361,72.2201],[362,69.0963],[363,62.1023],[364,52.63],[365,48.9975],[366,47.7868],[369,44.365],[370,43.3043],[371,42.416],[373,40.9432],[374,40.3093],[375,39.7882],[376,39.3706],[377,39.0381],[379,38.6707],[381,38.4667],[382,38.4742],[390,39.3858],[439,40.0761],[440,-34.8873]]
       }
       hand_r_ik_ctrl.Rotation.Z {
        type=MovieSceneScriptingFloatChannel
        default=-161.6039
        thinned_from=441
        keys [[0,-96.0939],[1,-98.3796],[56,-99.8221],[63,-98.901],[71,-96.2574],[77,-93.5184],[96,-90.5092],[111,-89.5232],[113,-89.2186],[115,-89.4043],[116,-89.7375],[117,-90.2945],[123,-95.2995],[126,-97.2747],[128,-98.1555],[133,-99.3988],[137,-99.4991],[144,-98.4541],[153,-99.3176],[155,-99.9598],[165,-104.3471],[185,-107.6564],[194,-107.0647],[197,-106.094],[200,-104.4928],[202,-102.8421],[205,-99.477],[208,-94.7852],[214,-86.5929],[216,-84.3694],[218,-82.5308],[222,-79.8286],[224,-78.9203],[241,-76.862],[244,-76.9451],[246,-77.4833],[254,-81.9127],[256,-82.6643],[258,-82.93],[260,-82.7326],[262,-82.0077],[263,-83.1322],[264,-86.2264],[266,-93.964],[267,-96.659],[268,-68.9817],[269,-60.4426],[270,-43.5515],[271,-23.4382],[272,8.8482],[273,45.5069],[274,68.7892],[275,90.4564],[276,99.0322],[277,108.741],[278,117.6843],[279,122.9883],[280,126.4891],[281,126.9584],[283,129.2071],[285,131.0254],[286,132.2588],[287,133.7634],[288,134.5327],[289,134.2322],[292,132.0711],[299,130.0968],[347,128.2176],[348,128.3214],[349,129.2584],[350,131.4015],[351,134.1794],[352,136.3736],[353,139.2254],[355,147.3505],[356,129.6798],[357,127.5],[358,125],[359,117.5],[360,100],[361,81.8872],[362,58.7334],[363,43.6216],[364,32.2426],[365,18.2794],[366,14.4193],[367,10.8723],[368,7.6174],[369,4.7553],[370,2.3659],[371,0.5137],[372,-1.0232],[373,-2.4728],[374,-3.7881],[375,-4.9269],[376,-5.9457],[377,-6.7623],[378,-7.132],[379,-7.1934],[380,-7.2327],[381,-7.2287],[382,-7.132],[383,-6.9275],[384,-6.6528],[385,-6.3623],[386,-6.1116],[389,-5.4845],[391,-5.098],[392,-4.9292],[393,-4.7811],[394,-4.6615],[395,-4.5741],[400,-4.2381],[402,-4.1207],[404,-4.0216],[406,-3.9442],[408,-3.8944],[433,-3.7196],[439,-3.7103],[440,-30.7556]]
       }
       hand_r_ik_ctrl.Scale.X {
        type=MovieSceneScriptingFloatChannel
        default=1
        thinned_from=441
        keys [[0,0.96],[269,0.964],[270,0.9658],[271,0.9697],[272,0.9795],[273,1],[274,0.96],[355,0.96],[356,1],[439,1],[440,0.96]]
       }
       hand_r_ik_ctrl.Scale.Y {
        type=MovieSceneScriptingFloatChannel
        default=1
        thinned_from=441
        keys [[0,0.96],[269,0.964],[270,0.9658],[271,0.9697],[272,0.9795],[273,1],[274,0.96],[355,0.96],[356,1],[439,1],[440,0.96]]
       }
       hand_r_ik_ctrl.Scale.Z {
        type=MovieSceneScriptingFloatChannel
        default=1
        thinned_from=441
        keys [[0,0.96],[269,0.964],[270,0.9658],[271,0.9697],[272,0.9795],[273,1],[274,0.96],[355,0.96],[356,1],[439,1],[440,0.96]]
       }
       Softness_2 {type=MovieSceneScriptingFloatChannel, default=0, thinned_from=113, keys [[0,0],[439,0]]}
       hand_r_ikorient_ctrl.Rotation.Roll {type=MovieSceneScriptingFloatChannel, default=0, thinned_from=64, keys [[0,0],[439,0]]}
       hand_r_ikorient_ctrl.Rotation.Pitch {type=MovieSceneScriptingFloatChannel, default=0, thinned_from=64, keys [[0,0],[439,0]]}
       hand_r_ikorient_ctrl.Rotation.Yaw {type=MovieSceneScriptingFloatChannel, default=0, thinned_from=64, keys [[0,0],[439,0]]}
       arm_r_pv_ik_ctrl.Location.X {
        type=MovieSceneScriptingFloatChannel
        default=13.8523
        thinned_from=118
        keys [[0,-2.3793],[8,12.9204],[15,-2],[204,-2],[209,-1],[246,-1],[251,5.8145],[257,-1],[266,-1],[267,-0.875],[270,-0.125],[271,0.25],[272,2.1613],[273,2.1613],[274,11.4672],[275,-1],[283,-1],[284,-1.2018],[285,-1.5707],[286,-2.0541],[287,-2.7141],[288,-3.3044],[289,-3.6332],[290,-3.8584],[291,-3.9646],[292,-4],[339,-4],[341,-3.9832],[342,-3.9355],[343,-3.8616],[344,-3.7656],[345,-3.6521],[346,-3.5254],[348,-3.25],[349,-3.054],[350,-2.7604],[351,-2.3887],[352,-1.9441],[353,-1.4889],[355,-1.4889],[356,18.75],[357,18.75],[358,16.25],[359,12.5],[360,21.25],[361,19.6604],[362,17.6332],[363,11.25],[364,8.75],[365,7],[366,10.2959],[367,4],[369,1.5],[370,0.5],[371,2.3018],[375,-1.8262],[378,-2.4957],[382,-2.8201],[387,-2.5622],[395,-2.1639],[409,-1.993],[439,-2],[440,-2.3796]]
       }
       arm_r_pv_ik_ctrl.Location.Y {
        type=MovieSceneScriptingFloatChannel
        default=38.9648
        thinned_from=114
        keys [[0,-17.3218],[8,37.951],[15,0],[246,0],[251,-0.0217],[257,0],[266,0],[267,0.6605],[268,2.1137],[269,6.9209],[270,15.1488],[271,23.4592],[272,31.6304],[273,39.4649],[274,45.3086],[275,46.0886],[276,47.5126],[278,51.0244],[279,51.779],[280,52.2318],[281,52.434],[282,53.2413],[283,54.342],[284,56.0244],[285,58.4113],[286,61.3098],[287,65.194],[288,68.6389],[289,70.5478],[290,71.8493],[291,72.4613],[292,72.6653],[342,72.6011],[348,71.8095],[350,71.2441],[353,69.776],[355,69.776],[356,38.9648],[357,38.9648],[358,29.7981],[359,20.9981],[360,13.9987],[361,6.8528],[362,-0.0276],[363,1.3354],[364,0.5008],[365,0.183],[366,-10.2449],[367,0],[370,0],[371,-15.8067],[375,-17.3042],[378,-17.5906],[382,-17.6445],[387,-17.5022],[409,-17.328],[439,0],[440,-17.322]]
       }
       arm_r_pv_ik_ctrl.Location.Z {
        type=MovieSceneScriptingFloatChannel
        default=13.6646
        thinned_from=114
        keys [[0,8.1229],[8,7.9536],[15,0],[246,0],[251,-1.2935],[257,0],[266,0],[267,0.0058],[268,0.0185],[269,0.0593],[270,0.1189],[271,4.6406],[272,6.9709],[273,11.0209],[274,9.6019],[275,0.3951],[276,0.5672],[277,2.0856],[278,3.5804],[279,4.2278],[280,4.6163],[281,4.752],[282,5.2938],[283,6.0389],[284,7.2504],[285,8.9818],[286,11.0898],[287,13.9166],[288,16.4244],[289,17.8143],[290,18.7621],[291,19.2078],[292,19.3564],[340,19.3495],[341,19.318],[342,19.1884],[343,18.987],[344,18.7258],[346,18.0719],[348,17.3222],[349,16.7887],[350,15.9894],[351,14.9774],[352,13.7671],[353,12.5281],[355,12.5281],[356,21.8827],[357,15.5569],[358,15.7207],[359,22.4624],[360,14.975],[361,4.2289],[362,9.6134],[363,1.4285],[364,0.5357],[365,0.1958],[366,16.8652],[367,0],[370,0],[371,15.8365],[375,13.6032],[378,12.0489],[382,11.413],[387,10.9135],[395,10.5158],[409,10.3275],[439,0],[440,8.1124]]
       }
       arm_r_pv_ik_ctrl.Rotation.X {
        type=MovieSceneScriptingFloatChannel
        default=-11.9556
        thinned_from=114
        keys [[0,23.0215],[8,5.9802],[15,0],[246,0],[251,2.1676],[257,0],[266,0],[267,-0.1394],[268,-0.4461],[269,-1.4659],[270,-3.2514],[271,-5.1026],[272,-6.9362],[273,-8.6835],[274,3.8646],[275,-10.0095],[276,-10.1812],[278,-10.3403],[280,-10.395],[281,-10.4041],[283,-10.4906],[284,-10.5668],[285,-10.6749],[286,-10.8062],[288,-11.1381],[289,-11.2246],[290,-11.2836],[292,-11.3205],[340,-11.3282],[341,-11.3439],[342,-11.4084],[343,-11.5087],[344,-11.6387],[346,-11.9641],[348,-12.3372],[349,-12.6954],[350,-13.297],[351,-14.0045],[353,-15.1866],[355,-15.1866],[356,-11.9556],[357,-11.9556],[358,-8.95],[359,-6.3068],[360,-4.2046],[361,24.6737],[362,34.6188],[363,-0.4011],[364,-0.1504],[365,-0.055],[366,43.9043],[367,0],[370,0],[371,39.8988],[375,35.2523],[378,32.016],[382,30.4115],[387,29.6787],[395,29.6355],[409,29.6588],[439,0],[440,23.0012]]
       }
       arm_r_pv_ik_ctrl.Rotation.Y {
        type=MovieSceneScriptingFloatChannel
        default=38.8737
        thinned_from=114
        keys [[0,-10.1038],[8,14.0949],[15,0],[246,0],[251,4.0496],[257,0],[266,0],[267,0.2968],[268,0.9497],[269,3.0792],[270,6.4951],[271,9.7365],[272,12.7197],[273,15.553],[274,19.6307],[275,19.218],[276,20.4651],[278,25.3562],[279,26.4072],[280,27.0378],[281,27.322],[282,28.4547],[283,29.9935],[284,32.3432],[285,35.6766],[286,39.7244],[287,45.1487],[288,49.9594],[289,52.6253],[290,54.4428],[291,55.2974],[292,55.5822],[341,55.5303],[343,55.0528],[345,54.2299],[348,52.6505],[349,51.8808],[350,50.7274],[351,49.2672],[353,45.7331],[355,45.7331],[356,38.8737],[357,38.8737],[358,29.5165],[359,20.7996],[360,13.8664],[361,-16.3925],[362,-20.1977],[363,1.3228],[364,0.496],[365,0.1813],[366,-18.6087],[367,0],[370,0],[371,-15.5167],[375,-13.3849],[378,-11.4997],[382,-10.7674],[387,-11.521],[395,-12.5664],[409,-13.037],[439,0],[440,-10.0916]]
       }
       arm_r_pv_ik_ctrl.Rotation.Z {
        type=MovieSceneScriptingFloatChannel
        default=84.3686
        thinned_from=114
        keys [[0,-4.1019],[8,63.9052],[15,0],[246,0],[251,9.1433],[257,0],[266,0],[267,0.9834],[268,3.1469],[269,10.2878],[271,34.4677],[272,46.3037],[273,57.6447],[274,71.4386],[275,67.822],[276,70.2707],[278,77.2481],[279,78.7474],[280,79.647],[281,80.0499],[282,81.6577],[283,83.8474],[284,87.1931],[285,91.9397],[286,97.7037],[287,105.4278],[288,112.2782],[289,116.0743],[290,118.6624],[291,119.8793],[292,120.285],[341,120.156],[343,119.078],[345,117.2208],[348,113.6557],[349,111.9183],[350,109.315],[351,106.019],[353,98.0418],[355,98.0418],[356,84.3686],[357,84.3686],[358,64.351],[359,45.3468],[360,30.2312],[361,34.9324],[362,26.5089],[363,2.8839],[364,1.0815],[365,0.3953],[366,10.3432],[367,0],[370,0],[371,-1.3996],[375,-7.0085],[382,-7.297],[387,-6.8243],[395,-6.3794],[409,-6.1801],[439,0],[440,-4.0963]]
       }
       arm_r_pv_ik_ctrl.Scale.X {type=MovieSceneScriptingFloatChannel, default=1, thinned_from=114, keys [[0,1],[440,1]]}
       arm_r_pv_ik_ctrl.Scale.Y {type=MovieSceneScriptingFloatChannel, default=1, thinned_from=114, keys [[0,1],[440,1]]}
       arm_r_pv_ik_ctrl.Scale.Z {type=MovieSceneScriptingFloatChannel, default=1, thinned_from=114, keys [[0,1],[440,1]]}
       body_ctrl.Location.X {
        type=MovieSceneScriptingFloatChannel
        default=-8.9882
        thinned_from=58
        keys [[0,-9.9],[8,-9.89],[15,-9.8625],[30,-9.7375],[51,-9.675],[56,-9.6],[59,-9.5],[62,-9.3222],[66,-9.0222],[70,-8.8267],[75,-8.7084],[83,-8.625],[94,-8.5889],[101,-8.6389],[107,-8.8],[115,-9.2059],[121,-9.3685],[130,-9.478],[143,-9.6],[153,-9.9],[160,-10.05],[172,-10.15],[187,-10.2],[194,-10],[199,-9.25],[204,-7.5],[209,-4.9585],[215,-2.4561],[223,-1.2258],[233,-0.6719],[244,-0.2923],[253,0.3114],[261,1.4564],[267,2.4564],[271,1.9564],[287,-8.75],[292,-10.5],[299,-11],[309,-10.75],[327,-10.5893],[337,-10.375],[345,-9.75],[351,-8.875],[356,-7.5],[361,-5.5],[364,-5.25],[367,-5.75],[373,-8.25],[378,-9.75],[384,-10.25],[394,-10.0664],[411,-9.9562],[439,-9.9]]
       }
       body_ctrl.Location.Y {
        type=MovieSceneScriptingFloatChannel
        default=-0.0136
        thinned_from=59
        keys [[0,4.2659],[8,4.27],[22,4.3],[30,4.31],[45,4.315],[51,4.29],[59,4.225],[62,4.21],[66,4.205],[70,4.215],[75,4.25],[83,4.29],[92,4.31],[101,4.315],[107,4.3],[111,4.2659],[115,4.22],[121,4.19],[130,4.175],[143,4.17],[148,4.16],[160,4.1],[172,4.0625],[187,4.05],[194,4.0625],[199,4.0875],[204,4.1375],[209,4.2],[215,4.25],[223,4.2875],[239,4.3],[246,4.275],[251,4.1519],[257,3.8097],[262,3.25],[275,0.7951],[279,0.125],[283,-0.25],[287,-0.375],[292,-0.25],[299,-0.0537],[309,0.0803],[327,0.1175],[337,0.125],[345,0],[351,-0.1],[356,0.1],[361,0.6],[371,2.839],[375,4],[378,4.5],[382,4.6],[387,4.5],[395,4.3],[409,4.2784],[439,4.2659]]
       }
       body_ctrl.Location.Z {
        type=MovieSceneScriptingFloatChannel
        default=0.2642
        thinned_from=109
        keys [[0,0.56],[8,0.55],[15,0.525],[22,0.5125],[30,0.525],[45,0.56],[51,0.5625],[56,0.5504],[62,0.5216],[66,0.5375],[70,0.5625],[76,0.5781],[94,0.5875],[101,0.5875],[107,0.585],[115,0.57],[121,0.5625],[130,0.5592],[138,0.5585],[160,0.5625],[172,0.5633],[187,0.5625],[215,0.5562],[225,0.5575],[233,0.5612],[241,0.58],[246,0.625],[251,0.675],[253,0.6722],[257,0.6561],[258,0.65],[259,0.6417],[261,0.6206],[262,0.6088],[263,0.5929],[264,0.57],[265,0.5418],[266,0.51],[267,0.4765],[268,0.4369],[270,0.3446],[271,0.3064],[272,0.2747],[273,0.2444],[274,0.2167],[275,0.193],[276,0.1721],[277,0.154],[278,0.1411],[279,0.1363],[280,0.1379],[281,0.1422],[282,0.1482],[283,0.1552],[284,0.1636],[286,0.1839],[287,0.193],[290,0.2165],[292,0.2308],[294,0.2439],[296,0.2558],[297,0.2609],[298,0.2652],[302,0.2768],[305,0.2836],[307,0.2864],[309,0.2875],[327,0.2875],[337,0.2686],[345,0.2308],[351,0.2497],[356,0.2686],[361,0.2497],[366,0.2071],[371,0.1787],[375,0.2354],[378,0.3205],[382,0.4],[387,0.47],[395,0.5263],[409,0.5477],[439,0.56]]
       }
       body_ctrl.Rotation.X {
        type=MovieSceneScriptingFloatChannel
        default=-12.1485
        thinned_from=59
        keys [[0,-22.5],[246,-22.5],[251,-22.4226],[257,-22.1537],[262,-20.6646],[275,-15],[283,-13.75],[292,-13.4076],[337,-13.3949],[345,-13.2212],[351,-12.7],[356,-11.7],[361,-11],[366,-13],[375,-20],[378,-21.6415],[382,-22.3956],[387,-22.5],[439,-22.5]]
       }
       body_ctrl.Rotation.Y {
        type=MovieSceneScriptingFloatChannel
        default=1.3115
        thinned_from=59
        keys [[0,-1.8488],[8,-1.888],[22,-1.9646],[30,-1.9911],[45,-2],[51,-1.9696],[56,-1.85],[59,-1.6],[62,-1.5],[66,-1.625],[70,-2.05],[75,-2.5],[83,-2.75],[94,-2.9],[101,-2.9534],[107,-3.0304],[111,-3.1436],[115,-3.273],[121,-3.3931],[130,-3.45],[138,-3.4],[143,-3.25],[148,-3.0559],[153,-3.25],[160,-3.65],[170,-4],[182,-4.15],[192,-4.2],[199,-3.75],[204,-2.25],[209,1],[213,2.5239],[215,2.7292],[218,2.8101],[229,2.4675],[246,2.337],[252,2.1998],[257,1.5],[262,-2],[271,-9.5],[275,-10],[279,-9],[284,-3.5],[293,1],[305,1.625],[321,2],[327,2.067],[337,2.125],[345,2],[351,1.625],[361,0.25],[366,-1.25],[371,-2.9858],[375,-3.3874],[379,-3.2463],[397,-2.1764],[413,-1.7802],[439,-1.8488]]
       }
       body_ctrl.Rotation.Z {
        type=MovieSceneScriptingFloatChannel
        default=-18.42
        thinned_from=63
        keys [[0,18.0042],[8,17.9162],[15,17.9774],[30,18.3482],[45,18.45],[51,18.4],[55,18.25],[59,17.5],[62,16.3],[66,14.6615],[70,13.2754],[75,12.3212],[83,11.7228],[94,11.5],[101,11.5407],[107,11.8255],[111,12.5507],[116,13.8611],[122,14.6092],[130,14.9601],[138,15.25],[143,16],[148,18],[153,19.4033],[160,20.0795],[172,20.3838],[185,20.5],[194,19],[199,16.25],[204,12.5],[209,10.5],[215,9.75],[223,9.375],[233,9],[241,8.625],[247,8],[251,7],[257,5],[262,0],[267,-7.5],[271,-13.75],[275,-17.5],[283,-20],[287,-19.25],[292,-18.25],[299,-17.5],[309,-17.125],[327,-17],[356,-17],[361,-10],[366,2.5],[371,13.2282],[375,18.75],[378,20.75],[380,21.1297],[386,20],[395,18.75],[409,18.2958],[439,18.0042]]
       }
       body_ctrl.Scale.X {type=MovieSceneScriptingFloatChannel, default=1, thinned_from=64, keys [[0,1],[439,1]]}
       body_ctrl.Scale.Y {type=MovieSceneScriptingFloatChannel, default=1, thinned_from=64, keys [[0,1],[439,1]]}
       body_ctrl.Scale.Z {type=MovieSceneScriptingFloatChannel, default=1, thinned_from=64, keys [[0,1],[439,1]]}
       hips_ctrl.Location.X {
        type=MovieSceneScriptingFloatChannel
        default=0.4
        thinned_from=59
        keys [[0,0.4263],[271,0.4263],[292,0.4],[345,0.4],[360,0.4211],[366,0.4263],[439,0.4263]]
       }
       hips_ctrl.Location.Y {
        type=MovieSceneScriptingFloatChannel
        default=0
        thinned_from=59
        keys [[0,-1.8826],[271,-1.8826],[292,0],[345,0],[360,-1.5094],[366,-1.8826],[439,-1.8826]]
       }
       hips_ctrl.Location.Z {
        type=MovieSceneScriptingFloatChannel
        default=0
        thinned_from=59
        keys [[0,0.6644],[271,0.6644],[292,0],[345,0],[360,0.5327],[366,0.6644],[439,0.6644]]
       }
       hips_ctrl.Rotation.X {
        type=MovieSceneScriptingFloatChannel
        default=1.9396
        thinned_from=57
        keys [[0,19.4608],[246,19.4608],[251,18.8955],[264,13.5822],[279,1.8352],[283,0.4588],[287,0.1147],[292,0.0287],[299,0.0072],[309,0.0022],[327,0.0006],[337,0],[345,0.2571],[351,1.0286],[360,5.7184],[375,16.8139],[378,18.3887],[382,19.0587],[387,19.3603],[395,19.4231],[439,19.4608]]
       }
       hips_ctrl.Rotation.Y {
        type=MovieSceneScriptingFloatChannel
        default=-3.3042
        thinned_from=62
        keys [[0,-2.7609],[8,-2.7625],[15,-2.775],[22,-2.8001],[30,-2.8125],[45,-2.8125],[51,-2.8],[56,-2.75],[62,-2.6],[66,-2.3407],[70,-2.05],[75,-1.9],[83,-1.85],[107,-1.85],[111,-1.775],[115,-1.4245],[121,-1.175],[130,-0.95],[143,-0.85],[153,-0.85],[160,-0.9],[172,-1.075],[181,-1.2],[199,-1.2604],[204,-1.3],[209,-1.4],[215,-1.6],[223,-1.8],[233,-1.885],[241,-2],[246,-2.125],[251,-2.375],[257,-2.75],[267,-3.5],[275,-3.75],[279,-3.975],[283,-3.7],[287,-3.2737],[292,-2.975],[299,-2.75],[309,-2.65],[327,-2.725],[337,-2.85],[345,-3.05],[351,-3.2598],[356,-3.325],[360,-3.3486],[361,-3.35],[366,-3.3],[371,-3.1],[375,-2.8286],[378,-2.7],[382,-2.65],[387,-2.675],[395,-2.725],[409,-2.75],[439,-2.7609]]
       }
       hips_ctrl.Rotation.Z {
        type=MovieSceneScriptingFloatChannel
        default=-0.1738
        thinned_from=62
        keys [[0,-0.1738],[51,-0.1738],[56,-0.1713],[59,-0.1675],[62,-0.16],[66,-0.1198],[70,-0.1],[75,-0.0935],[246,-0.0935],[251,-0.0955],[257,-0.1017],[264,-0.1188],[279,-0.1671],[283,-0.1721],[287,-0.1733],[309,-0.1738],[439,-0.1738]]
       }
       hips_ctrl.Scale.X {type=MovieSceneScriptingFloatChannel, default=1, thinned_from=66, keys [[0,1],[439,1]]}
       hips_ctrl.Scale.Y {type=MovieSceneScriptingFloatChannel, default=1, thinned_from=66, keys [[0,1],[439,1]]}
       hips_ctrl.Scale.Z {type=MovieSceneScriptingFloatChannel, default=1, thinned_from=66, keys [[0,1],[439,1]]}
       hips_tan_ctrl.Location.X {
        type=MovieSceneScriptingFloatChannel
        default=-1.1969
        thinned_from=65
        keys [[0,-0.5737],[8,-0.5889],[15,-0.6343],[22,-0.7173],[30,-0.726],[45,-0.7354],[51,-0.6],[56,-0.7331],[62,-0.7263],[66,-0.6861],[70,-0.6624],[75,-0.657],[101,-0.6546],[107,-0.6501],[111,-0.6362],[115,-0.5778],[121,-0.5632],[130,-0.5584],[148,-0.5584],[153,-0.5679],[160,-0.5715],[172,-0.5617],[187,-0.5584],[194,-0.5593],[199,-0.5622],[204,-0.5737],[215,-0.5737],[223,-0.585],[233,-0.6191],[241,-0.6304],[246,-0.6342],[251,-0.6472],[257,-0.6865],[267,-0.8433],[275,-1.0991],[279,-1.196],[283,-1.2396],[287,-1.2505],[292,-1.2532],[337,-1.2541],[345,-1.2482],[351,-1.2306],[356,-1.1601],[360,-1.0747],[361,-1.0456],[366,-1.0075],[371,-0.9038],[375,-0.8038],[378,-0.7611],[382,-0.7451],[387,-0.7378],[409,-0.7357],[439,-0.7354]]
       }
       hips_tan_ctrl.Location.Y {
        type=MovieSceneScriptingFloatChannel
        default=-0.3212
        thinned_from=65
        keys [[0,0.0977],[8,-0.1657],[15,-0.9559],[22,-2.793],[30,-2.5471],[45,-2.7118],[51,0.1],[56,-2.7118],[148,-2.7176],[153,-0.9581],[160,-0.2982],[172,-2.1128],[187,-2.7176],[194,-2.5416],[199,-2.0138],[204,0.0977],[215,0.0977],[223,-0.4297],[233,-2.0119],[241,-2.5393],[246,-2.7151],[251,-2.6494],[257,-2.4523],[262,-2.0605],[267,-1.6641],[271,-1.1739],[275,-0.6894],[279,-0.2614],[283,-0.0688],[287,-0.0207],[292,-0.0086],[299,-0.0056],[309,-0.0049],[337,-0.0046],[345,-0.0371],[351,-0.1347],[356,-0.5249],[360,-0.786],[361,-1.1397],[366,-1.3446],[371,-1.872],[375,-2.3811],[378,-2.5878],[382,-2.6653],[387,-2.7001],[395,-2.7074],[439,-2.7118]]
       }
       hips_tan_ctrl.Location.Z {
        type=MovieSceneScriptingFloatChannel
        default=-0.1161
        thinned_from=65
        keys [[0,-0.0214],[8,-0.0889],[15,-0.2914],[22,-0.735],[30,-0.6993],[45,-0.7415],[51,0],[56,-0.7412],[62,-0.7405],[70,-0.7339],[107,-0.7329],[111,-0.7319],[115,-0.7272],[148,-0.7256],[153,-0.2854],[160,-0.1204],[172,-0.5743],[187,-0.7256],[194,-0.6816],[199,-0.5495],[204,-0.0214],[215,-0.0214],[223,-0.1545],[233,-0.5541],[241,-0.6873],[246,-0.7317],[251,-0.707],[257,-0.6329],[262,-0.4854],[267,-0.3363],[271,-0.2361],[275,-0.1371],[279,-0.116],[283,-0.1065],[287,-0.1041],[292,-0.1035],[337,-0.1033],[345,-0.1047],[351,-0.1086],[356,-0.1242],[360,-0.1327],[361,-0.2192],[366,-0.2509],[371,-0.4222],[375,-0.5875],[378,-0.6837],[382,-0.7198],[387,-0.7361],[395,-0.7394],[409,-0.7412],[439,-0.7415]]
       }
       hips_tan_ctrl.Rotation.X {type=MovieSceneScriptingFloatChannel, default=0, thinned_from=65, keys [[0,0],[439,0]]}
       hips_tan_ctrl.Rotation.Y {type=MovieSceneScriptingFloatChannel, default=0, thinned_from=65, keys [[0,0],[439,0]]}
       hips_tan_ctrl.Rotation.Z {type=MovieSceneScriptingFloatChannel, default=0, thinned_from=65, keys [[0,0],[439,0]]}
       hips_tan_ctrl.Scale.X {type=MovieSceneScriptingFloatChannel, default=1, thinned_from=65, keys [[0,1],[439,1]]}
       hips_tan_ctrl.Scale.Y {type=MovieSceneScriptingFloatChannel, default=1, thinned_from=65, keys [[0,1],[439,1]]}
       hips_tan_ctrl.Scale.Z {type=MovieSceneScriptingFloatChannel, default=1, thinned_from=65, keys [[0,1],[439,1]]}
       spine_01_ctrl.Location.X {type=MovieSceneScriptingFloatChannel, default=0, thinned_from=65, keys [[0,0],[439,0]]}
       spine_01_ctrl.Location.Y {type=MovieSceneScriptingFloatChannel, default=0, thinned_from=65, keys [[0,0],[439,0]]}
       spine_01_ctrl.Location.Z {type=MovieSceneScriptingFloatChannel, default=0, thinned_from=65, keys [[0,0],[439,0]]}
       spine_01_ctrl.Rotation.X {
        type=MovieSceneScriptingFloatChannel
        default=0.6721
        thinned_from=62
        keys [[0,5.9324],[246,5.9324],[251,5.9],[257,5.75],[262,5.5],[267,4.9],[271,3.75],[275,1.8375],[279,0.6891],[283,0.1723],[287,0.0431],[292,0.0108],[299,0.0027],[309,0.0008],[337,0],[345,0.069],[351,0.2762],[356,1.1047],[366,4.1],[371,5.125],[376,5.625],[381,5.85],[387,5.9054],[395,5.9223],[439,5.9324]]
       }
       spine_01_ctrl.Rotation.Y {type=MovieSceneScriptingFloatChannel, default=2.7657, const [2.7657,65]}
       spine_01_ctrl.Rotation.Z {type=MovieSceneScriptingFloatChannel, default=2.6153, const [2.6153,65]}
       spine_01_ctrl.Scale.X {type=MovieSceneScriptingFloatChannel, default=1, thinned_from=65, keys [[0,1],[439,1]]}
       spine_01_ctrl.Scale.Y {type=MovieSceneScriptingFloatChannel, default=1, thinned_from=65, keys [[0,1],[439,1]]}
       spine_01_ctrl.Scale.Z {type=MovieSceneScriptingFloatChannel, default=1, thinned_from=65, keys [[0,1],[439,1]]}
       spine_02_ctrl.Location.X {type=MovieSceneScriptingFloatChannel, default=0, thinned_from=56, keys [[0,0.311],[267,0.311],[287,0],[337,0],[366,0.311],[439,0.311]]}
       spine_02_ctrl.Location.Y {type=MovieSceneScriptingFloatChannel, default=0, thinned_from=56, keys [[0,0.8458],[267,0.8458],[287,0],[337,0],[366,0.8458],[439,0.8458]]}
       spine_02_ctrl.Location.Z {
        type=MovieSceneScriptingFloatChannel
        default=0
        thinned_from=56
        keys [[0,-0.2624],[267,-0.2624],[287,0],[337,0],[366,-0.2624],[439,-0.2624]]
       }
       spine_02_ctrl.Rotation.X {
        type=MovieSceneScriptingFloatChannel
        default=1.0703
        thinned_from=59
        keys [[0,8.839],[246,8.839],[251,8.7803],[257,8.6042],[267,7.8],[275,6.3],[283,3.7],[287,1.5],[292,0.375],[299,0.0938],[309,0.0288],[327,0.0072],[337,0],[345,0.1029],[351,0.5],[356,1.6459],[361,3.5],[366,5.5],[371,7.3494],[375,8.2655],[382,8.6782],[395,8.8239],[439,8.839]]
       }
       spine_02_ctrl.Rotation.Y {
        type=MovieSceneScriptingFloatChannel
        default=8.2165
        thinned_from=63
        keys [[0,-2.1957],[8,-2.1626],[15,-2.1538],[22,-2.1608],[30,-2.1774],[45,-2.1981],[51,-2.2],[66,-2.1957],[107,-2.1988],[251,-2.1957],[257,-1.8941],[262,-0.1751],[267,2.5651],[271,4.7243],[275,6.8583],[279,8.4683],[283,9.1927],[287,9.3739],[292,9.4191],[337,9.4342],[345,9.2989],[351,8.8928],[356,7.5],[366,2],[371,-0.5],[375,-2],[378,-2.5],[381,-2.5777],[387,-2.4828],[396,-2.3145],[407,-2.2154],[439,-2.1957]]
       }
       spine_02_ctrl.Rotation.Z {type=MovieSceneScriptingFloatChannel, default=0, const [0,64]}
       spine_02_ctrl.Scale.X {type=MovieSceneScriptingFloatChannel, default=1, thinned_from=65, keys [[0,1],[439,1]]}
       spine_02_ctrl.Scale.Y {type=MovieSceneScriptingFloatChannel, default=1, thinned_from=65, keys [[0,1],[439,1]]}
       spine_02_ctrl.Scale.Z {type=MovieSceneScriptingFloatChannel, default=1, thinned_from=65, keys [[0,1],[439,1]]}
       spine_03_ctrl.Location.X {type=MovieSceneScriptingFloatChannel, default=0, const [0,65]}
       spine_03_ctrl.Location.Y {type=MovieSceneScriptingFloatChannel, default=0, const [0,65]}
       spine_03_ctrl.Location.Z {type=MovieSceneScriptingFloatChannel, default=0, const [0,65]}
       spine_03_ctrl.Rotation.X {
        type=MovieSceneScriptingFloatChannel
        default=6.342
        thinned_from=59
        keys [[0,15.5],[8,15.5],[30,15.6791],[45,15.7041],[56,15.8041],[59,15.7541],[62,15.6041],[66,15.4918],[72,15.4291],[94,15.3541],[101,15.3791],[115,15.6041],[121,15.6291],[138,15.5541],[148,15.4291],[153,15.4041],[183,15.5541],[204,15.6041],[209,15.5791],[223,15.4291],[233,15.4041],[246,15.4541],[251,15.4041],[257,14.968],[267,13.1],[271,11.25],[275,9.3],[279,8.9],[283,8.7],[287,8],[292,4.8956],[299,3.6239],[309,3],[327,2.75],[337,3.2],[345,4.25],[351,5.5],[356,7],[361,9],[371,14],[378,16.5291],[382,17.0291],[387,16.7791],[395,16.2791],[409,15.7791],[439,15.4918]]
       }
       spine_03_ctrl.Rotation.Y {
        type=MovieSceneScriptingFloatChannel
        default=4.8113
        thinned_from=64
        keys [[0,-1.7965],[8,-1.7965],[30,-1.725],[45,-1.7],[51,-1.65],[56,-1.575],[59,-1.475],[62,-1.3],[66,-0.8],[70,-0.45],[75,-0.5254],[83,-0.6],[94,-0.6375],[101,-0.6625],[107,-0.775],[111,-1.25],[116,-2.1644],[121,-2.8721],[128,-3.2392],[138,-3.349],[143,-3.2434],[148,-2.95],[153,-2.7],[160,-2.6],[172,-2.55],[185,-2.65],[194,-2.9],[199,-3.1],[209,-3.6],[215,-3.65],[223,-3.5545],[233,-3.4226],[246,-3.3307],[251,-3.2592],[257,-2.737],[262,-1.9945],[267,-2.3009],[271,-3.25],[275,-4],[279,-3],[283,0.5],[287,3],[292,4.5],[299,5.35],[309,5.5],[327,5.625],[337,5.625],[345,5.5],[351,5.25],[356,4.3332],[361,2.6596],[366,1],[371,-0.25],[375,-1.2021],[378,-1.6613],[382,-1.9119],[387,-1.9901],[395,-1.9133],[409,-1.814],[439,-1.7965]]
       }
       spine_03_ctrl.Rotation.Z {
        type=MovieSceneScriptingFloatChannel
        default=-8.304
        keys [[0,2.761],[8,2.6425],[15,2.7387],[22,2.9748],[30,3.1941],[45,3.25],[51,3.2],[56,2.9865],[59,2.7297],[62,2.125],[66,0.5046],[70,-1.1103],[75,-2.1761],[83,-2.7794],[94,-2.8],[101,-2.3323],[107,-1.4731],[120,1.3025],[137,2.6193],[138,2.625],[139,2.6106],[140,2.5714],[141,2.5137],[142,2.4436],[143,2.3673],[144,2.2485],[145,2.0768],[146,1.8978],[147,1.757],[148,1.7],[153,2.5],[160,4.5],[172,5.375],[186,5.625],[194,6],[199,6.625],[204,7],[209,5.875],[214,3.6943],[221,2.3193],[233,1.7021],[241,1.6663],[246,1.9263],[251,2.75],[257,3],[262,2.5],[267,1],[271,-1],[275,-3.4],[279,-6.5],[283,-9.7],[287,-12.1199],[292,-13.5],[299,-14.2135],[309,-14.5],[327,-14],[337,-13],[345,-11.5],[351,-9.5],[378,2.2994],[382,3.7495],[386,4.4063],[389,4.5046],[393,4.3433],[409,3.4581],[439,2.8191]]
       }
       spine_03_ctrl.Scale.X {type=MovieSceneScriptingFloatChannel, default=1, thinned_from=65, keys [[0,1],[439,1]]}
       spine_03_ctrl.Scale.Y {type=MovieSceneScriptingFloatChannel, default=1, thinned_from=65, keys [[0,1],[439,1]]}
       spine_03_ctrl.Scale.Z {type=MovieSceneScriptingFloatChannel, default=1, thinned_from=65, keys [[0,1],[439,1]]}
       spine_04_ctrl.Location.X {type=MovieSceneScriptingFloatChannel, default=0, const [0,64]}
       spine_04_ctrl.Location.Y {type=MovieSceneScriptingFloatChannel, default=0, const [0,64]}
       spine_04_ctrl.Location.Z {type=MovieSceneScriptingFloatChannel, default=0, const [0,64]}
       spine_04_ctrl.Rotation.X {
        type=MovieSceneScriptingFloatChannel
        default=0.2606
        thinned_from=63
        keys [[0,0],[246,0],[251,0.0187],[257,0.075],[262,0.1868],[267,0.3],[271,0.2471],[279,0.1081],[283,0.0543],[287,0.0229],[292,0.0057],[299,0.0014],[309,0.0004],[327,0.0001],[337,-0.0014],[409,-0.0014],[439,0]]
       }
       spine_04_ctrl.Rotation.Y {
        type=MovieSceneScriptingFloatChannel
        default=4.7057
        thinned_from=56
        keys [[0,7.2486],[8,7.2486],[15,7.3],[22,7.375],[30,7.4125],[45,7.4375],[51,7.425],[56,7.2991],[59,7.2],[62,7.4504],[66,8.3333],[70,8.9],[75,9.1],[83,9.15],[94,9.05],[101,8.895],[111,8.6],[115,8.5025],[121,8.4402],[130,8.4194],[138,8.4402],[143,8.5025],[148,8.7521],[153,8.95],[160,9.1],[172,9.2],[187,9.25],[209,9.2625],[215,9.35],[223,9.55],[233,9.7],[246,9.75],[251,9.75],[257,9.5216],[279,3.75],[283,2.75],[287,2.25],[292,2],[299,2.125],[323,2.375],[345,2.375],[351,3.5],[356,5.5],[361,6.75],[366,7.25],[371,7.4372],[375,7.3697],[378,7.2985],[382,7.2486],[439,7.2486]]
       }
       spine_04_ctrl.Rotation.Z {
        type=MovieSceneScriptingFloatChannel
        default=-0.0389
        thinned_from=57
        keys [[0,-0.4428],[246,-0.4428],[251,-0.4314],[257,-0.3975],[275,-0.0981],[279,-0.0368],[283,-0.0092],[287,-0.0023],[292,-0.0006],[299,-0.0001],[337,0],[345,-0.0052],[351,-0.0206],[375,-0.3855],[378,-0.4213],[382,-0.4347],[387,-0.4408],[395,-0.442],[439,-0.4428]]
       }
       spine_04_ctrl.Scale.X {type=MovieSceneScriptingFloatChannel, default=1, thinned_from=64, keys [[0,1],[439,1]]}
       spine_04_ctrl.Scale.Y {type=MovieSceneScriptingFloatChannel, default=1, thinned_from=64, keys [[0,1],[439,1]]}
       spine_04_ctrl.Scale.Z {type=MovieSceneScriptingFloatChannel, default=1, thinned_from=64, keys [[0,1],[439,1]]}
       head_ctrl.Location.X {type=MovieSceneScriptingFloatChannel, default=0, const [0,64]}
       head_ctrl.Location.Y {type=MovieSceneScriptingFloatChannel, default=0, const [0,64]}
       head_ctrl.Location.Z {type=MovieSceneScriptingFloatChannel, default=0, const [0,64]}
       head_ctrl.Rotation.X {
        type=MovieSceneScriptingFloatChannel
        default=-5.142
        thinned_from=62
        keys [[0,14.6517],[8,14.4309],[15,14.7347],[22,15.7784],[30,16.25],[45,16.375],[51,16.5],[56,16.375],[59,15.7884],[62,13.7123],[66,9.7733],[70,6.8937],[75,5.4797],[83,5.125],[94,5.125],[101,5.5375],[107,6.75],[111,8.0661],[115,8.7457],[121,9.0302],[130,9.125],[138,8.7453],[143,8.2109],[148,7.8394],[153,8.0083],[160,8.411],[172,8.7559],[187,8.8709],[194,9.0532],[199,9.6],[204,10.1],[209,9.2542],[215,8.75],[223,9],[233,9.8],[239,10.125],[246,10.25],[251,8.75],[257,5.5],[262,2.25],[267,0.25],[271,-1],[275,-2],[279,-1],[283,1],[287,2],[290,2.25],[299,-0.125],[309,-1.875],[327,-2.625],[337,-3],[345,-3.5],[351,-4.5],[356,-5.5],[361,-4.95],[366,-2.5538],[378,12.6088],[382,15.6116],[387,16.5],[409,15.0938],[439,14.6517]]
       }
       head_ctrl.Rotation.Y {
        type=MovieSceneScriptingFloatChannel
        default=0.9997
        keys [[0,5.1301],[8,5.2998],[15,5.809],[22,6.25],[30,6.5],[45,6.625],[51,6.75],[56,7.5],[59,8.75],[62,10.4],[66,11.8],[70,11.2],[75,9.9],[83,8.55],[93,7.75],[101,7],[107,5.2776],[111,0.6],[115,-2.9],[121,-4.2855],[130,-5.25],[138,-5.5],[144,-5],[148,-3],[153,2.4242],[160,4.9083],[172,6.0262],[187,6.3988],[194,6.1801],[199,5.5241],[204,3.4],[209,-0.1],[215,-2.9],[223,-4.25],[232,-5],[241,-5.325],[246,-5.3222],[251,-5.1],[257,-3.111],[267,5.3],[271,7.1608],[275,8],[279,5.2],[283,-2.2216],[287,-4.5],[292,-5.5],[299,-6],[309,-5.5],[327,-4],[337,-3.25],[345,-2.5],[351,-1],[356,3],[361,8],[366,10.6],[370,9.596],[375,7.9551],[382,6.274],[387,5.5354],[395,5.125],[409,5.0813],[439,5.1301]]
       }
       head_ctrl.Rotation.Z {
        type=MovieSceneScriptingFloatChannel
        default=-17.3456
        thinned_from=87
        keys [[0,11.593],[8,11.3819],[15,11.2909],[22,11.7595],[30,12.4427],[45,13.4123],[51,12.1623],[56,8.2471],[59,2.5],[62,-6.2962],[66,-17.7911],[70,-24.8282],[75,-28.6767],[83,-30.9904],[94,-32],[101,-31.5],[107,-28.75],[111,-23.0372],[115,-15.5313],[121,-9.3317],[129,-7.0651],[138,-5.25],[143,-2.9873],[148,4.9985],[153,16.3619],[161,25.6662],[171,30.5946],[184,33],[194,33.5],[199,31.5],[204,23.75],[209,9.8124],[215,-3.1418],[223,-8.5738],[233,-10.5554],[241,-11],[246,-10.5],[251,-9],[257,-6],[262,-4],[267,-3.5],[271,-4.5],[275,-9.5],[279,-16.25],[283,-21.6216],[287,-25],[292,-26.25],[299,-25],[309,-23.5],[325,-23],[337,-22.7204],[345,-22.4372],[351,-20],[364,3.123],[365,4.5086],[366,5.4806],[367,6.5086],[369,8.326],[370,9.2191],[371,9.9356],[372,10.5304],[373,11.0735],[374,11.5904],[375,12.0082],[376,12.3183],[377,12.5548],[379,12.9631],[380,13.1101],[382,13.3139],[385,13.4469],[390,13.4954],[405,13.465],[424,11.9908],[439,11.593]]
       }
       head_ctrl.Scale.X {type=MovieSceneScriptingFloatChannel, default=1, thinned_from=64, keys [[0,1],[439,1]]}
       head_ctrl.Scale.Y {type=MovieSceneScriptingFloatChannel, default=1, thinned_from=64, keys [[0,1],[439,1]]}
       head_ctrl.Scale.Z {type=MovieSceneScriptingFloatChannel, default=1, thinned_from=64, keys [[0,1],[439,1]]}
       clavicle_l_ctrl.Location.X {type=MovieSceneScriptingFloatChannel, default=0, thinned_from=64, keys [[0,0],[439,0]]}
       clavicle_l_ctrl.Location.Y {type=MovieSceneScriptingFloatChannel, default=0, thinned_from=64, keys [[0,0],[439,0]]}
       clavicle_l_ctrl.Location.Z {type=MovieSceneScriptingFloatChannel, default=0, thinned_from=64, keys [[0,0],[439,0]]}
       clavicle_l_ctrl.Rotation.X {type=MovieSceneScriptingFloatChannel, default=0, const [0,64]}
       clavicle_l_ctrl.Rotation.Y {
        type=MovieSceneScriptingFloatChannel
        default=-8.3327
        thinned_from=62
        keys [[0,0],[15,0],[22,0.1565],[30,0.3565],[45,0.4065],[51,0.3565],[56,-0.0498],[59,-0.4935],[62,-0.9685],[66,-1.4685],[70,-1.5935],[75,-1.4685],[83,-1.2185],[94,-1.0935],[101,-1.0935],[107,-1.2185],[111,-1.4685],[115,-1.8435],[121,-2.0935],[130,-2.2185],[143,-2.2185],[148,-1.9685],[153,-1.5935],[160,-1.1833],[172,-0.8435],[187,-1.2185],[194,-1.3234],[199,-1.1133],[204,-0.2726],[209,-0.8435],[215,-1.8435],[223,-3.2185],[233,-4.4685],[241,-4.9685],[251,-5.4685],[257,-5.8435],[262,-6.3435],[267,-7.3435],[271,-8.2295],[279,-8.7576],[283,-8.8319],[287,-7.9239],[292,-5.881],[299,-5.3702],[309,-5.125],[327,-5],[337,-4.875],[345,-5.375],[351,-7.1],[356,-8.7176],[361,-8.9109],[369,-8.4322],[375,-6.0935],[378,-3.5935],[382,-1.5935],[387,-0.5935],[395,0.0315],[409,0.1565],[439,0]]
       }
       clavicle_l_ctrl.Rotation.Z {
        type=MovieSceneScriptingFloatChannel
        default=30.8509
        thinned_from=57
        keys [[0,0],[45,0],[51,0.0044],[56,0],[66,0],[70,0.0033],[75,0.0042],[94,0.0044],[101,0.0043],[107,0.0032],[111,0],[246,0],[251,0.9226],[257,3.6903],[262,9.1934],[279,31.9005],[283,36.1307],[287,37.5357],[337,37.5407],[345,36.8536],[351,34.792],[356,30.0911],[375,5.2544],[380,1.5044],[387,0],[439,0]]
       }
       clavicle_l_ctrl.Scale.X {type=MovieSceneScriptingFloatChannel, default=1, thinned_from=64, keys [[0,1],[439,1]]}
       clavicle_l_ctrl.Scale.Y {type=MovieSceneScriptingFloatChannel, default=1, thinned_from=64, keys [[0,1],[439,1]]}
       clavicle_l_ctrl.Scale.Z {type=MovieSceneScriptingFloatChannel, default=1, thinned_from=64, keys [[0,1],[439,1]]}
       clavicle_r_ctrl.Location.X {type=MovieSceneScriptingFloatChannel, default=0, const [0,64]}
       clavicle_r_ctrl.Location.Y {type=MovieSceneScriptingFloatChannel, default=0, const [0,64]}
       clavicle_r_ctrl.Location.Z {type=MovieSceneScriptingFloatChannel, default=0, const [0,64]}
       clavicle_r_ctrl.Rotation.X {type=MovieSceneScriptingFloatChannel, default=0, const [0,64]}
       clavicle_r_ctrl.Rotation.Y {
        type=MovieSceneScriptingFloatChannel
        default=-4.1298
        thinned_from=63
        keys [[0,-1],[45,-1],[51,-1.0223],[56,-0.9416],[59,-0.8551],[62,-0.645],[66,-0.1733],[70,0.1621],[75,0.2789],[83,0.3083],[94,0.3179],[101,0.3017],[107,0.1882],[111,-0.0739],[115,-0.4638],[121,-0.8272],[130,-1],[148,-1],[153,-1.2014],[160,-1.277],[172,-1.311],[187,-1.3223],[194,-1.3098],[199,-1.2723],[204,-1.1223],[209,-1.5473],[215,-2.0535],[223,-2.0852],[233,-2.1801],[241,-2.2117],[246,-2.2223],[251,-2.1035],[257,-1.137],[262,0.4164],[267,3.2095],[271,6.6489],[275,8.8858],[279,2.2839],[283,-0.6869],[287,-0.0688],[292,1.3221],[299,1.6699],[309,1.7502],[327,1.7769],[337,1.7858],[345,0.6858],[351,-2.5002],[356,0.0839],[361,2.1853],[366,2.6295],[371,1.6138],[378,-1.2265],[382,-1.9344],[387,-1.3714],[395,-1.0876],[409,-1.032],[439,-1]]
       }
       clavicle_r_ctrl.Rotation.Z {
        type=MovieSceneScriptingFloatChannel
        default=22.6873
        thinned_from=55
        keys [[0,0],[257,0],[262,2.1831],[283,29.79],[292,30.7623],[299,30.9306],[337,30.9867],[345,29.54],[351,25.2],[371,5],[374,2.5],[382,0],[439,0]]
       }
       clavicle_r_ctrl.Scale.X {type=MovieSceneScriptingFloatChannel, default=1, thinned_from=64, keys [[0,1],[439,1]]}
       clavicle_r_ctrl.Scale.Y {type=MovieSceneScriptingFloatChannel, default=1, thinned_from=64, keys [[0,1],[439,1]]}
       clavicle_r_ctrl.Scale.Z {type=MovieSceneScriptingFloatChannel, default=1, thinned_from=64, keys [[0,1],[439,1]]}
       neck_01_ctrl.Location.X {
        type=MovieSceneScriptingFloatChannel
        default=2.6354
        thinned_from=64
        keys [[0,1.3609],[107,1.3609],[111,1.4],[115,1.625],[121,1.6715],[267,1.675],[271,1.8487],[275,2.257],[279,2.6331],[283,2.7569],[287,2.7879],[292,2.7956],[337,2.7982],[345,2.7815],[351,2.7313],[356,2.5306],[361,2.2112],[366,1.7922],[371,1.5013],[375,1.3633],[439,1.3609]]
       }
       neck_01_ctrl.Location.Y {
        type=MovieSceneScriptingFloatChannel
        default=0.518
        thinned_from=64
        keys [[0,-2.9424],[267,-2.9424],[271,-2.3219],[275,-0.8993],[279,0.4114],[283,0.823],[287,0.9259],[292,0.9516],[299,0.958],[337,0.9602],[345,0.9148],[351,0.7785],[356,0.2335],[361,-0.6336],[366,-1.7712],[371,-2.5612],[375,-2.9358],[439,-2.9424]]
       }
       neck_01_ctrl.Location.Z {
        type=MovieSceneScriptingFloatChannel
        default=-1.487
        thinned_from=64
        keys [[0,-0.217],[267,-0.217],[271,-0.4447],[275,-0.9668],[279,-1.4478],[283,-1.5989],[287,-1.6366],[292,-1.6461],[337,-1.6492],[345,-1.6326],[351,-1.5826],[356,-1.3825],[361,-1.0643],[366,-0.6468],[371,-0.3569],[375,-0.2195],[395,-0.2195],[409,-0.2179],[439,-0.217]]
       }
       neck_01_ctrl.Rotation.X {type=MovieSceneScriptingFloatChannel, default=0, const [0,64]}
       neck_01_ctrl.Rotation.Y {type=MovieSceneScriptingFloatChannel, default=0, thinned_from=64, keys [[0,0],[439,0]]}
       neck_01_ctrl.Rotation.Z {type=MovieSceneScriptingFloatChannel, default=0, thinned_from=64, keys [[0,0],[439,0]]}
       neck_01_ctrl.Scale.X {type=MovieSceneScriptingFloatChannel, default=1, thinned_from=64, keys [[0,1],[439,1]]}
       neck_01_ctrl.Scale.Y {type=MovieSceneScriptingFloatChannel, default=1, thinned_from=64, keys [[0,1],[439,1]]}
       neck_01_ctrl.Scale.Z {type=MovieSceneScriptingFloatChannel, default=1, thinned_from=64, keys [[0,1],[439,1]]}
       neck_02_ctrl.Location.X {type=MovieSceneScriptingFloatChannel, default=0, thinned_from=64, keys [[0,0],[439,0]]}
       neck_02_ctrl.Location.Y {type=MovieSceneScriptingFloatChannel, default=0, thinned_from=64, keys [[0,0],[439,0]]}
       neck_02_ctrl.Location.Z {type=MovieSceneScriptingFloatChannel, default=0, thinned_from=64, keys [[0,0],[439,0]]}
       neck_02_ctrl.Rotation.X {type=MovieSceneScriptingFloatChannel, default=0, thinned_from=64, keys [[0,0],[439,0]]}
       neck_02_ctrl.Rotation.Y {type=MovieSceneScriptingFloatChannel, default=0, thinned_from=64, keys [[0,0],[439,0]]}
       neck_02_ctrl.Rotation.Z {type=MovieSceneScriptingFloatChannel, default=0, thinned_from=64, keys [[0,0],[439,0]]}
       neck_02_ctrl.Scale.X {type=MovieSceneScriptingFloatChannel, default=1, thinned_from=64, keys [[0,1],[439,1]]}
       neck_02_ctrl.Scale.Y {type=MovieSceneScriptingFloatChannel, default=1, thinned_from=64, keys [[0,1],[439,1]]}
       neck_02_ctrl.Scale.Z {type=MovieSceneScriptingFloatChannel, default=1, thinned_from=64, keys [[0,1],[439,1]]}
       head_ik_ctrl.Location.X {
        type=MovieSceneScriptingFloatChannel
        default=2.971
        thinned_from=64
        keys [[0,1.1879],[8,1.3016],[15,1.6427],[22,2.4003],[30,2.3415],[45,2.4164],[51,1.2],[56,2.3863],[62,2.3423],[66,2.0769],[70,1.972],[75,1.9479],[83,1.916],[94,1.9056],[101,1.9113],[107,1.9516],[111,2.075],[115,2.4012],[121,2.4855],[130,2.5136],[138,2.5189],[143,2.5348],[148,2.5985],[153,1.7168],[160,1.3862],[172,2.4648],[187,2.8244],[194,2.7221],[199,2.4153],[204,1.1879],[215,1.1879],[223,1.3924],[233,2.0059],[241,2.2119],[246,2.2806],[251,2.2216],[257,2.0446],[262,1.6928],[267,1.3368],[271,1.8144],[275,2.2865],[279,2.9276],[283,3.2161],[287,3.2883],[292,3.3063],[337,3.3123],[345,3.278],[351,3.1751],[356,2.7636],[361,2.3608],[366,2.2265],[375,2.1828],[378,2.3203],[382,2.3719],[387,2.3951],[409,2.4109],[439,2.4029]]
       }
       head_ik_ctrl.Location.Y {
        type=MovieSceneScriptingFloatChannel
        default=-0.6582
        thinned_from=64
        keys [[0,0.0643],[8,-0.1429],[15,-0.7646],[22,-2.1006],[30,-2.0033],[45,-2.1284],[51,0.1],[56,-2.1613],[62,-2.2064],[66,-2.4425],[70,-2.5121],[83,-2.5529],[94,-2.5608],[101,-2.5587],[107,-2.5443],[111,-2.5001],[115,-2.2803],[121,-2.2247],[130,-2.2061],[138,-2.1996],[143,-2.1801],[148,-2.1021],[153,-0.7481],[160,-0.2404],[172,-1.4045],[187,-1.7926],[194,-1.6765],[199,-1.3284],[204,0.0643],[215,0.0643],[223,-0.4004],[233,-1.7946],[241,-2.2591],[246,-2.414],[251,-2.4526],[257,-2.5687],[267,-3.0331],[271,-2.1258],[275,-1.229],[279,-0.8408],[283,-0.6661],[287,-0.6224],[292,-0.6115],[299,-0.6088],[337,-0.6079],[345,-0.613],[351,-0.6285],[356,-0.6906],[361,-0.9425],[366,-1.0265],[371,-1.4846],[375,-1.9268],[378,-2.0624],[382,-2.1132],[387,-2.1361],[395,-2.1409],[409,-2.1327],[439,-2.1437]]
       }
       head_ik_ctrl.Location.Z {
        type=MovieSceneScriptingFloatChannel
        default=-0.4704
        thinned_from=64
        keys [[0,-0.069],[8,-0.022],[15,0.1193],[22,0.3987],[30,0.4039],[45,0.4333],[51,-0.1],[56,0.4377],[62,0.4514],[66,0.4576],[70,0.4757],[75,0.4755],[83,0.4731],[101,0.4712],[107,0.4634],[111,0.4394],[115,0.4211],[121,0.4032],[130,0.3973],[138,0.3984],[143,0.4018],[148,0.4151],[153,0.1125],[160,-0.0009],[172,0.2926],[187,0.3905],[194,0.3618],[199,0.2756],[204,-0.069],[215,-0.069],[223,0.0503],[233,0.4084],[241,0.523],[246,0.5612],[251,0.5634],[257,0.5701],[267,0.597],[271,0.0159],[275,-0.5584],[279,-0.4356],[283,-0.3803],[287,-0.3665],[292,-0.363],[299,-0.3622],[337,-0.3619],[345,-0.3709],[351,-0.3978],[356,-0.5055],[361,-0.4521],[366,-0.4343],[371,-0.1435],[375,0.1373],[378,0.3223],[382,0.3917],[387,0.4229],[395,0.4295],[409,0.4299],[439,0.4334]]
       }
       head_ik_ctrl.Rotation.X {
        type=MovieSceneScriptingFloatChannel
        default=-2.2149
        thinned_from=64
        keys [[0,14.6517],[15,14.6517],[22,15.4736],[30,14.9129],[45,15],[51,14.7],[56,15],[59,11.9544],[62,8.8367],[66,6.6561],[70,5.639],[75,5.3848],[83,5.3208],[94,5.3],[101,5.394],[107,6.0522],[111,8.0661],[115,8.7457],[121,8.9156],[130,8.9722],[138,8.8177],[143,8.3542],[148,6.5],[153,11.5948],[160,13.5054],[172,10.0295],[187,8.8709],[194,9.2322],[199,10.3161],[204,14.6517],[215,14.6517],[223,13.5868],[233,10.3921],[241,9.3272],[246,8.9722],[251,8.3802],[257,6.6042],[262,3.0729],[267,-0.5],[271,-3.2661],[275,-6],[279,-3.3878],[283,-2.2123],[287,-1.9184],[292,-1.8449],[299,-1.8266],[309,-1.8223],[337,-1.8204],[345,-1.861],[351,-1.9825],[356,-2.4687],[361,-5.6422],[366,-6.7],[371,-0.493],[375,5.5],[378,11.2198],[382,13.3647],[387,14.3299],[395,14.531],[409,14.9614],[439,14.6517]]
       }
       head_ik_ctrl.Rotation.Y {
        type=MovieSceneScriptingFloatChannel
        default=-1.714
        thinned_from=64
        keys [[0,9.8268],[8,9.3865],[15,8.0655],[22,6.1001],[30,6.2077],[45,6.2],[51,9.8],[56,6.1275],[59,7.6062],[62,9.1198],[66,9.9706],[70,10.3676],[75,9.4169],[83,8.4989],[94,8.2],[101,7.8893],[107,5.7145],[111,-0.9403],[115,-3.7956],[121,-4.5094],[130,-4.7474],[138,-4.5757],[143,-4.0605],[148,-2],[153,5.3917],[160,8.1636],[172,6.84],[187,6.3988],[194,6.613],[199,7.2558],[204,9.8268],[215,9.8268],[223,7.0583],[233,-1.2472],[241,-4.5868],[246,-5.7],[251,-4.4063],[257,-0.525],[262,7.1921],[267,15],[271,11.9825],[275,9],[279,0.4403],[283,-3.4115],[287,-4.3745],[292,-4.6152],[299,-4.6754],[309,-4.6893],[337,-4.6954],[345,-4.3892],[351,-3.4704],[356,0.2046],[361,6.0511],[366,8],[371,10.5439],[375,13],[378,8.0813],[382,6.2368],[387,5.4068],[395,5.2338],[409,5.168],[439,5.1301]]
       }
       head_ik_ctrl.Rotation.Z {
        type=MovieSceneScriptingFloatChannel
        default=-18.8327
        thinned_from=64
        keys [[0,8.187],[8,8.5063],[15,9.4642],[22,12.4001],[30,12.0737],[45,12.5],[51,8.2],[56,7.5],[59,-3.3396],[62,-14.4358],[66,-24.1966],[70,-28.6241],[75,-29.731],[83,-30.0094],[94,-30.1],[101,-29.7476],[107,-27.2806],[115,-12.1703],[121,-10.2799],[130,-9.6497],[138,-7.9841],[143,-2.9873],[148,17],[153,11.4919],[160,9.4263],[172,25.9816],[187,31.5],[194,30.0429],[199,25.6717],[204,8.187],[215,8.187],[223,4.8426],[233,-5.1905],[241,-8.5349],[246,-9.6497],[251,-9.7341],[257,-9.9873],[267,-11],[275,-11],[279,-18.3253],[283,-21.6216],[287,-22.4457],[292,-22.6517],[299,-22.7032],[337,-22.7204],[345,-22.3211],[351,-21.123],[356,-16.3309],[361,-8.707],[366,-6.1657],[371,3.5853],[375,13],[378,12.1206],[382,11.7908],[387,11.6424],[395,11.6115],[439,11.593]]
       }
       head_ik_ctrl.Scale.X {type=MovieSceneScriptingFloatChannel, default=1, thinned_from=64, keys [[0,1],[439,1]]}
       head_ik_ctrl.Scale.Y {type=MovieSceneScriptingFloatChannel, default=1, thinned_from=64, keys [[0,1],[439,1]]}
       head_ik_ctrl.Scale.Z {type=MovieSceneScriptingFloatChannel, default=1, thinned_from=64, keys [[0,1],[439,1]]}
       chest_ctrl.Location.X {
        type=MovieSceneScriptingFloatChannel
        default=-1.5546
        thinned_from=64
        keys [[0,-0.458],[8,-0.4553],[15,-0.4475],[22,-0.4007],[30,-0.4318],[45,-0.4301],[51,-0.5],[56,-0.4301],[66,-0.4301],[70,-0.4352],[101,-0.435],[107,-0.4338],[111,-0.4301],[148,-0.4301],[153,-0.4475],[160,-0.454],[172,-0.4361],[187,-0.4301],[194,-0.4319],[199,-0.4371],[204,-0.458],[215,-0.458],[223,-0.4527],[233,-0.4371],[241,-0.4319],[246,-0.4301],[251,-0.4652],[257,-0.5705],[262,-0.7799],[267,-0.9917],[271,-1.2171],[275,-1.4398],[279,-1.6108],[283,-1.6877],[287,-1.7069],[292,-1.7117],[337,-1.7133],[345,-1.6989],[351,-1.6554],[360,-1.2951],[361,-1.1983],[366,-1.104],[371,-0.8473],[375,-0.5995],[378,-0.4936],[382,-0.454],[387,-0.4361],[395,-0.4324],[409,-0.431],[439,-0.4301]]
       }
       chest_ctrl.Location.Y {
        type=MovieSceneScriptingFloatChannel
        default=0.239
        thinned_from=64
        keys [[0,0.2745],[8,0.4776],[15,1.0869],[22,2.3839],[30,2.3139],[45,2.4408],[51,0.3],[56,2.4408],[66,2.4408],[70,2.4315],[107,2.434],[111,2.4408],[148,2.4408],[153,1.0869],[160,0.5791],[172,1.9754],[187,2.4408],[194,2.3054],[199,1.8992],[204,0.2745],[215,0.2745],[223,0.6807],[233,1.8992],[241,2.3054],[246,2.4408],[251,2.417],[257,2.3458],[267,2.0607],[271,1.2978],[275,0.5438],[279,0.1542],[283,-0.0212],[287,-0.065],[292,-0.076],[299,-0.0787],[309,-0.0794],[337,-0.0796],[345,-0.0504],[351,0.0373],[360,0.8406],[361,0.9496],[366,1.1367],[371,1.6352],[375,2.1165],[378,2.3192],[382,2.3952],[387,2.4294],[395,2.4365],[439,2.4408]]
       }
       chest_ctrl.Location.Z {
        type=MovieSceneScriptingFloatChannel
        default=-0.1142
        thinned_from=64
        keys [[0,0.0044],[8,0.0011],[15,-0.009],[22,-0.0615],[30,-0.0292],[45,-0.0313],[51,0],[56,-0.0313],[66,-0.0313],[70,-0.0301],[101,-0.0301],[107,-0.0304],[111,-0.0313],[148,-0.0313],[153,-0.009],[160,-0.0006],[172,-0.0236],[187,-0.0313],[194,-0.029],[199,-0.0223],[204,0.0044],[215,0.0044],[223,-0.0023],[233,-0.0223],[241,-0.029],[246,-0.0313],[251,-0.031],[257,-0.0303],[267,-0.0275],[275,-0.0601],[279,-0.125],[283,-0.1542],[287,-0.1615],[292,-0.1633],[299,-0.1638],[337,-0.164],[345,-0.1586],[351,-0.1424],[360,-0.0263],[361,-0.0183],[366,0.0015],[371,-0.0006],[375,-0.0026],[378,-0.0205],[382,-0.0272],[387,-0.0302],[395,-0.0309],[409,-0.0311],[439,-0.0313]]
       }
       chest_ctrl.Rotation.X {
        type=MovieSceneScriptingFloatChannel
        default=5.7996
        thinned_from=64
        keys [[0,2.8184],[8,5.46],[15,13.3849],[22,31.0999],[30,29.4102],[45,31.083],[51,2.8],[56,30.9956],[66,30.9573],[70,30.8766],[107,30.8877],[115,30.9989],[148,31.0084],[153,13.3896],[160,6.7826],[172,24.9472],[187,31.0021],[194,29.2406],[199,23.9561],[204,2.8184],[215,2.8184],[223,8.1042],[233,23.9617],[241,29.2475],[246,31.0094],[251,30.7048],[257,29.7909],[267,26.1353],[271,16.0045],[275,5.9914],[279,3.533],[283,2.4267],[287,2.1501],[292,2.081],[299,2.0637],[309,2.0597],[337,2.0579],[345,2.4295],[351,3.5441],[360,13.0905],[361,13.3707],[366,15.1601],[371,21.5093],[375,27.6396],[378,29.7371],[382,30.5236],[387,30.8776],[395,30.9513],[439,30.9956]]
       }
       chest_ctrl.Rotation.Y {
        type=MovieSceneScriptingFloatChannel
        default=15.5282
        thinned_from=64
        keys [[0,13.4277],[8,11.985],[15,7.6569],[22,-1.9999],[30,-1.0594],[45,-1.961],[51,13.4],[56,-1.8768],[62,-1.6242],[66,-0.1153],[70,0.5611],[75,0.7252],[83,0.8238],[94,0.8559],[101,0.8196],[107,0.5658],[111,-0.2107],[115,-2.4637],[121,-3.0269],[130,-3.2147],[138,-3.2005],[143,-3.158],[148,-2.9878],[153,7.2719],[160,11.1193],[172,-0.1639],[187,-3.9249],[194,-2.8404],[199,0.4132],[204,13.4277],[215,13.4277],[223,10.3411],[233,1.0813],[241,-2.0053],[246,-3.0342],[251,-2.5739],[257,-1.1929],[262,1.5528],[267,4.3309],[271,8.1841],[275,11.9925],[279,15.5066],[283,17.0879],[287,17.4833],[292,17.5821],[337,17.6151],[345,17.4323],[351,16.8839],[360,11.8117],[361,10.6356],[366,9.284],[371,5.0819],[375,1.0247],[378,-0.8506],[382,-1.5539],[387,-1.8704],[395,-1.9363],[409,-1.9783],[439,-1.9759]]
       }
       chest_ctrl.Rotation.Z {
        type=MovieSceneScriptingFloatChannel
        default=-8.494
        thinned_from=64
        keys [[0,7.2983],[8,7.0815],[15,6.431],[22,4.9999],[30,5.121],[45,4.9855],[51,7.3],[56,4.8285],[62,4.3577],[66,1.474],[70,0.1674],[75,-0.1565],[83,-0.4569],[94,-0.5547],[101,-0.5078],[107,-0.1796],[111,0.8247],[115,3.2546],[121,3.862],[130,4.0645],[138,4.0108],[143,3.8495],[148,3.2044],[153,5.7631],[160,6.7226],[172,6.7495],[187,6.7584],[194,6.7921],[199,6.8934],[204,7.2983],[215,7.2983],[223,6.5637],[233,4.3598],[241,3.6251],[246,3.3803],[251,3.1651],[257,2.5196],[262,1.2361],[267,-0.0625],[271,-2.6471],[275,-5.2016],[279,-8.7567],[283,-10.3565],[287,-10.7564],[292,-10.8564],[299,-10.8814],[337,-10.8897],[345,-10.6617],[351,-9.9775],[360,-2.8689],[361,-3.3768],[366,-2.0888],[371,0.7398],[375,3.4709],[378,4.4526],[382,4.8208],[387,4.9864],[395,5.0209],[409,5.1089],[439,5.0416]]
       }
       chest_ctrl.Scale.X {type=MovieSceneScriptingFloatChannel, default=1, thinned_from=64, keys [[0,1],[439,1]]}
       chest_ctrl.Scale.Y {type=MovieSceneScriptingFloatChannel, default=1, thinned_from=64, keys [[0,1],[439,1]]}
       chest_ctrl.Scale.Z {type=MovieSceneScriptingFloatChannel, default=1, thinned_from=64, keys [[0,1],[439,1]]}
       chest_tan_ctrl.Location.X {
        type=MovieSceneScriptingFloatChannel
        default=0.0754
        thinned_from=64
        keys [[0,-0.7418],[8,-0.6676],[15,-0.4453],[22,-0.0595],[30,0.0026],[45,0.0489],[51,-0.7],[56,0.0437],[62,0.0281],[66,-0.0625],[70,-0.1021],[75,-0.1116],[83,-0.114],[94,-0.1147],[101,-0.1119],[107,-0.0918],[111,-0.0303],[115,0.1648],[121,0.2136],[130,0.2298],[148,0.2297],[153,-0.3775],[160,-0.6052],[172,0.0215],[187,0.2304],[194,0.1697],[199,-0.0126],[204,-0.7418],[215,-0.7418],[223,-0.5596],[233,-0.0131],[241,0.169],[246,0.2298],[251,0.2453],[257,0.292],[262,0.3847],[267,0.4786],[275,0.36],[279,0.1702],[283,0.0848],[287,0.0635],[292,0.0582],[299,0.0568],[309,0.0565],[337,0.0564],[345,0.0586],[351,0.0654],[360,0.2645],[361,0.1056],[366,0.11],[371,0.091],[375,0.0727],[378,0.0579],[382,0.0523],[387,0.0498],[395,0.0493],[409,0.0488],[439,0.049]]
       }
       chest_tan_ctrl.Location.Y {
        type=MovieSceneScriptingFloatChannel
        default=0.821
        thinned_from=64
        keys [[0,-0.2592],[8,-0.1032],[15,0.3649],[22,1.0125],[30,1.3153],[45,1.4154],[51,-0.3],[56,1.4044],[62,1.4029],[66,1.3934],[70,1.3897],[107,1.3883],[111,1.3918],[115,1.3997],[143,1.4016],[148,1.3997],[153,0.3629],[160,-0.0259],[172,1.0514],[187,1.4105],[194,1.3061],[199,0.9931],[204,-0.2592],[215,-0.2592],[223,0.052],[233,0.9854],[241,1.2965],[246,1.4002],[251,1.3837],[257,1.3342],[267,1.1362],[271,0.8247],[275,0.5169],[279,0.6956],[283,0.776],[287,0.7961],[292,0.8011],[337,0.8028],[345,0.8072],[351,0.8205],[360,0.8452],[361,0.8217],[366,0.8045],[371,1.0514],[375,1.2898],[378,1.3619],[382,1.3889],[387,1.4011],[409,1.4048],[439,1.4051]]
       }
       chest_tan_ctrl.Location.Z {
        type=MovieSceneScriptingFloatChannel
        default=1.2561
        thinned_from=64
        keys [[0,1.2329],[8,1.2517],[15,1.3081],[22,2.6142],[30,1.4241],[45,1.4366],[51,1.2],[56,1.4334],[111,1.4296],[115,1.4348],[148,1.4357],[153,1.3089],[160,1.2614],[172,1.3945],[187,1.4389],[194,1.426],[199,1.3874],[204,1.2329],[215,1.2329],[223,1.2709],[233,1.3851],[241,1.4232],[246,1.4358],[251,1.4321],[257,1.421],[267,1.3766],[275,1.2395],[279,1.2289],[283,1.2241],[337,1.2225],[345,1.2264],[351,1.2383],[360,2.5497],[361,1.3074],[366,1.3147],[371,1.3687],[375,1.4209],[378,1.4289],[387,1.4332],[439,1.4336]]
       }
       chest_tan_ctrl.Rotation.X {
        type=MovieSceneScriptingFloatChannel
        default=0
        thinned_from=64
        keys [[0,1.3545],[8,1.2275],[15,0.8466],[22,0],[30,0.0794],[45,0],[51,1.4],[56,0],[148,0],[153,0.8466],[160,1.164],[172,0.291],[187,0],[194,0.0847],[199,0.3386],[204,1.3545],[215,1.3545],[223,1.1006],[233,0.3386],[241,0.0847],[246,0],[439,0]]
       }
       chest_tan_ctrl.Rotation.Y {
        type=MovieSceneScriptingFloatChannel
        default=0
        thinned_from=64
        keys [[0,0],[8,0.0788],[15,0.3151],[22,1.5984],[30,0.7871],[45,0.835],[51,0],[56,0.8493],[62,0.8758],[66,1.0666],[70,1.1704],[75,1.1963],[83,1.2658],[94,1.2885],[101,1.2951],[107,1.3416],[111,1.4839],[115,2.1616],[121,2.331],[130,2.3875],[138,2.4031],[143,2.45],[148,2.6378],[153,0.9892],[160,0.3709],[172,1.2933],[187,1.6008],[194,1.5007],[199,1.2006],[204,0],[215,0],[223,0.485],[233,1.94],[241,2.425],[246,2.5866],[251,2.5603],[257,2.4812],[267,2.165],[271,1.0762],[275,0],[366,0],[371,0.2581],[375,0.5072],[378,0.7053],[382,0.7795],[387,0.8129],[395,0.8199],[409,0.799],[439,0.8241]]
       }
       chest_tan_ctrl.Rotation.Z {
        type=MovieSceneScriptingFloatChannel
        default=11.3311
        thinned_from=64
        keys [[0,0],[62,0],[66,1.4846],[70,2.1395],[83,2.5617],[94,2.6459],[101,2.5877],[107,2.1803],[111,0.9335],[115,0.2334],[121,0.0583],[130,0],[246,0],[251,0.1345],[257,0.5381],[262,1.3406],[267,2.1525],[271,6.4326],[275,10.663],[279,12.3796],[283,13.1521],[287,13.3452],[292,13.3935],[337,13.4096],[345,13.2207],[351,12.6543],[360,1.1314],[361,6.6436],[366,5.3953],[371,2.6503],[375,0],[439,0]]
       }
       chest_tan_ctrl.Scale.X {type=MovieSceneScriptingFloatChannel, default=1, thinned_from=64, keys [[0,1],[439,1]]}
       chest_tan_ctrl.Scale.Y {type=MovieSceneScriptingFloatChannel, default=1, thinned_from=64, keys [[0,1],[439,1]]}
       chest_tan_ctrl.Scale.Z {type=MovieSceneScriptingFloatChannel, default=1, thinned_from=64, keys [[0,1],[439,1]]}
       Weight {type=MovieSceneScriptingFloatChannel, default=1}
       TickableParentConstraint_0 {type=MovieSceneScriptingConstraintChannel, const [true,1]}
       ParentConstraint_2 {type=MovieSceneScriptingConstraintChannel}
       ParentConstraint_9 {type=MovieSceneScriptingConstraintChannel, const [true,1]}
      ]
     }
     {
      type=MovieSceneControlRigParameterSection
      channels [
       global_ctrl.Location.X {type=MovieSceneScriptingFloatChannel, default=0, thinned_from=64, keys [[0,0],[439,0]]}
       global_ctrl.Location.Y {type=MovieSceneScriptingFloatChannel, default=0, thinned_from=64, keys [[0,0],[439,0]]}
       global_ctrl.Location.Z {type=MovieSceneScriptingFloatChannel, default=0, thinned_from=64, keys [[0,0],[439,0]]}
       global_ctrl.Rotation.X {type=MovieSceneScriptingFloatChannel, default=0, thinned_from=64, keys [[0,0],[439,0]]}
       global_ctrl.Rotation.Y {type=MovieSceneScriptingFloatChannel, default=0, thinned_from=64, keys [[0,0],[439,0]]}
       global_ctrl.Rotation.Z {type=MovieSceneScriptingFloatChannel, default=0, thinned_from=64, keys [[0,0],[439,0]]}
       root_ctrl.Location.X {type=MovieSceneScriptingFloatChannel, default=0, thinned_from=64, keys [[0,0],[439,0]]}
       root_ctrl.Location.Y {type=MovieSceneScriptingFloatChannel, default=0, thinned_from=64, keys [[0,0],[439,0]]}
       root_ctrl.Location.Z {type=MovieSceneScriptingFloatChannel, default=0, thinned_from=64, keys [[0,0],[439,0]]}
       root_ctrl.Rotation.X {type=MovieSceneScriptingFloatChannel, default=0, thinned_from=64, keys [[0,0],[439,0]]}
       root_ctrl.Rotation.Y {type=MovieSceneScriptingFloatChannel, default=0, thinned_from=64, keys [[0,0],[439,0]]}
       root_ctrl.Rotation.Z {type=MovieSceneScriptingFloatChannel, default=0, thinned_from=64, keys [[0,0],[439,0]]}
       root_ctrl.Scale.X {type=MovieSceneScriptingFloatChannel, default=1, thinned_from=64, keys [[0,1],[439,1]]}
       root_ctrl.Scale.Y {type=MovieSceneScriptingFloatChannel, default=1, thinned_from=64, keys [[0,1],[439,1]]}
       root_ctrl.Scale.Z {type=MovieSceneScriptingFloatChannel, default=1, thinned_from=64, keys [[0,1],[439,1]]}
       body_offset_ctrl.Location.X {type=MovieSceneScriptingFloatChannel, default=0, thinned_from=64, keys [[0,0],[439,0]]}
       body_offset_ctrl.Location.Y {type=MovieSceneScriptingFloatChannel, default=0, thinned_from=64, keys [[0,0],[439,0]]}
       body_offset_ctrl.Location.Z {type=MovieSceneScriptingFloatChannel, default=0, thinned_from=64, keys [[0,0],[439,0]]}
       body_offset_ctrl.Rotation.X {type=MovieSceneScriptingFloatChannel, default=0, thinned_from=64, keys [[0,0],[439,0]]}
       body_offset_ctrl.Rotation.Y {type=MovieSceneScriptingFloatChannel, default=0, thinned_from=64, keys [[0,0],[439,0]]}
       body_offset_ctrl.Rotation.Z {type=MovieSceneScriptingFloatChannel, default=0, thinned_from=64, keys [[0,0],[439,0]]}
       body_offset_ctrl.Scale.X {type=MovieSceneScriptingFloatChannel, default=1, thinned_from=64, keys [[0,1],[439,1]]}
       body_offset_ctrl.Scale.Y {type=MovieSceneScriptingFloatChannel, default=1, thinned_from=64, keys [[0,1],[439,1]]}
       body_offset_ctrl.Scale.Z {type=MovieSceneScriptingFloatChannel, default=1, thinned_from=64, keys [[0,1],[439,1]]}
       thigh_l_fk_ctrl.Location.X {type=MovieSceneScriptingFloatChannel, default=0, thinned_from=64, keys [[0,0],[439,0]]}
       thigh_l_fk_ctrl.Location.Y {type=MovieSceneScriptingFloatChannel, default=0, thinned_from=64, keys [[0,0],[439,0]]}
       thigh_l_fk_ctrl.Location.Z {type=MovieSceneScriptingFloatChannel, default=0, thinned_from=64, keys [[0,0],[439,0]]}
       thigh_l_fk_ctrl.Rotation.X {
        type=MovieSceneScriptingFloatChannel
        default=59.9342
        thinned_from=64
        keys [[0,59.9342],[15,59.9342],[22,60.2529],[45,60.3505],[51,59.9],[56,59.7258],[62,59.1019],[66,55.7313],[70,54.1317],[83,53.2779],[94,53.1226],[101,53.1598],[107,53.4205],[111,54.2181],[115,55.2352],[121,55.535],[130,55.635],[138,55.8344],[143,56.4328],[148,58.8263],[153,42.4378],[160,36.2922],[172,55.4883],[187,61.887],[194,60.0569],[199,54.5665],[204,32.6048],[215,32.6048],[223,36.074],[233,46.4817],[241,49.9626],[246,51.123],[251,50.2691],[257,47.7075],[267,37.461],[275,49.6693],[279,46.1887],[283,44.6224],[287,44.2308],[292,44.1329],[337,44.1003],[345,44.0067],[351,43.7261],[356,42.6035],[366,42.7416],[371,48.9039],[375,54.8536],[378,58.029],[382,59.2198],[387,59.7556],[409,60.0875],[439,59.9342]]
       }
       thigh_l_fk_ctrl.Rotation.Y {
        type=MovieSceneScriptingFloatChannel
        default=12.1109
        thinned_from=64
        keys [[0,12.1109],[15,12.1109],[22,12.1694],[45,12.1872],[51,12.1],[56,12.0034],[62,11.6797],[66,12.4289],[70,12.2764],[75,12.4931],[83,12.6544],[94,12.7069],[101,12.7184],[107,12.7993],[111,13.0468],[115,11.9604],[121,12.1442],[130,12.2054],[138,12.1507],[143,11.9866],[148,11.33],[153,14.4516],[160,15.6222],[172,12.9571],[187,12.0688],[194,12.3348],[199,13.1327],[204,16.3246],[215,16.3246],[223,14.4249],[233,8.726],[241,6.9954],[246,6.4185],[251,6.5955],[257,7.1263],[267,9.2494],[275,9.6179],[279,7.9835],[283,7.2481],[287,7.0642],[292,7.0182],[337,7.0029],[345,7.1082],[351,7.4243],[356,8.6884],[361,10.2061],[366,10.7119],[371,11.0289],[375,11.335],[378,11.82],[382,12.0018],[387,12.0836],[409,12.1483],[439,12.1109]]
       }
       thigh_l_fk_ctrl.Rotation.Z {
        type=MovieSceneScriptingFloatChannel
        default=7.3236
        thinned_from=64
        keys [[0,7.3236],[15,7.3236],[22,7.4096],[30,7.4078],[45,7.4358],[51,7.3],[56,7.2635],[62,7.0828],[66,5.9964],[70,5.5224],[75,5.4121],[83,5.2648],[94,5.2168],[101,5.2256],[107,5.2871],[111,5.4754],[115,5.7505],[121,5.8323],[130,5.8596],[138,5.9162],[143,6.0861],[148,6.7656],[153,2.9766],[160,1.5557],[172,6.0694],[187,7.5739],[194,7.1445],[199,5.8562],[204,0.7032],[215,0.7032],[223,1.3759],[233,3.394],[241,4.0702],[246,4.2956],[257,4.3104],[267,4.3551],[275,5.3543],[279,5.7035],[283,5.8606],[287,5.8999],[345,5.9175],[351,5.9312],[356,5.9857],[361,5.4397],[366,5.2577],[371,6.102],[375,6.9171],[378,7.1712],[382,7.2664],[387,7.3093],[395,7.3183],[409,7.3922],[439,7.3236]]
       }
       thigh_l_fk_ctrl.Scale.X {type=MovieSceneScriptingFloatChannel, default=1, thinned_from=64, keys [[0,1],[439,1]]}
       thigh_l_fk_ctrl.Scale.Y {type=MovieSceneScriptingFloatChannel, default=1, thinned_from=64, keys [[0,1],[439,1]]}
       thigh_l_fk_ctrl.Scale.Z {type=MovieSceneScriptingFloatChannel, default=1, thinned_from=64, keys [[0,1],[439,1]]}
       calf_l_fk_ctrl.Location.X {
        type=MovieSceneScriptingFloatChannel
        default=-1.1428
        thinned_from=64
        keys [[0,-1.1428],[15,-1.1428],[22,-1.132],[30,-1.1322],[45,-1.1287],[51,-1.1],[56,-1.1456],[62,-1.1543],[66,-1.0796],[70,-1.0794],[75,-1.0687],[83,-1.0591],[101,-1.0549],[107,-1.047],[111,-1.023],[115,-1.0336],[121,-1.0171],[130,-1.0116],[143,-1.0123],[148,-1.0144],[153,-0.3804],[160,-0.1426],[172,-0.7377],[187,-0.9361],[194,-0.8775],[199,-0.702],[204,0],[215,0],[223,-0.1874],[233,-0.7495],[241,-0.9282],[246,-0.9878],[251,-0.9811],[257,-0.961],[267,-0.8806],[271,-0.5993],[275,-0.3214],[279,-0.3862],[283,-0.4153],[287,-0.4226],[292,-0.4244],[337,-0.425],[345,-0.4279],[351,-0.4367],[356,-0.4715],[361,-0.5607],[366,-0.5905],[371,-0.6376],[375,-0.6831],[378,-0.9704],[382,-1.0782],[387,-1.1267],[395,-1.1368],[409,-1.1346],[439,-1.1428]]
       }
       calf_l_fk_ctrl.Location.Y {
        type=MovieSceneScriptingFloatChannel
        default=0.1497
        thinned_from=64
        keys [[0,0.1497],[15,0.1497],[22,0.1483],[30,0.1483],[45,0.1479],[51,0.1],[56,0.1501],[62,0.1512],[66,0.1414],[70,0.1414],[75,0.14],[83,0.1387],[101,0.1382],[107,0.1372],[111,0.134],[115,0.1354],[121,0.1332],[130,0.1325],[143,0.1326],[148,0.1329],[153,0.0498],[160,0.0187],[172,0.0966],[187,0.1226],[194,0.115],[199,0.092],[204,0],[215,0],[223,0.0245],[233,0.0982],[241,0.1216],[246,0.1294],[251,0.1285],[257,0.1259],[267,0.1154],[271,0.0785],[275,0.0421],[279,0.0506],[283,0.0544],[287,0.0554],[292,0.0556],[337,0.0557],[345,0.0561],[351,0.0572],[356,0.0618],[361,0.0735],[366,0.0774],[371,0.0835],[375,0.0895],[378,0.1271],[382,0.1412],[387,0.1476],[395,0.1489],[409,0.1486],[439,0.1497]]
       }
       calf_l_fk_ctrl.Location.Z {
        type=MovieSceneScriptingFloatChannel
        default=-0.0212
        thinned_from=64
        keys [[0,-0.0212],[15,-0.0212],[22,-0.021],[45,-0.021],[51,0],[56,-0.0213],[62,-0.0214],[66,-0.0201],[83,-0.0197],[107,-0.0194],[111,-0.019],[115,-0.0192],[121,-0.0189],[148,-0.0188],[153,-0.0071],[160,-0.0026],[172,-0.0137],[187,-0.0174],[194,-0.0163],[199,-0.013],[204,0],[215,0],[223,-0.0035],[233,-0.0139],[241,-0.0172],[246,-0.0183],[251,-0.0182],[257,-0.0178],[267,-0.0164],[275,-0.006],[279,-0.0072],[283,-0.0077],[345,-0.0079],[351,-0.0081],[356,-0.0088],[361,-0.0104],[366,-0.011],[375,-0.0127],[378,-0.018],[382,-0.02],[387,-0.0209],[395,-0.0211],[439,-0.0212]]
       }
       calf_l_fk_ctrl.Rotation.X {
        type=MovieSceneScriptingFloatChannel
        default=-0.0692
        thinned_from=64
        keys [[0,-0.0692],[45,-0.0692],[51,-0.1],[56,-0.0692],[148,-0.0692],[153,0.012],[160,0.0425],[172,-0.0413],[187,-0.0692],[194,-0.0611],[199,-0.0367],[204,0.0607],[215,0.0607],[223,0.0364],[233,-0.0367],[241,-0.0611],[246,-0.0692],[439,-0.0692]]
       }
       calf_l_fk_ctrl.Rotation.Y {type=MovieSceneScriptingFloatChannel, default=1.0551, thinned_from=64, keys [[0,1.0551],[45,1.0551],[51,1.1],[56,1.0551],[439,1.0551]]}
       calf_l_fk_ctrl.Rotation.Z {
        type=MovieSceneScriptingFloatChannel
        default=-7.4634
        thinned_from=64
        keys [[0,-7.4634],[45,-7.4634],[51,-7.5],[56,-7.4634],[148,-7.4634],[153,-3.0605],[160,-1.4094],[172,-5.9499],[187,-7.4634],[194,-7.0231],[199,-5.7022],[204,-0.4187],[215,-0.4187],[223,-1.7396],[233,-5.7022],[241,-7.0231],[246,-7.4634],[439,-7.4634]]
       }
       calf_l_fk_ctrl.Scale.X {type=MovieSceneScriptingFloatChannel, default=1, thinned_from=64, keys [[0,1],[439,1]]}
       calf_l_fk_ctrl.Scale.Y {type=MovieSceneScriptingFloatChannel, default=1, thinned_from=64, keys [[0,1],[439,1]]}
       calf_l_fk_ctrl.Scale.Z {type=MovieSceneScriptingFloatChannel, default=1, thinned_from=64, keys [[0,1],[439,1]]}
       foot_l_fk_ctrl.Location.X {
        type=MovieSceneScriptingFloatChannel
        default=-1.2685
        thinned_from=64
        keys [[0,-1.2685],[15,-1.2685],[22,-1.2565],[30,-1.2567],[45,-1.2528],[51,-1.3],[56,-1.2716],[62,-1.2812],[66,-1.1983],[70,-1.198],[75,-1.1862],[83,-1.1756],[101,-1.1708],[107,-1.1621],[111,-1.1355],[115,-1.1472],[121,-1.129],[130,-1.1229],[143,-1.1236],[148,-1.1259],[153,-0.4222],[160,-0.1583],[172,-0.8188],[187,-1.039],[194,-0.974],[199,-0.7792],[204,0],[215,0],[223,-0.208],[233,-0.8319],[241,-1.0303],[246,-1.0964],[251,-1.0889],[257,-1.0666],[267,-0.9774],[271,-0.6652],[275,-0.3567],[279,-0.4286],[283,-0.461],[287,-0.4691],[292,-0.4711],[337,-0.4718],[345,-0.475],[351,-0.4847],[356,-0.5233],[361,-0.6224],[366,-0.6554],[371,-0.7077],[375,-0.7582],[378,-1.0771],[382,-1.1967],[387,-1.2505],[395,-1.2617],[409,-1.2593],[439,-1.2685]]
       }
       foot_l_fk_ctrl.Location.Y {type=MovieSceneScriptingFloatChannel, default=0.0627, thinned_from=64, keys [[0,0.0627],[45,0.0627],[51,0.1],[56,0.0627],[439,0.0627]]}
       foot_l_fk_ctrl.Location.Z {type=MovieSceneScriptingFloatChannel, default=-0.8453, thinned_from=64, keys [[0,-0.8453],[45,-0.8453],[51,-0.8],[56,-0.8453],[439,-0.8453]]}
       foot_l_fk_ctrl.Rotation.X {
        type=MovieSceneScriptingFloatChannel
        default=-3.9014
        thinned_from=64
        keys [[0,-3.9014],[45,-3.8945],[62,-3.8921],[66,-3.8184],[70,-3.7657],[111,-3.7533],[115,-3.6968],[138,-3.6981],[143,-3.6894],[148,-3.6546],[153,-2.9278],[160,-2.6552],[172,-3.4023],[187,-3.6513],[194,-3.5788],[199,-3.3614],[204,-2.4917],[215,-2.4917],[223,-2.6013],[233,-2.93],[241,-3.0452],[246,-3.0835],[251,-3.1067],[257,-3.1761],[267,-3.4539],[271,-2.658],[275,-1.8713],[279,-1.7155],[283,-1.6454],[287,-1.6278],[292,-1.6234],[337,-1.622],[345,-1.6764],[351,-1.8398],[356,-2.4933],[361,-3.1564],[366,-3.3774],[371,-3.4589],[375,-3.5376],[378,-3.765],[382,-3.8502],[387,-3.8886],[409,-3.9034],[439,-3.9014]]
       }
       foot_l_fk_ctrl.Rotation.Y {
        type=MovieSceneScriptingFloatChannel
        default=-7.5233
        thinned_from=64
        keys [[0,-7.5233],[45,-7.5263],[51,-7.5],[56,-7.5114],[62,-7.4755],[66,-7.1355],[70,-6.9531],[83,-6.8921],[101,-6.8823],[107,-6.8885],[111,-6.9074],[115,-6.8541],[121,-6.871],[138,-6.881],[143,-6.8944],[148,-6.948],[153,-4.9494],[160,-4.1999],[172,-6.3443],[187,-7.0592],[194,-6.8523],[199,-6.2319],[204,-3.7502],[215,-3.7502],[223,-4.0419],[233,-4.917],[241,-5.2206],[246,-5.3218],[251,-5.2984],[257,-5.2284],[267,-4.9482],[271,-1.9728],[275,0.9681],[279,1.4834],[283,1.7153],[287,1.7733],[292,1.7878],[309,1.7922],[337,1.7926],[345,1.6244],[351,1.1198],[356,-0.8987],[361,-3.842],[366,-4.8231],[371,-5.5977],[375,-6.3457],[378,-7.0817],[382,-7.3577],[387,-7.4819],[409,-7.5398],[439,-7.5233]]
       }
       foot_l_fk_ctrl.Rotation.Z {
        type=MovieSceneScriptingFloatChannel
        default=15.3756
        thinned_from=64
        keys [[0,15.3756],[15,15.3756],[45,15.3409],[51,15.4],[62,15.3911],[66,15.4685],[70,15.4476],[83,15.4937],[107,15.4872],[111,15.4465],[115,15.2426],[138,15.2152],[143,15.1568],[148,14.9231],[153,11.4116],[160,10.0947],[172,13.6027],[187,14.772],[194,14.4303],[199,13.4051],[204,9.3046],[215,9.3046],[223,10.1943],[233,12.8635],[241,13.7682],[246,14.0697],[251,14.2362],[257,14.7358],[267,16.7339],[275,19.3338],[279,19.8084],[283,20.022],[287,20.0754],[337,20.0932],[345,20.0283],[351,19.8336],[356,19.055],[361,17.1418],[366,16.504],[371,15.7969],[375,15.1141],[378,15.2776],[382,15.3388],[395,15.3722],[439,15.3756]]
       }
       foot_l_fk_ctrl.Scale.X {type=MovieSceneScriptingFloatChannel, default=1, thinned_from=64, keys [[0,1],[439,1]]}
       foot_l_fk_ctrl.Scale.Y {type=MovieSceneScriptingFloatChannel, default=1, thinned_from=64, keys [[0,1],[439,1]]}
       foot_l_fk_ctrl.Scale.Z {type=MovieSceneScriptingFloatChannel, default=1, thinned_from=64, keys [[0,1],[439,1]]}
       ball_l_fk_ctrl.Location.X {type=MovieSceneScriptingFloatChannel, default=0, thinned_from=64, keys [[0,0],[439,0]]}
       ball_l_fk_ctrl.Location.Y {type=MovieSceneScriptingFloatChannel, default=0, thinned_from=64, keys [[0,0],[439,0]]}
       ball_l_fk_ctrl.Location.Z {type=MovieSceneScriptingFloatChannel, default=0, thinned_from=64, keys [[0,0],[439,0]]}
       ball_l_fk_ctrl.Rotation.X {type=MovieSceneScriptingFloatChannel, default=0, thinned_from=64, keys [[0,0],[439,0]]}
       ball_l_fk_ctrl.Rotation.Y {type=MovieSceneScriptingFloatChannel, default=0, thinned_from=64, keys [[0,0],[439,0]]}
       ball_l_fk_ctrl.Rotation.Z {type=MovieSceneScriptingFloatChannel, default=0, thinned_from=64, keys [[0,0],[439,0]]}
       ball_l_fk_ctrl.Scale.X {type=MovieSceneScriptingFloatChannel, default=1, thinned_from=64, keys [[0,1],[439,1]]}
       ball_l_fk_ctrl.Scale.Y {type=MovieSceneScriptingFloatChannel, default=1, thinned_from=64, keys [[0,1],[439,1]]}
       ball_l_fk_ctrl.Scale.Z {type=MovieSceneScriptingFloatChannel, default=1, thinned_from=64, keys [[0,1],[439,1]]}
       foot_l_ik_ctrl.Location.X {
        type=MovieSceneScriptingFloatChannel
        default=4.9
        thinned_from=145
        keys [[0,4.9],[267,4.9],[268,4.8219],[270,4.4781],[271,4.4],[351,4.4],[352,4.452],[353,4.576],[354,4.724],[355,4.848],[356,4.9],[439,4.9]]
       }
       foot_l_ik_ctrl.Location.Y {type=MovieSceneScriptingFloatChannel, default=11.784, thinned_from=145, keys [[0,11.784],[439,11.784]]}
       foot_l_ik_ctrl.Location.Z {type=MovieSceneScriptingFloatChannel, default=0, thinned_from=145, keys [[0,0],[439,0]]}
       foot_l_ik_ctrl.Rotation.X {type=MovieSceneScriptingFloatChannel, default=0, thinned_from=145, keys [[0,0],[439,0]]}
       foot_l_ik_ctrl.Rotation.Y {type=MovieSceneScriptingFloatChannel, default=0, thinned_from=145, keys [[0,0],[439,0]]}
       foot_l_ik_ctrl.Rotation.Z {
        type=MovieSceneScriptingFloatChannel
        default=-38.3023
        thinned_from=141
        keys [[0,-38.3023],[265,-38.3023],[266,-38.75],[267,-41.25],[268,-45],[270,-55],[271,-58.0409],[272,-59.8313],[273,-60.5],[348,-60.5],[349,-60],[350,-57.5],[353,-46.25],[354,-43.75],[356,-40],[359,-38.3023],[439,-38.3023]]
       }
       foot_l_ik_ctrl.Scale.X {type=MovieSceneScriptingFloatChannel, default=1, thinned_from=145, keys [[0,1],[439,1]]}
       foot_l_ik_ctrl.Scale.Y {type=MovieSceneScriptingFloatChannel, default=1, thinned_from=145, keys [[0,1],[439,1]]}
       foot_l_ik_ctrl.Scale.Z {type=MovieSceneScriptingFloatChannel, default=1, thinned_from=145, keys [[0,1],[439,1]]}
       ball_l_ik_ctrl.Location.X {type=MovieSceneScriptingFloatChannel, default=0, thinned_from=145, keys [[0,0],[439,0]]}
       ball_l_ik_ctrl.Location.Y {type=MovieSceneScriptingFloatChannel, default=0, thinned_from=145, keys [[0,0],[439,0]]}
       ball_l_ik_ctrl.Location.Z {type=MovieSceneScriptingFloatChannel, default=0, thinned_from=145, keys [[0,0],[439,0]]}
       ball_l_ik_ctrl.Rotation.X {type=MovieSceneScriptingFloatChannel, default=0, thinned_from=145, keys [[0,0],[439,0]]}
       ball_l_ik_ctrl.Rotation.Y {type=MovieSceneScriptingFloatChannel, default=0, thinned_from=145, keys [[0,0],[439,0]]}
       ball_l_ik_ctrl.Rotation.Z {type=MovieSceneScriptingFloatChannel, default=0, thinned_from=145, keys [[0,0],[439,0]]}
       ball_l_ik_ctrl.Scale.X {type=MovieSceneScriptingFloatChannel, default=1, thinned_from=145, keys [[0,1],[439,1]]}
       ball_l_ik_ctrl.Scale.Y {type=MovieSceneScriptingFloatChannel, default=1, thinned_from=145, keys [[0,1],[439,1]]}
       ball_l_ik_ctrl.Scale.Z {type=MovieSceneScriptingFloatChannel, default=1, thinned_from=145, keys [[0,1],[439,1]]}
       foot_bk1_l_ctrl.Location.X {type=MovieSceneScriptingFloatChannel, default=0, thinned_from=145, keys [[0,0],[439,0]]}
       foot_bk1_l_ctrl.Location.Y {type=MovieSceneScriptingFloatChannel, default=0, thinned_from=145, keys [[0,0],[439,0]]}
       foot_bk1_l_ctrl.Location.Z {type=MovieSceneScriptingFloatChannel, default=0, thinned_from=145, keys [[0,0],[439,0]]}
       foot_bk1_l_ctrl.Rotation.X {type=MovieSceneScriptingFloatChannel, default=0, thinned_from=145, keys [[0,0],[439,0]]}
       foot_bk1_l_ctrl.Rotation.Y {type=MovieSceneScriptingFloatChannel, default=0, thinned_from=145, keys [[0,0],[439,0]]}
       foot_bk1_l_ctrl.Rotation.Z {
        type=MovieSceneScriptingFloatChannel
        default=0
        thinned_from=145
        keys [[0,0],[260,0],[261,-0.125],[262,-0.375],[263,-0.75],[265,-1.75],[266,-2.24],[267,-2.5],[271,-2.5],[272,-2.1094],[274,-0.3906],[275,0],[345,0],[346,-0.2001],[347,-0.7003],[348,-1.3506],[349,-2.2101],[350,-2.9469],[351,-3.438],[353,-3.9292],[354,-4],[355,-3.5],[356,-2.5],[357,-1.4064],[358,-0.5],[359,-0.125],[360,0],[439,0]]
       }
       foot_bk1_l_ctrl.Scale.X {type=MovieSceneScriptingFloatChannel, default=1, thinned_from=145, keys [[0,1],[439,1]]}
       foot_bk1_l_ctrl.Scale.Y {type=MovieSceneScriptingFloatChannel, default=1, thinned_from=145, keys [[0,1],[439,1]]}
       foot_bk1_l_ctrl.Scale.Z {type=MovieSceneScriptingFloatChannel, default=1, thinned_from=145, keys [[0,1],[439,1]]}
       foot_roll_l_ctrl.Location.X {type=MovieSceneScriptingFloatChannel, default=0, thinned_from=145, keys [[0,0],[439,0]]}
       foot_roll_l_ctrl.Location.Y {type=MovieSceneScriptingFloatChannel, default=0, thinned_from=145, keys [[0,0],[439,0]]}
       foot_roll_l_ctrl.Location.Z {type=MovieSceneScriptingFloatChannel, default=0, thinned_from=145, keys [[0,0],[439,0]]}
       foot_roll_l_ctrl.Rotation.X {type=MovieSceneScriptingFloatChannel, default=0, thinned_from=145, keys [[0,0],[439,0]]}
       foot_roll_l_ctrl.Rotation.Y {type=MovieSceneScriptingFloatChannel, default=0, thinned_from=145, keys [[0,0],[439,0]]}
       foot_roll_l_ctrl.Rotation.Z {type=MovieSceneScriptingFloatChannel, default=0, thinned_from=145, keys [[0,0],[439,0]]}
       foot_roll_l_ctrl.Scale.X {type=MovieSceneScriptingFloatChannel, default=1, thinned_from=145, keys [[0,1],[439,1]]}
       foot_roll_l_ctrl.Scale.Y {type=MovieSceneScriptingFloatChannel, default=1, thinned_from=145, keys [[0,1],[439,1]]}
       foot_roll_l_ctrl.Scale.Z {type=MovieSceneScriptingFloatChannel, default=1, thinned_from=145, keys [[0,1],[439,1]]}
       roll_blend_l_ctrl {type=MovieSceneScriptingFloatChannel, default=1, thinned_from=145, keys [[0,1],[439,1]]}
       tip_l_ctrl.Location.X {type=MovieSceneScriptingFloatChannel, default=0, thinned_from=145, keys [[0,0],[439,0]]}
       tip_l_ctrl.Location.Y {type=MovieSceneScriptingFloatChannel, default=0, thinned_from=145, keys [[0,0],[439,0]]}
       tip_l_ctrl.Location.Z {type=MovieSceneScriptingFloatChannel, default=0, thinned_from=145, keys [[0,0],[439,0]]}
       tip_l_ctrl.Rotation.X {type=MovieSceneScriptingFloatChannel, default=0, thinned_from=145, keys [[0,0],[439,0]]}
       tip_l_ctrl.Rotation.Y {type=MovieSceneScriptingFloatChannel, default=0, thinned_from=145, keys [[0,0],[439,0]]}
       tip_l_ctrl.Rotation.Z {type=MovieSceneScriptingFloatChannel, default=0, thinned_from=145, keys [[0,0],[439,0]]}
       tip_l_ctrl.Scale.X {type=MovieSceneScriptingFloatChannel, default=1, thinned_from=145, keys [[0,1],[439,1]]}
       tip_l_ctrl.Scale.Y {type=MovieSceneScriptingFloatChannel, default=1, thinned_from=145, keys [[0,1],[439,1]]}
       tip_l_ctrl.Scale.Z {type=MovieSceneScriptingFloatChannel, default=1, thinned_from=145, keys [[0,1],[439,1]]}
       heel_l_ctrl.Location.X {type=MovieSceneScriptingFloatChannel, default=0, thinned_from=145, keys [[0,0],[439,0]]}
       heel_l_ctrl.Location.Y {type=MovieSceneScriptingFloatChannel, default=0, thinned_from=145, keys [[0,0],[439,0]]}
       heel_l_ctrl.Location.Z {type=MovieSceneScriptingFloatChannel, default=0, thinned_from=145, keys [[0,0],[439,0]]}
       heel_l_ctrl.Rotation.X {type=MovieSceneScriptingFloatChannel, default=0, thinned_from=145, keys [[0,0],[439,0]]}
       heel_l_ctrl.Rotation.Y {type=MovieSceneScriptingFloatChannel, default=0, thinned_from=145, keys [[0,0],[439,0]]}
       heel_l_ctrl.Rotation.Z {type=MovieSceneScriptingFloatChannel, default=0, thinned_from=145, keys [[0,0],[439,0]]}
       heel_l_ctrl.Scale.X {type=MovieSceneScriptingFloatChannel, default=1, thinned_from=145, keys [[0,1],[439,1]]}
       heel_l_ctrl.Scale.Y {type=MovieSceneScriptingFloatChannel, default=1, thinned_from=145, keys [[0,1],[439,1]]}
       heel_l_ctrl.Scale.Z {type=MovieSceneScriptingFloatChannel, default=1, thinned_from=145, keys [[0,1],[439,1]]}
       leg_l_softness {
        type=MovieSceneScriptingFloatChannel
        default=0
        thinned_from=155
        keys [[0,0],[256,0],[257,0.0125],[258,0.05],[260,0.15],[261,0.195],[262,0.235],[263,0.27],[264,0.29],[265,0.305],[266,0.315],[267,0.3],[268,0],[272,0],[273,0.35],[274,0.4],[281,0.225],[282,0.175],[283,0.1],[284,0.0375],[285,0],[439,0]]
       }
       leg_l_pv_ik_ctrl.Location.X {type=MovieSceneScriptingFloatChannel, default=0, thinned_from=64, keys [[0,0],[439,0]]}
       leg_l_pv_ik_ctrl.Location.Y {type=MovieSceneScriptingFloatChannel, default=0, thinned_from=64, keys [[0,0],[439,0]]}
       leg_l_pv_ik_ctrl.Location.Z {type=MovieSceneScriptingFloatChannel, default=0, thinned_from=64, keys [[0,0],[439,0]]}
       leg_l_pv_ik_ctrl.Rotation.X {type=MovieSceneScriptingFloatChannel, default=0, thinned_from=64, keys [[0,0],[439,0]]}
       leg_l_pv_ik_ctrl.Rotation.Y {type=MovieSceneScriptingFloatChannel, default=0, thinned_from=64, keys [[0,0],[439,0]]}
       leg_l_pv_ik_ctrl.Rotation.Z {type=MovieSceneScriptingFloatChannel, default=0, thinned_from=64, keys [[0,0],[439,0]]}
       leg_l_pv_ik_ctrl.Scale.X {type=MovieSceneScriptingFloatChannel, default=1, thinned_from=64, keys [[0,1],[439,1]]}
       leg_l_pv_ik_ctrl.Scale.Y {type=MovieSceneScriptingFloatChannel, default=1, thinned_from=64, keys [[0,1],[439,1]]}
       leg_l_pv_ik_ctrl.Scale.Z {type=MovieSceneScriptingFloatChannel, default=1, thinned_from=64, keys [[0,1],[439,1]]}
       thigh_r_fk_ctrl.Location.X {type=MovieSceneScriptingFloatChannel, default=0, thinned_from=64, keys [[0,0],[439,0]]}
       thigh_r_fk_ctrl.Location.Y {type=MovieSceneScriptingFloatChannel, default=0, thinned_from=64, keys [[0,0],[439,0]]}
       thigh_r_fk_ctrl.Location.Z {type=MovieSceneScriptingFloatChannel, default=0, thinned_from=64, keys [[0,0],[439,0]]}
       thigh_r_fk_ctrl.Rotation.X {
        type=MovieSceneScriptingFloatChannel
        default=-4.4148
        thinned_from=64
        keys [[0,-4.4148],[15,-4.4148],[22,-4.7258],[30,-4.7195],[45,-4.821],[51,-4.4],[56,-4.2115],[62,-3.6013],[66,-0.4375],[70,1.0608],[75,1.4002],[83,1.843],[94,1.9872],[101,1.952],[107,1.7055],[111,0.9511],[115,0.015],[121,-0.2802],[130,-0.3786],[138,-0.5694],[143,-1.1417],[148,-3.4309],[153,18.7617],[160,27.0839],[172,1.9773],[187,-6.3915],[194,-3.9872],[199,3.2257],[204,32.0772],[215,32.0772],[223,26.8311],[233,11.093],[241,5.8248],[246,4.0687],[251,3.563],[257,2.0458],[262,-0.9709],[267,-4.0231],[271,0.1576],[275,4.2897],[279,7.7527],[283,9.3111],[287,9.7007],[292,9.7981],[299,9.8224],[337,9.8306],[345,9.4173],[351,8.1774],[356,3.2178],[361,-6.1938],[366,-9.3311],[371,-4.5339],[375,0.0978],[378,-2.7226],[382,-3.7802],[387,-4.2561],[395,-4.3553],[409,-4.5831],[439,-4.4148]]
       }
       thigh_r_fk_ctrl.Rotation.Y {
        type=MovieSceneScriptingFloatChannel
        default=-12.1768
        thinned_from=64
        keys [[0,-12.1768],[15,-12.1768],[22,-12.2215],[45,-12.2352],[51,-12.2],[56,-12.0791],[62,-11.7864],[66,-12.6734],[70,-12.5782],[75,-12.7988],[83,-12.9829],[94,-13.0429],[101,-13.0527],[107,-13.1214],[111,-13.3317],[115,-12.2125],[121,-12.3706],[130,-12.4233],[138,-12.3605],[143,-12.172],[148,-11.4181],[153,-15.7767],[160,-17.4112],[172,-13.3813],[187,-12.038],[194,-12.4351],[199,-13.6265],[204,-18.3919],[215,-18.3919],[223,-16.2198],[233,-9.7035],[241,-7.6961],[246,-7.0269],[251,-7.2982],[257,-8.1118],[267,-11.3665],[275,-15.4673],[279,-14.6092],[283,-14.2231],[287,-14.1265],[337,-14.0944],[345,-14.0548],[351,-13.9361],[356,-13.4613],[361,-12.3375],[366,-11.9629],[371,-11.7257],[375,-11.4966],[378,-11.9217],[382,-12.0811],[387,-12.1529],[409,-12.224],[439,-12.1768]]
       }
       thigh_r_fk_ctrl.Rotation.Z {
        type=MovieSceneScriptingFloatChannel
        default=9.3708
        thinned_from=64
        keys [[0,9.3708],[15,9.3708],[22,9.3422],[45,9.3333],[51,9.4],[56,9.3826],[62,9.4172],[66,9.5426],[70,9.6157],[83,9.6685],[101,9.6731],[107,9.6467],[111,9.5658],[115,9.4377],[138,9.394],[143,9.3422],[148,9.1353],[153,9.2064],[160,9.233],[172,8.9485],[187,8.8536],[194,8.8783],[199,8.9525],[204,9.249],[223,9.2418],[246,9.2148],[251,9.5226],[257,10.446],[267,14.1397],[275,17.9072],[279,19.0819],[283,19.6105],[287,19.7427],[337,19.7867],[345,19.6916],[351,19.4062],[356,18.2647],[361,14.5538],[366,13.3168],[371,11.9939],[375,10.7167],[378,9.8755],[382,9.5601],[387,9.4181],[395,9.3886],[439,9.3708]]
       }
       thigh_r_fk_ctrl.Scale.X {type=MovieSceneScriptingFloatChannel, default=1, thinned_from=64, keys [[0,1],[439,1]]}
       thigh_r_fk_ctrl.Scale.Y {type=MovieSceneScriptingFloatChannel, default=1, thinned_from=64, keys [[0,1],[439,1]]}
       thigh_r_fk_ctrl.Scale.Z {type=MovieSceneScriptingFloatChannel, default=1, thinned_from=64, keys [[0,1],[439,1]]}
       calf_r_fk_ctrl.Location.X {
        type=MovieSceneScriptingFloatChannel
        default=-0.9551
        thinned_from=64
        keys [[0,-0.9551],[15,-0.9551],[22,-0.9577],[45,-0.9585],[51,-1],[56,-0.944],[62,-0.9106],[66,-0.9873],[70,-1.0006],[75,-1.0214],[83,-1.0362],[94,-1.0411],[101,-1.0417],[107,-1.0462],[111,-1.0601],[115,-0.9636],[121,-0.9705],[130,-0.9728],[138,-0.9681],[143,-0.9541],[148,-0.8982],[153,-0.9698],[160,-0.9966],[172,-0.9533],[187,-0.9388],[194,-0.9434],[199,-0.9573],[204,-1.0127],[215,-1.0127],[223,-0.9386],[233,-0.7162],[241,-0.6526],[246,-0.6314],[251,-0.6262],[257,-0.6105],[267,-0.5478],[275,-0.5224],[279,-0.3481],[283,-0.2697],[287,-0.2501],[292,-0.2452],[299,-0.244],[337,-0.2436],[345,-0.2498],[351,-0.2685],[356,-0.3431],[361,-0.4073],[366,-0.4286],[371,-0.5248],[375,-0.6177],[378,-0.8286],[382,-0.9077],[387,-0.9433],[395,-0.9507],[439,-0.9551]]
       }
       calf_r_fk_ctrl.Location.Y {
        type=MovieSceneScriptingFloatChannel
        default=0.1276
        thinned_from=64
        keys [[0,0.1276],[15,0.1276],[22,0.1279],[45,0.128],[51,0.1],[56,0.1261],[62,0.1216],[66,0.1318],[70,0.1336],[75,0.1364],[83,0.1384],[94,0.139],[101,0.1391],[107,0.1397],[111,0.1416],[115,0.1287],[121,0.1296],[130,0.1299],[138,0.1293],[143,0.1274],[148,0.12],[153,0.1295],[160,0.1331],[172,0.1273],[187,0.1254],[194,0.126],[199,0.1278],[204,0.1352],[215,0.1352],[223,0.1253],[233,0.0957],[241,0.0872],[246,0.0843],[251,0.0836],[257,0.0815],[267,0.0732],[275,0.0698],[279,0.0465],[283,0.036],[287,0.0334],[292,0.0327],[337,0.0325],[345,0.0334],[351,0.0359],[356,0.0458],[361,0.0544],[366,0.0572],[371,0.0701],[375,0.0825],[378,0.1107],[382,0.1212],[387,0.126],[395,0.127],[439,0.1276]]
       }
       calf_r_fk_ctrl.Location.Z {
        type=MovieSceneScriptingFloatChannel
        default=-0.0174
        thinned_from=64
        keys [[0,-0.0174],[45,-0.0175],[51,0],[56,-0.0172],[62,-0.0166],[66,-0.018],[75,-0.0186],[83,-0.0189],[107,-0.0191],[111,-0.0194],[115,-0.0176],[130,-0.0178],[138,-0.0177],[143,-0.0174],[148,-0.0164],[153,-0.0177],[160,-0.0182],[172,-0.0174],[187,-0.0171],[194,-0.0172],[199,-0.0175],[204,-0.0185],[215,-0.0185],[223,-0.0171],[233,-0.0131],[241,-0.0119],[246,-0.0115],[257,-0.0111],[267,-0.01],[275,-0.0095],[279,-0.0064],[283,-0.0049],[287,-0.0046],[337,-0.0044],[345,-0.0046],[351,-0.0049],[361,-0.0074],[366,-0.0078],[371,-0.0096],[375,-0.0113],[378,-0.0151],[382,-0.0166],[387,-0.0172],[439,-0.0174]]
       }
       calf_r_fk_ctrl.Rotation.X {type=MovieSceneScriptingFloatChannel, default=-0.0638, thinned_from=64, keys [[0,-0.0638],[45,-0.0638],[51,-0.1],[56,-0.0638],[439,-0.0638]]}
       calf_r_fk_ctrl.Rotation.Y {type=MovieSceneScriptingFloatChannel, default=1.0367, thinned_from=64, keys [[0,1.0367],[45,1.0367],[51,1],[56,1.0367],[439,1.0367]]}
       calf_r_fk_ctrl.Rotation.Z {type=MovieSceneScriptingFloatChannel, default=-7.6066, thinned_from=64, keys [[0,-7.6066],[439,-7.6066]]}
       calf_r_fk_ctrl.Scale.X {type=MovieSceneScriptingFloatChannel, default=1, thinned_from=64, keys [[0,1],[439,1]]}
       calf_r_fk_ctrl.Scale.Y {type=MovieSceneScriptingFloatChannel, default=1, thinned_from=64, keys [[0,1],[439,1]]}
       calf_r_fk_ctrl.Scale.Z {type=MovieSceneScriptingFloatChannel, default=1, thinned_from=64, keys [[0,1],[439,1]]}
       foot_r_fk_ctrl.Location.X {
        type=MovieSceneScriptingFloatChannel
        default=-1.0605
        thinned_from=64
        keys [[0,-1.0605],[15,-1.0605],[22,-1.0633],[45,-1.0642],[51,-1.1],[56,-1.0481],[62,-1.0111],[66,-1.0962],[70,-1.1109],[75,-1.134],[83,-1.1505],[94,-1.1559],[101,-1.1566],[107,-1.1616],[111,-1.177],[115,-1.0698],[121,-1.0775],[130,-1.0801],[138,-1.0749],[143,-1.0594],[148,-0.9973],[153,-1.0767],[160,-1.1065],[172,-1.0584],[187,-1.0424],[194,-1.0475],[199,-1.0629],[204,-1.1244],[215,-1.1244],[223,-1.0421],[233,-0.7952],[241,-0.7246],[246,-0.701],[251,-0.6952],[257,-0.6778],[267,-0.6082],[275,-0.5801],[279,-0.3865],[283,-0.2994],[287,-0.2777],[292,-0.2722],[299,-0.2709],[337,-0.2704],[345,-0.2773],[351,-0.2981],[356,-0.381],[361,-0.4522],[366,-0.4759],[371,-0.5827],[375,-0.6859],[378,-0.92],[382,-1.0078],[387,-1.0473],[395,-1.0555],[439,-1.0605]]
       }
       foot_r_fk_ctrl.Location.Y {type=MovieSceneScriptingFloatChannel, default=-0.0543, thinned_from=64, keys [[0,-0.0543],[45,-0.0543],[51,-0.1],[56,-0.0543],[439,-0.0543]]}
       foot_r_fk_ctrl.Location.Z {type=MovieSceneScriptingFloatChannel, default=-0.8609, thinned_from=64, keys [[0,-0.8609],[45,-0.8609],[51,-0.9],[56,-0.8609],[439,-0.8609]]}
       foot_r_fk_ctrl.Rotation.X {
        type=MovieSceneScriptingFloatChannel
        default=-0.6843
        thinned_from=64
        keys [[0,-0.6843],[45,-0.6845],[51,-0.7],[56,-0.6849],[62,-0.6864],[66,-0.6971],[70,-0.7042],[83,-0.7063],[107,-0.7059],[111,-0.7042],[115,-0.7075],[143,-0.7075],[148,-0.7099],[153,-0.8051],[160,-0.8408],[172,-0.7446],[187,-0.7125],[194,-0.7219],[199,-0.75],[204,-0.8622],[215,-0.8622],[223,-0.8442],[233,-0.7902],[241,-0.7717],[246,-0.7655],[251,-0.7533],[257,-0.7168],[267,-0.5707],[275,-0.5301],[279,-0.525],[283,-0.5227],[337,-0.522],[345,-0.5228],[351,-0.5253],[356,-0.5354],[361,-0.5386],[366,-0.5397],[371,-0.6314],[375,-0.7198],[378,-0.6977],[382,-0.6893],[387,-0.6856],[409,-0.6835],[439,-0.6843]]
       }
       foot_r_fk_ctrl.Rotation.Y {
        type=MovieSceneScriptingFloatChannel
        default=5.3149
        thinned_from=64
        keys [[0,5.3149],[15,5.3149],[22,5.2858],[45,5.2768],[56,5.3172],[62,5.3232],[70,5.0203],[75,5.0052],[107,4.9971],[111,5.0017],[115,4.8332],[130,4.8302],[138,4.8168],[143,4.7765],[148,4.6154],[153,3.1089],[160,2.5439],[172,3.9905],[187,4.4727],[194,4.331],[199,3.9058],[204,2.2049],[215,2.2049],[223,2.3564],[233,2.8107],[241,2.9785],[246,3.0345],[251,3.3242],[257,4.1934],[262,5.9216],[267,7.6702],[275,8.4994],[279,8.5499],[287,8.5783],[337,8.5802],[345,8.5521],[351,8.4681],[356,8.132],[361,7.4726],[366,7.2527],[371,5.7927],[375,4.3829],[378,4.9654],[382,5.1839],[387,5.2822],[395,5.3027],[409,5.3228],[439,5.3149]]
       }
       foot_r_fk_ctrl.Rotation.Z {
        type=MovieSceneScriptingFloatChannel
        default=2.1622
        thinned_from=64
        keys [[0,2.1622],[15,2.1622],[22,2.207],[30,2.2061],[45,2.2207],[51,2.2],[62,2.057],[66,1.786],[70,1.6667],[83,1.5962],[94,1.5823],[101,1.5867],[107,1.6172],[111,1.7105],[115,1.8431],[121,1.8811],[130,1.8937],[138,1.9174],[143,1.9883],[148,2.2721],[153,3.3794],[160,3.7947],[172,2.9496],[187,2.6679],[194,2.7539],[199,3.0119],[204,4.0438],[215,4.0438],[223,3.6331],[233,2.401],[241,1.9908],[246,1.854],[251,1.895],[257,2.0178],[267,2.5093],[275,2.5457],[279,2.58],[283,2.5954],[337,2.6005],[345,2.6072],[351,2.6273],[356,2.7077],[361,3.222],[366,3.3934],[371,2.7746],[375,2.1772],[378,2.1678],[387,2.1627],[395,2.1624],[409,2.1906],[439,2.1622]]
       }
       foot_r_fk_ctrl.Scale.X {type=MovieSceneScriptingFloatChannel, default=1, thinned_from=64, keys [[0,1],[439,1]]}
       foot_r_fk_ctrl.Scale.Y {type=MovieSceneScriptingFloatChannel, default=1, thinned_from=64, keys [[0,1],[439,1]]}
       foot_r_fk_ctrl.Scale.Z {type=MovieSceneScriptingFloatChannel, default=1, thinned_from=64, keys [[0,1],[439,1]]}
       ball_r_fk_ctrl.Location.X {type=MovieSceneScriptingFloatChannel, default=0, thinned_from=64, keys [[0,0],[439,0]]}
       ball_r_fk_ctrl.Location.Y {type=MovieSceneScriptingFloatChannel, default=0, thinned_from=64, keys [[0,0],[439,0]]}
       ball_r_fk_ctrl.Location.Z {type=MovieSceneScriptingFloatChannel, default=0, thinned_from=64, keys [[0,0],[439,0]]}
       ball_r_fk_ctrl.Rotation.X {type=MovieSceneScriptingFloatChannel, default=0, thinned_from=64, keys [[0,0],[439,0]]}
       ball_r_fk_ctrl.Rotation.Y {type=MovieSceneScriptingFloatChannel, default=0, thinned_from=64, keys [[0,0],[439,0]]}
       ball_r_fk_ctrl.Rotation.Z {type=MovieSceneScriptingFloatChannel, default=0, thinned_from=64, keys [[0,0],[439,0]]}
       ball_r_fk_ctrl.Scale.X {type=MovieSceneScriptingFloatChannel, default=1, thinned_from=64, keys [[0,1],[439,1]]}
       ball_r_fk_ctrl.Scale.Y {type=MovieSceneScriptingFloatChannel, default=1, thinned_from=64, keys [[0,1],[439,1]]}
       ball_r_fk_ctrl.Scale.Z {type=MovieSceneScriptingFloatChannel, default=1, thinned_from=64, keys [[0,1],[439,1]]}
       foot_r_ik_ctrl.Location.X {type=MovieSceneScriptingFloatChannel, default=-4.2937, const [-4.2937,64]}
       foot_r_ik_ctrl.Location.Y {
        type=MovieSceneScriptingFloatChannel
        default=-1.8187
        thinned_from=64
        keys [[0,-1.8187],[66,-1.8187],[70,-1.8047],[75,-1.8012],[395,-1.8006],[409,-1.8084],[439,-1.8187]]
       }
       foot_r_ik_ctrl.Location.Z {type=MovieSceneScriptingFloatChannel, default=0, thinned_from=64, keys [[0,0],[439,0]]}
       foot_r_ik_ctrl.Rotation.X {
        type=MovieSceneScriptingFloatChannel
        default=2.4895
        thinned_from=64
        keys [[0,2.4895],[45,2.4895],[51,2.5],[56,2.4895],[66,2.4895],[70,2.4974],[395,2.4996],[439,2.4895]]
       }
       foot_r_ik_ctrl.Rotation.Y {
        type=MovieSceneScriptingFloatChannel
        default=-2.0871
        thinned_from=64
        keys [[0,-2.0871],[45,-2.0871],[51,-2.1],[56,-2.0871],[66,-2.0871],[70,-2.0968],[395,-2.0996],[439,-2.0871]]
       }
       foot_r_ik_ctrl.Rotation.Z {
        type=MovieSceneScriptingFloatChannel
        default=-14.9285
        thinned_from=64
        keys [[0,-14.9285],[251,-14.9285],[257,-5.25],[262,-2.75],[361,-2.75],[366,-12.4464],[371,-14.9285],[439,-14.9285]]
       }
       foot_r_ik_ctrl.Scale.X {type=MovieSceneScriptingFloatChannel, default=1, thinned_from=64, keys [[0,1],[439,1]]}
       foot_r_ik_ctrl.Scale.Y {type=MovieSceneScriptingFloatChannel, default=1, thinned_from=64, keys [[0,1],[439,1]]}
       foot_r_ik_ctrl.Scale.Z {type=MovieSceneScriptingFloatChannel, default=1, thinned_from=64, keys [[0,1],[439,1]]}
       ball_r_ik_ctrl.Location.X {type=MovieSceneScriptingFloatChannel, default=0, thinned_from=64, keys [[0,0],[439,0]]}
       ball_r_ik_ctrl.Location.Y {type=MovieSceneScriptingFloatChannel, default=0, thinned_from=64, keys [[0,0],[439,0]]}
       ball_r_ik_ctrl.Location.Z {type=MovieSceneScriptingFloatChannel, default=0, thinned_from=64, keys [[0,0],[439,0]]}
       ball_r_ik_ctrl.Rotation.X {type=MovieSceneScriptingFloatChannel, default=0, thinned_from=64, keys [[0,0],[439,0]]}
       ball_r_ik_ctrl.Rotation.Y {type=MovieSceneScriptingFloatChannel, default=0, thinned_from=64, keys [[0,0],[439,0]]}
       ball_r_ik_ctrl.Rotation.Z {type=MovieSceneScriptingFloatChannel, default=0, thinned_from=64, keys [[0,0],[439,0]]}
       ball_r_ik_ctrl.Scale.X {type=MovieSceneScriptingFloatChannel, default=1, thinned_from=64, keys [[0,1],[439,1]]}
       ball_r_ik_ctrl.Scale.Y {type=MovieSceneScriptingFloatChannel, default=1, thinned_from=64, keys [[0,1],[439,1]]}
       ball_r_ik_ctrl.Scale.Z {type=MovieSceneScriptingFloatChannel, default=1, thinned_from=64, keys [[0,1],[439,1]]}
       foot_bk1_r_ctrl.Location.X {type=MovieSceneScriptingFloatChannel, default=0, thinned_from=64, keys [[0,0],[439,0]]}
       foot_bk1_r_ctrl.Location.Y {type=MovieSceneScriptingFloatChannel, default=0, thinned_from=64, keys [[0,0],[439,0]]}
       foot_bk1_r_ctrl.Location.Z {type=MovieSceneScriptingFloatChannel, default=0, thinned_from=64, keys [[0,0],[439,0]]}
       foot_bk1_r_ctrl.Rotation.X {type=MovieSceneScriptingFloatChannel, default=0, thinned_from=64, keys [[0,0],[439,0]]}
       foot_bk1_r_ctrl.Rotation.Y {type=MovieSceneScriptingFloatChannel, default=0, thinned_from=64, keys [[0,0],[439,0]]}
       foot_bk1_r_ctrl.Rotation.Z {type=MovieSceneScriptingFloatChannel, default=0, thinned_from=64, keys [[0,0],[351,0],[356,-0.0104],[409,-0.0104],[439,0]]}
       foot_bk1_r_ctrl.Scale.X {type=MovieSceneScriptingFloatChannel, default=1, thinned_from=64, keys [[0,1],[439,1]]}
       foot_bk1_r_ctrl.Scale.Y {type=MovieSceneScriptingFloatChannel, default=1, thinned_from=64, keys [[0,1],[439,1]]}
       foot_bk1_r_ctrl.Scale.Z {type=MovieSceneScriptingFloatChannel, default=1, thinned_from=64, keys [[0,1],[439,1]]}
       foot_roll_r_ctrl.Location.X {type=MovieSceneScriptingFloatChannel, default=0, thinned_from=64, keys [[0,0],[439,0]]}
       foot_roll_r_ctrl.Location.Y {type=MovieSceneScriptingFloatChannel, default=0, thinned_from=64, keys [[0,0],[439,0]]}
       foot_roll_r_ctrl.Location.Z {type=MovieSceneScriptingFloatChannel, default=0, thinned_from=64, keys [[0,0],[439,0]]}
       foot_roll_r_ctrl.Rotation.X {type=MovieSceneScriptingFloatChannel, default=0, thinned_from=64, keys [[0,0],[439,0]]}
       foot_roll_r_ctrl.Rotation.Y {type=MovieSceneScriptingFloatChannel, default=0, thinned_from=64, keys [[0,0],[439,0]]}
       foot_roll_r_ctrl.Rotation.Z {type=MovieSceneScriptingFloatChannel, default=0, thinned_from=64, keys [[0,0],[439,0]]}
       foot_roll_r_ctrl.Scale.X {type=MovieSceneScriptingFloatChannel, default=1, thinned_from=64, keys [[0,1],[439,1]]}
       foot_roll_r_ctrl.Scale.Y {type=MovieSceneScriptingFloatChannel, default=1, thinned_from=64, keys [[0,1],[439,1]]}
       foot_roll_r_ctrl.Scale.Z {type=MovieSceneScriptingFloatChannel, default=1, thinned_from=64, keys [[0,1],[439,1]]}
       roll_blend_r_ctrl {type=MovieSceneScriptingFloatChannel, default=1, thinned_from=64, keys [[0,1],[439,1]]}
       tip_r_ctrl.Location.X {type=MovieSceneScriptingFloatChannel, default=0, thinned_from=64, keys [[0,0],[439,0]]}
       tip_r_ctrl.Location.Y {type=MovieSceneScriptingFloatChannel, default=0, thinned_from=64, keys [[0,0],[439,0]]}
       tip_r_ctrl.Location.Z {type=MovieSceneScriptingFloatChannel, default=0, thinned_from=64, keys [[0,0],[439,0]]}
       tip_r_ctrl.Rotation.X {type=MovieSceneScriptingFloatChannel, default=0, thinned_from=64, keys [[0,0],[439,0]]}
       tip_r_ctrl.Rotation.Y {type=MovieSceneScriptingFloatChannel, default=0, thinned_from=64, keys [[0,0],[439,0]]}
       tip_r_ctrl.Rotation.Z {type=MovieSceneScriptingFloatChannel, default=0, thinned_from=64, keys [[0,0],[439,0]]}
       tip_r_ctrl.Scale.X {type=MovieSceneScriptingFloatChannel, default=1, thinned_from=64, keys [[0,1],[439,1]]}
       tip_r_ctrl.Scale.Y {type=MovieSceneScriptingFloatChannel, default=1, thinned_from=64, keys [[0,1],[439,1]]}
       tip_r_ctrl.Scale.Z {type=MovieSceneScriptingFloatChannel, default=1, thinned_from=64, keys [[0,1],[439,1]]}
       heel_r_ctrl.Location.X {type=MovieSceneScriptingFloatChannel, default=0, thinned_from=64, keys [[0,0],[439,0]]}
       heel_r_ctrl.Location.Y {type=MovieSceneScriptingFloatChannel, default=0, thinned_from=64, keys [[0,0],[439,0]]}
       heel_r_ctrl.Location.Z {type=MovieSceneScriptingFloatChannel, default=0, thinned_from=64, keys [[0,0],[439,0]]}
       heel_r_ctrl.Rotation.X {type=MovieSceneScriptingFloatChannel, default=0, thinned_from=64, keys [[0,0],[439,0]]}
       heel_r_ctrl.Rotation.Y {type=MovieSceneScriptingFloatChannel, default=0, thinned_from=64, keys [[0,0],[439,0]]}
       heel_r_ctrl.Rotation.Z {
        type=MovieSceneScriptingFloatChannel
        default=0
        thinned_from=64
        keys [[0,0],[246,0],[251,0.55],[257,1.275],[262,0.425],[267,0],[356,0],[366,0.3],[371,0.35],[378,0],[439,0]]
       }
       heel_r_ctrl.Scale.X {type=MovieSceneScriptingFloatChannel, default=1, thinned_from=64, keys [[0,1],[439,1]]}
       heel_r_ctrl.Scale.Y {type=MovieSceneScriptingFloatChannel, default=1, thinned_from=64, keys [[0,1],[439,1]]}
       heel_r_ctrl.Scale.Z {type=MovieSceneScriptingFloatChannel, default=1, thinned_from=64, keys [[0,1],[439,1]]}
       leg_r_softness {type=MovieSceneScriptingFloatChannel, default=0, thinned_from=64, keys [[0,0],[439,0]]}
       leg_r_pv_ik_ctrl.Location.X {type=MovieSceneScriptingFloatChannel, default=0, thinned_from=64, keys [[0,0],[439,0]]}
       leg_r_pv_ik_ctrl.Location.Y {type=MovieSceneScriptingFloatChannel, default=0, thinned_from=64, keys [[0,0],[439,0]]}
       leg_r_pv_ik_ctrl.Location.Z {type=MovieSceneScriptingFloatChannel, default=0, thinned_from=64, keys [[0,0],[439,0]]}
       leg_r_pv_ik_ctrl.Rotation.X {type=MovieSceneScriptingFloatChannel, default=0, thinned_from=64, keys [[0,0],[439,0]]}
       leg_r_pv_ik_ctrl.Rotation.Y {type=MovieSceneScriptingFloatChannel, default=0, thinned_from=64, keys [[0,0],[439,0]]}
       leg_r_pv_ik_ctrl.Rotation.Z {type=MovieSceneScriptingFloatChannel, default=0, thinned_from=64, keys [[0,0],[439,0]]}
       leg_r_pv_ik_ctrl.Scale.X {type=MovieSceneScriptingFloatChannel, default=1, thinned_from=64, keys [[0,1],[439,1]]}
       leg_r_pv_ik_ctrl.Scale.Y {type=MovieSceneScriptingFloatChannel, default=1, thinned_from=64, keys [[0,1],[439,1]]}
       leg_r_pv_ik_ctrl.Scale.Z {type=MovieSceneScriptingFloatChannel, default=1, thinned_from=64, keys [[0,1],[439,1]]}
       upperarm_l_fk_ctrl.Location.X {type=MovieSceneScriptingFloatChannel, default=0, thinned_from=64, keys [[0,0],[439,0]]}
       upperarm_l_fk_ctrl.Location.Y {type=MovieSceneScriptingFloatChannel, default=0, thinned_from=64, keys [[0,0],[439,0]]}
       upperarm_l_fk_ctrl.Location.Z {type=MovieSceneScriptingFloatChannel, default=0, thinned_from=64, keys [[0,0],[439,0]]}
       upperarm_l_fk_ctrl.Rotation.X {
        type=MovieSceneScriptingFloatChannel
        default=17.0338
        thinned_from=64
        keys [[0,17.0338],[15,17.0338],[22,16.7873],[30,16.7922],[45,16.7117],[51,17],[56,17.1383],[62,17.4525],[66,18.9775],[70,19.5124],[83,19.9734],[94,20.0677],[101,20.0534],[107,19.954],[111,19.6497],[115,19.23],[121,19.172],[130,19.1526],[138,18.9883],[143,18.4955],[148,16.5239],[153,27.041],[160,30.9849],[172,18.8302],[187,14.7787],[194,15.9395],[199,19.4218],[204,33.3512],[215,33.3512],[223,30.6123],[233,22.3956],[241,19.6786],[246,18.7729],[251,19.4435],[257,21.4555],[267,29.5033],[271,42.8317],[275,56.0052],[279,77.7538],[283,87.5406],[287,89.9874],[292,90.599],[309,90.7873],[337,90.8029],[345,88.2598],[351,80.6306],[356,50.1136],[361,30.3942],[366,23.821],[371,20.9969],[375,18.2701],[378,17.4749],[382,17.1768],[387,17.0426],[395,17.0146],[409,33.6898],[439,16.9979]]
       }
       upperarm_l_fk_ctrl.Rotation.Y {
        type=MovieSceneScriptingFloatChannel
        default=-13.8012
        thinned_from=64
        keys [[0,-13.8012],[15,-13.8012],[22,-14.0091],[30,-14.0049],[45,-14.0728],[51,-13.8],[56,-13.6836],[62,-13.3299],[66,-11.5195],[70,-10.7943],[75,-10.5862],[83,-10.3254],[94,-10.2405],[101,-10.2663],[107,-10.4465],[111,-10.9979],[115,-12.1573],[121,-12.4062],[130,-12.4892],[138,-12.6303],[143,-13.0535],[148,-14.7462],[153,-4.4612],[160,-0.6044],[172,-12.4302],[187,-16.3721],[194,-15.242],[199,-11.8517],[204,1.7097],[215,1.7097],[223,-1.0331],[233,-9.2614],[241,-11.9853],[246,-12.8932],[251,-11.9387],[257,-9.075],[262,-3.3811],[267,2.3797],[271,13.8273],[275,25.1418],[279,12.4713],[283,6.7696],[287,5.3442],[292,4.9878],[299,4.8987],[309,4.8781],[337,4.869],[345,4.9615],[351,5.239],[356,6.3489],[361,-0.484],[366,-2.7616],[371,-7.4852],[375,-12.0458],[378,-13.1361],[382,-13.545],[387,-13.7289],[395,-13.7673],[409,-0.5334],[439,-13.7903]]
       }
       upperarm_l_fk_ctrl.Rotation.Z {
        type=MovieSceneScriptingFloatChannel
        default=-38.6045
        thinned_from=64
        keys [[0,-38.6045],[15,-38.6045],[22,-38.7095],[45,-38.7411],[56,-38.4847],[62,-38.1256],[66,-36.0874],[70,-35.1474],[75,-34.9176],[83,-34.6912],[94,-34.6175],[101,-34.658],[107,-34.942],[111,-35.8109],[115,-38.0108],[121,-38.5462],[130,-38.7247],[143,-38.8058],[148,-39.0493],[153,-23.3188],[160,-17.4198],[172,-35.2374],[187,-41.1766],[194,-39.4706],[199,-34.3526],[204,-13.8804],[215,-13.8804],[223,-18.465],[233,-32.2187],[241,-36.7969],[246,-38.3229],[251,-37.794],[257,-36.207],[267,-29.8592],[271,-0.9118],[275,27.699],[279,46.0379],[283,54.2904],[287,56.3535],[292,56.8693],[299,56.9982],[337,57.0412],[345,56.3095],[351,54.1142],[356,45.333],[361,-13.6693],[366,-33.3367],[371,-35.5691],[375,-37.7245],[378,-38.2692],[382,-38.4735],[387,-38.5654],[395,-38.5846],[409,-32.3324],[439,-38.5961]]
       }
       upperarm_l_fk_ctrl.Scale.X {type=MovieSceneScriptingFloatChannel, default=1, thinned_from=64, keys [[0,1],[439,1]]}
       upperarm_l_fk_ctrl.Scale.Y {type=MovieSceneScriptingFloatChannel, default=1, thinned_from=64, keys [[0,1],[439,1]]}
       upperarm_l_fk_ctrl.Scale.Z {type=MovieSceneScriptingFloatChannel, default=1, thinned_from=64, keys [[0,1],[439,1]]}
       lowerarm_l_fk_ctrl.Location.X {
        type=MovieSceneScriptingFloatChannel
        default=0
        thinned_from=64
        keys [[0,0],[15,0],[22,-0.0003],[45,-0.0003],[51,0],[66,0],[70,-0.0025],[75,-0.0029],[94,-0.0031],[101,-0.003],[107,-0.0022],[111,0],[275,0],[279,0.2154],[283,0.3124],[287,0.3366],[292,0.3427],[299,0.3442],[337,0.3447],[345,0.3231],[351,0.2585],[356,0],[439,0]]
       }
       lowerarm_l_fk_ctrl.Location.Y {
        type=MovieSceneScriptingFloatChannel
        default=0
        thinned_from=64
        keys [[0,0],[15,0],[22,0.0002],[45,0.0002],[66,0],[70,0.0013],[75,0.0015],[101,0.0015],[107,0.0012],[111,0],[275,0],[279,-0.1127],[283,-0.1634],[287,-0.1761],[292,-0.1792],[299,-0.18],[337,-0.1803],[345,-0.169],[351,-0.1352],[356,0],[439,0]]
       }
       lowerarm_l_fk_ctrl.Location.Z {
        type=MovieSceneScriptingFloatChannel
        default=0
        thinned_from=64
        keys [[0,0],[275,0],[279,0.0019],[283,0.0027],[292,0.003],[337,0.003],[345,0.0028],[351,0.0023],[356,0],[439,0]]
       }
       lowerarm_l_fk_ctrl.Rotation.X {
        type=MovieSceneScriptingFloatChannel
        default=0.4425
        thinned_from=64
        keys [[0,0.4425],[45,0.4425],[51,0.4],[56,0.4423],[62,0.442],[66,0.4396],[70,0.4385],[101,0.4381],[107,0.4386],[111,0.4403],[115,0.4444],[121,0.4454],[143,0.4455],[148,0.445],[172,0.447],[199,0.447],[204,0.4457],[246,0.4466],[251,0.4474],[257,0.4497],[267,0.4592],[275,0.4327],[279,0.0892],[283,-0.0654],[287,-0.1041],[292,-0.1137],[299,-0.1161],[309,-0.1167],[337,-0.1169],[345,-0.0989],[351,-0.0446],[356,0.1723],[361,0.3864],[366,0.4577],[375,0.4538],[378,0.4467],[382,0.444],[387,0.4428],[439,0.4424]]
       }
       lowerarm_l_fk_ctrl.Rotation.Y {
        type=MovieSceneScriptingFloatChannel
        default=0.1249
        thinned_from=64
        keys [[0,0.1249],[45,0.1248],[51,0.1],[56,0.1253],[62,0.1266],[66,0.1344],[70,0.1381],[75,0.139],[83,0.1396],[94,0.1398],[101,0.1395],[107,0.1378],[111,0.1325],[115,0.1177],[121,0.114],[130,0.1128],[138,0.113],[143,0.1135],[148,0.1157],[153,0.114],[160,0.1134],[172,0.1075],[187,0.1056],[194,0.106],[199,0.1074],[204,0.113],[215,0.113],[223,0.1123],[233,0.1103],[241,0.1096],[246,0.1094],[251,0.1039],[257,0.0877],[262,0.0553],[267,0.0226],[271,0.0894],[275,0.1555],[279,0.3362],[283,0.4175],[287,0.4379],[292,0.443],[299,0.4442],[337,0.4446],[345,0.4435],[351,0.44],[356,0.4263],[361,0.1392],[366,0.0436],[371,0.059],[375,0.0738],[378,0.1059],[382,0.118],[387,0.1234],[395,0.1245],[409,0.1245],[439,0.1252]]
       }
       lowerarm_l_fk_ctrl.Rotation.Z {
        type=MovieSceneScriptingFloatChannel
        default=61.3561
        thinned_from=64
        keys [[0,61.3561],[51,61.4],[62,61.1357],[66,60.1275],[70,59.6356],[83,59.4456],[101,59.4532],[107,59.6802],[111,60.3747],[115,62.2834],[121,62.7606],[130,62.9197],[143,62.8279],[148,62.5527],[153,62.7647],[160,62.8441],[172,63.5963],[187,63.847],[194,63.7873],[199,63.6082],[204,62.8918],[223,62.9792],[233,63.2411],[246,63.3598],[251,64.0442],[257,66.0975],[267,74.3106],[275,57.3492],[279,4.2482],[283,-19.6472],[287,-25.6211],[292,-27.1145],[299,-27.4879],[309,-27.5741],[337,-27.6123],[345,-25.3157],[351,-18.4259],[356,9.1334],[361,56.0475],[366,71.6856],[371,69.7498],[375,67.8807],[378,63.7828],[382,62.246],[387,61.5545],[395,61.4104],[439,61.324]]
       }
       lowerarm_l_fk_ctrl.Scale.X {type=MovieSceneScriptingFloatChannel, default=1, thinned_from=64, keys [[0,1],[439,1]]}
       lowerarm_l_fk_ctrl.Scale.Y {type=MovieSceneScriptingFloatChannel, default=1, thinned_from=64, keys [[0,1],[439,1]]}
       lowerarm_l_fk_ctrl.Scale.Z {type=MovieSceneScriptingFloatChannel, default=1, thinned_from=64, keys [[0,1],[439,1]]}
       hand_l_fk_ctrl.Location.X {
        type=MovieSceneScriptingFloatChannel
        default=0
        thinned_from=64
        keys [[0,0],[15,0],[22,-0.0003],[45,-0.0003],[51,0],[66,0],[70,-0.0024],[75,-0.0028],[94,-0.003],[101,-0.0029],[107,-0.0022],[111,0],[275,0],[279,0.2088],[283,0.3028],[287,0.3263],[292,0.3322],[299,0.3337],[337,0.3341],[345,0.3133],[351,0.2506],[356,0],[439,0]]
       }
       hand_l_fk_ctrl.Location.Y {type=MovieSceneScriptingFloatChannel, default=-0.1788, thinned_from=64, keys [[0,-0.1788],[45,-0.1788],[51,-0.2],[56,-0.1788],[439,-0.1788]]}
       hand_l_fk_ctrl.Location.Z {type=MovieSceneScriptingFloatChannel, default=0.0249, thinned_from=64, keys [[0,0.0249],[45,0.0249],[51,0],[56,0.0249],[439,0.0249]]}
       hand_l_fk_ctrl.Rotation.X {
        type=MovieSceneScriptingFloatChannel
        default=13.5746
        thinned_from=64
        keys [[0,13.5746],[15,13.5746],[22,13.8152],[30,13.8098],[45,13.8883],[51,13.6],[56,13.4355],[62,13.0162],[66,10.7816],[70,9.8545],[75,9.599],[83,9.2567],[94,9.1452],[101,9.1882],[107,9.4889],[111,10.409],[115,12.5198],[121,13.0042],[138,13.3056],[143,13.7254],[148,15.4045],[153,5.2644],[160,1.4618],[172,13.5498],[187,17.5791],[194,16.4292],[199,12.9794],[204,-0.8197],[215,-0.8197],[223,1.7089],[233,9.2948],[241,11.8032],[246,12.6393],[251,11.5802],[257,8.4026],[262,2.0846],[267,-4.3078],[271,-1.3561],[275,1.5613],[279,21.1913],[283,30.0249],[287,32.2332],[292,32.7853],[299,32.9234],[337,32.9694],[345,38.6391],[351,55.6482],[356,123.6848],[361,32.4636],[366,2.0565],[371,6.1557],[375,10.1135],[378,12.2627],[382,13.0687],[387,13.4313],[395,13.5069],[409,-1.5922],[439,13.5522]]
       }
       hand_l_fk_ctrl.Rotation.Y {
        type=MovieSceneScriptingFloatChannel
        default=-17.9359
        thinned_from=64
        keys [[0,-17.9359],[15,-17.9359],[22,-17.9895],[45,-18.0069],[51,-17.9],[56,-17.9123],[62,-17.8405],[66,-17.49],[70,-17.4108],[83,-17.3275],[101,-17.3266],[107,-17.4381],[111,-17.7794],[115,-18.7397],[121,-18.9672],[138,-19.0876],[143,-19.2212],[148,-19.7558],[153,-13.7467],[160,-11.4933],[172,-18.0505],[187,-20.2362],[194,-19.6053],[199,-17.7124],[204,-10.1412],[215,-10.1412],[223,-11.9537],[233,-17.391],[241,-19.1988],[246,-19.8015],[251,-20.7152],[257,-23.4565],[262,-28.9069],[267,-34.4215],[271,1.1769],[275,36.3613],[279,43.1282],[283,46.1733],[287,46.9346],[292,47.1249],[337,47.1883],[345,46.5466],[351,44.6213],[356,36.9201],[361,-2.0978],[366,-15.1038],[371,-18.1401],[375,-21.0717],[378,-19.1167],[382,-18.3836],[387,-18.0537],[395,-17.985],[409,-14.4792],[439,-17.9437]]
       }
       hand_l_fk_ctrl.Rotation.Z {
        type=MovieSceneScriptingFloatChannel
        default=19.8888
        thinned_from=64
        keys [[0,19.8888],[15,19.8888],[22,20.0822],[30,20.0784],[45,20.1416],[51,19.9],[56,19.7908],[62,19.4985],[66,18.1304],[70,17.6291],[83,17.256],[94,17.1846],[107,17.206],[111,17.2634],[115,17.1327],[121,17.0637],[130,17.0407],[138,17.1642],[143,17.5345],[148,19.0157],[153,28.6879],[160,32.315],[172,23.5419],[187,20.6175],[194,21.4846],[199,24.0859],[204,34.4912],[215,34.4912],[223,31.386],[233,22.0702],[241,18.9482],[246,17.9075],[251,17.3186],[257,15.5519],[262,12.0392],[267,8.4852],[271,-0.4414],[275,-9.2642],[279,-38.0472],[283,-50.9996],[287,-54.2377],[292,-55.0472],[299,-55.2496],[337,-55.317],[345,-47.5269],[351,-24.1564],[356,69.3253],[361,26.1304],[366,11.7321],[371,15.4447],[375,19.0294],[378,19.6021],[382,19.8168],[387,19.9135],[395,19.9336],[409,6.9805],[439,19.9457]]
       }
       hand_l_fk_ctrl.Scale.X {type=MovieSceneScriptingFloatChannel, default=1, thinned_from=64, keys [[0,1],[439,1]]}
       hand_l_fk_ctrl.Scale.Y {type=MovieSceneScriptingFloatChannel, default=1, thinned_from=64, keys [[0,1],[439,1]]}
       hand_l_fk_ctrl.Scale.Z {type=MovieSceneScriptingFloatChannel, default=1, thinned_from=64, keys [[0,1],[439,1]]}
       thumb_01_l_ctrl.Location.X {type=MovieSceneScriptingFloatChannel, default=0, thinned_from=73, keys [[0,0],[439,0]]}
       thumb_01_l_ctrl.Location.Y {type=MovieSceneScriptingFloatChannel, default=0, thinned_from=73, keys [[0,0],[439,0]]}
       thumb_01_l_ctrl.Location.Z {type=MovieSceneScriptingFloatChannel, default=0, thinned_from=73, keys [[0,0],[439,0]]}
       thumb_01_l_ctrl.Rotation.X {
        type=MovieSceneScriptingFloatChannel
        default=-37.7411
        thinned_from=73
        keys [[0,-37.7411],[51,-37.7],[269,-37.7411],[271,-29.6074],[272,-26.5615],[273,-24.501],[274,-18.5363],[279,-8.6086],[283,-5.0333],[287,-4.1395],[292,-3.916],[299,-3.8601],[309,-3.8472],[351,-3.8415],[355,-11.0988],[356,-13.1027],[357,-14.8516],[358,-22.1651],[359,-22.1651],[360,-27.4231],[361,-33.6984],[366,-36.7304],[371,-37.7411],[439,-37.7411]]
       }
       thumb_01_l_ctrl.Rotation.Y {
        type=MovieSceneScriptingFloatChannel
        default=11.9985
        thinned_from=73
        keys [[0,11.9985],[269,11.9985],[271,16.03],[272,17.5397],[273,18.561],[274,21.5174],[275,22.5],[279,31.4739],[283,35.5122],[287,36.5218],[292,36.7742],[309,36.8519],[351,36.8583],[355,31.5363],[356,30.0667],[357,28.7842],[358,23.4209],[359,19],[360,17.3546],[361,14.0971],[366,12.5231],[371,11.9985],[439,11.9985]]
       }
       thumb_01_l_ctrl.Rotation.Z {
        type=MovieSceneScriptingFloatChannel
        default=5.4741
        thinned_from=73
        keys [[0,5.4741],[45,5.4741],[51,5.5],[56,5.4741],[66,5.4741],[70,5.4935],[107,5.4883],[111,5.4741],[246,5.4741],[251,5.1632],[257,4.2306],[262,2.3762],[267,0.5],[269,0.5],[271,1.0743],[272,1.2894],[273,1.4349],[274,1.856],[279,2.557],[283,2.8094],[287,2.8725],[292,2.8883],[351,2.8936],[355,0.7433],[356,0.1495],[357,0.6881],[358,2.9404],[359,0],[360,1.8388],[361,4.0498],[366,5.118],[371,5.4741],[439,5.4741]]
       }
       thumb_01_l_ctrl.Scale.X {type=MovieSceneScriptingFloatChannel, default=1, thinned_from=73, keys [[0,1],[439,1]]}
       thumb_01_l_ctrl.Scale.Y {type=MovieSceneScriptingFloatChannel, default=1, thinned_from=73, keys [[0,1],[439,1]]}
       thumb_01_l_ctrl.Scale.Z {type=MovieSceneScriptingFloatChannel, default=1, thinned_from=73, keys [[0,1],[439,1]]}
       thumb_02_l_ctrl.Location.X {type=MovieSceneScriptingFloatChannel, default=0, thinned_from=73, keys [[0,0],[439,0]]}
       thumb_02_l_ctrl.Location.Y {type=MovieSceneScriptingFloatChannel, default=0, thinned_from=73, keys [[0,0],[439,0]]}
       thumb_02_l_ctrl.Location.Z {type=MovieSceneScriptingFloatChannel, default=0, thinned_from=73, keys [[0,0],[439,0]]}
       thumb_02_l_ctrl.Rotation.X {type=MovieSceneScriptingFloatChannel, default=0, thinned_from=73, keys [[0,0],[439,0]]}
       thumb_02_l_ctrl.Rotation.Y {type=MovieSceneScriptingFloatChannel, default=0, thinned_from=73, keys [[0,0],[439,0]]}
       thumb_02_l_ctrl.Rotation.Z {
        type=MovieSceneScriptingFloatChannel
        default=-4.9669
        thinned_from=73
        keys [[0,-4.9669],[45,-4.9669],[51,-5],[56,-4.9669],[66,-4.9669],[70,-4.9917],[107,-4.985],[111,-4.9669],[246,-4.9669],[251,-5.2778],[257,-6.2105],[262,-8.0649],[267,-9.9411],[269,-9.9411],[271,-0.6041],[272,2.8924],[273,5.2577],[274,12.1048],[279,23.5011],[283,27.6054],[287,28.6314],[292,28.8879],[299,28.9521],[351,28.9734],[355,23.1814],[356,21.5821],[357,20.5866],[358,16.4233],[359,16.4233],[360,8.8289],[361,0.4385],[366,-3.6156],[371,-4.9669],[439,-4.9669]]
       }
       thumb_02_l_ctrl.Scale.X {type=MovieSceneScriptingFloatChannel, default=1, thinned_from=73, keys [[0,1],[439,1]]}
       thumb_02_l_ctrl.Scale.Y {type=MovieSceneScriptingFloatChannel, default=1, thinned_from=73, keys [[0,1],[439,1]]}
       thumb_02_l_ctrl.Scale.Z {type=MovieSceneScriptingFloatChannel, default=1, thinned_from=73, keys [[0,1],[439,1]]}
       thumb_03_l_ctrl.Location.X {type=MovieSceneScriptingFloatChannel, default=0, thinned_from=73, keys [[0,0],[439,0]]}
       thumb_03_l_ctrl.Location.Y {type=MovieSceneScriptingFloatChannel, default=0, thinned_from=73, keys [[0,0],[439,0]]}
       thumb_03_l_ctrl.Location.Z {type=MovieSceneScriptingFloatChannel, default=0, thinned_from=73, keys [[0,0],[439,0]]}
       thumb_03_l_ctrl.Rotation.X {type=MovieSceneScriptingFloatChannel, default=0, thinned_from=73, keys [[0,0],[439,0]]}
       thumb_03_l_ctrl.Rotation.Y {type=MovieSceneScriptingFloatChannel, default=0, thinned_from=73, keys [[0,0],[439,0]]}
       thumb_03_l_ctrl.Rotation.Z {
        type=MovieSceneScriptingFloatChannel
        default=41.8526
        thinned_from=73
        keys [[0,41.8526],[45,41.8526],[51,41.9],[246,41.8526],[251,41.5417],[257,40.609],[267,36.8785],[269,36.8785],[272,36.2913],[273,36.183],[274,35.8698],[279,35.3483],[283,35.1606],[351,35.098],[357,37.2918],[358,38.749],[359,38.749],[360,39.7967],[361,41.0471],[366,41.6512],[371,41.8526],[439,41.8526]]
       }
       thumb_03_l_ctrl.Scale.X {type=MovieSceneScriptingFloatChannel, default=1, thinned_from=73, keys [[0,1],[439,1]]}
       thumb_03_l_ctrl.Scale.Y {type=MovieSceneScriptingFloatChannel, default=1, thinned_from=73, keys [[0,1],[439,1]]}
       thumb_03_l_ctrl.Scale.Z {type=MovieSceneScriptingFloatChannel, default=1, thinned_from=73, keys [[0,1],[439,1]]}
       index_metacarpal_l_ctrl.Location.X {type=MovieSceneScriptingFloatChannel, default=0, thinned_from=73, keys [[0,0],[439,0]]}
       index_metacarpal_l_ctrl.Location.Y {type=MovieSceneScriptingFloatChannel, default=0, thinned_from=73, keys [[0,0],[439,0]]}
       index_metacarpal_l_ctrl.Location.Z {type=MovieSceneScriptingFloatChannel, default=0, thinned_from=73, keys [[0,0],[439,0]]}
       index_metacarpal_l_ctrl.Rotation.X {
        type=MovieSceneScriptingFloatChannel
        default=0.0346
        thinned_from=73
        keys [[0,0.0346],[45,0.0346],[51,0],[56,0.0346],[66,0.0346],[70,0.0086],[75,0.0022],[83,0.0005],[94,0],[101,0.0012],[107,0.0094],[111,0.0346],[269,0.0346],[271,0.0263],[272,0.0232],[273,0.0211],[274,0.015],[279,0.0049],[283,0.0012],[287,0.0003],[292,0.0001],[351,0],[355,0.0074],[356,0.0095],[357,0.0112],[358,0.0187],[359,0.0238],[360,0.0257],[361,0.0295],[366,0.0346],[439,0.0346]]
       }
       index_metacarpal_l_ctrl.Rotation.Y {
        type=MovieSceneScriptingFloatChannel
        default=2.3559
        thinned_from=73
        keys [[0,2.3559],[45,2.3559],[51,2.4],[56,2.3559],[66,2.3559],[70,2.389],[75,2.3973],[101,2.3985],[107,2.388],[111,2.3559],[269,2.3559],[271,-5.0748],[272,-7.8574],[273,-9.7398],[274,-15.1889],[275,-17],[279,-10.2728],[283,-7.2455],[287,-6.4887],[292,-6.2995],[299,-6.2522],[351,-6.2364],[355,-4.397],[356,-3.889],[357,-3.4458],[358,-1.5921],[359,-0.3391],[360,0.1518],[361,1.0734],[366,2.3559],[439,2.3559]]
       }
       index_metacarpal_l_ctrl.Rotation.Z {
        type=MovieSceneScriptingFloatChannel
        default=-18.2
        thinned_from=73
        keys [[0,-18.2],[269,-18.2],[271,-7.95],[272,-4.1115],[273,-1.5149],[274,6.0018],[275,8.5],[279,8.9632],[283,9.1716],[287,9.2237],[351,9.2411],[355,8.2918],[356,8.0297],[357,6.45],[358,-0.1561],[359,-4.6214],[360,-6.3707],[361,-11.3169],[366,-18.2],[439,-18.2]]
       }
       index_metacarpal_l_ctrl.Scale.X {type=MovieSceneScriptingFloatChannel, default=1, thinned_from=73, keys [[0,1],[439,1]]}
       index_metacarpal_l_ctrl.Scale.Y {type=MovieSceneScriptingFloatChannel, default=1, thinned_from=73, keys [[0,1],[439,1]]}
       index_metacarpal_l_ctrl.Scale.Z {type=MovieSceneScriptingFloatChannel, default=1, thinned_from=73, keys [[0,1],[439,1]]}
       index_01_l_ctrl.Location.X {type=MovieSceneScriptingFloatChannel, default=0, thinned_from=73, keys [[0,0],[439,0]]}
       index_01_l_ctrl.Location.Y {type=MovieSceneScriptingFloatChannel, default=0, thinned_from=73, keys [[0,0],[439,0]]}
       index_01_l_ctrl.Location.Z {type=MovieSceneScriptingFloatChannel, default=0, thinned_from=73, keys [[0,0],[439,0]]}
       index_01_l_ctrl.Rotation.X {
        type=MovieSceneScriptingFloatChannel
        default=0.6292
        thinned_from=74
        keys [[0,0.6292],[45,0.6292],[51,0.6],[56,0.6292],[66,0.6292],[70,0.6073],[101,0.6081],[107,0.6133],[111,0.6292],[269,0.6292],[271,-0.0571],[272,-0.3141],[273,-0.4879],[274,-0.9911],[279,-1.8287],[283,-2.1304],[287,-2.2058],[292,-2.2247],[299,-2.2294],[351,-2.231],[354,2.0391],[355,1.0138],[356,0.7307],[357,0.682],[358,0.4786],[359,0.3411],[360,0.2872],[361,0.4302],[366,0.6292],[439,0.6292]]
       }
       index_01_l_ctrl.Rotation.Y {
        type=MovieSceneScriptingFloatChannel
        default=-2.7506
        thinned_from=74
        keys [[0,-2.7506],[45,-2.7506],[51,-2.8],[56,-2.7506],[66,-2.7506],[70,-2.7877],[101,-2.7864],[107,-2.7776],[111,-2.7506],[269,-2.7506],[271,-4.4964],[272,-5.1501],[273,-5.5924],[274,-6.8726],[279,-9.0034],[283,-9.7707],[287,-9.9626],[292,-10.0105],[351,-10.0265],[354,-9.092],[355,-8.6717],[356,-8.5556],[357,-8.5046],[358,-8.2915],[359,-6.2761],[360,-5.4866],[361,-4.3426],[366,-2.7506],[439,-2.7506]]
       }
       index_01_l_ctrl.Rotation.Z {
        type=MovieSceneScriptingFloatChannel
        default=-12.8893
        thinned_from=74
        keys [[0,-12.8893],[246,-12.8893],[251,-13.0524],[257,-13.542],[267,-15.5],[269,-15.5],[271,5.5616],[272,14.7705],[273,21],[274,18],[275,12.929],[279,23.5899],[283,28.3873],[287,29.5867],[292,29.8865],[299,29.9615],[351,29.9864],[354,-3.463],[355,15.315],[356,20.5],[357,26],[358,-2.7368],[359,-4.8483],[360,-5.6754],[361,-8.6917],[366,-12.8893],[439,-12.8893]]
       }
       index_01_l_ctrl.Scale.X {type=MovieSceneScriptingFloatChannel, default=1, thinned_from=73, keys [[0,1],[439,1]]}
       index_01_l_ctrl.Scale.Y {type=MovieSceneScriptingFloatChannel, default=1, thinned_from=73, keys [[0,1],[439,1]]}
       index_01_l_ctrl.Scale.Z {type=MovieSceneScriptingFloatChannel, default=1, thinned_from=73, keys [[0,1],[439,1]]}
       index_02_l_ctrl.Location.X {type=MovieSceneScriptingFloatChannel, default=0, thinned_from=73, keys [[0,0],[439,0]]}
       index_02_l_ctrl.Location.Y {type=MovieSceneScriptingFloatChannel, default=0, thinned_from=73, keys [[0,0],[439,0]]}
       index_02_l_ctrl.Location.Z {type=MovieSceneScriptingFloatChannel, default=0, thinned_from=73, keys [[0,0],[439,0]]}
       index_02_l_ctrl.Rotation.X {type=MovieSceneScriptingFloatChannel, default=0, thinned_from=73, keys [[0,0],[439,0]]}
       index_02_l_ctrl.Rotation.Y {type=MovieSceneScriptingFloatChannel, default=0, thinned_from=73, keys [[0,0],[439,0]]}
       index_02_l_ctrl.Rotation.Z {
        type=MovieSceneScriptingFloatChannel
        default=28.6778
        thinned_from=73
        keys [[0,28.6778],[246,28.6778],[251,28.5146],[257,28.0251],[267,26.0671],[269,26.0671],[271,30.0923],[272,31.5997],[273,32.6194],[274,35.5713],[279,40.4843],[283,42.2537],[287,42.696],[292,42.8066],[351,42.8435],[355,24.584],[356,19.5421],[357,21.5095],[358,-22],[359,2.7802],[360,12.4877],[361,19.2572],[366,28.6778],[439,28.6778]]
       }
       index_02_l_ctrl.Scale.X {type=MovieSceneScriptingFloatChannel, default=1, thinned_from=73, keys [[0,1],[439,1]]}
       index_02_l_ctrl.Scale.Y {type=MovieSceneScriptingFloatChannel, default=1, thinned_from=73, keys [[0,1],[439,1]]}
       index_02_l_ctrl.Scale.Z {type=MovieSceneScriptingFloatChannel, default=1, thinned_from=73, keys [[0,1],[439,1]]}
       index_03_l_ctrl.Location.X {type=MovieSceneScriptingFloatChannel, default=0, thinned_from=73, keys [[0,0],[439,0]]}
       index_03_l_ctrl.Location.Y {type=MovieSceneScriptingFloatChannel, default=0, thinned_from=73, keys [[0,0],[439,0]]}
       index_03_l_ctrl.Location.Z {type=MovieSceneScriptingFloatChannel, default=0, thinned_from=73, keys [[0,0],[439,0]]}
       index_03_l_ctrl.Rotation.X {type=MovieSceneScriptingFloatChannel, default=0, thinned_from=73, keys [[0,0],[439,0]]}
       index_03_l_ctrl.Rotation.Y {type=MovieSceneScriptingFloatChannel, default=0, thinned_from=73, keys [[0,0],[439,0]]}
       index_03_l_ctrl.Rotation.Z {
        type=MovieSceneScriptingFloatChannel
        default=24.3552
        thinned_from=73
        keys [[0,24.3552],[45,24.3552],[51,24.4],[66,24.3552],[107,24.3797],[246,24.3552],[251,24.192],[257,23.7025],[267,21.7444],[269,21.7444],[271,25.9506],[272,27.5257],[273,28.5912],[274,31.6757],[279,36.8096],[283,38.6585],[287,39.1207],[292,39.2363],[351,39.2748],[355,36.0808],[356,35.1988],[357,34.4291],[358,31.2104],[359,29.0347],[360,28.1824],[361,26.5821],[366,24.3552],[439,24.3552]]
       }
       index_03_l_ctrl.Scale.X {type=MovieSceneScriptingFloatChannel, default=1, thinned_from=73, keys [[0,1],[439,1]]}
       index_03_l_ctrl.Scale.Y {type=MovieSceneScriptingFloatChannel, default=1, thinned_from=73, keys [[0,1],[439,1]]}
       index_03_l_ctrl.Scale.Z {type=MovieSceneScriptingFloatChannel, default=1, thinned_from=73, keys [[0,1],[439,1]]}
       middle_metacarpal_l_ctrl.Location.X {type=MovieSceneScriptingFloatChannel, default=0, thinned_from=73, keys [[0,0],[439,0]]}
       middle_metacarpal_l_ctrl.Location.Y {type=MovieSceneScriptingFloatChannel, default=0, thinned_from=73, keys [[0,0],[439,0]]}
       middle_metacarpal_l_ctrl.Location.Z {type=MovieSceneScriptingFloatChannel, default=0, thinned_from=73, keys [[0,0],[439,0]]}
       middle_metacarpal_l_ctrl.Rotation.X {
        type=MovieSceneScriptingFloatChannel
        default=-0.0167
        thinned_from=73
        keys [[0,-0.0167],[45,-0.0167],[51,0],[56,-0.0167],[66,-0.0167],[70,-0.0042],[75,-0.0011],[83,-0.0003],[94,0],[101,-0.0006],[107,-0.0046],[111,-0.0167],[269,-0.0167],[271,-0.1154],[272,-0.1524],[273,-0.1774],[274,-0.2498],[279,-0.3703],[283,-0.4136],[287,-0.4245],[292,-0.4272],[351,-0.4281],[355,-0.34],[356,-0.3157],[357,-0.2945],[358,-0.2058],[359,-0.1458],[360,-0.1223],[361,-0.0781],[366,-0.0167],[439,-0.0167]]
       }
       middle_metacarpal_l_ctrl.Rotation.Y {
        type=MovieSceneScriptingFloatChannel
        default=1.8417
        thinned_from=73
        keys [[0,1.8417],[45,1.8417],[51,1.8],[56,1.8417],[66,1.8417],[70,1.8104],[75,1.8026],[101,1.8014],[107,1.8113],[111,1.8417],[269,1.8417],[271,-0.8217],[272,-1.8192],[273,-2.4939],[274,-4.4471],[279,-7.698],[283,-8.8688],[287,-9.1615],[292,-9.2347],[309,-9.2572],[351,-9.2591],[355,-6.8826],[356,-6.2264],[357,-5.6537],[358,-3.2588],[359,-1.64],[360,-1.0059],[361,0.1848],[366,1.8417],[439,1.8417]]
       }
       middle_metacarpal_l_ctrl.Rotation.Z {
        type=MovieSceneScriptingFloatChannel
        default=-12.6457
        thinned_from=73
        keys [[0,-12.6457],[45,-12.6457],[51,-12.6],[56,-12.6457],[66,-12.6457],[70,-12.6114],[107,-12.6124],[111,-12.6457],[269,-12.6457],[271,-1.043],[272,3.302],[273,6.2413],[274,14.7499],[279,28.9117],[283,34.0119],[287,35.287],[292,35.6058],[299,35.6855],[351,35.712],[355,24.1047],[356,20.8997],[357,18.4467],[358,8.1887],[359,1.255],[360,-1.4612],[361,-6.1378],[366,-12.6457],[439,-12.6457]]
       }
       middle_metacarpal_l_ctrl.Scale.X {type=MovieSceneScriptingFloatChannel, default=1, thinned_from=73, keys [[0,1],[439,1]]}
       middle_metacarpal_l_ctrl.Scale.Y {type=MovieSceneScriptingFloatChannel, default=1, thinned_from=73, keys [[0,1],[439,1]]}
       middle_metacarpal_l_ctrl.Scale.Z {type=MovieSceneScriptingFloatChannel, default=1, thinned_from=73, keys [[0,1],[439,1]]}
       middle_01_l_ctrl.Location.X {type=MovieSceneScriptingFloatChannel, default=0, thinned_from=73, keys [[0,0],[439,0]]}
       middle_01_l_ctrl.Location.Y {type=MovieSceneScriptingFloatChannel, default=0, thinned_from=73, keys [[0,0],[439,0]]}
       middle_01_l_ctrl.Location.Z {type=MovieSceneScriptingFloatChannel, default=0, thinned_from=73, keys [[0,0],[439,0]]}
       middle_01_l_ctrl.Rotation.X {
        type=MovieSceneScriptingFloatChannel
        default=-1.8457
        thinned_from=73
        keys [[0,-1.8457],[45,-1.8457],[51,-1.8],[56,-1.8457],[66,-1.8457],[70,-1.8114],[101,-1.8126],[107,-1.8208],[111,-1.8457],[269,-1.8457],[271,-1.4029],[272,-1.237],[273,-1.1249],[274,-0.8001],[279,-0.2596],[283,-0.0649],[287,-0.0162],[292,-0.0041],[299,-0.001],[309,-0.0003],[351,0],[355,-0.3951],[356,-0.5042],[357,-0.5995],[358,-0.9977],[359,-1.2668],[360,-1.3723],[361,-1.5702],[366,-1.8457],[439,-1.8457]]
       }
       middle_01_l_ctrl.Rotation.Y {
        type=MovieSceneScriptingFloatChannel
        default=-10.2135
        thinned_from=73
        keys [[0,-10.2135],[45,-10.2135],[51,-10.2],[66,-10.2135],[269,-10.2135],[271,-12.4349],[272,-13.2668],[273,-13.8295],[274,-15.4586],[275,-16],[279,-8.004],[283,-4.4059],[287,-3.5063],[292,-3.2814],[299,-3.2252],[309,-3.2122],[351,-3.2065],[355,-4.7065],[356,-5.1207],[357,-5.4822],[358,-6.9939],[359,-8.0157],[360,-8.416],[361,-9.1676],[366,-10.2135],[439,-10.2135]]
       }
       middle_01_l_ctrl.Rotation.Z {
        type=MovieSceneScriptingFloatChannel
        default=-20.8534
        thinned_from=73
        keys [[0,-20.8534],[45,-20.8534],[51,-20.9],[66,-20.8534],[107,-20.8788],[246,-20.8534],[251,-21.0165],[257,-21.506],[267,-23.4641],[269,-23.4641],[271,-13.3218],[272,-9.5509],[273,-7],[274,0.977],[275,3.5],[279,16.8206],[283,22.8149],[287,24.3134],[292,24.6881],[299,24.7817],[351,24.813],[355,8.3961],[356,3.8631],[357,-0.9643],[358,-21.1515],[359,-18.4055],[360,-17.3298],[361,-18.8031],[366,-20.8534],[439,-20.8534]]
       }
       middle_01_l_ctrl.Scale.X {type=MovieSceneScriptingFloatChannel, default=1, thinned_from=73, keys [[0,1],[439,1]]}
       middle_01_l_ctrl.Scale.Y {type=MovieSceneScriptingFloatChannel, default=1, thinned_from=73, keys [[0,1],[439,1]]}
       middle_01_l_ctrl.Scale.Z {type=MovieSceneScriptingFloatChannel, default=1, thinned_from=73, keys [[0,1],[439,1]]}
       middle_02_l_ctrl.Location.X {type=MovieSceneScriptingFloatChannel, default=0, thinned_from=73, keys [[0,0],[439,0]]}
       middle_02_l_ctrl.Location.Y {type=MovieSceneScriptingFloatChannel, default=0, thinned_from=73, keys [[0,0],[439,0]]}
       middle_02_l_ctrl.Location.Z {type=MovieSceneScriptingFloatChannel, default=0, thinned_from=73, keys [[0,0],[439,0]]}
       middle_02_l_ctrl.Rotation.X {type=MovieSceneScriptingFloatChannel, default=0, thinned_from=73, keys [[0,0],[439,0]]}
       middle_02_l_ctrl.Rotation.Y {type=MovieSceneScriptingFloatChannel, default=0, thinned_from=73, keys [[0,0],[439,0]]}
       middle_02_l_ctrl.Rotation.Z {
        type=MovieSceneScriptingFloatChannel
        default=11.1202
        thinned_from=73
        keys [[0,11.1202],[45,11.1202],[51,11.1],[66,11.1202],[246,11.1202],[251,10.9571],[257,10.4676],[267,8.5095],[269,8.5095],[271,20.5544],[272,25.0651],[273,28.1164],[274,36.9492],[279,51.6508],[283,56.9454],[287,58.269],[292,58.5999],[351,58.7102],[355,32.5365],[356,25.3094],[357,15.5934],[358,-25.0371],[359,-5.0596],[360,2.7664],[361,6.2594],[366,11.1202],[439,11.1202]]
       }
       middle_02_l_ctrl.Scale.X {type=MovieSceneScriptingFloatChannel, default=1, thinned_from=73, keys [[0,1],[439,1]]}
       middle_02_l_ctrl.Scale.Y {type=MovieSceneScriptingFloatChannel, default=1, thinned_from=73, keys [[0,1],[439,1]]}
       middle_02_l_ctrl.Scale.Z {type=MovieSceneScriptingFloatChannel, default=1, thinned_from=73, keys [[0,1],[439,1]]}
       middle_03_l_ctrl.Location.X {type=MovieSceneScriptingFloatChannel, default=0, thinned_from=73, keys [[0,0],[439,0]]}
       middle_03_l_ctrl.Location.Y {type=MovieSceneScriptingFloatChannel, default=0, thinned_from=73, keys [[0,0],[439,0]]}
       middle_03_l_ctrl.Location.Z {type=MovieSceneScriptingFloatChannel, default=0, thinned_from=73, keys [[0,0],[439,0]]}
       middle_03_l_ctrl.Rotation.X {type=MovieSceneScriptingFloatChannel, default=0, thinned_from=73, keys [[0,0],[439,0]]}
       middle_03_l_ctrl.Rotation.Y {type=MovieSceneScriptingFloatChannel, default=0, thinned_from=73, keys [[0,0],[439,0]]}
       middle_03_l_ctrl.Rotation.Z {
        type=MovieSceneScriptingFloatChannel
        default=24.3552
        thinned_from=73
        keys [[0,24.3552],[45,24.3552],[51,24.4],[66,24.3552],[107,24.3797],[246,24.3552],[251,24.192],[257,23.7025],[267,21.7444],[269,21.7444],[271,25.5912],[272,27.0317],[273,28.0062],[274,30.8271],[279,35.5223],[283,37.2132],[287,37.6359],[292,37.7416],[351,37.7768],[355,34.9035],[356,34.1101],[357,33.4177],[358,30.5221],[359,28.5648],[360,27.7981],[361,26.3585],[366,24.3552],[439,24.3552]]
       }
       middle_03_l_ctrl.Scale.X {type=MovieSceneScriptingFloatChannel, default=1, thinned_from=73, keys [[0,1],[439,1]]}
       middle_03_l_ctrl.Scale.Y {type=MovieSceneScriptingFloatChannel, default=1, thinned_from=73, keys [[0,1],[439,1]]}
       middle_03_l_ctrl.Scale.Z {type=MovieSceneScriptingFloatChannel, default=1, thinned_from=73, keys [[0,1],[439,1]]}
       ring_metacarpal_l_ctrl.Location.X {type=MovieSceneScriptingFloatChannel, default=0, thinned_from=73, keys [[0,0],[439,0]]}
       ring_metacarpal_l_ctrl.Location.Y {type=MovieSceneScriptingFloatChannel, default=0, thinned_from=73, keys [[0,0],[439,0]]}
       ring_metacarpal_l_ctrl.Location.Z {type=MovieSceneScriptingFloatChannel, default=0, thinned_from=73, keys [[0,0],[439,0]]}
       ring_metacarpal_l_ctrl.Rotation.X {
        type=MovieSceneScriptingFloatChannel
        default=0.0204
        thinned_from=73
        keys [[0,0.0204],[45,0.0204],[51,0],[56,0.0204],[66,0.0204],[70,0.0051],[75,0.0013],[83,0.0003],[94,0],[101,0.0007],[107,0.0055],[111,0.0204],[269,0.0204],[271,0.0155],[272,0.0137],[273,0.0124],[274,0.0088],[279,0.0029],[283,0.0007],[287,0.0002],[351,0],[355,-2.4812],[356,-3.1663],[357,-3.0824],[358,-2.7319],[359,-2.4949],[360,-2.4021],[361,-1.3892],[366,0.0204],[439,0.0204]]
       }
       ring_metacarpal_l_ctrl.Rotation.Y {
        type=MovieSceneScriptingFloatChannel
        default=-1.3752
        thinned_from=73
        keys [[0,-1.3752],[45,-1.3752],[51,-1.4],[56,-1.3752],[66,-1.3752],[70,-1.3938],[75,-1.3985],[101,-1.3992],[107,-1.3933],[111,-1.3752],[269,-1.3752],[271,-3.9022],[272,-4.5564],[273,-4.9989],[274,-5.375],[275,-5.5],[279,-4.4324],[283,-3.952],[287,-3.8318],[292,-3.8018],[309,-3.7926],[351,-3.7918],[355,-4.677],[356,-4.9214],[357,-5.5155],[358,-8],[359,-5.0194],[360,-3.8517],[361,-2.8162],[366,-1.3752],[439,-1.3752]]
       }
       ring_metacarpal_l_ctrl.Rotation.Z {
        type=MovieSceneScriptingFloatChannel
        default=-3.712
        thinned_from=73
        keys [[0,-3.712],[45,-3.712],[51,-3.7],[56,-3.712],[66,-3.712],[70,-3.703],[107,-3.7032],[111,-3.712],[269,-3.712],[271,2.5086],[272,4.8381],[273,6.414],[274,10.9757],[279,18.5683],[283,21.3027],[287,21.9863],[292,22.1572],[309,22.2098],[351,22.2141],[355,16.6638],[356,15.1313],[357,13.7937],[358,8.2004],[359,4.4197],[360,2.9387],[361,0.1578],[366,-3.712],[439,-3.712]]
       }
       ring_metacarpal_l_ctrl.Scale.X {type=MovieSceneScriptingFloatChannel, default=1, thinned_from=73, keys [[0,1],[439,1]]}
       ring_metacarpal_l_ctrl.Scale.Y {type=MovieSceneScriptingFloatChannel, default=1, thinned_from=73, keys [[0,1],[439,1]]}
       ring_metacarpal_l_ctrl.Scale.Z {type=MovieSceneScriptingFloatChannel, default=1, thinned_from=73, keys [[0,1],[439,1]]}
       ring_01_l_ctrl.Location.X {type=MovieSceneScriptingFloatChannel, default=0, thinned_from=73, keys [[0,0],[439,0]]}
       ring_01_l_ctrl.Location.Y {type=MovieSceneScriptingFloatChannel, default=0, thinned_from=73, keys [[0,0],[439,0]]}
       ring_01_l_ctrl.Location.Z {type=MovieSceneScriptingFloatChannel, default=0, thinned_from=73, keys [[0,0],[439,0]]}
       ring_01_l_ctrl.Rotation.X {
        type=MovieSceneScriptingFloatChannel
        default=-9.9706
        thinned_from=73
        keys [[0,-9.9706],[45,-9.9706],[51,-10],[56,-9.9706],[66,-9.9706],[70,-9.9926],[107,-9.9866],[269,-9.9706],[271,-7.5783],[272,-6.6824],[273,-6.0764],[274,-4.322],[279,-1.4021],[283,-0.3505],[287,-0.0876],[292,-0.0219],[299,-0.0055],[309,-0.0017],[327,-0.0004],[337,0],[351,0],[355,-2.1345],[356,-2.7239],[357,-3.2383],[358,-5.3893],[359,-6.8433],[360,-7.4129],[361,-8.4823],[366,-9.9706],[439,-9.9706]]
       }
       ring_01_l_ctrl.Rotation.Y {
        type=MovieSceneScriptingFloatChannel
        default=-4.2911
        thinned_from=73
        keys [[0,-4.2911],[45,-4.2911],[51,-4.3],[66,-4.2911],[107,-4.296],[269,-4.2911],[271,-4.9472],[272,-5.1928],[273,-5.359],[274,-5.8401],[275,-6],[279,-2.6452],[283,-1.1356],[287,-0.7582],[292,-0.6638],[299,-0.6402],[309,-0.6348],[351,-0.6324],[355,-1.4156],[356,-1.6319],[357,-1.8207],[358,-2.61],[359,-3.1436],[360,-3.3526],[361,-3.745],[366,-4.2911],[439,-4.2911]]
       }
       ring_01_l_ctrl.Rotation.Z {
        type=MovieSceneScriptingFloatChannel
        default=-24.8375
        thinned_from=73
        keys [[0,-24.8375],[45,-24.8375],[51,-24.8],[66,-24.8375],[246,-24.8375],[251,-25.0006],[257,-25.4901],[267,-27.4482],[269,-27.4482],[271,-14.7995],[272,-10.0628],[273,-6.8585],[274,2.4171],[275,5.5],[279,23.7942],[283,32.0267],[287,34.0848],[292,34.5993],[299,34.7279],[351,34.7708],[355,13.4636],[356,7.5803],[357,1.5957],[358,-23.4309],[359,-20.9036],[360,-19.9135],[361,-21.9723],[366,-24.8375],[439,-24.8375]]
       }
       ring_01_l_ctrl.Scale.X {type=MovieSceneScriptingFloatChannel, default=1, thinned_from=73, keys [[0,1],[439,1]]}
       ring_01_l_ctrl.Scale.Y {type=MovieSceneScriptingFloatChannel, default=1, thinned_from=73, keys [[0,1],[439,1]]}
       ring_01_l_ctrl.Scale.Z {type=MovieSceneScriptingFloatChannel, default=1, thinned_from=73, keys [[0,1],[439,1]]}
       ring_02_l_ctrl.Location.X {type=MovieSceneScriptingFloatChannel, default=0, thinned_from=73, keys [[0,0],[439,0]]}
       ring_02_l_ctrl.Location.Y {type=MovieSceneScriptingFloatChannel, default=0, thinned_from=73, keys [[0,0],[439,0]]}
       ring_02_l_ctrl.Location.Z {type=MovieSceneScriptingFloatChannel, default=0, thinned_from=73, keys [[0,0],[439,0]]}
       ring_02_l_ctrl.Rotation.X {type=MovieSceneScriptingFloatChannel, default=0, thinned_from=73, keys [[0,0],[439,0]]}
       ring_02_l_ctrl.Rotation.Y {type=MovieSceneScriptingFloatChannel, default=0, thinned_from=73, keys [[0,0],[439,0]]}
       ring_02_l_ctrl.Rotation.Z {
        type=MovieSceneScriptingFloatChannel
        default=-1.1252
        thinned_from=73
        keys [[0,-1.1252],[45,-1.1252],[51,-1.1],[56,-1.1252],[66,-1.1252],[70,-1.1063],[101,-1.107],[107,-1.1115],[111,-1.1252],[246,-1.1252],[251,-1.2884],[257,-1.7779],[262,-2.7512],[267,-3.736],[269,-3.736],[271,8.8623],[272,13.5801],[273,16.7716],[274,26.0102],[279,41.3871],[283,46.9249],[287,48.3093],[292,48.6555],[309,48.762],[351,48.7708],[355,21.1165],[356,13.4806],[357,3.6785],[358,-37.3122],[359,-17.578],[360,-9.8473],[361,-6.2004],[366,-1.1252],[439,-1.1252]]
       }
       ring_02_l_ctrl.Scale.X {type=MovieSceneScriptingFloatChannel, default=1, thinned_from=73, keys [[0,1],[439,1]]}
       ring_02_l_ctrl.Scale.Y {type=MovieSceneScriptingFloatChannel, default=1, thinned_from=73, keys [[0,1],[439,1]]}
       ring_02_l_ctrl.Scale.Z {type=MovieSceneScriptingFloatChannel, default=1, thinned_from=73, keys [[0,1],[439,1]]}
       ring_03_l_ctrl.Location.X {type=MovieSceneScriptingFloatChannel, default=0, thinned_from=73, keys [[0,0],[439,0]]}
       ring_03_l_ctrl.Location.Y {type=MovieSceneScriptingFloatChannel, default=0, thinned_from=73, keys [[0,0],[439,0]]}
       ring_03_l_ctrl.Location.Z {type=MovieSceneScriptingFloatChannel, default=0, thinned_from=73, keys [[0,0],[439,0]]}
       ring_03_l_ctrl.Rotation.X {type=MovieSceneScriptingFloatChannel, default=0, thinned_from=73, keys [[0,0],[439,0]]}
       ring_03_l_ctrl.Rotation.Y {type=MovieSceneScriptingFloatChannel, default=0, thinned_from=73, keys [[0,0],[439,0]]}
       ring_03_l_ctrl.Rotation.Z {
        type=MovieSceneScriptingFloatChannel
        default=24.3552
        thinned_from=73
        keys [[0,24.3552],[45,24.3552],[51,24.4],[66,24.3552],[107,24.3797],[246,24.3552],[251,24.192],[257,23.7025],[267,21.7444],[269,21.7444],[271,23.191],[272,23.7327],[273,24.0992],[274,25.16],[279,26.9256],[283,27.5614],[287,27.7204],[351,27.7734],[357,26.6632],[358,25.9257],[359,25.4273],[360,25.232],[361,24.8654],[366,24.3552],[439,24.3552]]
       }
       ring_03_l_ctrl.Scale.X {type=MovieSceneScriptingFloatChannel, default=1, thinned_from=73, keys [[0,1],[439,1]]}
       ring_03_l_ctrl.Scale.Y {type=MovieSceneScriptingFloatChannel, default=1, thinned_from=73, keys [[0,1],[439,1]]}
       ring_03_l_ctrl.Scale.Z {type=MovieSceneScriptingFloatChannel, default=1, thinned_from=73, keys [[0,1],[439,1]]}
       pinky_metacarpal_l_ctrl.Location.X {type=MovieSceneScriptingFloatChannel, default=0, thinned_from=73, keys [[0,0],[439,0]]}
       pinky_metacarpal_l_ctrl.Location.Y {type=MovieSceneScriptingFloatChannel, default=0, thinned_from=73, keys [[0,0],[439,0]]}
       pinky_metacarpal_l_ctrl.Location.Z {type=MovieSceneScriptingFloatChannel, default=0, thinned_from=73, keys [[0,0],[439,0]]}
       pinky_metacarpal_l_ctrl.Rotation.X {
        type=MovieSceneScriptingFloatChannel
        default=-0.3335
        thinned_from=73
        keys [[0,-0.3335],[45,-0.3335],[51,-0.3],[56,-0.3335],[66,-0.3335],[70,-0.3084],[75,-0.3021],[83,-0.3005],[94,-0.3],[101,-0.3011],[107,-0.3091],[111,-0.3335],[269,-0.3335],[271,-0.1988],[272,-0.1484],[273,-0.1142],[274,-0.0154],[279,0.149],[283,0.2082],[287,0.223],[292,0.2267],[299,0.2276],[351,0.2279],[355,0.1077],[356,0.0745],[357,0.0456],[358,-0.0756],[359,-0.1574],[360,-0.1895],[361,-0.2497],[366,-0.3335],[439,-0.3335]]
       }
       pinky_metacarpal_l_ctrl.Rotation.Y {
        type=MovieSceneScriptingFloatChannel
        default=-4.5453
        thinned_from=73
        keys [[0,-4.5453],[45,-4.5453],[51,-4.5],[56,-4.5453],[66,-4.5453],[70,-4.5113],[101,-4.5015],[107,-4.5123],[111,-4.5453],[269,-4.5453],[271,-7.0427],[272,-7.3181],[273,-7.5044],[274,-6],[275,-5.5],[279,-1.8165],[283,-0.1589],[287,0.2555],[292,0.3591],[299,0.385],[309,0.3909],[337,0.3936],[351,0.3936],[355,-3.3912],[356,-4.4363],[357,-10],[358,-13],[359,-8.6091],[360,-6.889],[361,-5.909],[366,-4.5453],[439,-4.5453]]
       }
       pinky_metacarpal_l_ctrl.Rotation.Z {
        type=MovieSceneScriptingFloatChannel
        default=-7.1098
        thinned_from=73
        keys [[0,-7.1098],[45,-7.1098],[51,-7.1],[66,-7.1098],[107,-7.1027],[269,-7.1098],[271,-4.3242],[272,-3.281],[273,-2.5754],[274,-0.5326],[275,0.1463],[279,2.8674],[283,4.0918],[287,4.398],[292,4.4745],[299,4.4936],[351,4.5],[355,2.0145],[356,1.3283],[357,0.7293],[358,-1.7754],[359,-3.4684],[360,-4.1316],[361,-5.3769],[366,-7.1098],[439,-7.1098]]
       }
       pinky_metacarpal_l_ctrl.Scale.X {type=MovieSceneScriptingFloatChannel, default=1, thinned_from=73, keys [[0,1],[439,1]]}
       pinky_metacarpal_l_ctrl.Scale.Y {type=MovieSceneScriptingFloatChannel, default=1, thinned_from=73, keys [[0,1],[439,1]]}
       pinky_metacarpal_l_ctrl.Scale.Z {type=MovieSceneScriptingFloatChannel, default=1, thinned_from=73, keys [[0,1],[439,1]]}
       pinky_01_l_ctrl.Location.X {type=MovieSceneScriptingFloatChannel, default=0, thinned_from=73, keys [[0,0],[439,0]]}
       pinky_01_l_ctrl.Location.Y {type=MovieSceneScriptingFloatChannel, default=0, thinned_from=73, keys [[0,0],[439,0]]}
       pinky_01_l_ctrl.Location.Z {type=MovieSceneScriptingFloatChannel, default=0, thinned_from=73, keys [[0,0],[439,0]]}
       pinky_01_l_ctrl.Rotation.X {
        type=MovieSceneScriptingFloatChannel
        default=-8.0273
        thinned_from=73
        keys [[0,-8.0273],[45,-8.0273],[51,-8],[56,-8.0273],[66,-8.0273],[70,-8.0068],[107,-8.0124],[269,-8.0273],[271,-6.1013],[272,-5.38],[273,-4.8921],[274,-3.4797],[279,-1.1288],[283,-0.2822],[287,-0.0706],[292,-0.0176],[299,-0.0044],[309,-0.0014],[327,-0.0003],[337,0],[351,0],[355,-5.161],[356,-6.5861],[357,-31.5],[358,-16],[359,-11.899],[360,-10.2925],[361,-9.3453],[366,-8.0273],[439,-8.0273]]
       }
       pinky_01_l_ctrl.Rotation.Y {
        type=MovieSceneScriptingFloatChannel
        default=-9.4708
        thinned_from=73
        keys [[0,-9.4708],[45,-9.4708],[51,-9.5],[56,-9.4708],[66,-9.4708],[70,-9.4927],[107,-9.4867],[269,-9.4708],[271,-6.3367],[272,-5.1631],[273,-4.3691],[274,-2.0708],[279,1.7545],[283,3.1321],[287,3.4766],[292,3.5627],[299,3.5842],[351,3.5914],[355,0.795],[356,0.0229],[357,-0.651],[358,-3.4691],[359,-5.3738],[360,-6.12],[361,-7.5211],[366,-9.4708],[439,-9.4708]]
       }
       pinky_01_l_ctrl.Rotation.Z {
        type=MovieSceneScriptingFloatChannel
        default=-9.9044
        thinned_from=73
        keys [[0,-9.9044],[246,-9.9044],[251,-10.0676],[257,-10.5571],[267,-12.5151],[269,-12.5151],[271,0.0126],[272,4.7041],[273,7.8777],[274,17.0647],[279,32.3556],[283,37.8624],[287,39.2391],[292,39.5833],[299,39.6694],[351,39.698],[355,19.4874],[356,13.9068],[357,8.4732],[358,-14.2491],[359,-10.1642],[360,-8.5639],[361,-9.1244],[366,-9.9044],[439,-9.9044]]
       }
       pinky_01_l_ctrl.Scale.X {type=MovieSceneScriptingFloatChannel, default=1, thinned_from=73, keys [[0,1],[439,1]]}
       pinky_01_l_ctrl.Scale.Y {type=MovieSceneScriptingFloatChannel, default=1, thinned_from=73, keys [[0,1],[439,1]]}
       pinky_01_l_ctrl.Scale.Z {type=MovieSceneScriptingFloatChannel, default=1, thinned_from=73, keys [[0,1],[439,1]]}
       pinky_02_l_ctrl.Location.X {type=MovieSceneScriptingFloatChannel, default=0, thinned_from=73, keys [[0,0],[439,0]]}
       pinky_02_l_ctrl.Location.Y {type=MovieSceneScriptingFloatChannel, default=0, thinned_from=73, keys [[0,0],[439,0]]}
       pinky_02_l_ctrl.Location.Z {type=MovieSceneScriptingFloatChannel, default=0, thinned_from=73, keys [[0,0],[439,0]]}
       pinky_02_l_ctrl.Rotation.X {type=MovieSceneScriptingFloatChannel, default=0, thinned_from=73, keys [[0,0],[439,0]]}
       pinky_02_l_ctrl.Rotation.Y {type=MovieSceneScriptingFloatChannel, default=0, thinned_from=73, keys [[0,0],[439,0]]}
       pinky_02_l_ctrl.Rotation.Z {
        type=MovieSceneScriptingFloatChannel
        default=-3.5066
        thinned_from=73
        keys [[0,-3.5066],[45,-3.5066],[51,-3.5],[246,-3.5066],[251,-3.6698],[257,-4.1593],[262,-5.1326],[267,-6.1173],[269,-6.1173],[271,7.8497],[272,13.0802],[273,16.6184],[274,26.8609],[279,43.9085],[283,50.048],[287,51.5829],[292,51.9666],[309,52.0847],[351,52.0945],[355,22.1733],[356,13.9115],[357,3.8499],[358,-38.226],[359,-19.2254],[360,-11.782],[361,-8.3218],[366,-3.5066],[439,-3.5066]]
       }
       pinky_02_l_ctrl.Scale.X {type=MovieSceneScriptingFloatChannel, default=1, thinned_from=73, keys [[0,1],[439,1]]}
       pinky_02_l_ctrl.Scale.Y {type=MovieSceneScriptingFloatChannel, default=1, thinned_from=73, keys [[0,1],[439,1]]}
       pinky_02_l_ctrl.Scale.Z {type=MovieSceneScriptingFloatChannel, default=1, thinned_from=73, keys [[0,1],[439,1]]}
       pinky_03_l_ctrl.Location.X {type=MovieSceneScriptingFloatChannel, default=0, thinned_from=73, keys [[0,0],[439,0]]}
       pinky_03_l_ctrl.Location.Y {type=MovieSceneScriptingFloatChannel, default=0, thinned_from=73, keys [[0,0],[439,0]]}
       pinky_03_l_ctrl.Location.Z {type=MovieSceneScriptingFloatChannel, default=0, thinned_from=73, keys [[0,0],[439,0]]}
       pinky_03_l_ctrl.Rotation.X {type=MovieSceneScriptingFloatChannel, default=0, thinned_from=73, keys [[0,0],[439,0]]}
       pinky_03_l_ctrl.Rotation.Y {type=MovieSceneScriptingFloatChannel, default=0, thinned_from=73, keys [[0,0],[439,0]]}
       pinky_03_l_ctrl.Rotation.Z {
        type=MovieSceneScriptingFloatChannel
        default=16.7626
        thinned_from=73
        keys [[0,16.7626],[45,16.7626],[51,16.8],[66,16.7626],[107,16.783],[246,16.7626],[251,16.5994],[257,16.1099],[267,14.1518],[269,14.1518],[271,18.1227],[272,19.6098],[273,20.6157],[274,23.5277],[279,28.3744],[283,30.1199],[287,30.5563],[292,30.6654],[351,30.7018],[355,27.7176],[356,26.8937],[357,26.1745],[358,23.1673],[359,21.1346],[360,20.3383],[361,18.8432],[366,16.7626],[439,16.7626]]
       }
       pinky_03_l_ctrl.Scale.X {type=MovieSceneScriptingFloatChannel, default=1, thinned_from=73, keys [[0,1],[439,1]]}
       pinky_03_l_ctrl.Scale.Y {type=MovieSceneScriptingFloatChannel, default=1, thinned_from=73, keys [[0,1],[439,1]]}
       pinky_03_l_ctrl.Scale.Z {type=MovieSceneScriptingFloatChannel, default=1, thinned_from=73, keys [[0,1],[439,1]]}
       hand_l_ik_ctrl.Location.X {
        type=MovieSceneScriptingFloatChannel
        default=-20.23
        thinned_from=117
        keys [[0,14.7782],[45,14.7782],[51,14.8],[66,14.7782],[107,14.7941],[263,14.7782],[268,14.8004],[269,15.1906],[270,16.1548],[271,17.1999],[272,16.9399],[273,15.9481],[274,15.8276],[275,14.9467],[275.9987,14.8272],[276,17.927],[278,18.9119],[279,19.1236],[280,19.2506],[281,19.2715],[283,19.4681],[284,19.8957],[285,20.2545],[287,20.8354],[288,21.0859],[289,21.2167],[290,21.3016],[292,21.3526],[351,21.3537],[352,22.7027],[353,23.8773],[354,24.704],[354.9987,22.9514],[355,17.7747],[356,16.6529],[357,15.0969],[358,14.6376],[359,14.2824],[360,14.6706],[361,14.7676],[439,14.8]]
       }
       hand_l_ik_ctrl.Location.Y {
        type=MovieSceneScriptingFloatChannel
        default=22.2594
        thinned_from=117
        keys [[0,4.0897],[45,4.0897],[51,4.1],[66,4.0897],[75,4.0994],[107,4.0972],[246,4.0897],[251,4.239],[257,4.6871],[262,4.9111],[263,4.9111],[264,4.9377],[268,4.9377],[269,5.018],[270,7.1404],[271,10.924],[272,16.5799],[273,23.1409],[274,24.7653],[275,26.1967],[275.9987,27.4525],[276,13.7954],[277,11.8519],[278,9.9387],[279,9.1099],[280,8.6127],[281,8.4768],[282,7.9341],[283,7.1875],[284,5.793],[285,3.9845],[286,1.8798],[287,-0.9089],[288,-3.3692],[289,-4.7281],[290,-5.6521],[291,-6.0857],[292,-6.2302],[350,-6.2378],[351,-4.7798],[352,-1.6241],[353,4.3116],[354,6.0371],[354.9987,12.9516],[355,29.7027],[356,25.9056],[357,21.3952],[358,14.7284],[359,7.5964],[360,4.9741],[361,4.3185],[362,4.2099],[363,4.1412],[364,4.1],[439,4.1]]
       }
       hand_l_ik_ctrl.Location.Z {
        type=MovieSceneScriptingFloatChannel
        default=20.5629
        thinned_from=117
        keys [[0,3.6056],[45,3.6056],[51,3.6],[66,3.6056],[107,3.6015],[246,3.6056],[251,3.8292],[257,4.4998],[262,4.8351],[263,4.8351],[264,4.875],[268,4.875],[269,4.9446],[270,4.8713],[271,4.6434],[272,5.4212],[273,6.6188],[274,8.4205],[275,10.1471],[275.9987,12.6085],[276,2.2587],[277,2.6146],[278,2.9648],[279,3.1166],[280,3.2076],[281,3.2133],[282,3.2362],[283,3.2677],[284,3.4077],[285,3.5185],[286,3.5989],[287,3.6879],[288,3.7592],[289,3.796],[290,3.8197],[292,3.8339],[351,3.8342],[352,6.9246],[353,7.9976],[354,10.2126],[354.9987,13.8438],[355,10.4378],[356,8.5464],[357,6.6872],[358,4.9158],[359,3.929],[360,3.6822],[361,3.6206],[364,3.6],[439,3.6]]
       }
       hand_l_ik_ctrl.Rotation.X {
        type=MovieSceneScriptingFloatChannel
        default=39.7174
        thinned_from=117
        keys [[0,18.7273],[45,18.7273],[51,18.7],[66,18.7273],[107,18.7074],[246,18.7273],[251,19.0599],[257,20.0576],[262,20.5565],[268,20.5565],[269,19.9696],[270,24.7957],[271,30.6199],[272,34.973],[273,42.3069],[275,58.2],[275.9987,68.5387],[276,16.1358],[278,13.0709],[279,12.4123],[280,12.0171],[281,11.9032],[282,11.449],[283,10.8316],[284,9.8879],[285,8.55],[286,6.9257],[287,4.7491],[288,2.8188],[289,1.7491],[290,1.0199],[291,0.677],[292,0.5627],[327,0.5627],[350,0.5569],[351,3.4729],[352,5.5439],[353,22.7778],[354,37.0887],[354.9987,73.0183],[355,106.3238],[356,81.1677],[357,58.5186],[358,31.5429],[359,25.6611],[360,19.4802],[361,18.2751],[362,18.4863],[363,18.6199],[364,18.7],[439,18.7]]
       }
       hand_l_ik_ctrl.Rotation.Y {
        type=MovieSceneScriptingFloatChannel
        default=26.5084
        thinned_from=117
        keys [[0,-38.6604],[51,-38.7],[246,-38.6604],[251,-39.4961],[257,-42.0033],[262,-43.2569],[268,-43.2569],[269,-41.5153],[270,-34.6313],[271,-25.8506],[272,-18.5054],[273,-10.9545],[274,-1.049],[275,8.973],[275.9987,7.6933],[276,9.6732],[278,12.193],[279,12.7345],[280,13.0594],[281,13.0955],[282,13.2397],[283,13.4357],[284,13.7352],[285,14.1598],[286,14.6754],[287,15.3662],[288,15.9788],[289,16.3183],[290,16.5498],[291,16.6586],[292,16.6949],[350,16.6967],[351,19.2471],[352,33.8462],[353,41.6027],[354,50.095],[354.9987,57.4846],[355,-1.9022],[356,-2.6059],[357,-15.6926],[358,-23.6217],[359,-28.4244],[360,-32.2467],[361,-34.5316],[362,-36.6036],[363,-37.9138],[364,-38.7],[439,-38.7]]
       }
       hand_l_ik_ctrl.Rotation.Z {
        type=MovieSceneScriptingFloatChannel
        default=53.1504
        thinned_from=117
        keys [[0,105.7776],[246,105.7776],[251,105.3136],[257,103.9216],[262,103.2256],[268,103.2256],[269,104.096],[270,107.1811],[272,117.1069],[273,125.2004],[275,148.8761],[275.9987,155.4498],[276,64.8606],[277,54.893],[278,45.0808],[279,40.8304],[280,38.2802],[281,37.4201],[282,33.9916],[283,29.3306],[284,22.2069],[285,12.107],[286,-0.1548],[287,-16.5858],[288,-31.1575],[289,-39.2323],[290,-44.7374],[291,-47.326],[292,-48.1888],[350,-48.232],[351,-40.9381],[352,-26.1596],[353,5.913],[354,25.8846],[354.9987,74.1172],[355,139.0441],[356,127.8447],[357,119.8519],[359,111.4216],[360,108.8747],[361,107.628],[362,106.7193],[363,106.1448],[364,105.8],[439,105.8]]
       }
       hand_l_ik_ctrl.Scale.X {type=MovieSceneScriptingFloatChannel, default=1, thinned_from=116, keys [[0,1],[439,1]]}
       hand_l_ik_ctrl.Scale.Y {type=MovieSceneScriptingFloatChannel, default=1, thinned_from=116, keys [[0,1],[439,1]]}
       hand_l_ik_ctrl.Scale.Z {type=MovieSceneScriptingFloatChannel, default=1, thinned_from=116, keys [[0,1],[439,1]]}
       Softness {
        type=MovieSceneScriptingFloatChannel
        default=0
        thinned_from=115
        keys [[0,0],[287,0],[288,0.2107],[289,0.3037],[290,0.3483],[291,0.3683],[292,0.375],[337,0.375],[338,0.3729],[339,0.3669],[340,0.3577],[341,0.3457],[342,0.3315],[343,0.3157],[345,0.2812],[346,0.2482],[347,0.1927],[348,0.1274],[349,0.0651],[350,0.0184],[354,0.0184],[355,0],[439,0]]
       }
       hand_l_ikorient_ctrl.Rotation.Roll {
        type=MovieSceneScriptingFloatChannel
        default=0
        thinned_from=64
        keys [[0,0],[267,0],[271,19.1435],[275,38.0644],[279,14.2741],[283,3.5685],[287,0.8921],[292,0.223],[299,0.0558],[309,0.0171],[327,0.0043],[337,0],[439,0]]
       }
       hand_l_ikorient_ctrl.Rotation.Pitch {
        type=MovieSceneScriptingFloatChannel
        default=0
        thinned_from=64
        keys [[0,0],[267,0],[271,-3.2884],[275,-6.5385],[279,-2.452],[283,-0.613],[287,-0.1532],[292,-0.0383],[299,-0.0096],[309,-0.0029],[327,-0.0007],[337,0],[439,0]]
       }
       hand_l_ikorient_ctrl.Rotation.Yaw {
        type=MovieSceneScriptingFloatChannel
        default=0
        thinned_from=64
        keys [[0,0],[267,0],[271,53.8779],[275,107.1294],[279,40.1735],[283,10.0434],[287,2.5108],[292,0.6277],[299,0.1569],[309,0.0482],[327,0.012],[337,0],[439,0]]
       }
       arm_l_pv_ik_ctrl.Location.X {
        type=MovieSceneScriptingFloatChannel
        default=0.5
        thinned_from=115
        keys [[0,0.5],[262,0.5],[263,1.2138],[264,1.5815],[265,1.2487],[266,0.9273],[267,0.4195],[268,-0.3527],[269,-0.9067],[270,0.5654],[271,3.0446],[272,7.6976],[273,15.6117],[274,15.6117],[275,-6.05],[337,-6.05],[338,-6.0132],[339,-5.9093],[340,-5.7478],[341,-5.5383],[342,-5.2904],[343,-5.0138],[344,-4.7179],[345,-4.4125],[346,-3.8354],[347,-2.866],[348,-1.726],[349,-0.6372],[350,0.1788],[351,0.5],[357,0.5],[358,-1.3609],[359,-3.9369],[360,-5],[361,-4],[363,0],[364,0.5],[439,0.5]]
       }
       arm_l_pv_ik_ctrl.Location.Y {
        type=MovieSceneScriptingFloatChannel
        default=10
        thinned_from=115
        keys [[0,10],[274,10],[275,24.6234],[276,25.6271],[278,26.3487],[279,26.5038],[280,26.5968],[281,26.9358],[282,27.7927],[283,28.7231],[284,29.8254],[285,31.3392],[286,33.1561],[287,35.5836],[288,37.7334],[289,38.9237],[290,39.7346],[291,40.1157],[292,40.2427],[337,40.2487],[339,40.0865],[341,39.6591],[345,38.362],[346,37.8691],[347,37.1305],[348,36.1953],[349,35.1128],[351,32.7021],[352,30.5086],[353,26.9995],[354,23.0668],[355,10],[439,10]]
       }
       arm_l_pv_ik_ctrl.Location.Z {
        type=MovieSceneScriptingFloatChannel
        default=0
        thinned_from=115
        keys [[0,0],[271,0],[272,0.5647],[273,0.5647],[274,4.4563],[275,18.6488],[276,19.9312],[278,23.6036],[279,24.3927],[280,24.8662],[281,24.8847],[282,25.142],[283,25.5805],[284,26.3689],[285,27.5058],[286,28.8942],[287,30.7577],[288,32.4114],[289,33.3282],[290,33.9534],[291,34.2475],[292,34.3455],[337,34.3506],[339,34.2122],[341,33.8474],[345,32.7404],[346,32.3197],[347,31.6893],[348,30.8912],[349,29.9673],[351,27.9099],[352,25.4817],[353,21.1205],[354,16.0412],[355,9.5889],[356,1.179],[357,-0.6924],[358,-0.2597],[359,-0.0974],[360,-0.0243],[361,-0.0061],[362,-0.0031],[363,-0.0011],[364,0],[439,0]]
       }
       arm_l_pv_ik_ctrl.Rotation.X {
        type=MovieSceneScriptingFloatChannel
        default=0
        thinned_from=115
        keys [[0,0],[274,0],[275,22.0679],[276,22.778],[278,23.4192],[279,23.557],[280,23.6396],[281,23.941],[282,24.7027],[283,25.5297],[284,26.5094],[285,27.855],[286,29.4699],[287,31.6275],[288,33.5383],[289,34.5963],[290,35.3171],[291,35.6558],[292,35.7687],[337,35.774],[339,35.6299],[341,35.25],[345,34.0971],[346,33.659],[347,33.0025],[348,32.1713],[349,31.2091],[351,29.0664],[352,26.5376],[353,21.9957],[354,16.7059],[355,0],[439,0]]
       }
       arm_l_pv_ik_ctrl.Rotation.Y {
        type=MovieSceneScriptingFloatChannel
        default=0
        thinned_from=115
        keys [[0,0],[274,0],[275,26.9719],[276,27.8397],[278,28.6234],[279,28.7918],[280,28.8928],[281,29.2612],[282,30.1921],[283,31.2028],[284,32.4003],[285,34.0449],[286,36.0187],[287,38.6557],[288,40.9912],[289,42.2842],[290,43.1652],[291,43.5792],[292,43.7172],[337,43.7237],[339,43.5475],[341,43.0832],[345,41.6741],[346,41.1386],[347,40.3362],[348,39.3203],[349,38.1443],[351,35.5255],[352,32.4348],[353,26.8836],[354,20.4183],[355,0],[439,0]]
       }
       arm_l_pv_ik_ctrl.Rotation.Z {
        type=MovieSceneScriptingFloatChannel
        default=0
        thinned_from=115
        keys [[0,0],[274,0],[275,53.4325],[276,55.1517],[278,56.7043],[279,57.0379],[280,57.238],[281,57.9678],[282,59.812],[283,61.8143],[284,64.1866],[285,67.4445],[286,71.3548],[287,76.5789],[288,81.2055],[289,83.7671],[290,85.5124],[291,86.3325],[292,86.6059],[337,86.6187],[339,86.2697],[341,85.3498],[345,82.5584],[346,81.4975],[347,79.908],[348,77.8955],[349,75.5658],[351,70.3777],[352,64.2548],[353,53.2576],[354,40.4496],[355,0],[439,0]]
       }
       arm_l_pv_ik_ctrl.Scale.X {type=MovieSceneScriptingFloatChannel, default=1, thinned_from=115, keys [[0,1],[439,1]]}
       arm_l_pv_ik_ctrl.Scale.Y {type=MovieSceneScriptingFloatChannel, default=1, thinned_from=115, keys [[0,1],[439,1]]}
       arm_l_pv_ik_ctrl.Scale.Z {type=MovieSceneScriptingFloatChannel, default=1, thinned_from=115, keys [[0,1],[439,1]]}
       upperarm_r_fk_ctrl.Location.X {
        type=MovieSceneScriptingFloatChannel
        default=0
        thinned_from=115
        keys [[0,0],[274,0],[275,0.0012],[276,0.0039],[277,0.0334],[278,0.0624],[279,0.075],[280,0.0825],[281,0.0851],[282,0.0955],[283,0.1097],[284,0.1326],[285,0.1652],[286,0.2047],[287,0.2577],[288,0.3047],[289,0.3308],[290,0.3486],[291,0.3569],[292,0.3597],[337,0.3598],[338,0.3577],[339,0.352],[340,0.3432],[341,0.3316],[342,0.318],[343,0.3028],[345,0.2698],[346,0.2381],[347,0.1849],[348,0.1223],[349,0.0706],[350,0.0176],[351,0.011],[352,0.0054],[354,0.0054],[356,0.0029],[358,0.001],[360,0.0002],[439,0]]
       }
       upperarm_r_fk_ctrl.Location.Y {
        type=MovieSceneScriptingFloatChannel
        default=0
        thinned_from=115
        keys [[0,0],[274,0],[276,0.0004],[278,0.0062],[279,0.0075],[280,0.0082],[281,0.0085],[282,0.0095],[283,0.011],[284,0.0133],[285,0.0165],[286,0.0205],[287,0.0258],[288,0.0305],[289,0.0331],[290,0.0348],[291,0.0357],[292,0.0359],[337,0.036],[338,0.0358],[339,0.0352],[340,0.0343],[341,0.0331],[343,0.0303],[345,0.027],[346,0.0238],[347,0.0185],[348,0.0122],[350,0.0018],[352,0.0005],[387,0],[439,0]]
       }
       upperarm_r_fk_ctrl.Location.Z {
        type=MovieSceneScriptingFloatChannel
        default=0
        thinned_from=115
        keys [[0,0],[274,0],[275,0.0031],[276,0.01],[277,0.085],[278,0.1588],[279,0.1908],[280,0.21],[281,0.2166],[282,0.2429],[283,0.2791],[284,0.3374],[285,0.4202],[286,0.5208],[287,0.6557],[288,0.7753],[289,0.8416],[290,0.8868],[291,0.9081],[292,0.9151],[337,0.9152],[338,0.9101],[339,0.8956],[340,0.873],[341,0.8437],[342,0.8091],[343,0.7704],[345,0.6864],[346,0.6058],[347,0.4703],[348,0.311],[349,0.1795],[350,0.0449],[351,0.0281],[352,0.0138],[354,0.0138],[355,0.0104],[356,0.0073],[358,0.0024],[359,0.0012],[360,0.0005],[362,0.0001],[439,0]]
       }
       upperarm_r_fk_ctrl.Rotation.X {
        type=MovieSceneScriptingFloatChannel
        default=5.993
        thinned_from=115
        keys [[0,5.993],[8,8.6897],[15,16.7799],[22,35.0577],[30,33.3662],[45,35.1495],[51,6],[56,34.6559],[62,34.3472],[66,32.8265],[70,32.2628],[83,31.7907],[94,31.6986],[101,31.7207],[107,31.8753],[111,32.3485],[115,33.4438],[121,33.6186],[130,33.6769],[138,33.8662],[143,34.4342],[148,36.706],[153,18.1627],[160,11.2089],[172,31.7045],[187,38.5364],[194,36.5676],[199,30.6614],[204,7.0367],[215,7.0367],[223,12.4527],[233,28.7009],[241,34.0845],[246,35.879],[251,36.088],[257,36.7151],[266,38.9713],[267,39.3178],[268,39.7734],[269,40.7038],[272,45.0631],[273,46.4454],[274,47.4902],[276,47.4189],[278,48.6407],[279,48.9032],[280,49.0607],[281,48.375],[283,48.6873],[284,49.3511],[285,50.3318],[286,51.5393],[287,53.1633],[288,54.6058],[289,55.406],[290,55.952],[291,56.209],[292,56.2946],[338,56.3283],[340,56.6223],[345,58.1014],[346,58.7086],[347,59.7023],[348,60.8608],[350,62.7863],[351,62.6263],[352,60.5],[353,53],[354,52.9144],[355,50.1597],[356,47.5492],[358,43.3964],[359,42.3764],[360,41.7161],[361,41.4685],[368,41.32],[369,40.4854],[370,39.4719],[371,38.3691],[375,34.7716],[395,34.749],[409,36.37],[439,34.7487]]
       }
       upperarm_r_fk_ctrl.Rotation.Y {
        type=MovieSceneScriptingFloatChannel
        default=0.3823
        thinned_from=115
        keys [[0,0.3823],[8,0.2928],[15,0.0245],[22,-0.3145],[30,-0.264],[45,-0.236],[51,0.4],[56,-0.708],[59,-0.9111],[62,-1.1191],[66,-3.3186],[70,-4.2022],[75,-4.4635],[83,-4.78],[94,-4.8831],[101,-4.8558],[107,-4.6646],[111,-4.0795],[115,-2.7224],[121,-2.4738],[130,-2.391],[138,-2.2244],[143,-1.7247],[148,0.2742],[153,-14.635],[160,-20.226],[172,-3.3638],[187,2.2569],[194,0.6421],[199,-4.2025],[204,-23.5806],[215,-23.5806],[223,-19.1656],[233,-5.9206],[241,-1.535],[246,-0.0731],[251,-0.2292],[257,-0.6974],[262,-1.6284],[263,-1.8209],[265,-2.1915],[266,-2.382],[267,-2.6781],[268,-3.1002],[269,-4.0358],[270,-4.9017],[271,-5.1813],[272,-5.3181],[274,-5.3416],[275,-4.5617],[276,-4.8799],[277,-4.1658],[278,-3.4628],[279,-3.1582],[280,-2.9755],[281,-3.0869],[282,-3.6039],[283,-4.4088],[284,-6.3786],[285,-9.2926],[286,-12.8821],[287,-17.71],[288,-21.999],[289,-24.3782],[290,-26.0016],[291,-26.7655],[292,-27.0201],[337,-27.0114],[338,-26.8822],[339,-26.5195],[340,-25.9567],[341,-25.2271],[342,-24.3641],[343,-23.401],[345,-21.3081],[346,-19.957],[347,-18.1077],[348,-15.8597],[349,-13.2441],[350,-10.5666],[351,-8.5186],[352,-5.1379],[353,-3],[354,-3.01],[355,5.2536],[356,13.1027],[358,25.5888],[359,28.6556],[360,30.6408],[361,31.3852],[362,31.6686],[363,31.7911],[368,31.8319],[369,27.4651],[370,22.1626],[371,16.3921],[375,-2.4303],[378,-1.2688],[382,-0.8332],[387,-0.6372],[395,-0.5964],[409,0.4057],[439,-0.5719]]
       }
       upperarm_r_fk_ctrl.Rotation.Z {
        type=MovieSceneScriptingFloatChannel
        default=-29.3057
        thinned_from=115
        keys [[0,-29.3057],[8,-29.8789],[15,-31.5984],[22,-35.2689],[30,-34.9136],[45,-35.2225],[51,-29.3],[56,-35.5745],[62,-36.0422],[66,-38.7662],[70,-39.8631],[75,-40.1736],[83,-40.5141],[94,-40.625],[101,-40.6044],[107,-40.4607],[111,-40.021],[115,-39.1649],[121,-39.0115],[138,-38.8684],[143,-38.5926],[148,-37.4891],[153,-25.4106],[160,-20.8812],[172,-31.0099],[187,-34.3861],[194,-33.3722],[199,-30.3305],[204,-18.1636],[215,-18.1636],[223,-21.8658],[233,-32.9727],[241,-36.6961],[246,-37.9373],[251,-38.3],[257,-39.3881],[262,-41.5515],[265,-43.3008],[266,-43.5977],[267,-42.4737],[268,-40.0008],[269,-31.6696],[270,-18.4447],[271,-6.446],[272,4.9694],[273,15.7737],[274,23.318],[275,27.3216],[276,31.8481],[277,38.3439],[278,44.7384],[279,47.5084],[280,49.1703],[281,49.6798],[282,51.616],[283,54.1271],[284,57.1355],[285,61.247],[286,66.1729],[287,72.7507],[288,78.575],[289,81.7993],[290,83.9957],[291,85.0279],[292,85.3719],[337,85.3942],[339,85.0462],[341,84.1288],[345,81.345],[346,80.3653],[347,78.9965],[348,77.3163],[350,73.3335],[351,70.6241],[352,64.2983],[353,55],[354,54.9868],[355,30.0143],[356,6.3996],[358,-31.1657],[359,-40.3923],[360,-46.3649],[361,-48.6046],[362,-49.4572],[363,-49.8256],[364,-49.9484],[368,-49.9484],[369,-48.3712],[370,-46.4561],[371,-44.372],[375,-37.5739],[378,-36.2129],[382,-35.7025],[387,-35.4728],[409,-35.3415],[439,-35.3963]]
       }
       upperarm_r_fk_ctrl.Scale.X {type=MovieSceneScriptingFloatChannel, default=1, thinned_from=115, keys [[0,1],[439,1]]}
       upperarm_r_fk_ctrl.Scale.Y {type=MovieSceneScriptingFloatChannel, default=1, thinned_from=115, keys [[0,1],[439,1]]}
       upperarm_r_fk_ctrl.Scale.Z {type=MovieSceneScriptingFloatChannel, default=1, thinned_from=115, keys [[0,1],[439,1]]}
       lowerarm_r_fk_ctrl.Location.X {
        type=MovieSceneScriptingFloatChannel
        default=0
        thinned_from=115
        keys [[0,0],[66,0],[70,-0.0025],[75,-0.0029],[94,-0.0031],[101,-0.003],[107,-0.0022],[111,0],[439,0]]
       }
       lowerarm_r_fk_ctrl.Location.Y {
        type=MovieSceneScriptingFloatChannel
        default=0
        thinned_from=115
        keys [[0,0],[66,0],[70,0.0013],[75,0.0015],[101,0.0015],[107,0.0012],[111,0],[439,0]]
       }
       lowerarm_r_fk_ctrl.Location.Z {type=MovieSceneScriptingFloatChannel, default=0, thinned_from=115, keys [[0,0],[439,0]]}
       lowerarm_r_fk_ctrl.Rotation.X {
        type=MovieSceneScriptingFloatChannel
        default=0.4185
        thinned_from=115
        keys [[0,0.4185],[8,0.4224],[15,0.434],[22,0.4598],[30,0.4573],[45,0.4598],[51,0.4],[56,0.4598],[251,0.4592],[257,0.4583],[266,0.4551],[267,0.4486],[268,0.4347],[269,0.3893],[270,0.3094],[271,0.2259],[272,0.1431],[273,0.0642],[274,0.0194],[275,0.0061],[276,0],[337,0],[338,0.0015],[339,0.0059],[340,0.0127],[341,0.0215],[342,0.0319],[343,0.0435],[344,0.056],[345,0.0688],[346,0.0877],[347,0.1168],[348,0.153],[349,0.1936],[350,0.2351],[351,0.2644],[352,0.3073],[353,0.3073],[354,0.3666],[355,0.3866],[356,0.4051],[358,0.4346],[359,0.4419],[360,0.4466],[361,0.4483],[368,0.4494],[375,0.4582],[382,0.4595],[439,0.4598]]
       }
       lowerarm_r_fk_ctrl.Rotation.Y {
        type=MovieSceneScriptingFloatChannel
        default=0.1904
        thinned_from=115
        keys [[0,0.1904],[8,0.1729],[15,0.1204],[22,0.0037],[30,0.0146],[45,0.0036],[51,0.2],[56,0.0033],[62,0.0021],[66,-0.0046],[70,-0.008],[75,-0.0088],[83,-0.0094],[101,-0.0095],[107,-0.0092],[111,-0.0082],[115,-0.007],[121,-0.0067],[130,-0.0065],[138,-0.0067],[143,-0.0072],[148,-0.0093],[153,0.0077],[160,0.014],[172,-0.0005],[187,-0.0053],[194,-0.0039],[199,0.0005],[204,0.0178],[215,0.0178],[223,0.0113],[233,-0.0084],[241,-0.015],[246,-0.0172],[251,-0.0203],[257,-0.0298],[262,-0.0486],[263,-0.0535],[265,-0.0639],[266,-0.0664],[267,-0.0659],[268,-0.0647],[269,-0.0582],[270,-0.0461],[272,-0.0213],[273,-0.0096],[274,-0.0029],[275,-0.0009],[276,0],[337,0],[338,0.0021],[339,0.0079],[340,0.017],[341,0.0288],[342,0.0427],[343,0.0583],[344,0.0749],[345,0.0921],[346,0.1245],[347,0.179],[348,0.2431],[349,0.296],[350,0.3502],[351,0.3533],[352,0.3425],[353,0.3425],[354,0.2775],[355,0.1889],[356,0.1046],[358,-0.0294],[359,-0.0623],[360,-0.0836],[361,-0.0916],[362,-0.0946],[363,-0.0959],[364,-0.0964],[368,-0.0964],[369,-0.0884],[370,-0.0794],[371,-0.0707],[375,-0.0382],[378,-0.012],[382,-0.0021],[387,0.0023],[395,0.0032],[409,-0.0027],[439,0.0038]]
       }
       lowerarm_r_fk_ctrl.Rotation.Z {
        type=MovieSceneScriptingFloatChannel
        default=52.6578
        thinned_from=115
        keys [[0,52.6578],[8,54.9086],[15,61.661],[22,76.6688],[30,75.262],[45,76.6696],[51,52.7],[56,76.7138],[62,76.8562],[66,77.7019],[70,78.1216],[107,78.2674],[115,77.9912],[143,78.0247],[148,78.2785],[153,76.1689],[160,75.3779],[172,77.1857],[187,77.7883],[194,77.608],[199,77.067],[204,74.9032],[215,74.9032],[223,75.7215],[233,78.1763],[241,78.993],[251,79.6605],[257,80.8461],[262,83.2035],[265,85.111],[266,85.4336],[267,85.1005],[268,84.3675],[269,81.6046],[270,76.6474],[273,61.3304],[274,58.7984],[276,58.915],[278,59.8657],[280,60.1925],[281,59.1849],[282,55.2499],[283,50.1315],[284,43.2593],[285,33.6758],[286,22.1092],[287,6.6337],[288,-7.081],[289,-14.6777],[290,-19.8549],[291,-22.2886],[292,-23.0998],[327,-23.0998],[337,-23.134],[338,-22.8699],[339,-22.1235],[340,-20.9637],[341,-19.4595],[342,-17.6797],[343,-15.6931],[344,-13.5688],[345,-11.3756],[346,-8.0163],[347,-2.7814],[348,3.6782],[349,10.592],[350,17.6694],[351,25],[353,40],[354,39.9981],[355,51.6727],[356,62.766],[358,80.4126],[359,84.7469],[360,87.5525],[361,88.6046],[362,89.0052],[368,89.2359],[371,85.9898],[375,81.8867],[378,78.6167],[382,77.3904],[387,76.8386],[395,76.7236],[409,77.4667],[439,76.6546]]
       }
       lowerarm_r_fk_ctrl.Scale.X {type=MovieSceneScriptingFloatChannel, default=1, thinned_from=115, keys [[0,1],[439,1]]}
       lowerarm_r_fk_ctrl.Scale.Y {type=MovieSceneScriptingFloatChannel, default=1, thinned_from=115, keys [[0,1],[439,1]]}
       lowerarm_r_fk_ctrl.Scale.Z {type=MovieSceneScriptingFloatChannel, default=1, thinned_from=115, keys [[0,1],[439,1]]}
       hand_r_fk_ctrl.Location.X {
        type=MovieSceneScriptingFloatChannel
        default=0
        thinned_from=115
        keys [[0,0],[66,0],[70,-0.0024],[75,-0.0028],[94,-0.003],[101,-0.0029],[107,-0.0022],[111,0],[439,0]]
       }
       hand_r_fk_ctrl.Location.Y {
        type=MovieSceneScriptingFloatChannel
        default=-0.1788
        thinned_from=115
        keys [[0,-0.1788],[45,-0.1788],[51,-0.2],[56,-0.1788],[274,-0.1788],[275,-0.1782],[276,-0.1768],[278,-0.1466],[279,-0.1401],[280,-0.1362],[281,-0.1346],[282,-0.1282],[283,-0.1192],[284,-0.1074],[285,-0.0916],[286,-0.0728],[287,-0.0479],[288,-0.0258],[289,-0.0136],[290,-0.0053],[291,-0.0014],[292,-0.0001],[337,0],[338,-0.001],[339,-0.0038],[340,-0.0082],[341,-0.014],[342,-0.0207],[343,-0.0283],[344,-0.0364],[345,-0.0447],[346,-0.0604],[347,-0.0869],[348,-0.118],[349,-0.1437],[350,-0.17],[351,-0.17],[352,-0.1745],[353,-0.1745],[354,-0.1788],[439,-0.1788]]
       }
       hand_r_fk_ctrl.Location.Z {
        type=MovieSceneScriptingFloatChannel
        default=0.0249
        thinned_from=115
        keys [[0,0.0249],[45,0.0249],[51,0],[56,0.0249],[275,0.0248],[276,0.0246],[278,0.0204],[279,0.0195],[280,0.019],[281,0.0187],[282,0.0178],[283,0.0166],[284,0.0149],[285,0.0127],[286,0.0101],[287,0.0067],[288,0.0036],[289,0.0019],[290,0.0007],[291,0.0002],[292,0],[337,0],[338,0.0001],[339,0.0005],[341,0.0019],[345,0.0062],[346,0.0084],[347,0.0121],[348,0.0164],[350,0.0237],[351,0.0237],[352,0.0243],[353,0.0243],[354,0.0249],[439,0.0249]]
       }
       hand_r_fk_ctrl.Rotation.X {
        type=MovieSceneScriptingFloatChannel
        default=64.892
        thinned_from=115
        keys [[0,64.892],[8,59.7417],[15,44.2909],[22,9.7359],[30,12.9594],[45,9.6688],[51,64.9],[56,10.1052],[62,10.5556],[66,12.9895],[70,13.9934],[75,14.2803],[83,14.6152],[94,14.7243],[101,14.6902],[107,14.4517],[111,13.7219],[115,12.012],[121,11.6629],[130,11.5465],[138,11.4161],[143,11.025],[148,9.4605],[153,9.0549],[160,8.9028],[172,7.3386],[187,6.8172],[194,6.9418],[199,7.3158],[204,8.8116],[215,8.8116],[223,9.1252],[233,10.0661],[241,10.4046],[246,10.5174],[251,11.0424],[257,12.6174],[262,15.749],[263,16.5651],[265,18.2859],[266,18.7126],[267,18.4165],[268,17.7649],[269,15.1691],[270,10.4774],[271,5.6177],[272,0.8098],[273,-3.9807],[274,-6.3715],[275,-5.1796],[276,-2.5574],[277,11.642],[278,25.6199],[279,31.6747],[280,35.3076],[281,35.734],[282,37.4781],[283,39.9644],[284,43.4846],[285,48.2038],[286,53.817],[287,61.2984],[288,67.9169],[289,71.5788],[290,74.0722],[291,75.2435],[292,75.6339],[337,75.6552],[338,75.4859],[339,75.0074],[340,74.2638],[341,73.2993],[342,72.1582],[343,70],[344,67.5],[345,62.5],[346,53.75],[347,42.5],[348,35],[350,22.5],[352,12.5],[353,8.75],[354,8.7367],[355,22.8743],[356,35.7741],[358,56.2947],[359,61.3348],[360,64.5974],[361,65.8209],[362,66.2867],[363,66.4879],[368,66.555],[369,59.8779],[370,51.77],[371,42.9467],[375,14.1661],[378,11.5261],[382,10.5361],[387,10.0906],[395,9.9978],[409,9.6228],[439,9.9421]]
       }
       hand_r_fk_ctrl.Rotation.Y {
        type=MovieSceneScriptingFloatChannel
        default=27.5152
        thinned_from=115
        keys [[0,27.5152],[8,23.6403],[15,12.0154],[22,-13.8449],[30,-11.4227],[45,-13.8535],[51,27.5],[56,-13.774],[62,-13.6416],[66,-12.8285],[70,-12.5958],[83,-12.4124],[94,-12.3771],[107,-12.3663],[111,-12.3375],[115,-12.266],[121,-12.2338],[130,-12.2231],[138,-12.2508],[143,-12.334],[148,-12.6666],[153,-23.7371],[160,-27.8885],[172,-17.5701],[187,-14.1307],[194,-15.1462],[199,-18.1928],[204,-30.3793],[215,-30.3793],[223,-27.2293],[233,-17.7791],[241,-14.6217],[246,-13.5693],[251,-13.109],[257,-11.728],[262,-8.9822],[263,-8.4155],[265,-7.3209],[266,-6.7598],[267,-5.6692],[268,-3.9455],[269,0.356],[270,6.3778],[271,11.3923],[272,15.8757],[273,20.1916],[275,29.0512],[276,33.3348],[278,42.0209],[279,43.8874],[280,45.0073],[281,45.1646],[282,45.8039],[283,46.706],[284,47.9796],[285,49.6869],[286,51.7175],[287,54.4239],[288,56.8181],[289,58.1428],[290,59.0448],[291,59.4685],[292,59.6097],[337,59.6174],[338,59.4748],[339,59.0718],[340,58.4455],[341,57.6332],[343,55.5994],[345,53.268],[346,51.5031],[347,48.7842],[348,45.4039],[349,41.6609],[350,37.8292],[351,37.8292],[352,32.5292],[354,32.5298],[355,24.8288],[356,17.636],[358,6.194],[359,3.3837],[360,1.5646],[361,0.8824],[362,0.6226],[363,0.5104],[364,0.473],[368,0.473],[369,-1.4821],[370,-3.8562],[371,-6.4398],[375,-14.8671],[378,-14.2179],[382,-13.9745],[387,-13.8649],[395,-13.8421],[409,-14.1947],[439,-13.8284]]
       }
       hand_r_fk_ctrl.Rotation.Z {
        type=MovieSceneScriptingFloatChannel
        default=-62.7236
        thinned_from=115
        keys [[0,-62.7236],[8,-58.0403],[15,-43.9904],[22,-13.0807],[30,-16.0013],[45,-13.1762],[51,-62.7],[56,-12.6478],[62,-12.2795],[66,-10.2417],[70,-9.4122],[75,-9.1489],[83,-8.7959],[94,-8.681],[101,-8.7051],[107,-8.8742],[111,-9.3914],[115,-10.3344],[121,-10.4598],[130,-10.5016],[138,-10.7086],[143,-11.3296],[148,-13.8137],[153,10.9053],[160,20.175],[172,-6.83],[187,-15.8316],[194,-13.2336],[199,-5.4395],[204,25.7368],[215,25.7368],[223,18.3923],[233,-3.641],[241,-10.9497],[246,-13.3859],[251,-13.6518],[257,-14.4495],[262,-16.0355],[265,-17.3051],[266,-17.5357],[268,-17.6049],[269,-17.5337],[275,-16.9703],[276,-16.8275],[278,-14.8455],[279,-14.4196],[280,-14.164],[281,-14.0297],[282,-13.4885],[283,-12.7452],[284,-11.7973],[285,-10.5435],[286,-9.0598],[287,-7.0851],[288,-5.3392],[289,-4.3737],[290,-3.7164],[291,-3.4077],[292,-3.3048],[327,-3.3048],[337,-3.2993],[338,-3.2717],[339,-3.1951],[340,-3.0766],[341,-2.9231],[342,-2.7417],[343,-2.5393],[344,-2.3229],[345,-2.0995],[346,-1.6769],[347,-0.9667],[348,-0.1313],[349,0.5585],[350,1.2646],[351,1.2646],[352,-4.714],[354,-4.7127],[355,-32.5767],[356,-58.913],[358,-100.8076],[359,-111.0975],[360,-117.7584],[361,-120.2562],[362,-121.2071],[363,-121.6179],[368,-121.7549],[369,-107.9715],[370,-91.2345],[371,-73.0207],[375,-13.6094],[378,-13.0782],[382,-12.8791],[387,-12.7895],[395,-12.7708],[409,-14.4618],[439,-12.7596]]
       }
       hand_r_fk_ctrl.Scale.X {type=MovieSceneScriptingFloatChannel, default=1, thinned_from=115, keys [[0,1],[439,1]]}
       hand_r_fk_ctrl.Scale.Y {type=MovieSceneScriptingFloatChannel, default=1, thinned_from=115, keys [[0,1],[439,1]]}
       hand_r_fk_ctrl.Scale.Z {type=MovieSceneScriptingFloatChannel, default=1, thinned_from=115, keys [[0,1],[439,1]]}
       thumb_01_r_ctrl.Location.X {type=MovieSceneScriptingFloatChannel, default=0, thinned_from=71, keys [[0,0],[439,0]]}
       thumb_01_r_ctrl.Location.Y {type=MovieSceneScriptingFloatChannel, default=0, thinned_from=71, keys [[0,0],[439,0]]}
       thumb_01_r_ctrl.Location.Z {type=MovieSceneScriptingFloatChannel, default=0, thinned_from=71, keys [[0,0],[439,0]]}
       thumb_01_r_ctrl.Rotation.X {
        type=MovieSceneScriptingFloatChannel
        default=-36.8112
        thinned_from=71
        keys [[0,-36.8112],[267,-36.8112],[271,-33.5494],[273,-30.0572],[275,-26.4822],[279,-21.7674],[283,-19.9211],[287,-19.4595],[292,-19.3441],[351,-19.3057],[353,-21.8394],[355,-23.6593],[357,-24.2518],[359,-24.2518],[361,-29.819],[362,-33.1046],[363,-35.7152],[365,-36.5854],[366,-36.8112],[439,-36.8112]]
       }
       thumb_01_r_ctrl.Rotation.Y {
        type=MovieSceneScriptingFloatChannel
        default=22.85
        thinned_from=71
        keys [[0,22.85],[45,22.85],[51,22.9],[66,22.85],[107,22.8864],[275,22.8635],[279,24.0312],[283,25.6202],[287,26.0175],[292,26.1168],[351,26.1499],[353,26.6562],[355,26.8144],[359,26.8659],[361,25.1869],[362,24.196],[363,23.2598],[365,22.9477],[366,22.85],[439,22.85]]
       }
       thumb_01_r_ctrl.Rotation.Z {
        type=MovieSceneScriptingFloatChannel
        default=10.3087
        thinned_from=71
        keys [[0,10.3087],[267,10.3087],[271,9.0754],[273,7.7549],[275,6.4032],[279,4.6205],[283,3.9224],[287,3.7479],[292,3.7043],[299,3.6934],[351,3.6897],[353,9.2436],[355,12.2731],[357,13.2594],[359,13.2594],[361,12.4238],[362,11.9306],[363,10.8432],[366,10.3087],[439,10.3087]]
       }
       thumb_01_r_ctrl.Scale.X {type=MovieSceneScriptingFloatChannel, default=1, thinned_from=71, keys [[0,1],[439,1]]}
       thumb_01_r_ctrl.Scale.Y {type=MovieSceneScriptingFloatChannel, default=1, thinned_from=71, keys [[0,1],[439,1]]}
       thumb_01_r_ctrl.Scale.Z {type=MovieSceneScriptingFloatChannel, default=1, thinned_from=71, keys [[0,1],[439,1]]}
       thumb_02_r_ctrl.Location.X {type=MovieSceneScriptingFloatChannel, default=0, thinned_from=71, keys [[0,0],[439,0]]}
       thumb_02_r_ctrl.Location.Y {type=MovieSceneScriptingFloatChannel, default=0, thinned_from=71, keys [[0,0],[439,0]]}
       thumb_02_r_ctrl.Location.Z {type=MovieSceneScriptingFloatChannel, default=0, thinned_from=71, keys [[0,0],[439,0]]}
       thumb_02_r_ctrl.Rotation.X {type=MovieSceneScriptingFloatChannel, default=0, thinned_from=71, keys [[0,0],[439,0]]}
       thumb_02_r_ctrl.Rotation.Y {
        type=MovieSceneScriptingFloatChannel
        default=0
        thinned_from=71
        keys [[0,0],[351,0],[353,5.0625],[355,7.6416],[357,8.4813],[359,8.4813],[361,5.2421],[362,3.3304],[363,1.0452],[365,0.2835],[366,0],[439,0]]
       }
       thumb_02_r_ctrl.Rotation.Z {
        type=MovieSceneScriptingFloatChannel
        default=-8.742
        thinned_from=71
        keys [[0,-8.742],[45,-8.742],[51,-8.7],[56,-8.742],[66,-8.742],[70,-8.7105],[107,-8.7191],[111,-8.742],[257,-8.742],[267,-24],[271,-14.7467],[273,-4.8395],[275,5.3022],[279,23.6697],[283,27.6475],[287,28.642],[292,28.8906],[299,28.9527],[351,28.9734],[353,23.5146],[355,19.5936],[357,18.317],[359,18.317],[361,6.5],[362,-0.6898],[363,-6.3642],[365,-8.2556],[366,-8.742],[439,-8.742]]
       }
       thumb_02_r_ctrl.Scale.X {type=MovieSceneScriptingFloatChannel, default=1, thinned_from=71, keys [[0,1],[439,1]]}
       thumb_02_r_ctrl.Scale.Y {type=MovieSceneScriptingFloatChannel, default=1, thinned_from=71, keys [[0,1],[439,1]]}
       thumb_02_r_ctrl.Scale.Z {type=MovieSceneScriptingFloatChannel, default=1, thinned_from=71, keys [[0,1],[439,1]]}
       thumb_03_r_ctrl.Location.X {type=MovieSceneScriptingFloatChannel, default=0, thinned_from=71, keys [[0,0],[439,0]]}
       thumb_03_r_ctrl.Location.Y {type=MovieSceneScriptingFloatChannel, default=0, thinned_from=71, keys [[0,0],[439,0]]}
       thumb_03_r_ctrl.Location.Z {type=MovieSceneScriptingFloatChannel, default=0, thinned_from=71, keys [[0,0],[439,0]]}
       thumb_03_r_ctrl.Rotation.X {type=MovieSceneScriptingFloatChannel, default=0, thinned_from=71, keys [[0,0],[439,0]]}
       thumb_03_r_ctrl.Rotation.Y {type=MovieSceneScriptingFloatChannel, default=0, thinned_from=71, keys [[0,0],[439,0]]}
       thumb_03_r_ctrl.Rotation.Z {
        type=MovieSceneScriptingFloatChannel
        default=41.8526
        thinned_from=71
        keys [[0,41.8526],[45,41.8526],[51,41.9],[257,41.8526],[267,26.5945],[271,27.5619],[275,29.6577],[279,36.0478],[283,35.3354],[287,35.1573],[351,35.098],[353,36.0756],[355,36.7778],[357,37.0065],[359,37.0065],[362,40.4888],[363,41.4463],[365,41.7655],[366,41.8526],[439,41.8526]]
       }
       thumb_03_r_ctrl.Scale.X {type=MovieSceneScriptingFloatChannel, default=1, thinned_from=71, keys [[0,1],[439,1]]}
       thumb_03_r_ctrl.Scale.Y {type=MovieSceneScriptingFloatChannel, default=1, thinned_from=71, keys [[0,1],[439,1]]}
       thumb_03_r_ctrl.Scale.Z {type=MovieSceneScriptingFloatChannel, default=1, thinned_from=71, keys [[0,1],[439,1]]}
       index_metacarpal_r_ctrl.Location.X {type=MovieSceneScriptingFloatChannel, default=0, thinned_from=71, keys [[0,0],[439,0]]}
       index_metacarpal_r_ctrl.Location.Y {type=MovieSceneScriptingFloatChannel, default=0, thinned_from=71, keys [[0,0],[439,0]]}
       index_metacarpal_r_ctrl.Location.Z {type=MovieSceneScriptingFloatChannel, default=0, thinned_from=71, keys [[0,0],[439,0]]}
       index_metacarpal_r_ctrl.Rotation.X {
        type=MovieSceneScriptingFloatChannel
        default=0.0346
        thinned_from=71
        keys [[0,0.0346],[45,0.0346],[51,0],[56,0.0346],[66,0.0346],[70,0.0086],[75,0.0022],[83,0.0005],[94,0],[101,0.0012],[107,0.0094],[111,0.0346],[267,0.0346],[271,0.0653],[273,0.0982],[275,0.1318],[279,0.1762],[283,0.1935],[287,0.1979],[292,0.199],[351,0.1993],[353,0.1755],[355,0.1584],[357,0.1528],[359,0.1528],[361,0.1004],[362,0.0695],[363,0.0449],[365,0.0367],[366,0.0346],[439,0.0346]]
       }
       index_metacarpal_r_ctrl.Rotation.Y {
        type=MovieSceneScriptingFloatChannel
        default=2.3559
        thinned_from=72
        keys [[0,2.3559],[45,2.3559],[51,2.4],[56,2.3559],[66,2.3559],[70,2.389],[75,2.3973],[101,2.3985],[107,2.388],[111,2.3559],[267,2.3559],[271,-0.5418],[273,-0.027],[274,0.5],[275,-6.8201],[279,-11.0087],[283,-12.6489],[287,-13.0589],[292,-13.1614],[309,-13.193],[351,-13.1956],[353,-9.7473],[355,-7.5204],[357,-6.7954],[359,-6.7954],[361,-2.6158],[362,-0.1492],[363,1.6295],[365,2.2224],[366,2.3559],[439,2.3559]]
       }
       index_metacarpal_r_ctrl.Rotation.Z {
        type=MovieSceneScriptingFloatChannel
        default=-18.2
        thinned_from=71
        keys [[0,-18.2],[267,-18.2],[271,-13.7841],[273,-9.056],[275,-4.2161],[279,2.1672],[283,4.6668],[287,5.2917],[292,5.4479],[299,5.487],[351,5.5],[353,4.75],[355,3.6515],[357,3.2939],[359,3.2939],[361,-5.9582],[362,-11.4186],[363,-16.1629],[365,-17.7443],[366,-18.2],[439,-18.2]]
       }
       index_metacarpal_r_ctrl.Scale.X {type=MovieSceneScriptingFloatChannel, default=1, thinned_from=71, keys [[0,1],[439,1]]}
       index_metacarpal_r_ctrl.Scale.Y {type=MovieSceneScriptingFloatChannel, default=1, thinned_from=71, keys [[0,1],[439,1]]}
       index_metacarpal_r_ctrl.Scale.Z {type=MovieSceneScriptingFloatChannel, default=1, thinned_from=71, keys [[0,1],[439,1]]}
       index_01_r_ctrl.Location.X {type=MovieSceneScriptingFloatChannel, default=0, thinned_from=71, keys [[0,0],[439,0]]}
       index_01_r_ctrl.Location.Y {type=MovieSceneScriptingFloatChannel, default=0, thinned_from=71, keys [[0,0],[439,0]]}
       index_01_r_ctrl.Location.Z {type=MovieSceneScriptingFloatChannel, default=0, thinned_from=71, keys [[0,0],[439,0]]}
       index_01_r_ctrl.Rotation.X {
        type=MovieSceneScriptingFloatChannel
        default=0.6292
        thinned_from=71
        keys [[0,0.6292],[45,0.6292],[51,0.6],[56,0.6292],[66,0.6292],[70,0.6073],[101,0.6081],[107,0.6133],[111,0.6292],[267,0.6292],[271,0.0877],[273,-0.492],[275,-1.0855],[279,-1.8682],[283,-2.1747],[287,-2.2513],[292,-2.2705],[299,-2.2752],[351,-2.2768],[353,-1.8562],[355,-1.5541],[357,-1.4557],[359,-1.4557],[361,-1.2313],[362,-0.248],[363,0.3818],[365,0.5917],[366,0.6292],[439,0.6292]]
       }
       index_01_r_ctrl.Rotation.Y {
        type=MovieSceneScriptingFloatChannel
        default=-2.7506
        thinned_from=71
        keys [[0,-2.7506],[45,-2.7506],[51,-2.8],[56,-2.7506],[66,-2.7506],[70,-2.7877],[101,-2.7864],[107,-2.7776],[111,-2.7506],[267,-2.7506],[271,-4.906],[273,-7.2137],[275,-9.576],[279,-10.6271],[283,-9.7843],[287,-9.5736],[292,-9.5209],[351,-9.5034],[353,-8.526],[355,-7.824],[357,-7.5954],[359,-7.5954],[361,-7.0738],[362,-4.789],[363,-3.3255],[365,-2.8377],[366,-2.7506],[439,-2.7506]]
       }
       index_01_r_ctrl.Rotation.Z {
        type=MovieSceneScriptingFloatChannel
        default=-12.8893
        thinned_from=72
        keys [[0,-12.8893],[262,-12.8893],[267,-17.5],[271,-13.3947],[273,-17.1529],[274,-21],[275,19.1592],[279,29.5719],[283,34.783],[287,36.0858],[292,36.4115],[299,36.4929],[351,36.5201],[353,24.7],[355,23.0404],[357,22.5],[359,22.5],[361,25],[362,9.3561],[363,-5.7766],[365,-10.8208],[366,-12.8893],[439,-12.8893]]
       }
       index_01_r_ctrl.Scale.X {type=MovieSceneScriptingFloatChannel, default=1, thinned_from=71, keys [[0,1],[439,1]]}
       index_01_r_ctrl.Scale.Y {type=MovieSceneScriptingFloatChannel, default=1, thinned_from=71, keys [[0,1],[439,1]]}
       index_01_r_ctrl.Scale.Z {type=MovieSceneScriptingFloatChannel, default=1, thinned_from=71, keys [[0,1],[439,1]]}
       index_02_r_ctrl.Location.X {type=MovieSceneScriptingFloatChannel, default=0, thinned_from=71, keys [[0,0],[439,0]]}
       index_02_r_ctrl.Location.Y {type=MovieSceneScriptingFloatChannel, default=0, thinned_from=71, keys [[0,0],[439,0]]}
       index_02_r_ctrl.Location.Z {type=MovieSceneScriptingFloatChannel, default=0, thinned_from=71, keys [[0,0],[439,0]]}
       index_02_r_ctrl.Rotation.X {type=MovieSceneScriptingFloatChannel, default=0, thinned_from=71, keys [[0,0],[439,0]]}
       index_02_r_ctrl.Rotation.Y {type=MovieSceneScriptingFloatChannel, default=0, thinned_from=71, keys [[0,0],[439,0]]}
       index_02_r_ctrl.Rotation.Z {
        type=MovieSceneScriptingFloatChannel
        default=28.6778
        thinned_from=71
        keys [[0,28.6778],[262,28.6778],[267,24.0671],[271,19.3091],[273,21.444],[275,23.6295],[279,40.8514],[283,42.3455],[287,42.719],[292,42.8123],[351,42.8435],[353,34.671],[355,35.2798],[357,35.478],[359,35.478],[361,31.3426],[362,28.9019],[363,28.6137],[365,28.5176],[366,28.6778],[439,28.6778]]
       }
       index_02_r_ctrl.Scale.X {type=MovieSceneScriptingFloatChannel, default=1, thinned_from=71, keys [[0,1],[439,1]]}
       index_02_r_ctrl.Scale.Y {type=MovieSceneScriptingFloatChannel, default=1, thinned_from=71, keys [[0,1],[439,1]]}
       index_02_r_ctrl.Scale.Z {type=MovieSceneScriptingFloatChannel, default=1, thinned_from=71, keys [[0,1],[439,1]]}
       index_03_r_ctrl.Location.X {type=MovieSceneScriptingFloatChannel, default=0, thinned_from=71, keys [[0,0],[439,0]]}
       index_03_r_ctrl.Location.Y {type=MovieSceneScriptingFloatChannel, default=0, thinned_from=71, keys [[0,0],[439,0]]}
       index_03_r_ctrl.Location.Z {type=MovieSceneScriptingFloatChannel, default=0, thinned_from=71, keys [[0,0],[439,0]]}
       index_03_r_ctrl.Rotation.X {type=MovieSceneScriptingFloatChannel, default=0, thinned_from=71, keys [[0,0],[439,0]]}
       index_03_r_ctrl.Rotation.Y {type=MovieSceneScriptingFloatChannel, default=0, thinned_from=71, keys [[0,0],[439,0]]}
       index_03_r_ctrl.Rotation.Z {
        type=MovieSceneScriptingFloatChannel
        default=24.3552
        thinned_from=71
        keys [[0,24.3552],[45,24.3552],[51,24.4],[66,24.3552],[107,24.3797],[262,24.3552],[267,19.7444],[271,23.9804],[273,28.5157],[275,33.1583],[279,37.1767],[283,38.7503],[287,39.1437],[292,39.242],[351,39.2748],[353,37.1154],[355,35.5643],[357,35.0593],[359,35.0593],[361,30.3145],[362,27.5142],[363,25.2892],[365,24.5475],[366,24.3552],[439,24.3552]]
       }
       index_03_r_ctrl.Scale.X {type=MovieSceneScriptingFloatChannel, default=1, thinned_from=71, keys [[0,1],[439,1]]}
       index_03_r_ctrl.Scale.Y {type=MovieSceneScriptingFloatChannel, default=1, thinned_from=71, keys [[0,1],[439,1]]}
       index_03_r_ctrl.Scale.Z {type=MovieSceneScriptingFloatChannel, default=1, thinned_from=71, keys [[0,1],[439,1]]}
       middle_metacarpal_r_ctrl.Location.X {type=MovieSceneScriptingFloatChannel, default=0, thinned_from=71, keys [[0,0],[439,0]]}
       middle_metacarpal_r_ctrl.Location.Y {type=MovieSceneScriptingFloatChannel, default=0, thinned_from=71, keys [[0,0],[439,0]]}
       middle_metacarpal_r_ctrl.Location.Z {type=MovieSceneScriptingFloatChannel, default=0, thinned_from=71, keys [[0,0],[439,0]]}
       middle_metacarpal_r_ctrl.Rotation.X {
        type=MovieSceneScriptingFloatChannel
        default=-0.0167
        thinned_from=71
        keys [[0,-0.0167],[45,-0.0167],[51,0],[56,-0.0167],[66,-0.0167],[70,-0.0042],[75,-0.0011],[83,-0.0003],[94,0],[101,-0.0006],[107,-0.0046],[111,-0.0167],[267,-0.0167],[271,-0.0672],[273,-0.1212],[275,-0.1764],[279,-0.2493],[283,-0.2779],[287,-0.285],[292,-0.2868],[351,-0.2874],[353,-0.2482],[355,-0.2201],[357,-0.2109],[359,-0.2109],[361,-0.1249],[362,-0.0741],[363,-0.0337],[365,-0.0202],[366,-0.0167],[439,-0.0167]]
       }
       middle_metacarpal_r_ctrl.Rotation.Y {
        type=MovieSceneScriptingFloatChannel
        default=1.8417
        thinned_from=72
        keys [[0,1.8417],[45,1.8417],[51,1.8],[56,1.8417],[66,1.8417],[70,1.8104],[75,1.8026],[101,1.8014],[107,1.8113],[111,1.8417],[267,1.8417],[271,1.0741],[273,1.0375],[274,1],[275,-0.5892],[279,-1.6988],[283,-2.1333],[287,-2.242],[292,-2.2691],[299,-2.2759],[351,-2.2782],[353,-1.6819],[355,-1.2536],[357,-1.1141],[359,-1.1141],[361,0.1961],[362,0.9694],[363,1.5838],[365,1.7886],[366,1.8417],[439,1.8417]]
       }
       middle_metacarpal_r_ctrl.Rotation.Z {
        type=MovieSceneScriptingFloatChannel
        default=-12.6457
        thinned_from=71
        keys [[0,-12.6457],[45,-12.6457],[51,-12.6],[56,-12.6457],[66,-12.6457],[70,-12.6114],[107,-12.6124],[111,-12.6457],[267,-12.6457],[271,-8.3341],[273,-3.7178],[275,1.0078],[279,7.2403],[283,9.6808],[287,10.2909],[292,10.4435],[299,10.4816],[351,10.4943],[353,7.1451],[355,4.7394],[357,3.9561],[359,3.9561],[361,-3.4029],[362,-7.7461],[363,-11.197],[365,-12.3473],[366,-12.6457],[439,-12.6457]]
       }
       middle_metacarpal_r_ctrl.Scale.X {type=MovieSceneScriptingFloatChannel, default=1, thinned_from=71, keys [[0,1],[439,1]]}
       middle_metacarpal_r_ctrl.Scale.Y {type=MovieSceneScriptingFloatChannel, default=1, thinned_from=71, keys [[0,1],[439,1]]}
       middle_metacarpal_r_ctrl.Scale.Z {type=MovieSceneScriptingFloatChannel, default=1, thinned_from=71, keys [[0,1],[439,1]]}
       middle_01_r_ctrl.Location.X {type=MovieSceneScriptingFloatChannel, default=0, thinned_from=71, keys [[0,0],[439,0]]}
       middle_01_r_ctrl.Location.Y {type=MovieSceneScriptingFloatChannel, default=0, thinned_from=71, keys [[0,0],[439,0]]}
       middle_01_r_ctrl.Location.Z {type=MovieSceneScriptingFloatChannel, default=0, thinned_from=71, keys [[0,0],[439,0]]}
       middle_01_r_ctrl.Rotation.X {
        type=MovieSceneScriptingFloatChannel
        default=-1.8457
        thinned_from=71
        keys [[0,-1.8457],[45,-1.8457],[51,-1.8],[56,-1.8457],[66,-1.8457],[70,-1.8114],[101,-1.8126],[107,-1.8208],[111,-1.8457],[267,-1.8457],[271,-1.5972],[273,-1.3311],[275,-1.0587],[279,-0.6994],[283,-0.5587],[287,-0.5235],[292,-0.5147],[299,-0.5125],[351,-0.5118],[353,-0.7049],[355,-0.8435],[357,-0.8887],[359,-0.8887],[361,-1.1008],[362,-1.4839],[363,-1.7424],[365,-1.8285],[366,-1.8457],[439,-1.8457]]
       }
       middle_01_r_ctrl.Rotation.Y {
        type=MovieSceneScriptingFloatChannel
        default=-10.2135
        thinned_from=71
        keys [[0,-10.2135],[45,-10.2135],[51,-10.2],[66,-10.2135],[267,-10.2135],[271,-10.3562],[275,-10.6654],[279,-10.8717],[283,-10.9525],[351,-10.9794],[353,-10.8686],[355,-10.7889],[357,-10.763],[359,-10.763],[361,-10.6412],[362,-10.4212],[363,-10.2728],[366,-10.2135],[439,-10.2135]]
       }
       middle_01_r_ctrl.Rotation.Z {
        type=MovieSceneScriptingFloatChannel
        default=-26.3692
        thinned_from=72
        keys [[0,-26.3692],[45,-26.3692],[51,-26.4],[262,-26.3692],[267,-30.9799],[271,-29.0915],[273,-28.058],[274,-27],[275,8.4756],[279,31.8969],[283,39.0478],[287,40.8355],[292,41.2824],[299,41.3942],[351,41.4314],[353,25.1446],[355,14.7979],[357,11.4292],[359,11.4292],[361,7.5],[362,-11.2228],[363,-22.1988],[365,-25.8575],[366,-26.3692],[439,-26.3692]]
       }
       middle_01_r_ctrl.Scale.X {type=MovieSceneScriptingFloatChannel, default=1, thinned_from=71, keys [[0,1],[439,1]]}
       middle_01_r_ctrl.Scale.Y {type=MovieSceneScriptingFloatChannel, default=1, thinned_from=71, keys [[0,1],[439,1]]}
       middle_01_r_ctrl.Scale.Z {type=MovieSceneScriptingFloatChannel, default=1, thinned_from=71, keys [[0,1],[439,1]]}
       middle_02_r_ctrl.Location.X {type=MovieSceneScriptingFloatChannel, default=0, thinned_from=71, keys [[0,0],[439,0]]}
       middle_02_r_ctrl.Location.Y {type=MovieSceneScriptingFloatChannel, default=0, thinned_from=71, keys [[0,0],[439,0]]}
       middle_02_r_ctrl.Location.Z {type=MovieSceneScriptingFloatChannel, default=0, thinned_from=71, keys [[0,0],[439,0]]}
       middle_02_r_ctrl.Rotation.X {type=MovieSceneScriptingFloatChannel, default=0, thinned_from=71, keys [[0,0],[439,0]]}
       middle_02_r_ctrl.Rotation.Y {type=MovieSceneScriptingFloatChannel, default=0, thinned_from=71, keys [[0,0],[439,0]]}
       middle_02_r_ctrl.Rotation.Z {
        type=MovieSceneScriptingFloatChannel
        default=11.1202
        thinned_from=71
        keys [[0,11.1202],[45,11.1202],[51,11.1],[66,11.1202],[262,11.1202],[267,6.5095],[271,-1.8619],[273,12.7208],[275,27.6486],[279,42.709],[283,46.5858],[287,47.555],[292,47.7972],[351,47.878],[353,29.3613],[355,37.0098],[357,39.5],[359,39.5],[361,23.8431],[362,14.6027],[363,11.7921],[365,10.8552],[366,11.1202],[439,11.1202]]
       }
       middle_02_r_ctrl.Scale.X {type=MovieSceneScriptingFloatChannel, default=1, thinned_from=71, keys [[0,1],[439,1]]}
       middle_02_r_ctrl.Scale.Y {type=MovieSceneScriptingFloatChannel, default=1, thinned_from=71, keys [[0,1],[439,1]]}
       middle_02_r_ctrl.Scale.Z {type=MovieSceneScriptingFloatChannel, default=1, thinned_from=71, keys [[0,1],[439,1]]}
       middle_03_r_ctrl.Location.X {type=MovieSceneScriptingFloatChannel, default=0, thinned_from=71, keys [[0,0],[439,0]]}
       middle_03_r_ctrl.Location.Y {type=MovieSceneScriptingFloatChannel, default=0, thinned_from=71, keys [[0,0],[439,0]]}
       middle_03_r_ctrl.Location.Z {type=MovieSceneScriptingFloatChannel, default=0, thinned_from=71, keys [[0,0],[439,0]]}
       middle_03_r_ctrl.Rotation.X {type=MovieSceneScriptingFloatChannel, default=0, thinned_from=71, keys [[0,0],[439,0]]}
       middle_03_r_ctrl.Rotation.Y {type=MovieSceneScriptingFloatChannel, default=0, thinned_from=71, keys [[0,0],[439,0]]}
       middle_03_r_ctrl.Rotation.Z {
        type=MovieSceneScriptingFloatChannel
        default=24.3552
        thinned_from=71
        keys [[0,24.3552],[45,24.3552],[51,24.4],[66,24.3552],[107,24.3797],[262,24.3552],[267,19.7444],[271,23.4274],[273,27.3707],[275,31.4073],[279,34.6264],[283,35.887],[287,36.2021],[292,36.2809],[351,36.3071],[353,34.5773],[355,33.3347],[357,32.9301],[359,32.9301],[361,29.1291],[362,26.8858],[363,25.1034],[365,24.5093],[366,24.3552],[439,24.3552]]
       }
       middle_03_r_ctrl.Scale.X {type=MovieSceneScriptingFloatChannel, default=1, thinned_from=71, keys [[0,1],[439,1]]}
       middle_03_r_ctrl.Scale.Y {type=MovieSceneScriptingFloatChannel, default=1, thinned_from=71, keys [[0,1],[439,1]]}
       middle_03_r_ctrl.Scale.Z {type=MovieSceneScriptingFloatChannel, default=1, thinned_from=71, keys [[0,1],[439,1]]}
       ring_metacarpal_r_ctrl.Location.X {type=MovieSceneScriptingFloatChannel, default=0, thinned_from=71, keys [[0,0],[439,0]]}
       ring_metacarpal_r_ctrl.Location.Y {type=MovieSceneScriptingFloatChannel, default=0, thinned_from=71, keys [[0,0],[439,0]]}
       ring_metacarpal_r_ctrl.Location.Z {type=MovieSceneScriptingFloatChannel, default=0, thinned_from=71, keys [[0,0],[439,0]]}
       ring_metacarpal_r_ctrl.Rotation.X {
        type=MovieSceneScriptingFloatChannel
        default=0.0204
        thinned_from=71
        keys [[0,0.0204],[45,0.0204],[51,0],[56,0.0204],[66,0.0204],[70,0.0051],[75,0.0013],[83,0.0003],[94,0],[101,0.0007],[107,0.0055],[111,0.0204],[267,0.0204],[271,0.0166],[275,0.0084],[279,0.0029],[283,0.0007],[287,0.0002],[351,0],[353,0.003],[355,0.0051],[357,0.0058],[359,0.0058],[361,0.0123],[362,0.0161],[363,0.0191],[365,0.0201],[366,0.0204],[439,0.0204]]
       }
       ring_metacarpal_r_ctrl.Rotation.Y {
        type=MovieSceneScriptingFloatChannel
        default=-1.3752
        thinned_from=71
        keys [[0,-1.3752],[45,-1.3752],[51,-1.4],[56,-1.3752],[66,-1.3752],[70,-1.3938],[75,-1.3985],[101,-1.3992],[107,-1.3933],[111,-1.3752],[267,-1.3752],[271,-1.119],[273,-0.8446],[275,-0.5638],[279,-0.1934],[283,-0.0483],[287,-0.0121],[292,-0.003],[299,-0.0008],[309,-0.0002],[351,0],[353,-0.199],[355,-0.342],[357,-0.3886],[359,-0.3886],[361,-0.8259],[362,-1.084],[363,-1.2891],[365,-1.3575],[366,-1.3752],[439,-1.3752]]
       }
       ring_metacarpal_r_ctrl.Rotation.Z {
        type=MovieSceneScriptingFloatChannel
        default=-0.8501
        thinned_from=71
        keys [[0,-0.8501],[45,-0.8501],[51,-0.9],[56,-0.8501],[66,-0.8501],[70,-0.8875],[75,-0.8969],[83,-0.8992],[94,-0.9],[101,-0.8983],[107,-0.8864],[111,-0.8501],[267,-0.8501],[271,-1.6586],[273,-2.5242],[275,-3.4103],[279,-2.8343],[283,-3.0778],[287,-3.1386],[292,-3.1539],[351,-3.1589],[353,-2.8248],[355,-2.5847],[357,-2.5066],[359,-2.5066],[361,-1.7723],[362,-1.339],[363,-0.9947],[365,-0.8799],[366,-0.8501],[439,-0.8501]]
       }
       ring_metacarpal_r_ctrl.Scale.X {type=MovieSceneScriptingFloatChannel, default=1, thinned_from=71, keys [[0,1],[439,1]]}
       ring_metacarpal_r_ctrl.Scale.Y {type=MovieSceneScriptingFloatChannel, default=1, thinned_from=71, keys [[0,1],[439,1]]}
       ring_metacarpal_r_ctrl.Scale.Z {type=MovieSceneScriptingFloatChannel, default=1, thinned_from=71, keys [[0,1],[439,1]]}
       ring_01_r_ctrl.Location.X {type=MovieSceneScriptingFloatChannel, default=0, thinned_from=71, keys [[0,0],[439,0]]}
       ring_01_r_ctrl.Location.Y {type=MovieSceneScriptingFloatChannel, default=0, thinned_from=71, keys [[0,0],[439,0]]}
       ring_01_r_ctrl.Location.Z {type=MovieSceneScriptingFloatChannel, default=0, thinned_from=71, keys [[0,0],[439,0]]}
       ring_01_r_ctrl.Rotation.X {
        type=MovieSceneScriptingFloatChannel
        default=-9.9706
        thinned_from=71
        keys [[0,-9.9706],[45,-9.9706],[51,-10],[56,-9.9706],[66,-9.9706],[70,-9.9926],[107,-9.9866],[267,-9.9706],[271,-8.0454],[273,-5.9842],[275,-3.8742],[279,-1.0914],[283,-0.0017],[287,0.2707],[292,0.3388],[299,0.3559],[309,0.3598],[337,0.3615],[351,0.3615],[353,-1.1339],[355,-2.2081],[357,-2.5578],[359,-2.5578],[361,-5.8436],[362,-7.7829],[363,-9.3237],[365,-9.8373],[366,-9.9706],[439,-9.9706]]
       }
       ring_01_r_ctrl.Rotation.Y {
        type=MovieSceneScriptingFloatChannel
        default=-5.7599
        thinned_from=71
        keys [[0,-5.7599],[45,-5.7599],[51,-5.8],[56,-5.7599],[66,-5.7599],[70,-5.79],[101,-5.789],[107,-5.7818],[111,-5.7599],[267,-5.7599],[271,-5.8624],[275,-6.0845],[279,-6.7749],[283,-7.5736],[287,-7.7732],[292,-7.8231],[351,-7.8398],[353,-7.5387],[355,-7.3225],[357,-7.2521],[359,-7.2521],[361,-6.5907],[362,-6.2003],[363,-5.8901],[365,-5.7867],[366,-5.7599],[439,-5.7599]]
       }
       ring_01_r_ctrl.Rotation.Z {
        type=MovieSceneScriptingFloatChannel
        default=-30.9995
        thinned_from=71
        keys [[0,-30.9995],[262,-30.9995],[267,-35.6103],[271,-33.1196],[273,-16.1177],[275,1.2866],[279,29.4076],[283,36.8212],[287,38.6746],[292,39.1379],[299,39.2538],[351,39.2924],[353,23.8077],[355,13.7944],[357,10.5342],[359,10.5342],[361,-8.4222],[362,-19.6099],[363,-27.6954],[365,-30.3906],[366,-30.9995],[439,-30.9995]]
       }
       ring_01_r_ctrl.Scale.X {type=MovieSceneScriptingFloatChannel, default=1, thinned_from=71, keys [[0,1],[439,1]]}
       ring_01_r_ctrl.Scale.Y {type=MovieSceneScriptingFloatChannel, default=1, thinned_from=71, keys [[0,1],[439,1]]}
       ring_01_r_ctrl.Scale.Z {type=MovieSceneScriptingFloatChannel, default=1, thinned_from=71, keys [[0,1],[439,1]]}
       ring_02_r_ctrl.Location.X {type=MovieSceneScriptingFloatChannel, default=0, thinned_from=71, keys [[0,0],[439,0]]}
       ring_02_r_ctrl.Location.Y {type=MovieSceneScriptingFloatChannel, default=0, thinned_from=71, keys [[0,0],[439,0]]}
       ring_02_r_ctrl.Location.Z {type=MovieSceneScriptingFloatChannel, default=0, thinned_from=71, keys [[0,0],[439,0]]}
       ring_02_r_ctrl.Rotation.X {type=MovieSceneScriptingFloatChannel, default=0, thinned_from=71, keys [[0,0],[439,0]]}
       ring_02_r_ctrl.Rotation.Y {type=MovieSceneScriptingFloatChannel, default=0, thinned_from=71, keys [[0,0],[439,0]]}
       ring_02_r_ctrl.Rotation.Z {
        type=MovieSceneScriptingFloatChannel
        default=2.5295
        thinned_from=71
        keys [[0,2.5295],[45,2.5295],[51,2.5],[56,2.5295],[66,2.5295],[70,2.5074],[101,2.5081],[107,2.5134],[111,2.5295],[262,2.5295],[267,-2.0812],[271,-1.7398],[273,10.3629],[275,22.752],[279,45.3663],[283,50.6236],[287,51.9379],[292,52.2665],[351,52.376],[353,34.985],[355,24.6185],[357,21.2433],[359,21.2433],[361,11.9021],[362,6.3892],[363,3.5491],[365,2.6024],[366,2.5295],[439,2.5295]]
       }
       ring_02_r_ctrl.Scale.X {type=MovieSceneScriptingFloatChannel, default=1, thinned_from=71, keys [[0,1],[439,1]]}
       ring_02_r_ctrl.Scale.Y {type=MovieSceneScriptingFloatChannel, default=1, thinned_from=71, keys [[0,1],[439,1]]}
       ring_02_r_ctrl.Scale.Z {type=MovieSceneScriptingFloatChannel, default=1, thinned_from=71, keys [[0,1],[439,1]]}
       ring_03_r_ctrl.Location.X {type=MovieSceneScriptingFloatChannel, default=0, thinned_from=71, keys [[0,0],[439,0]]}
       ring_03_r_ctrl.Location.Y {type=MovieSceneScriptingFloatChannel, default=0, thinned_from=71, keys [[0,0],[439,0]]}
       ring_03_r_ctrl.Location.Z {type=MovieSceneScriptingFloatChannel, default=0, thinned_from=71, keys [[0,0],[439,0]]}
       ring_03_r_ctrl.Rotation.X {type=MovieSceneScriptingFloatChannel, default=0, thinned_from=71, keys [[0,0],[439,0]]}
       ring_03_r_ctrl.Rotation.Y {type=MovieSceneScriptingFloatChannel, default=0, thinned_from=71, keys [[0,0],[439,0]]}
       ring_03_r_ctrl.Rotation.Z {
        type=MovieSceneScriptingFloatChannel
        default=24.3552
        thinned_from=71
        keys [[0,24.3552],[45,24.3552],[51,24.4],[66,24.3552],[107,24.3797],[262,24.3552],[267,19.7444],[271,22.5091],[273,25.4692],[275,28.4993],[279,30.3909],[283,31.1317],[287,31.3169],[351,31.3786],[353,30.362],[355,29.6319],[357,29.3941],[359,29.3941],[361,27.1605],[362,25.8423],[363,24.7949],[365,24.4457],[366,24.3552],[439,24.3552]]
       }
       ring_03_r_ctrl.Scale.X {type=MovieSceneScriptingFloatChannel, default=1, thinned_from=71, keys [[0,1],[439,1]]}
       ring_03_r_ctrl.Scale.Y {type=MovieSceneScriptingFloatChannel, default=1, thinned_from=71, keys [[0,1],[439,1]]}
       ring_03_r_ctrl.Scale.Z {type=MovieSceneScriptingFloatChannel, default=1, thinned_from=71, keys [[0,1],[439,1]]}
       pinky_metacarpal_r_ctrl.Location.X {type=MovieSceneScriptingFloatChannel, default=0, thinned_from=71, keys [[0,0],[439,0]]}
       pinky_metacarpal_r_ctrl.Location.Y {type=MovieSceneScriptingFloatChannel, default=0, thinned_from=71, keys [[0,0],[439,0]]}
       pinky_metacarpal_r_ctrl.Location.Z {type=MovieSceneScriptingFloatChannel, default=0, thinned_from=71, keys [[0,0],[439,0]]}
       pinky_metacarpal_r_ctrl.Rotation.X {
        type=MovieSceneScriptingFloatChannel
        default=-0.3335
        thinned_from=71
        keys [[0,-0.3335],[45,-0.3335],[51,-0.3],[56,-0.3335],[66,-0.3335],[70,-0.3084],[75,-0.3021],[83,-0.3005],[94,-0.3],[101,-0.3011],[107,-0.3091],[111,-0.3335],[267,-0.3335],[271,-0.2714],[273,-0.2048],[275,-0.1367],[279,-0.0469],[283,-0.0117],[287,-0.0029],[292,-0.0007],[299,-0.0002],[351,0],[353,-0.0483],[355,-0.0829],[357,-0.0942],[359,-0.0942],[361,-0.2003],[362,-0.2629],[363,-0.3126],[365,-0.3292],[366,-0.3335],[439,-0.3335]]
       }
       pinky_metacarpal_r_ctrl.Rotation.Y {
        type=MovieSceneScriptingFloatChannel
        default=-4.5453
        thinned_from=71
        keys [[0,-4.5453],[45,-4.5453],[51,-4.5],[56,-4.5453],[66,-4.5453],[70,-4.5113],[101,-4.5015],[107,-4.5123],[111,-4.5453],[275,-4.5402],[279,-1.6489],[283,-2.1659],[287,-2.2952],[292,-2.3275],[299,-2.3356],[351,-2.3383],[353,-2.6577],[355,-2.8872],[357,-2.9619],[359,-2.9619],[361,-3.6637],[362,-4.078],[363,-4.4071],[365,-4.5168],[366,-4.5453],[439,-4.5453]]
       }
       pinky_metacarpal_r_ctrl.Rotation.Z {
        type=MovieSceneScriptingFloatChannel
        default=-10.0111
        thinned_from=71
        keys [[0,-10.0111],[51,-10],[267,-10.0111],[271,-9.1032],[273,-8.1312],[275,-7.1361],[279,-1.4078],[283,-0.352],[287,-0.088],[292,-0.022],[299,-0.0055],[309,-0.0017],[327,-0.0004],[337,0],[351,0],[353,-1.449],[355,-2.4898],[357,-2.8286],[359,-2.8286],[361,-6.0124],[362,-7.8914],[363,-9.3844],[365,-9.882],[366,-10.0111],[439,-10.0111]]
       }
       pinky_metacarpal_r_ctrl.Scale.X {type=MovieSceneScriptingFloatChannel, default=1, thinned_from=71, keys [[0,1],[439,1]]}
       pinky_metacarpal_r_ctrl.Scale.Y {type=MovieSceneScriptingFloatChannel, default=1, thinned_from=71, keys [[0,1],[439,1]]}
       pinky_metacarpal_r_ctrl.Scale.Z {type=MovieSceneScriptingFloatChannel, default=1, thinned_from=71, keys [[0,1],[439,1]]}
       pinky_01_r_ctrl.Location.X {type=MovieSceneScriptingFloatChannel, default=0, thinned_from=71, keys [[0,0],[439,0]]}
       pinky_01_r_ctrl.Location.Y {type=MovieSceneScriptingFloatChannel, default=0, thinned_from=71, keys [[0,0],[439,0]]}
       pinky_01_r_ctrl.Location.Z {type=MovieSceneScriptingFloatChannel, default=0, thinned_from=71, keys [[0,0],[439,0]]}
       pinky_01_r_ctrl.Rotation.X {
        type=MovieSceneScriptingFloatChannel
        default=-8.0273
        thinned_from=71
        keys [[0,-8.0273],[45,-8.0273],[51,-8],[56,-8.0273],[66,-8.0273],[70,-8.0068],[107,-8.0124],[267,-8.0273],[271,-6.5991],[273,-5.0701],[275,-3.5049],[279,-1.4405],[283,-0.6321],[287,-0.43],[292,-0.3795],[299,-0.3669],[309,-0.364],[351,-0.3627],[353,-1.472],[355,-2.2689],[357,-2.5283],[359,-2.5283],[361,-4.9658],[362,-6.4044],[363,-7.5474],[365,-7.9284],[366,-8.0273],[439,-8.0273]]
       }
       pinky_01_r_ctrl.Rotation.Y {
        type=MovieSceneScriptingFloatChannel
        default=-9.4708
        thinned_from=71
        keys [[0,-9.4708],[45,-9.4708],[51,-9.5],[56,-9.4708],[66,-9.4708],[70,-9.4927],[107,-9.4867],[267,-9.4708],[271,-7.0068],[273,-4.3687],[275,-1.6681],[279,1.8936],[283,3.2883],[287,3.637],[292,3.7241],[299,3.7459],[351,3.7532],[353,1.8392],[355,0.4644],[357,0.0168],[359,0.0168],[361,-4.1887],[362,-6.6708],[363,-8.6429],[365,-9.3002],[366,-9.4708],[439,-9.4708]]
       }
       pinky_01_r_ctrl.Rotation.Z {
        type=MovieSceneScriptingFloatChannel
        default=-19.8844
        thinned_from=72
        keys [[0,-19.8844],[262,-19.8844],[267,-24.4951],[271,-26.9051],[273,-21.7634],[274,-16.5],[275,-9.8702],[279,5.8866],[283,9.0494],[287,9.8401],[292,10.0378],[299,10.0872],[351,10.1037],[353,3.6898],[355,-0.4842],[357,-1.8432],[359,-1.8432],[361,-10.0534],[362,-14.8989],[363,-18.4351],[365,-19.6138],[366,-19.8844],[439,-19.8844]]
       }
       pinky_01_r_ctrl.Scale.X {type=MovieSceneScriptingFloatChannel, default=1, thinned_from=71, keys [[0,1],[439,1]]}
       pinky_01_r_ctrl.Scale.Y {type=MovieSceneScriptingFloatChannel, default=1, thinned_from=71, keys [[0,1],[439,1]]}
       pinky_01_r_ctrl.Scale.Z {type=MovieSceneScriptingFloatChannel, default=1, thinned_from=71, keys [[0,1],[439,1]]}
       pinky_02_r_ctrl.Location.X {type=MovieSceneScriptingFloatChannel, default=0, thinned_from=71, keys [[0,0],[439,0]]}
       pinky_02_r_ctrl.Location.Y {type=MovieSceneScriptingFloatChannel, default=0, thinned_from=71, keys [[0,0],[439,0]]}
       pinky_02_r_ctrl.Location.Z {type=MovieSceneScriptingFloatChannel, default=0, thinned_from=71, keys [[0,0],[439,0]]}
       pinky_02_r_ctrl.Rotation.X {type=MovieSceneScriptingFloatChannel, default=0, thinned_from=71, keys [[0,0],[439,0]]}
       pinky_02_r_ctrl.Rotation.Y {type=MovieSceneScriptingFloatChannel, default=0, thinned_from=71, keys [[0,0],[439,0]]}
       pinky_02_r_ctrl.Rotation.Z {
        type=MovieSceneScriptingFloatChannel
        default=5.3847
        thinned_from=71
        keys [[0,5.3847],[45,5.3847],[51,5.4],[56,5.3847],[66,5.3847],[70,5.3962],[107,5.393],[262,5.3847],[267,0.7739],[271,3.2138],[273,14.2295],[275,25.5061],[279,45.8766],[283,50.846],[287,52.0884],[292,52.399],[351,52.5025],[353,39.5641],[355,31.5484],[357,28.9386],[359,28.9386],[361,17.869],[362,11.336],[363,7.0712],[365,5.6496],[366,5.3847],[439,5.3847]]
       }
       pinky_02_r_ctrl.Scale.X {type=MovieSceneScriptingFloatChannel, default=1, thinned_from=71, keys [[0,1],[439,1]]}
       pinky_02_r_ctrl.Scale.Y {type=MovieSceneScriptingFloatChannel, default=1, thinned_from=71, keys [[0,1],[439,1]]}
       pinky_02_r_ctrl.Scale.Z {type=MovieSceneScriptingFloatChannel, default=1, thinned_from=71, keys [[0,1],[439,1]]}
       pinky_03_r_ctrl.Location.X {type=MovieSceneScriptingFloatChannel, default=0, thinned_from=71, keys [[0,0],[439,0]]}
       pinky_03_r_ctrl.Location.Y {type=MovieSceneScriptingFloatChannel, default=0, thinned_from=71, keys [[0,0],[439,0]]}
       pinky_03_r_ctrl.Location.Z {type=MovieSceneScriptingFloatChannel, default=0, thinned_from=71, keys [[0,0],[439,0]]}
       pinky_03_r_ctrl.Rotation.X {type=MovieSceneScriptingFloatChannel, default=0, thinned_from=71, keys [[0,0],[439,0]]}
       pinky_03_r_ctrl.Rotation.Y {type=MovieSceneScriptingFloatChannel, default=0, thinned_from=71, keys [[0,0],[439,0]]}
       pinky_03_r_ctrl.Rotation.Z {
        type=MovieSceneScriptingFloatChannel
        default=16.7626
        thinned_from=71
        keys [[0,16.7626],[45,16.7626],[51,16.8],[66,16.7626],[107,16.783],[262,16.7626],[267,12.1518],[271,14.9842],[273,18.0167],[275,21.121],[279,23.1105],[283,23.8896],[287,24.0844],[299,24.1452],[351,24.1493],[353,23.0802],[355,22.3122],[357,22.0622],[359,22.0622],[361,19.713],[362,18.3266],[363,17.225],[365,16.8578],[366,16.7626],[439,16.7626]]
       }
       pinky_03_r_ctrl.Scale.X {type=MovieSceneScriptingFloatChannel, default=1, thinned_from=71, keys [[0,1],[439,1]]}
       pinky_03_r_ctrl.Scale.Y {type=MovieSceneScriptingFloatChannel, default=1, thinned_from=71, keys [[0,1],[439,1]]}
       pinky_03_r_ctrl.Scale.Z {type=MovieSceneScriptingFloatChannel, default=1, thinned_from=71, keys [[0,1],[439,1]]}
       hand_r_ik_ctrl.Location.X {
        type=MovieSceneScriptingFloatChannel
        default=-17.2624
        thinned_from=115
        keys [[0,-14.2042],[266,-14.2042],[267,-14.1763],[268,-14.1149],[269,-13.9208],[270,-13.5743],[271,-12.7898],[272,-11.4824],[273,-10.3632],[274,-9.0273],[275,-7.0499],[276,-6.2725],[278,-5.2124],[279,-4.9846],[280,-4.8479],[281,-4.7864],[282,-4.5413],[283,-4.208],[284,-3.6985],[285,-2.9763],[286,-2.0995],[287,-0.9245],[288,0.1175],[289,0.6949],[290,1.0886],[291,1.2737],[292,1.3354],[327,1.3354],[337,1.3385],[338,1.1855],[339,0.7531],[340,0.0814],[341,-0.79],[342,-1.8209],[343,-2.9716],[344,-4.2021],[345,-5.4725],[346,-7.873],[347,-11.9052],[348,-16.6469],[349,-20.5618],[350,-24.5694],[351,-24.4109],[352,-22.1426],[353,-22.1426],[354,-9.0396],[355,-11.7221],[356,-13.0666],[358,-13.825],[359,-14.0113],[360,-14.1319],[362,-14.1943],[439,-14.2042]]
       }
       hand_r_ik_ctrl.Location.Y {
        type=MovieSceneScriptingFloatChannel
        default=3.6547
        thinned_from=115
        keys [[0,0.8032],[45,0.8032],[51,0.8],[56,0.8032],[66,0.8032],[70,0.8008],[107,0.8009],[111,0.8032],[266,0.8032],[267,1.1474],[268,1.9046],[269,3.491],[270,5.9898],[271,10.2107],[272,17.2456],[273,22.8719],[274,26.1558],[275,25.8087],[276,27.6671],[278,30.2494],[279,30.8043],[280,31.1373],[281,31.2871],[282,31.8843],[283,32.6963],[284,33.9372],[285,35.6965],[286,37.8325],[287,40.6947],[288,43.233],[289,44.6397],[290,45.5986],[291,46.0495],[292,46.1998],[337,46.2074],[338,46.0969],[339,45.7847],[340,45.2996],[341,44.6705],[343,43.0952],[345,41.2894],[346,39.5561],[347,36.6446],[348,33.2208],[349,30.3939],[350,27.5002],[351,27.5174],[352,28.7583],[353,28.7583],[354,33.9829],[355,28.7072],[356,21.1841],[357,13.9121],[358,7.3576],[359,4.1379],[360,2.0537],[361,1.2721],[362,0.9746],[363,0.846],[364,0.8032],[439,0.8032]]
       }
       hand_r_ik_ctrl.Location.Z {
        type=MovieSceneScriptingFloatChannel
        default=5.3487
        thinned_from=115
        keys [[0,1.97],[45,1.97],[51,2],[56,1.97],[66,1.97],[70,1.9925],[75,1.9981],[101,1.999],[107,1.9919],[111,1.97],[266,1.97],[267,2.0346],[268,2.1766],[269,2.4761],[270,2.9491],[271,3.2527],[272,3.7587],[273,4.2339],[274,4.8151],[275,6.9477],[276,7.036],[278,7.4191],[279,7.5014],[280,7.5508],[281,7.5731],[282,7.6617],[283,7.7822],[284,7.9663],[285,8.2273],[286,8.5443],[287,8.9689],[288,9.3456],[289,9.5543],[290,9.6965],[291,9.7635],[292,9.7858],[337,9.7869],[338,10.0031],[339,10.6138],[340,11.5626],[341,12.7933],[342,14.2493],[343,15.8745],[344,17.6123],[345,19.4065],[346,22.7967],[347,28.4917],[348,35.1887],[349,40.7181],[350,46.3783],[351,45.9215],[352,41.678],[353,41.678],[354,7.2794],[355,6.528],[356,5.7039],[357,4.4593],[358,3.3812],[359,2.688],[360,2.2393],[361,2.071],[362,2.0069],[363,1.9792],[364,1.97],[439,1.97]]
       }
       hand_r_ik_ctrl.Rotation.X {
        type=MovieSceneScriptingFloatChannel
        default=-4.7449
        thinned_from=115
        keys [[0,24.3604],[45,24.3604],[51,24.4],[66,24.3604],[107,24.3892],[268,24.3604],[269,24.4399],[270,24.6149],[271,25.6945],[272,27.4939],[273,28.6831],[274,30.0785],[275,36.4217],[276,38.5567],[277,46.123],[278,53.5713],[279,56.7977],[280,58.7335],[281,59.6182],[282,63.1321],[283,67.8825],[284,75.1299],[285,85.4047],[286,97.8785],[287,114.5933],[288,129.4167],[289,137.6311],[290,143.2313],[291,145.8645],[292,146.7422],[337,146.7862],[338,146.5956],[339,146.0571],[341,144.1348],[345,138.3018],[346,136.0849],[347,132.7633],[348,128.558],[350,118.3796],[351,113.9843],[352,106.5658],[353,106.5658],[354,72.4001],[355,59.5073],[356,47.9206],[357,38.6753],[358,31.6849],[359,27.9498],[360,26.386],[361,25.12],[362,24.638],[363,24.4298],[364,24.3604],[439,24.3604]]
       }
       hand_r_ik_ctrl.Rotation.Y {
        type=MovieSceneScriptingFloatChannel
        default=40.611
        thinned_from=115
        keys [[0,19.6569],[45,19.6569],[51,19.7],[66,19.6569],[107,19.6883],[266,19.6569],[267,18.9146],[268,17.2818],[269,13.8843],[270,8.5474],[271,1.9979],[272,-8.9179],[273,-19.3553],[274,-28.876],[275,-31.2205],[276,-34.4584],[278,-29.6841],[279,-28.6582],[280,-28.0426],[281,-27.8314],[282,-26.9877],[283,-25.827],[284,-23.9293],[285,-21.217],[286,-17.9149],[287,-13.4869],[288,-9.5585],[289,-7.3812],[290,-5.8965],[291,-5.1983],[292,-4.9656],[327,-4.9656],[337,-4.9539],[338,-5.0048],[339,-5.1485],[340,-5.3719],[341,-5.6617],[342,-6.0045],[343,-6.3872],[345,-7.2189],[346,-7.8107],[347,-8.6974],[348,-9.8201],[349,-11.1628],[350,-12.5373],[351,-13.9473],[352,-16.8198],[354,-16.8198],[355,-19.8747],[356,-14.88],[357,-11.8836],[358,-4.5251],[359,1.7516],[360,9.5524],[361,15.8677],[362,18.2719],[363,19.3106],[364,19.6569],[439,19.6569]]
       }
       hand_r_ik_ctrl.Rotation.Z {
        type=MovieSceneScriptingFloatChannel
        default=-4.1753
        thinned_from=115
        keys [[0,-95.7552],[266,-95.7552],[267,-96.5206],[268,-98.2045],[269,-101.5968],[270,-106.8555],[271,-111.3065],[272,-118.7248],[273,-123.49],[274,-130.4318],[275,-147.3806],[276,-151.0484],[278,-154.2206],[280,-155.3112],[281,-155.4943],[283,-157.1922],[284,-158.7164],[285,-160.8781],[286,-163.5027],[287,-167.0198],[288,-170.1389],[289,-171.8674],[290,-173.0458],[291,-173.5999],[340,-173.4647],[346,-171.5448],[348,-169.963],[350,-167.8239],[351,-166.1813],[352,-161.6039],[354,-161.6039],[355,-139.2398],[357,-101.7398],[358,-101.5756],[359,-99.1791],[360,-97.6874],[361,-96.4798],[362,-96.02],[409,-95.7552],[439,-95.7552]]
       }
       hand_r_ik_ctrl.Scale.X {type=MovieSceneScriptingFloatChannel, default=1, thinned_from=115, keys [[0,1],[439,1]]}
       hand_r_ik_ctrl.Scale.Y {type=MovieSceneScriptingFloatChannel, default=1, thinned_from=115, keys [[0,1],[439,1]]}
       hand_r_ik_ctrl.Scale.Z {type=MovieSceneScriptingFloatChannel, default=1, thinned_from=115, keys [[0,1],[439,1]]}
       Softness_2 {type=MovieSceneScriptingFloatChannel, default=0, thinned_from=115, keys [[0,0],[439,0]]}
       hand_r_ikorient_ctrl.Rotation.Roll {type=MovieSceneScriptingFloatChannel, default=0, thinned_from=64, keys [[0,0],[439,0]]}
       hand_r_ikorient_ctrl.Rotation.Pitch {type=MovieSceneScriptingFloatChannel, default=0, thinned_from=64, keys [[0,0],[439,0]]}
       hand_r_ikorient_ctrl.Rotation.Yaw {type=MovieSceneScriptingFloatChannel, default=0, thinned_from=64, keys [[0,0],[439,0]]}
       arm_r_pv_ik_ctrl.Location.X {
        type=MovieSceneScriptingFloatChannel
        default=-2
        thinned_from=119
        keys [[0,-2],[204,-2],[209,-1],[266,-1],[267,-0.875],[270,-0.125],[271,0.25],[272,2.1613],[274,2.1613],[275,-1],[283,-1],[284,-1.2018],[285,-1.5707],[286,-2.0541],[287,-2.7141],[288,-3.3044],[289,-3.6332],[290,-3.8584],[291,-3.9646],[292,-4],[337,-4],[338,-3.9832],[339,-3.9355],[340,-3.8616],[341,-3.7656],[342,-3.6521],[343,-3.5254],[345,-3.25],[346,-3.054],[347,-2.7604],[348,-2.3887],[349,-1.9441],[350,-1.4889],[351,0.0745],[352,5.5031],[353,5.5031],[354,13.8523],[355,7.2025],[356,4.6794],[357,6.8786],[358,3.689],[359,0.8944],[360,-0.9146],[361,-1.593],[362,-1.8512],[363,-1.9628],[364,-2],[439,-2]]
       }
       arm_r_pv_ik_ctrl.Location.Y {
        type=MovieSceneScriptingFloatChannel
        default=0
        thinned_from=115
        keys [[0,0],[266,0],[267,0.6605],[268,2.1137],[269,6.9209],[270,15.1488],[271,23.4592],[272,31.6304],[273,39.4649],[274,44.1733],[275,46.0886],[276,47.5126],[278,51.0244],[279,51.779],[280,52.2318],[281,52.434],[282,53.2413],[283,54.342],[284,56.0244],[285,58.4113],[286,61.3098],[287,65.194],[288,68.6389],[289,70.5478],[290,71.8493],[291,72.4613],[292,72.6653],[339,72.6011],[345,71.8095],[347,71.2441],[350,69.776],[351,68.7659],[352,65.4653],[353,65.4653],[354,38.9648],[355,29.7981],[356,20.9981],[358,6.9994],[359,3.5611],[360,1.3354],[361,0.5008],[362,0.183],[363,0.0458],[364,0],[439,0]]
       }
       arm_r_pv_ik_ctrl.Location.Z {
        type=MovieSceneScriptingFloatChannel
        default=0
        thinned_from=115
        keys [[0,0],[266,0],[267,0.0058],[268,0.0185],[269,0.0593],[270,0.1189],[271,4.6406],[272,6.9709],[273,11.0209],[274,12.978],[275,0.3951],[276,0.5672],[277,2.0856],[278,3.5804],[279,4.2278],[280,4.6163],[281,4.752],[282,5.2938],[283,6.0389],[284,7.2504],[285,8.9818],[286,11.0898],[287,13.9166],[288,16.4244],[289,17.8143],[290,18.7621],[291,19.2078],[292,19.3564],[337,19.3639],[338,19.318],[339,19.1884],[340,18.987],[341,18.7258],[343,18.0719],[345,17.3222],[346,16.7887],[347,15.9894],[348,14.9774],[349,13.7671],[350,12.5281],[351,11.417],[352,9.4695],[353,9.4695],[354,13.6646],[355,15.7207],[356,22.4624],[358,7.4875],[359,3.8094],[360,1.4285],[361,0.5357],[362,0.1958],[363,0.0489],[364,0],[439,0]]
       }
       arm_r_pv_ik_ctrl.Rotation.X {
        type=MovieSceneScriptingFloatChannel
        default=0
        thinned_from=115
        keys [[0,0],[266,0],[267,-0.1394],[268,-0.4461],[269,-1.4659],[270,-3.2514],[271,-5.1026],[272,-6.9362],[273,-8.6835],[274,-9.687],[275,-10.0095],[276,-10.1812],[278,-10.3403],[280,-10.395],[281,-10.4041],[283,-10.4906],[284,-10.5668],[285,-10.6749],[286,-10.8062],[288,-11.1381],[289,-11.2246],[290,-11.2836],[292,-11.3205],[337,-11.321],[338,-11.3439],[339,-11.4084],[340,-11.5087],[341,-11.6387],[343,-11.9641],[345,-12.3372],[346,-12.6954],[347,-13.297],[348,-14.0045],[350,-15.1866],[351,-14.437],[352,-11.1341],[353,-11.1341],[354,-11.9556],[355,-8.95],[356,-6.3068],[358,-2.1023],[359,-1.0696],[360,-0.4011],[361,-0.1504],[362,-0.055],[363,-0.0137],[364,0],[439,0]]
       }
       arm_r_pv_ik_ctrl.Rotation.Y {
        type=MovieSceneScriptingFloatChannel
        default=0
        thinned_from=115
        keys [[0,0],[266,0],[267,0.2968],[268,0.9497],[269,3.0792],[270,6.4951],[271,9.7365],[272,12.7197],[273,15.553],[274,17.7623],[275,19.218],[276,20.4651],[278,25.3562],[279,26.4072],[280,27.0378],[281,27.322],[282,28.4547],[283,29.9935],[284,32.3432],[285,35.6766],[286,39.7244],[287,45.1487],[288,49.9594],[289,52.6253],[290,54.4428],[291,55.2974],[292,55.5822],[337,55.5965],[338,55.5303],[340,55.0528],[342,54.2299],[345,52.6505],[346,51.8808],[347,50.7274],[348,49.2672],[350,45.7331],[351,43.72],[352,39.1463],[353,39.1463],[354,38.8737],[355,29.5165],[356,20.7996],[358,6.9332],[359,3.5274],[360,1.3228],[361,0.496],[362,0.1813],[363,0.0453],[364,0],[439,0]]
       }
       arm_r_pv_ik_ctrl.Rotation.Z {
        type=MovieSceneScriptingFloatChannel
        default=0
        thinned_from=115
        keys [[0,0],[266,0],[267,0.9834],[268,3.1469],[269,10.2878],[271,34.4677],[272,46.3037],[273,57.6447],[274,64.6646],[275,67.822],[276,70.2707],[278,77.2481],[279,78.7474],[280,79.647],[281,80.0499],[282,81.6577],[283,83.8474],[284,87.1931],[285,91.9397],[286,97.7037],[287,105.4278],[288,112.2782],[289,116.0743],[290,118.6624],[291,119.8793],[292,120.285],[337,120.3053],[338,120.156],[340,119.078],[342,117.2208],[345,113.6557],[346,111.9183],[347,109.315],[348,106.019],[350,98.0418],[351,94.5614],[352,88.5862],[353,88.5862],[354,84.3686],[355,64.351],[356,45.3468],[358,15.1156],[359,7.6904],[360,2.8839],[361,1.0815],[362,0.3953],[363,0.0988],[364,0],[439,0]]
       }
       arm_r_pv_ik_ctrl.Scale.X {type=MovieSceneScriptingFloatChannel, default=1, thinned_from=115, keys [[0,1],[439,1]]}
       arm_r_pv_ik_ctrl.Scale.Y {type=MovieSceneScriptingFloatChannel, default=1, thinned_from=115, keys [[0,1],[439,1]]}
       arm_r_pv_ik_ctrl.Scale.Z {type=MovieSceneScriptingFloatChannel, default=1, thinned_from=115, keys [[0,1],[439,1]]}
       body_ctrl.Location.X {
        type=MovieSceneScriptingFloatChannel
        default=-9.9
        thinned_from=60
        keys [[0,-9.9],[8,-9.89],[15,-9.8625],[30,-9.7375],[51,-9.675],[56,-9.6],[59,-9.5],[62,-9.3],[66,-9],[70,-8.8],[75,-8.6906],[83,-8.625],[94,-8.6],[101,-8.65],[107,-8.8],[111,-9],[115,-9.25],[121,-9.4],[130,-9.5],[138,-9.55],[143,-9.6],[153,-9.9],[160,-10.05],[172,-10.15],[187,-10.2],[194,-10.05],[199,-9.725],[204,-9],[209,-7.75],[215,-6.75],[223,-6.375],[233,-6.25],[241,-6],[251,-4.75],[257,-3.5],[262,-3],[267,-3.5],[271,-4.75],[279,-9],[283,-10.5],[287,-11.125],[292,-11.25],[299,-11],[309,-10.75],[337,-10.5],[345,-10.25],[351,-9.5],[356,-8.75],[361,-8.5],[366,-8.75],[371,-9.5],[375,-10],[382,-10.25],[387,-10.2],[395,-10.075],[409,-9.9562],[439,-9.9]]
       }
       body_ctrl.Location.Y {
        type=MovieSceneScriptingFloatChannel
        default=4.2659
        thinned_from=58
        keys [[0,4.2659],[8,4.27],[22,4.3],[30,4.31],[45,4.315],[51,4.29],[59,4.225],[62,4.21],[66,4.205],[70,4.215],[75,4.25],[83,4.29],[92,4.31],[101,4.315],[107,4.3],[111,4.2659],[115,4.22],[121,4.19],[130,4.175],[143,4.17],[148,4.16],[160,4.1],[172,4.0625],[187,4.05],[194,4.0625],[199,4.0875],[204,4.1375],[209,4.2],[215,4.25],[223,4.2875],[239,4.3],[246,4.275],[251,4.1519],[257,3.8097],[262,3.25],[275,0.7951],[279,0.125],[283,-0.25],[287,-0.375],[292,-0.25],[299,-0.0537],[309,0.0803],[337,0.125],[345,0],[351,-0.1],[356,0.1],[361,0.6],[371,2.839],[375,4],[378,4.5],[382,4.6],[387,4.5],[395,4.3],[409,4.2784],[439,4.2659]]
       }
       body_ctrl.Location.Z {
        type=MovieSceneScriptingFloatChannel
        default=0.56
        thinned_from=109
        keys [[0,0.56],[8,0.55],[15,0.525],[22,0.5125],[30,0.525],[45,0.56],[51,0.5625],[56,0.5504],[62,0.5216],[66,0.5375],[70,0.5625],[76,0.5781],[94,0.5875],[101,0.5875],[107,0.585],[115,0.57],[121,0.5625],[130,0.5592],[138,0.5585],[160,0.5625],[172,0.5633],[187,0.5625],[215,0.5562],[225,0.5575],[233,0.5612],[241,0.58],[246,0.625],[251,0.675],[253,0.6722],[257,0.6561],[258,0.65],[259,0.6417],[261,0.6206],[262,0.6088],[263,0.5929],[264,0.57],[265,0.5418],[266,0.51],[267,0.4765],[268,0.4369],[270,0.3446],[271,0.3064],[272,0.2747],[273,0.2444],[274,0.2167],[275,0.193],[276,0.1721],[277,0.154],[278,0.1411],[279,0.1363],[280,0.1379],[281,0.1422],[282,0.1482],[283,0.1552],[284,0.1636],[286,0.1839],[287,0.193],[290,0.2165],[292,0.2308],[294,0.2439],[296,0.2558],[297,0.2609],[298,0.2652],[302,0.2768],[305,0.2836],[307,0.2864],[309,0.2875],[327,0.2875],[337,0.2686],[345,0.2308],[351,0.2497],[356,0.2686],[361,0.2497],[366,0.2071],[371,0.1787],[375,0.2354],[378,0.3205],[382,0.4],[387,0.47],[395,0.5263],[409,0.5477],[439,0.56]]
       }
       body_ctrl.Rotation.X {
        type=MovieSceneScriptingFloatChannel
        default=-23.1523
        thinned_from=59
        keys [[0,-23.1523],[45,-23.1523],[51,-23.2],[66,-23.1523],[107,-23.187],[246,-23.1523],[251,-22.9026],[257,-22.1537],[262,-20.6646],[275,-15],[283,-13.75],[292,-13.4076],[337,-13.3949],[345,-13.2212],[351,-12.7],[356,-11.7],[361,-11],[366,-13],[375,-20],[378,-22],[382,-23],[387,-23.25],[395,-23.1767],[439,-23.1523]]
       }
       body_ctrl.Rotation.Y {
        type=MovieSceneScriptingFloatChannel
        default=-1.8488
        thinned_from=60
        keys [[0,-1.8488],[8,-1.863],[15,-1.9055],[22,-1.9646],[30,-1.9911],[45,-2],[51,-1.9696],[56,-1.85],[59,-1.6],[62,-1.5],[66,-1.625],[70,-2.05],[75,-2.5],[83,-2.75],[94,-2.9],[101,-2.9144],[107,-2.95],[115,-3.25],[121,-3.4],[130,-3.45],[138,-3.4],[143,-3.25],[148,-3.0559],[153,-3.25],[160,-3.65],[170,-4],[182,-4.15],[192,-4.2],[199,-3.75],[204,-2.25],[209,1],[215,2.5],[223,2.875],[233,3.0437],[246,3.1],[252,2.75],[257,1.5],[262,-2],[267,-6.25],[271,-9.5],[275,-10],[279,-9],[284,-3.5],[293,1],[305,1.625],[321,2],[337,2.125],[345,2],[351,1.625],[361,0.25],[366,-1.25],[371,-2.25],[375,-2.75],[378,-3],[382,-3.125],[387,-2.75],[395,-2.125],[409,-1.904],[439,-1.8488]]
       }
       body_ctrl.Rotation.Z {
        type=MovieSceneScriptingFloatChannel
        default=18.0042
        keys [[0,18.0042],[8,18.0413],[15,18.1526],[22,18.3072],[30,18.3982],[45,18.45],[51,18.4],[55,18.25],[59,17.5],[62,16.3],[66,13.75],[70,12.5887],[75,12.1],[83,11.7228],[94,11.5],[101,11.5407],[107,11.8255],[111,12.6968],[115,14.25],[121,14.875],[130,15],[138,15.25],[143,16],[148,18],[153,19.4033],[160,20.0795],[172,20.3838],[185,20.5],[194,19],[199,16.25],[204,12.5],[209,10.5],[215,9.75],[223,9.375],[233,9],[241,8.625],[247,8],[251,7],[257,5],[262,0],[267,-7.5],[271,-13.75],[275,-17.5],[283,-20],[287,-19.25],[292,-18.25],[299,-17.5],[309,-17.125],[327,-17],[337,-17],[345,-19.1368],[351,-19.5],[356,-17],[361,-10],[366,2.5],[371,13.75],[375,18.75],[378,20.75],[381,21],[386,20],[395,18.75],[409,18.1478],[439,18.0042]]
       }
       body_ctrl.Scale.X {type=MovieSceneScriptingFloatChannel, default=1, thinned_from=64, keys [[0,1],[439,1]]}
       body_ctrl.Scale.Y {type=MovieSceneScriptingFloatChannel, default=1, thinned_from=64, keys [[0,1],[439,1]]}
       body_ctrl.Scale.Z {type=MovieSceneScriptingFloatChannel, default=1, thinned_from=64, keys [[0,1],[439,1]]}
       hips_ctrl.Location.X {type=MovieSceneScriptingFloatChannel, default=0.4, const [0.4,65]}
       hips_ctrl.Location.Y {type=MovieSceneScriptingFloatChannel, default=0, const [0,65]}
       hips_ctrl.Location.Z {type=MovieSceneScriptingFloatChannel, default=0, const [0,65]}
       hips_ctrl.Rotation.X {
        type=MovieSceneScriptingFloatChannel
        default=22.0947
        thinned_from=56
        keys [[0,22.0947],[246,22.0947],[251,21.5294],[264,13.5822],[279,1.8352],[283,0.4588],[287,0.1147],[292,0.0287],[299,0.0072],[309,0.0022],[327,0.0006],[337,0],[345,0.2571],[351,1.0286],[375,19.2359],[378,21.0226],[382,21.6926],[387,21.9942],[395,22.057],[439,22.0947]]
       }
       hips_ctrl.Rotation.Y {
        type=MovieSceneScriptingFloatChannel
        default=-2.7609
        thinned_from=61
        keys [[0,-2.7609],[8,-2.7625],[15,-2.775],[22,-2.8001],[30,-2.8125],[45,-2.8125],[51,-2.8],[56,-2.75],[62,-2.6],[66,-2.3407],[70,-2.05],[75,-1.9],[83,-1.85],[107,-1.85],[111,-1.775],[115,-1.4245],[121,-1.175],[130,-0.95],[143,-0.85],[153,-0.85],[160,-0.9],[172,-1.075],[181,-1.2],[199,-1.2604],[204,-1.3],[209,-1.4],[215,-1.6],[223,-1.8],[233,-1.885],[241,-2],[246,-2.125],[251,-2.375],[257,-2.75],[267,-3.5],[275,-3.75],[279,-3.975],[283,-3.7],[287,-3.2737],[292,-2.975],[299,-2.75],[309,-2.65],[327,-2.725],[337,-2.85],[345,-3.05],[351,-3.2598],[356,-3.325],[361,-3.35],[366,-3.3],[371,-3.1],[375,-2.8286],[378,-2.7],[382,-2.65],[387,-2.675],[395,-2.725],[409,-2.75],[439,-2.7609]]
       }
       hips_ctrl.Rotation.Z {
        type=MovieSceneScriptingFloatChannel
        default=-0.1738
        thinned_from=61
        keys [[0,-0.1738],[51,-0.1738],[56,-0.1713],[59,-0.1675],[62,-0.16],[66,-0.1198],[70,-0.1],[75,-0.0935],[246,-0.0935],[251,-0.0955],[257,-0.1017],[264,-0.1188],[279,-0.1671],[283,-0.1721],[287,-0.1733],[309,-0.1738],[439,-0.1738]]
       }
       hips_ctrl.Scale.X {type=MovieSceneScriptingFloatChannel, default=1, thinned_from=65, keys [[0,1],[439,1]]}
       hips_ctrl.Scale.Y {type=MovieSceneScriptingFloatChannel, default=1, thinned_from=65, keys [[0,1],[439,1]]}
       hips_ctrl.Scale.Z {type=MovieSceneScriptingFloatChannel, default=1, thinned_from=65, keys [[0,1],[439,1]]}
       hips_tan_ctrl.Location.X {
        type=MovieSceneScriptingFloatChannel
        default=-0.5737
        thinned_from=64
        keys [[0,-0.5737],[8,-0.5889],[15,-0.6343],[22,-0.7173],[30,-0.726],[45,-0.7354],[51,-0.6],[56,-0.7331],[62,-0.7263],[66,-0.6861],[70,-0.6624],[75,-0.657],[101,-0.6546],[107,-0.6501],[111,-0.6362],[115,-0.5778],[121,-0.5632],[130,-0.5584],[148,-0.5584],[153,-0.5679],[160,-0.5715],[172,-0.5617],[187,-0.5584],[194,-0.5593],[199,-0.5622],[204,-0.5737],[215,-0.5737],[223,-0.585],[233,-0.6191],[241,-0.6304],[246,-0.6342],[251,-0.6472],[257,-0.6865],[267,-0.8433],[275,-1.0991],[279,-1.196],[283,-1.2396],[287,-1.2505],[292,-1.2532],[337,-1.2541],[345,-1.2482],[351,-1.2306],[356,-1.1601],[361,-1.0456],[366,-1.0075],[371,-0.9038],[375,-0.8038],[378,-0.7611],[382,-0.7451],[387,-0.7378],[409,-0.7357],[439,-0.7354]]
       }
       hips_tan_ctrl.Location.Y {
        type=MovieSceneScriptingFloatChannel
        default=0.0977
        thinned_from=64
        keys [[0,0.0977],[8,-0.1657],[15,-0.9559],[22,-2.793],[30,-2.5471],[45,-2.7118],[51,0.1],[56,-2.7118],[148,-2.7176],[153,-0.9581],[160,-0.2982],[172,-2.1128],[187,-2.7176],[194,-2.5416],[199,-2.0138],[204,0.0977],[215,0.0977],[223,-0.4297],[233,-2.0119],[241,-2.5393],[246,-2.7151],[251,-2.6494],[257,-2.4523],[262,-2.0605],[267,-1.6641],[271,-1.1739],[275,-0.6894],[279,-0.2614],[283,-0.0688],[287,-0.0207],[292,-0.0086],[299,-0.0056],[309,-0.0049],[337,-0.0046],[345,-0.0371],[351,-0.1347],[356,-0.5249],[361,-1.1397],[366,-1.3446],[371,-1.872],[375,-2.3811],[378,-2.5878],[382,-2.6653],[387,-2.7001],[395,-2.7074],[439,-2.7118]]
       }
       hips_tan_ctrl.Location.Z {
        type=MovieSceneScriptingFloatChannel
        default=-0.0214
        thinned_from=64
        keys [[0,-0.0214],[8,-0.0889],[15,-0.2914],[22,-0.735],[30,-0.6993],[45,-0.7415],[51,0],[56,-0.7412],[62,-0.7405],[70,-0.7339],[107,-0.7329],[111,-0.7319],[115,-0.7272],[148,-0.7256],[153,-0.2854],[160,-0.1204],[172,-0.5743],[187,-0.7256],[194,-0.6816],[199,-0.5495],[204,-0.0214],[215,-0.0214],[223,-0.1545],[233,-0.5541],[241,-0.6873],[246,-0.7317],[251,-0.707],[257,-0.6329],[262,-0.4854],[267,-0.3363],[271,-0.2361],[275,-0.1371],[279,-0.116],[283,-0.1065],[287,-0.1041],[292,-0.1035],[337,-0.1033],[345,-0.1047],[351,-0.1086],[356,-0.1242],[361,-0.2192],[366,-0.2509],[371,-0.4222],[375,-0.5875],[378,-0.6837],[382,-0.7198],[387,-0.7361],[395,-0.7394],[409,-0.7412],[439,-0.7415]]
       }
       hips_tan_ctrl.Rotation.X {type=MovieSceneScriptingFloatChannel, default=0, thinned_from=64, keys [[0,0],[439,0]]}
       hips_tan_ctrl.Rotation.Y {type=MovieSceneScriptingFloatChannel, default=0, thinned_from=64, keys [[0,0],[439,0]]}
       hips_tan_ctrl.Rotation.Z {type=MovieSceneScriptingFloatChannel, default=0, thinned_from=64, keys [[0,0],[439,0]]}
       hips_tan_ctrl.Scale.X {type=MovieSceneScriptingFloatChannel, default=1, thinned_from=64, keys [[0,1],[439,1]]}
       hips_tan_ctrl.Scale.Y {type=MovieSceneScriptingFloatChannel, default=1, thinned_from=64, keys [[0,1],[439,1]]}
       hips_tan_ctrl.Scale.Z {type=MovieSceneScriptingFloatChannel, default=1, thinned_from=64, keys [[0,1],[439,1]]}
       spine_01_ctrl.Location.X {type=MovieSceneScriptingFloatChannel, default=0, thinned_from=64, keys [[0,0],[439,0]]}
       spine_01_ctrl.Location.Y {type=MovieSceneScriptingFloatChannel, default=0, thinned_from=64, keys [[0,0],[439,0]]}
       spine_01_ctrl.Location.Z {type=MovieSceneScriptingFloatChannel, default=0, thinned_from=64, keys [[0,0],[439,0]]}
       spine_01_ctrl.Rotation.X {
        type=MovieSceneScriptingFloatChannel
        default=5.9324
        thinned_from=62
        keys [[0,5.9324],[246,5.9324],[251,5.9],[257,5.75],[262,5.5],[267,4.9],[271,3.75],[275,1.8375],[279,0.6891],[283,0.1723],[287,0.0431],[292,0.0108],[299,0.0027],[309,0.0008],[337,0],[345,0.069],[351,0.2762],[356,1.1047],[366,4.1],[371,5.125],[376,5.625],[381,5.85],[387,5.9054],[395,5.9223],[439,5.9324]]
       }
       spine_01_ctrl.Rotation.Y {type=MovieSceneScriptingFloatChannel, default=2.7657, const [2.7657,64]}
       spine_01_ctrl.Rotation.Z {type=MovieSceneScriptingFloatChannel, default=2.6153, const [2.6153,64]}
       spine_01_ctrl.Scale.X {type=MovieSceneScriptingFloatChannel, default=1, thinned_from=64, keys [[0,1],[439,1]]}
       spine_01_ctrl.Scale.Y {type=MovieSceneScriptingFloatChannel, default=1, thinned_from=64, keys [[0,1],[439,1]]}
       spine_01_ctrl.Scale.Z {type=MovieSceneScriptingFloatChannel, default=1, thinned_from=64, keys [[0,1],[439,1]]}
       spine_02_ctrl.Location.X {type=MovieSceneScriptingFloatChannel, default=0, thinned_from=64, keys [[0,0],[439,0]]}
       spine_02_ctrl.Location.Y {type=MovieSceneScriptingFloatChannel, default=0, thinned_from=64, keys [[0,0],[439,0]]}
       spine_02_ctrl.Location.Z {type=MovieSceneScriptingFloatChannel, default=0, thinned_from=64, keys [[0,0],[439,0]]}
       spine_02_ctrl.Rotation.X {
        type=MovieSceneScriptingFloatChannel
        default=8.839
        thinned_from=59
        keys [[0,8.839],[246,8.839],[251,8.7803],[257,8.6042],[267,7.8],[275,6.3],[283,3.7],[287,1.5],[292,0.375],[299,0.0938],[309,0.0288],[327,0.0072],[337,0],[345,0.1029],[351,0.5],[356,1.6459],[361,3.5],[366,5.5],[371,7],[375,8],[382,8.6782],[395,8.8239],[439,8.839]]
       }
       spine_02_ctrl.Rotation.Y {
        type=MovieSceneScriptingFloatChannel
        default=-2.1957
        thinned_from=63
        keys [[0,-2.1957],[45,-2.1957],[51,-2.2],[66,-2.1957],[107,-2.1988],[246,-2.1957],[251,-1.8982],[257,-1.0055],[262,0.5],[267,2.5651],[271,4.7243],[275,6.8583],[279,8.4683],[283,9.1927],[287,9.3739],[292,9.4191],[337,9.4342],[345,9.2989],[351,8.8928],[356,7.5],[366,2],[371,-0.5],[375,-2],[378,-2.5],[382,-2.75],[387,-2.625],[395,-2.375],[409,-2.2502],[439,-2.1957]]
       }
       spine_02_ctrl.Rotation.Z {type=MovieSceneScriptingFloatChannel, default=0, const [0,64]}
       spine_02_ctrl.Scale.X {type=MovieSceneScriptingFloatChannel, default=1, thinned_from=64, keys [[0,1],[439,1]]}
       spine_02_ctrl.Scale.Y {type=MovieSceneScriptingFloatChannel, default=1, thinned_from=64, keys [[0,1],[439,1]]}
       spine_02_ctrl.Scale.Z {type=MovieSceneScriptingFloatChannel, default=1, thinned_from=64, keys [[0,1],[439,1]]}
       spine_03_ctrl.Location.X {type=MovieSceneScriptingFloatChannel, default=0, thinned_from=64, keys [[0,0],[439,0]]}
       spine_03_ctrl.Location.Y {type=MovieSceneScriptingFloatChannel, default=0, thinned_from=64, keys [[0,0],[439,0]]}
       spine_03_ctrl.Location.Z {type=MovieSceneScriptingFloatChannel, default=0, thinned_from=64, keys [[0,0],[439,0]]}
       spine_03_ctrl.Rotation.X {
        type=MovieSceneScriptingFloatChannel
        default=16.2127
        thinned_from=61
        keys [[0,16.2127],[8,16.2209],[30,16.4],[45,16.425],[56,16.525],[59,16.475],[62,16.325],[66,16.2127],[72,16.15],[94,16.075],[101,16.1],[115,16.325],[121,16.35],[138,16.275],[148,16.15],[153,16.125],[183,16.275],[204,16.325],[209,16.3],[223,16.15],[233,16.125],[246,16.175],[251,16.125],[257,15.5595],[262,14.5855],[267,13.1],[271,11.25],[275,9.3],[279,8.9],[283,8.7],[287,8],[292,4.8956],[299,3.6239],[309,3],[327,2.75],[337,3.2],[345,4.25],[351,5.5],[356,7],[361,9],[375,16],[378,17.25],[382,17.75],[387,17.5],[395,17],[409,16.5],[439,16.2127]]
       }
       spine_03_ctrl.Rotation.Y {
        type=MovieSceneScriptingFloatChannel
        default=-1.7965
        thinned_from=64
        keys [[0,-1.7965],[8,-1.7965],[30,-1.725],[45,-1.7],[51,-1.65],[56,-1.575],[59,-1.475],[62,-1.3],[66,-0.8],[70,-0.45],[75,-0.5254],[83,-0.6],[94,-0.6375],[101,-0.6625],[107,-0.775],[111,-1.25],[115,-2.5],[121,-3.25],[130,-3.5],[138,-3.45],[143,-3.2434],[148,-2.95],[153,-2.7],[160,-2.6],[172,-2.55],[185,-2.65],[194,-2.9],[199,-3.1],[209,-3.6],[215,-3.65],[223,-3.5],[233,-3.2434],[241,-3],[246,-2.7],[251,-2.25],[257,-1.375],[262,-0.625],[267,-1.6],[271,-3.25],[275,-4],[279,-3],[283,0.5],[287,3],[292,4.5],[299,5.35],[309,5.5],[327,5.625],[337,5.625],[345,5.5],[351,5.25],[356,4.3332],[361,2.6596],[366,1],[371,-0.25],[375,-1.125],[378,-1.625],[382,-1.875],[387,-1.9625],[395,-1.9],[409,-1.8094],[439,-1.7965]]
       }
       spine_03_ctrl.Rotation.Z {
        type=MovieSceneScriptingFloatChannel
        default=3.1598
        thinned_from=64
        keys [[0,3.1598],[8,3.175],[22,3.225],[30,3.2375],[45,3.25],[51,3.2],[56,2.9865],[59,2.7297],[62,2.125],[66,0.75],[70,-1.75],[75,-2.7],[83,-3],[94,-2.8],[101,-2.5],[107,-2],[111,-1],[115,1],[121,2.0923],[130,2.4918],[138,2.625],[143,2.3673],[148,1.7],[153,2.5],[160,4.5],[172,5.375],[186,5.625],[194,6],[199,6.625],[204,7],[209,5.875],[215,3.625],[223,2.25],[233,1.75],[241,1.5],[246,2],[251,2.75],[257,3],[262,2.5],[267,1],[271,-1],[275,-3.4],[279,-6.5],[283,-9.7],[287,-12.1199],[292,-13.5],[299,-14.2135],[309,-14.5],[327,-14],[337,-13],[345,-11.5],[351,-9.5],[371,-1.5],[375,0.5],[378,2.5],[382,3.9],[387,4.5],[395,4],[409,3.4134],[439,3.2179]]
       }
       spine_03_ctrl.Scale.X {type=MovieSceneScriptingFloatChannel, default=1, thinned_from=64, keys [[0,1],[439,1]]}
       spine_03_ctrl.Scale.Y {type=MovieSceneScriptingFloatChannel, default=1, thinned_from=64, keys [[0,1],[439,1]]}
       spine_03_ctrl.Scale.Z {type=MovieSceneScriptingFloatChannel, default=1, thinned_from=64, keys [[0,1],[439,1]]}
       spine_04_ctrl.Location.X {type=MovieSceneScriptingFloatChannel, default=0, thinned_from=64, keys [[0,0],[439,0]]}
       spine_04_ctrl.Location.Y {type=MovieSceneScriptingFloatChannel, default=0, thinned_from=64, keys [[0,0],[439,0]]}
       spine_04_ctrl.Location.Z {type=MovieSceneScriptingFloatChannel, default=0, thinned_from=64, keys [[0,0],[439,0]]}
       spine_04_ctrl.Rotation.X {
        type=MovieSceneScriptingFloatChannel
        default=2.3
        thinned_from=64
        keys [[0,2.3],[246,2.3],[251,2.3187],[257,2.375],[267,2.6],[271,1.7827],[275,0.975],[279,0.3656],[283,0.0914],[287,0.0229],[292,0.0057],[299,0.0014],[309,0.0004],[337,0],[345,0.0268],[351,0.1071],[356,0.4283],[361,1],[366,1.6],[371,2.1],[375,2.275],[378,2.2986],[439,2.3]]
       }
       spine_04_ctrl.Rotation.Y {
        type=MovieSceneScriptingFloatChannel
        default=7.2486
        thinned_from=57
        keys [[0,7.2486],[8,7.2486],[15,7.3],[22,7.375],[30,7.4125],[45,7.4375],[51,7.425],[56,7.2991],[59,7.2],[62,7.4504],[66,8.3333],[70,8.9],[75,9.1],[83,9.15],[94,9.05],[101,8.895],[111,8.6],[115,8.5025],[121,8.4402],[130,8.4194],[138,8.4402],[143,8.5025],[148,8.7521],[153,8.95],[160,9.1],[172,9.2],[187,9.25],[209,9.2625],[215,9.35],[223,9.55],[233,9.7],[246,9.75],[251,9.625],[257,9],[262,8],[283,2.75],[287,2.25],[292,2],[299,2.125],[323,2.375],[345,2.375],[351,3.5],[356,5.5],[361,6.75],[366,7.25],[371,7.5],[375,7.625],[378,7.5],[382,7.1854],[387,6.875],[395,7],[409,7.1865],[439,7.2486]]
       }
       spine_04_ctrl.Rotation.Z {
        type=MovieSceneScriptingFloatChannel
        default=-0.4428
        thinned_from=57
        keys [[0,-0.4428],[246,-0.4428],[251,-0.4314],[257,-0.3975],[275,-0.0981],[279,-0.0368],[283,-0.0092],[287,-0.0023],[292,-0.0006],[299,-0.0001],[337,0],[345,-0.0052],[351,-0.0206],[375,-0.3855],[378,-0.4213],[382,-0.4347],[387,-0.4408],[395,-0.442],[439,-0.4428]]
       }
       spine_04_ctrl.Scale.X {type=MovieSceneScriptingFloatChannel, default=1, thinned_from=64, keys [[0,1],[439,1]]}
       spine_04_ctrl.Scale.Y {type=MovieSceneScriptingFloatChannel, default=1, thinned_from=64, keys [[0,1],[439,1]]}
       spine_04_ctrl.Scale.Z {type=MovieSceneScriptingFloatChannel, default=1, thinned_from=64, keys [[0,1],[439,1]]}
       head_ctrl.Location.X {type=MovieSceneScriptingFloatChannel, default=0, const [0,64]}
       head_ctrl.Location.Y {type=MovieSceneScriptingFloatChannel, default=0, const [0,64]}
       head_ctrl.Location.Z {type=MovieSceneScriptingFloatChannel, default=0, const [0,64]}
       head_ctrl.Rotation.X {
        type=MovieSceneScriptingFloatChannel
        default=14.6517
        keys [[0,14.6517],[8,14.7069],[15,14.8727],[22,15.875],[30,16.25],[45,16.375],[51,16.5],[56,16.375],[59,14.7],[62,9.5],[66,6.6561],[70,5.639],[75,5.2535],[83,5.125],[94,5.125],[101,5.5375],[107,6.75],[111,8.0661],[115,8.7457],[121,9.0302],[130,9.125],[138,8.4542],[143,7.25],[148,6.5],[153,7.6443],[160,8.411],[172,8.7559],[187,8.8709],[194,9.0532],[199,9.6],[204,10.1],[209,9.2542],[215,8.75],[223,9],[233,9.8],[239,10.125],[246,10.25],[251,8.75],[257,5.5],[262,2.25],[267,0.25],[271,-1],[275,-2],[279,-1],[283,1],[287,2],[290,2.25],[299,-0.125],[309,-1.875],[327,-2.625],[337,-3],[345,-3.5],[351,-4.5],[356,-5.5],[361,-6],[366,-4.25],[371,1.25],[378,12.5],[382,15.5],[387,16.5],[395,16],[409,14.9614],[439,14.6517]]
       }
       head_ctrl.Rotation.Y {
        type=MovieSceneScriptingFloatChannel
        default=5.1301
        keys [[0,5.1301],[8,5.2998],[15,5.809],[22,6.25],[30,6.5],[45,6.625],[51,6.75],[56,7.5],[59,8.75],[62,10.4],[66,11.8],[70,11.2],[75,9.9],[83,8.55],[93,7.75],[101,7],[107,5.2776],[111,0.6],[115,-2.9],[121,-4.2855],[130,-5.25],[138,-5.5],[144,-5],[148,-3],[153,2.4242],[160,4.9083],[172,6.0262],[187,6.3988],[194,6.1801],[199,5.5241],[204,3.4],[209,-0.1],[215,-2.9],[223,-4.25],[232,-5],[241,-5.525],[246,-5.7],[251,-4.4063],[257,-1],[262,2.3],[267,5.3],[271,7.1608],[275,8],[279,5.2],[283,-2.2216],[287,-4.5],[292,-5.5],[299,-6],[309,-5.5],[327,-4],[337,-3.25],[345,-2.5],[351,-1],[356,3],[361,8],[366,10.6],[371,11],[375,10.5],[378,8.0813],[382,6.2368],[387,5.4068],[395,5.125],[409,5.1],[439,5.1301]]
       }
       head_ctrl.Rotation.Z {
        type=MovieSceneScriptingFloatChannel
        default=11.593
        keys [[0,11.593],[8,11.7369],[15,12.5],[22,13.5],[30,14],[45,14.25],[51,13],[56,9.5],[59,2.5],[62,-14.4358],[66,-24.1966],[70,-28.6241],[75,-30.5],[83,-31.5],[94,-32],[101,-31.5],[107,-28.75],[111,-20.5086],[115,-12.1703],[121,-8],[130,-6.25],[138,-5.25],[143,-2.9873],[148,7.5],[153,20],[161,27.5],[171,31],[184,33],[194,33.5],[199,31.5],[204,23.75],[209,2.5],[215,-5],[223,-8.5],[233,-10.5],[241,-11],[246,-10.5],[251,-9],[257,-6],[262,-4],[267,-3.5],[271,-4.5],[275,-9.5],[279,-16.25],[283,-21.6216],[287,-25],[292,-26.25],[299,-25],[309,-23.5],[325,-23],[337,-22.7204],[345,-22.4372],[351,-20],[356,-15],[361,-7.5],[366,1.25],[371,8.75],[375,13],[378,14.25],[382,13.5],[387,11.6424],[395,10.5],[409,11.3197],[439,11.593]]
       }
       head_ctrl.Scale.X {type=MovieSceneScriptingFloatChannel, default=1, thinned_from=64, keys [[0,1],[439,1]]}
       head_ctrl.Scale.Y {type=MovieSceneScriptingFloatChannel, default=1, thinned_from=64, keys [[0,1],[439,1]]}
       head_ctrl.Scale.Z {type=MovieSceneScriptingFloatChannel, default=1, thinned_from=64, keys [[0,1],[439,1]]}
       clavicle_l_ctrl.Location.X {type=MovieSceneScriptingFloatChannel, default=0, thinned_from=64, keys [[0,0],[439,0]]}
       clavicle_l_ctrl.Location.Y {type=MovieSceneScriptingFloatChannel, default=0, thinned_from=64, keys [[0,0],[439,0]]}
       clavicle_l_ctrl.Location.Z {type=MovieSceneScriptingFloatChannel, default=0, thinned_from=64, keys [[0,0],[439,0]]}
       clavicle_l_ctrl.Rotation.X {type=MovieSceneScriptingFloatChannel, default=0, const [0,64]}
       clavicle_l_ctrl.Rotation.Y {
        type=MovieSceneScriptingFloatChannel
        default=14.0935
        thinned_from=63
        keys [[0,14.0935],[15,14.0935],[30,14.45],[45,14.5],[51,14.45],[56,14.0437],[59,13.6],[62,13.125],[66,12.625],[70,12.5],[75,12.625],[83,12.875],[94,13],[101,13],[107,12.875],[111,12.625],[115,12.25],[121,12],[130,11.875],[143,11.875],[148,12.125],[153,12.5],[160,12.9102],[172,13.25],[187,12.875],[194,12.7701],[199,12.9802],[204,13.8208],[209,13.25],[215,12.25],[223,10.875],[233,9.625],[241,9.125],[251,8.625],[257,8.25],[262,7.75],[267,6.75],[271,3.731],[279,-5.9358],[283,-8.8319],[287,-7.9239],[292,-5.881],[299,-5.3702],[309,-5.125],[327,-5],[337,-4.875],[345,-5.375],[351,-7.1],[356,-8.7176],[361,-7.5],[366,-2.5],[371,3.75],[375,8],[378,10.5],[382,12.5],[387,13.5],[395,14.125],[409,14.25],[439,14.0935]]
       }
       clavicle_l_ctrl.Rotation.Z {
        type=MovieSceneScriptingFloatChannel
        default=-21.5044
        thinned_from=61
        keys [[0,-21.5044],[246,-21.5044],[251,-20.5819],[257,-17.8141],[262,-12.311],[267,-2.5],[275,22.5],[279,31.9005],[283,36.1307],[287,37.5357],[337,37.5407],[345,36.8536],[351,34.792],[356,26.5459],[361,13.427],[371,-8.75],[375,-16.25],[380,-20],[387,-21.5044],[439,-21.5044]]
       }
       clavicle_l_ctrl.Scale.X {type=MovieSceneScriptingFloatChannel, default=1, thinned_from=64, keys [[0,1],[439,1]]}
       clavicle_l_ctrl.Scale.Y {type=MovieSceneScriptingFloatChannel, default=1, thinned_from=64, keys [[0,1],[439,1]]}
       clavicle_l_ctrl.Scale.Z {type=MovieSceneScriptingFloatChannel, default=1, thinned_from=64, keys [[0,1],[439,1]]}
       clavicle_r_ctrl.Location.X {type=MovieSceneScriptingFloatChannel, default=0, thinned_from=64, keys [[0,0],[439,0]]}
       clavicle_r_ctrl.Location.Y {type=MovieSceneScriptingFloatChannel, default=0, thinned_from=64, keys [[0,0],[439,0]]}
       clavicle_r_ctrl.Location.Z {type=MovieSceneScriptingFloatChannel, default=0, thinned_from=64, keys [[0,0],[439,0]]}
       clavicle_r_ctrl.Rotation.X {type=MovieSceneScriptingFloatChannel, default=0, thinned_from=64, keys [[0,0],[439,0]]}
       clavicle_r_ctrl.Rotation.Y {
        type=MovieSceneScriptingFloatChannel
        default=0.8223
        thinned_from=64
        keys [[0,0.8223],[45,0.8223],[51,0.8],[56,0.8807],[59,0.9672],[62,1.0557],[66,2.0771],[70,2.5443],[75,2.6611],[83,2.6904],[94,2.7],[101,2.6838],[107,2.5703],[111,2.223],[115,1.1725],[121,0.9098],[130,0.8223],[148,0.8223],[153,0.6209],[160,0.5453],[172,0.5113],[187,0.5],[194,0.5125],[199,0.55],[204,0.7],[209,0.275],[215,-0.2312],[223,-0.2629],[233,-0.3578],[241,-0.3895],[246,-0.4],[251,-0.2812],[257,0.075],[262,0.7833],[267,1.5],[271,3.7632],[275,6],[279,-0.6018],[283,-3.5727],[287,-2.9545],[292,-1.5636],[299,-1.2159],[309,-1.1356],[327,-1.1089],[337,-1.1],[345,-2.2],[351,-5.5],[356,-2.8018],[361,-0.7005],[366,0],[371,-1.2719],[375,-2.5],[378,-1.6694],[382,-0.1121],[387,0.5887],[395,0.7347],[409,0.7903],[439,0.8223]]
       }
       clavicle_r_ctrl.Rotation.Z {
        type=MovieSceneScriptingFloatChannel
        default=-20
        thinned_from=64
        keys [[0,-20],[111,-20],[115,-15],[121,-12.5],[130,-11.25],[138,-10],[257,-10],[262,-7.3322],[267,-3.0524],[271,7.647],[275,18.222],[279,26.2],[283,29.79],[287,30.0892],[292,30.7623],[299,30.9306],[337,30.9867],[345,29.54],[351,25.2],[356,21.0121],[361,9.1106],[366,5.1434],[371,-5.435],[375,-15],[378,-17.5],[382,-20],[439,-20]]
       }
       clavicle_r_ctrl.Scale.X {type=MovieSceneScriptingFloatChannel, default=1, thinned_from=64, keys [[0,1],[439,1]]}
       clavicle_r_ctrl.Scale.Y {type=MovieSceneScriptingFloatChannel, default=1, thinned_from=64, keys [[0,1],[439,1]]}
       clavicle_r_ctrl.Scale.Z {type=MovieSceneScriptingFloatChannel, default=1, thinned_from=64, keys [[0,1],[439,1]]}
       neck_01_ctrl.Location.X {
        type=MovieSceneScriptingFloatChannel
        default=1.3609
        thinned_from=64
        keys [[0,1.3609],[107,1.3609],[111,1.4],[115,1.625],[121,1.6715],[267,1.675],[271,1.8487],[275,2.257],[279,2.6331],[283,2.7569],[287,2.7879],[292,2.7956],[337,2.7982],[345,2.7815],[351,2.7313],[356,2.5306],[361,2.2112],[366,1.7922],[371,1.5013],[375,1.3633],[439,1.3609]]
       }
       neck_01_ctrl.Location.Y {
        type=MovieSceneScriptingFloatChannel
        default=-2.9424
        thinned_from=64
        keys [[0,-2.9424],[267,-2.9424],[271,-2.3219],[275,-0.8993],[279,0.4114],[283,0.823],[287,0.9259],[292,0.9516],[299,0.958],[337,0.9602],[345,0.9148],[351,0.7785],[356,0.2335],[361,-0.6336],[366,-1.7712],[371,-2.5612],[375,-2.9358],[439,-2.9424]]
       }
       neck_01_ctrl.Location.Z {
        type=MovieSceneScriptingFloatChannel
        default=-0.217
        thinned_from=64
        keys [[0,-0.217],[267,-0.217],[271,-0.4447],[275,-0.9668],[279,-1.4478],[283,-1.5989],[287,-1.6366],[292,-1.6461],[337,-1.6492],[345,-1.6326],[351,-1.5826],[356,-1.3825],[361,-1.0643],[366,-0.6468],[371,-0.3569],[375,-0.2195],[395,-0.2195],[409,-0.2179],[439,-0.217]]
       }
       neck_01_ctrl.Rotation.X {type=MovieSceneScriptingFloatChannel, default=0, thinned_from=64, keys [[0,0],[439,0]]}
       neck_01_ctrl.Rotation.Y {type=MovieSceneScriptingFloatChannel, default=0, thinned_from=64, keys [[0,0],[439,0]]}
       neck_01_ctrl.Rotation.Z {type=MovieSceneScriptingFloatChannel, default=0, thinned_from=64, keys [[0,0],[439,0]]}
       neck_01_ctrl.Scale.X {type=MovieSceneScriptingFloatChannel, default=1, thinned_from=64, keys [[0,1],[439,1]]}
       neck_01_ctrl.Scale.Y {type=MovieSceneScriptingFloatChannel, default=1, thinned_from=64, keys [[0,1],[439,1]]}
       neck_01_ctrl.Scale.Z {type=MovieSceneScriptingFloatChannel, default=1, thinned_from=64, keys [[0,1],[439,1]]}
       neck_02_ctrl.Location.X {type=MovieSceneScriptingFloatChannel, default=0, thinned_from=64, keys [[0,0],[439,0]]}
       neck_02_ctrl.Location.Y {type=MovieSceneScriptingFloatChannel, default=0, thinned_from=64, keys [[0,0],[439,0]]}
       neck_02_ctrl.Location.Z {type=MovieSceneScriptingFloatChannel, default=0, thinned_from=64, keys [[0,0],[439,0]]}
       neck_02_ctrl.Rotation.X {type=MovieSceneScriptingFloatChannel, default=0, thinned_from=64, keys [[0,0],[439,0]]}
       neck_02_ctrl.Rotation.Y {type=MovieSceneScriptingFloatChannel, default=0, thinned_from=64, keys [[0,0],[439,0]]}
       neck_02_ctrl.Rotation.Z {type=MovieSceneScriptingFloatChannel, default=0, thinned_from=64, keys [[0,0],[439,0]]}
       neck_02_ctrl.Scale.X {type=MovieSceneScriptingFloatChannel, default=1, thinned_from=64, keys [[0,1],[439,1]]}
       neck_02_ctrl.Scale.Y {type=MovieSceneScriptingFloatChannel, default=1, thinned_from=64, keys [[0,1],[439,1]]}
       neck_02_ctrl.Scale.Z {type=MovieSceneScriptingFloatChannel, default=1, thinned_from=64, keys [[0,1],[439,1]]}
       head_ik_ctrl.Location.X {
        type=MovieSceneScriptingFloatChannel
        default=1.1879
        thinned_from=64
        keys [[0,1.1879],[8,1.3016],[15,1.6427],[22,2.4003],[30,2.3415],[45,2.4164],[51,1.2],[56,2.3863],[62,2.3423],[66,2.0769],[70,1.972],[75,1.9479],[83,1.916],[94,1.9056],[101,1.9113],[107,1.9516],[111,2.075],[115,2.4012],[121,2.4855],[130,2.5136],[138,2.5189],[143,2.5348],[148,2.5985],[153,1.7168],[160,1.3862],[172,2.4648],[187,2.8244],[194,2.7221],[199,2.4153],[204,1.1879],[215,1.1879],[223,1.3924],[233,2.0059],[241,2.2119],[246,2.2806],[251,2.2216],[257,2.0446],[262,1.6928],[267,1.3368],[271,1.8144],[275,2.2865],[279,2.9276],[283,3.2161],[287,3.2883],[292,3.3063],[337,3.3123],[345,3.278],[351,3.1751],[356,2.7636],[361,2.3608],[366,2.2265],[375,2.1828],[378,2.3203],[382,2.3719],[387,2.3951],[409,2.4109],[439,2.4029]]
       }
       head_ik_ctrl.Location.Y {
        type=MovieSceneScriptingFloatChannel
        default=0.0643
        thinned_from=64
        keys [[0,0.0643],[8,-0.1429],[15,-0.7646],[22,-2.1006],[30,-2.0033],[45,-2.1284],[51,0.1],[56,-2.1613],[62,-2.2064],[66,-2.4425],[70,-2.5121],[83,-2.5529],[94,-2.5608],[101,-2.5587],[107,-2.5443],[111,-2.5001],[115,-2.2803],[121,-2.2247],[130,-2.2061],[138,-2.1996],[143,-2.1801],[148,-2.1021],[153,-0.7481],[160,-0.2404],[172,-1.4045],[187,-1.7926],[194,-1.6765],[199,-1.3284],[204,0.0643],[215,0.0643],[223,-0.4004],[233,-1.7946],[241,-2.2591],[246,-2.414],[251,-2.4526],[257,-2.5687],[267,-3.0331],[271,-2.1258],[275,-1.229],[279,-0.8408],[283,-0.6661],[287,-0.6224],[292,-0.6115],[299,-0.6088],[337,-0.6079],[345,-0.613],[351,-0.6285],[356,-0.6906],[361,-0.9425],[366,-1.0265],[371,-1.4846],[375,-1.9268],[378,-2.0624],[382,-2.1132],[387,-2.1361],[395,-2.1409],[409,-2.1327],[439,-2.1437]]
       }
       head_ik_ctrl.Location.Z {
        type=MovieSceneScriptingFloatChannel
        default=-0.069
        thinned_from=64
        keys [[0,-0.069],[8,-0.022],[15,0.1193],[22,0.3987],[30,0.4039],[45,0.4333],[51,-0.1],[56,0.4377],[62,0.4514],[66,0.4576],[70,0.4757],[75,0.4755],[83,0.4731],[101,0.4712],[107,0.4634],[111,0.4394],[115,0.4211],[121,0.4032],[130,0.3973],[138,0.3984],[143,0.4018],[148,0.4151],[153,0.1125],[160,-0.0009],[172,0.2926],[187,0.3905],[194,0.3618],[199,0.2756],[204,-0.069],[215,-0.069],[223,0.0503],[233,0.4084],[241,0.523],[246,0.5612],[251,0.5634],[257,0.5701],[267,0.597],[271,0.0159],[275,-0.5584],[279,-0.4356],[283,-0.3803],[287,-0.3665],[292,-0.363],[299,-0.3622],[337,-0.3619],[345,-0.3709],[351,-0.3978],[356,-0.5055],[361,-0.4521],[366,-0.4343],[371,-0.1435],[375,0.1373],[378,0.3223],[382,0.3917],[387,0.4229],[395,0.4295],[409,0.4299],[439,0.4334]]
       }
       head_ik_ctrl.Rotation.X {
        type=MovieSceneScriptingFloatChannel
        default=14.6517
        thinned_from=64
        keys [[0,14.6517],[15,14.6517],[22,15.4736],[30,14.9129],[45,15],[51,14.7],[56,15],[59,11.9544],[62,8.8367],[66,6.6561],[70,5.639],[75,5.3848],[83,5.3208],[94,5.3],[101,5.394],[107,6.0522],[111,8.0661],[115,8.7457],[121,8.9156],[130,8.9722],[138,8.8177],[143,8.3542],[148,6.5],[153,11.5948],[160,13.5054],[172,10.0295],[187,8.8709],[194,9.2322],[199,10.3161],[204,14.6517],[215,14.6517],[223,13.5868],[233,10.3921],[241,9.3272],[246,8.9722],[251,8.3802],[257,6.6042],[262,3.0729],[267,-0.5],[271,-3.2661],[275,-6],[279,-3.3878],[283,-2.2123],[287,-1.9184],[292,-1.8449],[299,-1.8266],[309,-1.8223],[337,-1.8204],[345,-1.861],[351,-1.9825],[356,-2.4687],[361,-5.6422],[366,-6.7],[371,-0.493],[375,5.5],[378,11.2198],[382,13.3647],[387,14.3299],[395,14.531],[409,14.9614],[439,14.6517]]
       }
       head_ik_ctrl.Rotation.Y {
        type=MovieSceneScriptingFloatChannel
        default=9.8268
        thinned_from=64
        keys [[0,9.8268],[8,9.3865],[15,8.0655],[22,6.1001],[30,6.2077],[45,6.2],[51,9.8],[56,6.1275],[59,7.6062],[62,9.1198],[66,9.9706],[70,10.3676],[75,9.4169],[83,8.4989],[94,8.2],[101,7.8893],[107,5.7145],[111,-0.9403],[115,-3.7956],[121,-4.5094],[130,-4.7474],[138,-4.5757],[143,-4.0605],[148,-2],[153,5.3917],[160,8.1636],[172,6.84],[187,6.3988],[194,6.613],[199,7.2558],[204,9.8268],[215,9.8268],[223,7.0583],[233,-1.2472],[241,-4.5868],[246,-5.7],[251,-4.4063],[257,-0.525],[262,7.1921],[267,15],[271,11.9825],[275,9],[279,0.4403],[283,-3.4115],[287,-4.3745],[292,-4.6152],[299,-4.6754],[309,-4.6893],[337,-4.6954],[345,-4.3892],[351,-3.4704],[356,0.2046],[361,6.0511],[366,8],[371,10.5439],[375,13],[378,8.0813],[382,6.2368],[387,5.4068],[395,5.2338],[409,5.168],[439,5.1301]]
       }
       head_ik_ctrl.Rotation.Z {
        type=MovieSceneScriptingFloatChannel
        default=8.187
        thinned_from=64
        keys [[0,8.187],[8,8.5063],[15,9.4642],[22,12.4001],[30,12.0737],[45,12.5],[51,8.2],[56,7.5],[59,-3.3396],[62,-14.4358],[66,-24.1966],[70,-28.6241],[75,-29.731],[83,-30.0094],[94,-30.1],[101,-29.7476],[107,-27.2806],[115,-12.1703],[121,-10.2799],[130,-9.6497],[138,-7.9841],[143,-2.9873],[148,17],[153,11.4919],[160,9.4263],[172,25.9816],[187,31.5],[194,30.0429],[199,25.6717],[204,8.187],[215,8.187],[223,4.8426],[233,-5.1905],[241,-8.5349],[246,-9.6497],[251,-9.7341],[257,-9.9873],[267,-11],[275,-11],[279,-18.3253],[283,-21.6216],[287,-22.4457],[292,-22.6517],[299,-22.7032],[337,-22.7204],[345,-22.3211],[351,-21.123],[356,-16.3309],[361,-8.707],[366,-6.1657],[371,3.5853],[375,13],[378,12.1206],[382,11.7908],[387,11.6424],[395,11.6115],[439,11.593]]
       }
       head_ik_ctrl.Scale.X {type=MovieSceneScriptingFloatChannel, default=1, thinned_from=64, keys [[0,1],[439,1]]}
       head_ik_ctrl.Scale.Y {type=MovieSceneScriptingFloatChannel, default=1, thinned_from=64, keys [[0,1],[439,1]]}
       head_ik_ctrl.Scale.Z {type=MovieSceneScriptingFloatChannel, default=1, thinned_from=64, keys [[0,1],[439,1]]}
       chest_ctrl.Location.X {
        type=MovieSceneScriptingFloatChannel
        default=-0.458
        thinned_from=63
        keys [[0,-0.458],[8,-0.4553],[15,-0.4475],[22,-0.4007],[30,-0.4318],[45,-0.4301],[51,-0.5],[56,-0.4301],[66,-0.4301],[70,-0.4352],[101,-0.435],[107,-0.4338],[111,-0.4301],[148,-0.4301],[153,-0.4475],[160,-0.454],[172,-0.4361],[187,-0.4301],[194,-0.4319],[199,-0.4371],[204,-0.458],[215,-0.458],[223,-0.4527],[233,-0.4371],[241,-0.4319],[246,-0.4301],[251,-0.4652],[257,-0.5705],[262,-0.7799],[267,-0.9917],[271,-1.2171],[275,-1.4398],[279,-1.6108],[283,-1.6877],[287,-1.7069],[292,-1.7117],[337,-1.7133],[345,-1.6989],[351,-1.6554],[361,-1.1983],[366,-1.104],[371,-0.8473],[375,-0.5995],[378,-0.4936],[382,-0.454],[387,-0.4361],[395,-0.4324],[409,-0.431],[439,-0.4301]]
       }
       chest_ctrl.Location.Y {
        type=MovieSceneScriptingFloatChannel
        default=0.2745
        thinned_from=63
        keys [[0,0.2745],[8,0.4776],[15,1.0869],[22,2.3839],[30,2.3139],[45,2.4408],[51,0.3],[56,2.4408],[66,2.4408],[70,2.4315],[107,2.434],[111,2.4408],[148,2.4408],[153,1.0869],[160,0.5791],[172,1.9754],[187,2.4408],[194,2.3054],[199,1.8992],[204,0.2745],[215,0.2745],[223,0.6807],[233,1.8992],[241,2.3054],[246,2.4408],[251,2.417],[257,2.3458],[267,2.0607],[271,1.2978],[275,0.5438],[279,0.1542],[283,-0.0212],[287,-0.065],[292,-0.076],[299,-0.0787],[309,-0.0794],[337,-0.0796],[345,-0.0504],[351,0.0373],[361,0.9496],[366,1.1367],[371,1.6352],[375,2.1165],[378,2.3192],[382,2.3952],[387,2.4294],[395,2.4365],[439,2.4408]]
       }
       chest_ctrl.Location.Z {
        type=MovieSceneScriptingFloatChannel
        default=0.0044
        thinned_from=63
        keys [[0,0.0044],[8,0.0011],[15,-0.009],[22,-0.0615],[30,-0.0292],[45,-0.0313],[51,0],[56,-0.0313],[66,-0.0313],[70,-0.0301],[101,-0.0301],[107,-0.0304],[111,-0.0313],[148,-0.0313],[153,-0.009],[160,-0.0006],[172,-0.0236],[187,-0.0313],[194,-0.029],[199,-0.0223],[204,0.0044],[215,0.0044],[223,-0.0023],[233,-0.0223],[241,-0.029],[246,-0.0313],[251,-0.031],[257,-0.0303],[267,-0.0275],[275,-0.0601],[279,-0.125],[283,-0.1542],[287,-0.1615],[292,-0.1633],[299,-0.1638],[337,-0.164],[345,-0.1586],[351,-0.1424],[361,-0.0183],[366,0.0015],[371,-0.0006],[375,-0.0026],[378,-0.0205],[382,-0.0272],[387,-0.0302],[395,-0.0309],[409,-0.0311],[439,-0.0313]]
       }
       chest_ctrl.Rotation.X {
        type=MovieSceneScriptingFloatChannel
        default=2.8184
        thinned_from=63
        keys [[0,2.8184],[8,5.46],[15,13.3849],[22,31.0999],[30,29.4102],[45,31.083],[51,2.8],[56,30.9956],[66,30.9573],[70,30.8766],[107,30.8877],[115,30.9989],[148,31.0084],[153,13.3896],[160,6.7826],[172,24.9472],[187,31.0021],[194,29.2406],[199,23.9561],[204,2.8184],[215,2.8184],[223,8.1042],[233,23.9617],[241,29.2475],[246,31.0094],[251,30.7048],[257,29.7909],[267,26.1353],[271,16.0045],[275,5.9914],[279,3.533],[283,2.4267],[287,2.1501],[292,2.081],[299,2.0637],[309,2.0597],[337,2.0579],[345,2.4295],[351,3.5441],[361,13.3707],[366,15.1601],[371,21.5093],[375,27.6396],[378,29.7371],[382,30.5236],[387,30.8776],[395,30.9513],[439,30.9956]]
       }
       chest_ctrl.Rotation.Y {
        type=MovieSceneScriptingFloatChannel
        default=13.4277
        thinned_from=63
        keys [[0,13.4277],[8,11.985],[15,7.6569],[22,-1.9999],[30,-1.0594],[45,-1.961],[51,13.4],[56,-1.8768],[62,-1.6242],[66,-0.1153],[70,0.5611],[75,0.7252],[83,0.8238],[94,0.8559],[101,0.8196],[107,0.5658],[111,-0.2107],[115,-2.4637],[121,-3.0269],[130,-3.2147],[138,-3.2005],[143,-3.158],[148,-2.9878],[153,7.2719],[160,11.1193],[172,-0.1639],[187,-3.9249],[194,-2.8404],[199,0.4132],[204,13.4277],[215,13.4277],[223,10.3411],[233,1.0813],[241,-2.0053],[246,-3.0342],[251,-2.5739],[257,-1.1929],[262,1.5528],[267,4.3309],[271,8.1841],[275,11.9925],[279,15.5066],[283,17.0879],[287,17.4833],[292,17.5821],[337,17.6151],[345,17.4323],[351,16.8839],[361,10.6356],[366,9.284],[371,5.0819],[375,1.0247],[378,-0.8506],[382,-1.5539],[387,-1.8704],[395,-1.9363],[409,-1.9783],[439,-1.9759]]
       }
       chest_ctrl.Rotation.Z {
        type=MovieSceneScriptingFloatChannel
        default=7.2983
        thinned_from=63
        keys [[0,7.2983],[8,7.0815],[15,6.431],[22,4.9999],[30,5.121],[45,4.9855],[51,7.3],[56,4.8285],[62,4.3577],[66,1.474],[70,0.1674],[75,-0.1565],[83,-0.4569],[94,-0.5547],[101,-0.5078],[107,-0.1796],[111,0.8247],[115,3.2546],[121,3.862],[130,4.0645],[138,4.0108],[143,3.8495],[148,3.2044],[153,5.7631],[160,6.7226],[172,6.7495],[187,6.7584],[194,6.7921],[199,6.8934],[204,7.2983],[215,7.2983],[223,6.5637],[233,4.3598],[241,3.6251],[246,3.3803],[251,3.1651],[257,2.5196],[262,1.2361],[267,-0.0625],[271,-2.6471],[275,-5.2016],[279,-8.7567],[283,-10.3565],[287,-10.7564],[292,-10.8564],[299,-10.8814],[337,-10.8897],[345,-10.6617],[351,-9.9775],[361,-3.3768],[366,-2.0888],[371,0.7398],[375,3.4709],[378,4.4526],[382,4.8208],[387,4.9864],[395,5.0209],[409,5.1089],[439,5.0416]]
       }
       chest_ctrl.Scale.X {type=MovieSceneScriptingFloatChannel, default=1, thinned_from=63, keys [[0,1],[439,1]]}
       chest_ctrl.Scale.Y {type=MovieSceneScriptingFloatChannel, default=1, thinned_from=63, keys [[0,1],[439,1]]}
       chest_ctrl.Scale.Z {type=MovieSceneScriptingFloatChannel, default=1, thinned_from=63, keys [[0,1],[439,1]]}
       chest_tan_ctrl.Location.X {
        type=MovieSceneScriptingFloatChannel
        default=-0.7418
        thinned_from=63
        keys [[0,-0.7418],[8,-0.6676],[15,-0.4453],[22,-0.0595],[30,0.0026],[45,0.0489],[51,-0.7],[56,0.0437],[62,0.0281],[66,-0.0625],[70,-0.1021],[75,-0.1116],[83,-0.114],[94,-0.1147],[101,-0.1119],[107,-0.0918],[111,-0.0303],[115,0.1648],[121,0.2136],[130,0.2298],[148,0.2297],[153,-0.3775],[160,-0.6052],[172,0.0215],[187,0.2304],[194,0.1697],[199,-0.0126],[204,-0.7418],[215,-0.7418],[223,-0.5596],[233,-0.0131],[241,0.169],[246,0.2298],[251,0.2453],[257,0.292],[262,0.3847],[267,0.4786],[275,0.36],[279,0.1702],[283,0.0848],[287,0.0635],[292,0.0582],[299,0.0568],[309,0.0565],[337,0.0564],[345,0.0586],[351,0.0654],[361,0.1056],[366,0.11],[371,0.091],[375,0.0727],[378,0.0579],[382,0.0523],[387,0.0498],[395,0.0493],[409,0.0488],[439,0.049]]
       }
       chest_tan_ctrl.Location.Y {
        type=MovieSceneScriptingFloatChannel
        default=-0.2592
        thinned_from=63
        keys [[0,-0.2592],[8,-0.1032],[15,0.3649],[22,1.0125],[30,1.3153],[45,1.4154],[51,-0.3],[56,1.4044],[62,1.4029],[66,1.3934],[70,1.3897],[107,1.3883],[111,1.3918],[115,1.3997],[143,1.4016],[148,1.3997],[153,0.3629],[160,-0.0259],[172,1.0514],[187,1.4105],[194,1.3061],[199,0.9931],[204,-0.2592],[215,-0.2592],[223,0.052],[233,0.9854],[241,1.2965],[246,1.4002],[251,1.3837],[257,1.3342],[267,1.1362],[271,0.8247],[275,0.5169],[279,0.6956],[283,0.776],[287,0.7961],[292,0.8011],[337,0.8028],[345,0.8072],[351,0.8205],[361,0.8217],[366,0.8045],[371,1.0514],[375,1.2898],[378,1.3619],[382,1.3889],[387,1.4011],[409,1.4048],[439,1.4051]]
       }
       chest_tan_ctrl.Location.Z {
        type=MovieSceneScriptingFloatChannel
        default=1.2329
        thinned_from=63
        keys [[0,1.2329],[8,1.2517],[15,1.3081],[22,2.6142],[30,1.4241],[45,1.4366],[51,1.2],[56,1.4334],[111,1.4296],[115,1.4348],[148,1.4357],[153,1.3089],[160,1.2614],[172,1.3945],[187,1.4389],[194,1.426],[199,1.3874],[204,1.2329],[215,1.2329],[223,1.2709],[233,1.3851],[241,1.4232],[246,1.4358],[251,1.4321],[257,1.421],[267,1.3766],[275,1.2395],[279,1.2289],[283,1.2241],[337,1.2225],[345,1.2264],[351,1.2383],[361,1.3074],[366,1.3147],[371,1.3687],[375,1.4209],[378,1.4289],[387,1.4332],[439,1.4336]]
       }
       chest_tan_ctrl.Rotation.X {
        type=MovieSceneScriptingFloatChannel
        default=1.3545
        thinned_from=63
        keys [[0,1.3545],[8,1.2275],[15,0.8466],[22,0],[30,0.0794],[45,0],[51,1.4],[56,0],[148,0],[153,0.8466],[160,1.164],[172,0.291],[187,0],[194,0.0847],[199,0.3386],[204,1.3545],[215,1.3545],[223,1.1006],[233,0.3386],[241,0.0847],[246,0],[439,0]]
       }
       chest_tan_ctrl.Rotation.Y {
        type=MovieSceneScriptingFloatChannel
        default=0
        thinned_from=63
        keys [[0,0],[8,0.0788],[15,0.3151],[22,1.5984],[30,0.7871],[45,0.835],[51,0],[56,0.8493],[62,0.8758],[66,1.0666],[70,1.1704],[75,1.1963],[83,1.2658],[94,1.2885],[101,1.2951],[107,1.3416],[111,1.4839],[115,2.1616],[121,2.331],[130,2.3875],[138,2.4031],[143,2.45],[148,2.6378],[153,0.9892],[160,0.3709],[172,1.2933],[187,1.6008],[194,1.5007],[199,1.2006],[204,0],[215,0],[223,0.485],[233,1.94],[241,2.425],[246,2.5866],[251,2.5603],[257,2.4812],[267,2.165],[271,1.0762],[275,0],[366,0],[371,0.2581],[375,0.5072],[378,0.7053],[382,0.7795],[387,0.8129],[395,0.8199],[409,0.799],[439,0.8241]]
       }
       chest_tan_ctrl.Rotation.Z {
        type=MovieSceneScriptingFloatChannel
        default=0
        thinned_from=63
        keys [[0,0],[62,0],[66,1.4846],[70,2.1395],[83,2.5617],[94,2.6459],[101,2.5877],[107,2.1803],[111,0.9335],[115,0.2334],[121,0.0583],[130,0],[246,0],[251,0.1345],[257,0.5381],[262,1.3406],[267,2.1525],[271,6.4326],[275,10.663],[279,12.3796],[283,13.1521],[287,13.3452],[292,13.3935],[337,13.4096],[345,13.2207],[351,12.6543],[361,6.6436],[366,5.3953],[371,2.6503],[375,0],[439,0]]
       }
       chest_tan_ctrl.Scale.X {type=MovieSceneScriptingFloatChannel, default=1, thinned_from=63, keys [[0,1],[439,1]]}
       chest_tan_ctrl.Scale.Y {type=MovieSceneScriptingFloatChannel, default=1, thinned_from=63, keys [[0,1],[439,1]]}
       chest_tan_ctrl.Scale.Z {type=MovieSceneScriptingFloatChannel, default=1, thinned_from=63, keys [[0,1],[439,1]]}
       Weight {type=MovieSceneScriptingFloatChannel, default=1}
      ]
     }
     {
      type=MovieSceneControlRigParameterSection
      channels [
       thumb_01_r_ctrl.Location.X {type=MovieSceneScriptingFloatChannel, default=0, const [0,6]}
       thumb_01_r_ctrl.Location.Y {type=MovieSceneScriptingFloatChannel, default=0, const [0,6]}
       thumb_01_r_ctrl.Location.Z {type=MovieSceneScriptingFloatChannel, default=0, const [0,6]}
       thumb_01_r_ctrl.Rotation.X {type=MovieSceneScriptingFloatChannel, default=0, const [0,6]}
       thumb_01_r_ctrl.Rotation.Y {type=MovieSceneScriptingFloatChannel, default=0, const [0,6]}
       thumb_01_r_ctrl.Rotation.Z {type=MovieSceneScriptingFloatChannel, default=0, thinned_from=6, keys [[0,0],[439,0]]}
       thumb_01_r_ctrl.Scale.X {type=MovieSceneScriptingFloatChannel, default=0, const [0,6]}
       thumb_01_r_ctrl.Scale.Y {type=MovieSceneScriptingFloatChannel, default=0, const [0,6]}
       thumb_01_r_ctrl.Scale.Z {type=MovieSceneScriptingFloatChannel, default=0, const [0,6]}
       thumb_02_r_ctrl.Location.X {type=MovieSceneScriptingFloatChannel, default=0, const [0,6]}
       thumb_02_r_ctrl.Location.Y {type=MovieSceneScriptingFloatChannel, default=0, const [0,6]}
       thumb_02_r_ctrl.Location.Z {type=MovieSceneScriptingFloatChannel, default=0, const [0,6]}
       thumb_02_r_ctrl.Rotation.X {type=MovieSceneScriptingFloatChannel, default=0, const [0,6]}
       thumb_02_r_ctrl.Rotation.Y {type=MovieSceneScriptingFloatChannel, default=0, const [0,6]}
       thumb_02_r_ctrl.Rotation.Z {type=MovieSceneScriptingFloatChannel, default=0, const [0,6]}
       thumb_02_r_ctrl.Scale.X {type=MovieSceneScriptingFloatChannel, default=0, const [0,6]}
       thumb_02_r_ctrl.Scale.Y {type=MovieSceneScriptingFloatChannel, default=0, const [0,6]}
       thumb_02_r_ctrl.Scale.Z {type=MovieSceneScriptingFloatChannel, default=0, const [0,6]}
       thumb_03_r_ctrl.Location.X {type=MovieSceneScriptingFloatChannel, default=0, const [0,6]}
       thumb_03_r_ctrl.Location.Y {type=MovieSceneScriptingFloatChannel, default=0, const [0,6]}
       thumb_03_r_ctrl.Location.Z {type=MovieSceneScriptingFloatChannel, default=0, const [0,6]}
       thumb_03_r_ctrl.Rotation.X {type=MovieSceneScriptingFloatChannel, default=0, const [0,6]}
       thumb_03_r_ctrl.Rotation.Y {type=MovieSceneScriptingFloatChannel, default=0, const [0,6]}
       thumb_03_r_ctrl.Rotation.Z {type=MovieSceneScriptingFloatChannel, default=0, const [0,6]}
       thumb_03_r_ctrl.Scale.X {type=MovieSceneScriptingFloatChannel, default=0, const [0,6]}
       thumb_03_r_ctrl.Scale.Y {type=MovieSceneScriptingFloatChannel, default=0, const [0,6]}
       thumb_03_r_ctrl.Scale.Z {type=MovieSceneScriptingFloatChannel, default=0, const [0,6]}
       index_metacarpal_r_ctrl.Location.X {type=MovieSceneScriptingFloatChannel, default=0, const [0,6]}
       index_metacarpal_r_ctrl.Location.Y {type=MovieSceneScriptingFloatChannel, default=0, const [0,6]}
       index_metacarpal_r_ctrl.Location.Z {type=MovieSceneScriptingFloatChannel, default=0, const [0,6]}
       index_metacarpal_r_ctrl.Rotation.X {type=MovieSceneScriptingFloatChannel, default=0, const [0,6]}
       index_metacarpal_r_ctrl.Rotation.Y {type=MovieSceneScriptingFloatChannel, default=0, const [0,6]}
       index_metacarpal_r_ctrl.Rotation.Z {type=MovieSceneScriptingFloatChannel, default=0, const [0,6]}
       index_metacarpal_r_ctrl.Scale.X {type=MovieSceneScriptingFloatChannel, default=0, const [0,6]}
       index_metacarpal_r_ctrl.Scale.Y {type=MovieSceneScriptingFloatChannel, default=0, const [0,6]}
       index_metacarpal_r_ctrl.Scale.Z {type=MovieSceneScriptingFloatChannel, default=0, const [0,6]}
       index_01_r_ctrl.Location.X {type=MovieSceneScriptingFloatChannel, default=0, keys [[0,0.5341],[265,0.5341],[274,0],[364,0],[370,0.5341],[439,0.5341]]}
       index_01_r_ctrl.Location.Y {type=MovieSceneScriptingFloatChannel, default=0, keys [[0,0.0561],[265,0.0561],[274,0],[364,0],[370,0.0561],[439,0.0561]]}
       index_01_r_ctrl.Location.Z {type=MovieSceneScriptingFloatChannel, default=0, const [0,6]}
       index_01_r_ctrl.Rotation.X {type=MovieSceneScriptingFloatChannel, default=0, const [0,6]}
       index_01_r_ctrl.Rotation.Y {type=MovieSceneScriptingFloatChannel, default=0, const [0,6]}
       index_01_r_ctrl.Rotation.Z {type=MovieSceneScriptingFloatChannel, default=0, const [0,6]}
       index_01_r_ctrl.Scale.X {type=MovieSceneScriptingFloatChannel, default=0, const [0,6]}
       index_01_r_ctrl.Scale.Y {type=MovieSceneScriptingFloatChannel, default=0, const [0,6]}
       index_01_r_ctrl.Scale.Z {type=MovieSceneScriptingFloatChannel, default=0, const [0,6]}
       index_02_r_ctrl.Location.X {type=MovieSceneScriptingFloatChannel, default=0, const [0,6]}
       index_02_r_ctrl.Location.Y {type=MovieSceneScriptingFloatChannel, default=0, const [0,6]}
       index_02_r_ctrl.Location.Z {type=MovieSceneScriptingFloatChannel, default=0, const [0,6]}
       index_02_r_ctrl.Rotation.X {type=MovieSceneScriptingFloatChannel, default=0, const [0,6]}
       index_02_r_ctrl.Rotation.Y {type=MovieSceneScriptingFloatChannel, default=0, const [0,6]}
       index_02_r_ctrl.Rotation.Z {type=MovieSceneScriptingFloatChannel, default=0, keys [[0,-2],[265,-2],[274,0],[364,0],[370,-2],[439,-2]]}
       index_02_r_ctrl.Scale.X {type=MovieSceneScriptingFloatChannel, default=0, const [0,6]}
       index_02_r_ctrl.Scale.Y {type=MovieSceneScriptingFloatChannel, default=0, const [0,6]}
       index_02_r_ctrl.Scale.Z {type=MovieSceneScriptingFloatChannel, default=0, const [0,6]}
       index_03_r_ctrl.Location.X {type=MovieSceneScriptingFloatChannel, default=0, const [0,6]}
       index_03_r_ctrl.Location.Y {type=MovieSceneScriptingFloatChannel, default=0, const [0,6]}
       index_03_r_ctrl.Location.Z {type=MovieSceneScriptingFloatChannel, default=0, const [0,6]}
       index_03_r_ctrl.Rotation.X {type=MovieSceneScriptingFloatChannel, default=0, const [0,6]}
       index_03_r_ctrl.Rotation.Y {type=MovieSceneScriptingFloatChannel, default=0, const [0,6]}
       index_03_r_ctrl.Rotation.Z {type=MovieSceneScriptingFloatChannel, default=0, keys [[0,-11],[265,-11],[274,0],[364,0],[370,-11],[439,-11]]}
       index_03_r_ctrl.Scale.X {type=MovieSceneScriptingFloatChannel, default=0, const [0,6]}
       index_03_r_ctrl.Scale.Y {type=MovieSceneScriptingFloatChannel, default=0, const [0,6]}
       index_03_r_ctrl.Scale.Z {type=MovieSceneScriptingFloatChannel, default=0, const [0,6]}
       middle_metacarpal_r_ctrl.Location.X {type=MovieSceneScriptingFloatChannel, default=0, const [0,6]}
       middle_metacarpal_r_ctrl.Location.Y {type=MovieSceneScriptingFloatChannel, default=0, const [0,6]}
       middle_metacarpal_r_ctrl.Location.Z {type=MovieSceneScriptingFloatChannel, default=0, const [0,6]}
       middle_metacarpal_r_ctrl.Rotation.X {type=MovieSceneScriptingFloatChannel, default=0, const [0,6]}
       middle_metacarpal_r_ctrl.Rotation.Y {type=MovieSceneScriptingFloatChannel, default=0, const [0,6]}
       middle_metacarpal_r_ctrl.Rotation.Z {type=MovieSceneScriptingFloatChannel, default=0, const [0,6]}
       middle_metacarpal_r_ctrl.Scale.X {type=MovieSceneScriptingFloatChannel, default=0, const [0,6]}
       middle_metacarpal_r_ctrl.Scale.Y {type=MovieSceneScriptingFloatChannel, default=0, const [0,6]}
       middle_metacarpal_r_ctrl.Scale.Z {type=MovieSceneScriptingFloatChannel, default=0, const [0,6]}
       middle_01_r_ctrl.Location.X {type=MovieSceneScriptingFloatChannel, default=0, const [0,6]}
       middle_01_r_ctrl.Location.Y {type=MovieSceneScriptingFloatChannel, default=0, const [0,6]}
       middle_01_r_ctrl.Location.Z {type=MovieSceneScriptingFloatChannel, default=0, const [0,6]}
       middle_01_r_ctrl.Rotation.X {type=MovieSceneScriptingFloatChannel, default=0, keys [[0,-12],[265,-12],[274,0],[364,0],[370,-12],[439,-12]]}
       middle_01_r_ctrl.Rotation.Y {type=MovieSceneScriptingFloatChannel, default=0, keys [[0,0.5],[265,0.5],[274,0],[364,0],[370,0.5],[439,0.5]]}
       middle_01_r_ctrl.Rotation.Z {type=MovieSceneScriptingFloatChannel, default=0, const [0,6]}
       middle_01_r_ctrl.Scale.X {type=MovieSceneScriptingFloatChannel, default=0, const [0,6]}
       middle_01_r_ctrl.Scale.Y {type=MovieSceneScriptingFloatChannel, default=0, const [0,6]}
       middle_01_r_ctrl.Scale.Z {type=MovieSceneScriptingFloatChannel, default=0, const [0,6]}
       middle_02_r_ctrl.Location.X {type=MovieSceneScriptingFloatChannel, default=0, const [0,6]}
       middle_02_r_ctrl.Location.Y {type=MovieSceneScriptingFloatChannel, default=0, const [0,6]}
       middle_02_r_ctrl.Location.Z {type=MovieSceneScriptingFloatChannel, default=0, const [0,6]}
       middle_02_r_ctrl.Rotation.X {type=MovieSceneScriptingFloatChannel, default=0, const [0,6]}
       middle_02_r_ctrl.Rotation.Y {type=MovieSceneScriptingFloatChannel, default=0, const [0,6]}
       middle_02_r_ctrl.Rotation.Z {type=MovieSceneScriptingFloatChannel, default=0, keys [[0,16.5],[265,16.5],[274,0],[364,0],[370,16.5],[439,16.5]]}
       middle_02_r_ctrl.Scale.X {type=MovieSceneScriptingFloatChannel, default=0, const [0,6]}
       middle_02_r_ctrl.Scale.Y {type=MovieSceneScriptingFloatChannel, default=0, const [0,6]}
       middle_02_r_ctrl.Scale.Z {type=MovieSceneScriptingFloatChannel, default=0, const [0,6]}
       middle_03_r_ctrl.Location.X {type=MovieSceneScriptingFloatChannel, default=0, const [0,6]}
       middle_03_r_ctrl.Location.Y {type=MovieSceneScriptingFloatChannel, default=0, const [0,6]}
       middle_03_r_ctrl.Location.Z {type=MovieSceneScriptingFloatChannel, default=0, const [0,6]}
       middle_03_r_ctrl.Rotation.X {type=MovieSceneScriptingFloatChannel, default=0, const [0,6]}
       middle_03_r_ctrl.Rotation.Y {type=MovieSceneScriptingFloatChannel, default=0, const [0,6]}
       middle_03_r_ctrl.Rotation.Z {type=MovieSceneScriptingFloatChannel, default=0, keys [[0,45],[265,45],[274,0],[364,0],[370,45],[439,45]]}
       middle_03_r_ctrl.Scale.X {type=MovieSceneScriptingFloatChannel, default=0, const [0,6]}
       middle_03_r_ctrl.Scale.Y {type=MovieSceneScriptingFloatChannel, default=0, const [0,6]}
       middle_03_r_ctrl.Scale.Z {type=MovieSceneScriptingFloatChannel, default=0, const [0,6]}
       ring_metacarpal_r_ctrl.Location.X {type=MovieSceneScriptingFloatChannel, default=0, const [0,6]}
       ring_metacarpal_r_ctrl.Location.Y {type=MovieSceneScriptingFloatChannel, default=0, const [0,6]}
       ring_metacarpal_r_ctrl.Location.Z {type=MovieSceneScriptingFloatChannel, default=0, const [0,6]}
       ring_metacarpal_r_ctrl.Rotation.X {type=MovieSceneScriptingFloatChannel, default=0, const [0,6]}
       ring_metacarpal_r_ctrl.Rotation.Y {type=MovieSceneScriptingFloatChannel, default=0, const [0,6]}
       ring_metacarpal_r_ctrl.Rotation.Z {type=MovieSceneScriptingFloatChannel, default=0, const [0,6]}
       ring_metacarpal_r_ctrl.Scale.X {type=MovieSceneScriptingFloatChannel, default=0, const [0,6]}
       ring_metacarpal_r_ctrl.Scale.Y {type=MovieSceneScriptingFloatChannel, default=0, const [0,6]}
       ring_metacarpal_r_ctrl.Scale.Z {type=MovieSceneScriptingFloatChannel, default=0, const [0,6]}
       ring_01_r_ctrl.Location.X {type=MovieSceneScriptingFloatChannel, default=0, const [0,6]}
       ring_01_r_ctrl.Location.Y {type=MovieSceneScriptingFloatChannel, default=0, const [0,6]}
       ring_01_r_ctrl.Location.Z {type=MovieSceneScriptingFloatChannel, default=0, const [0,6]}
       ring_01_r_ctrl.Rotation.X {type=MovieSceneScriptingFloatChannel, default=0, keys [[0,-10.1011],[265,-10.1011],[274,0],[364,0],[370,-10.1011],[439,-10.1011]]}
       ring_01_r_ctrl.Rotation.Y {type=MovieSceneScriptingFloatChannel, default=0, keys [[0,-8.3238],[265,-8.3238],[274,0],[364,0],[370,-8.3238],[439,-8.3238]]}
       ring_01_r_ctrl.Rotation.Z {type=MovieSceneScriptingFloatChannel, default=0, keys [[0,1.5],[265,1.5],[274,0],[364,0],[370,1.5],[439,1.5]]}
       ring_01_r_ctrl.Scale.X {type=MovieSceneScriptingFloatChannel, default=0, const [0,6]}
       ring_01_r_ctrl.Scale.Y {type=MovieSceneScriptingFloatChannel, default=0, const [0,6]}
       ring_01_r_ctrl.Scale.Z {type=MovieSceneScriptingFloatChannel, default=0, const [0,6]}
       ring_02_r_ctrl.Location.X {type=MovieSceneScriptingFloatChannel, default=0, const [0,6]}
       ring_02_r_ctrl.Location.Y {type=MovieSceneScriptingFloatChannel, default=0, const [0,6]}
       ring_02_r_ctrl.Location.Z {type=MovieSceneScriptingFloatChannel, default=0, const [0,6]}
       ring_02_r_ctrl.Rotation.X {type=MovieSceneScriptingFloatChannel, default=0, const [0,6]}
       ring_02_r_ctrl.Rotation.Y {type=MovieSceneScriptingFloatChannel, default=0, const [0,6]}
       ring_02_r_ctrl.Rotation.Z {type=MovieSceneScriptingFloatChannel, default=0, keys [[0,26.5],[265,26.5],[274,0],[364,0],[370,26.5],[439,26.5]]}
       ring_02_r_ctrl.Scale.X {type=MovieSceneScriptingFloatChannel, default=0, const [0,6]}
       ring_02_r_ctrl.Scale.Y {type=MovieSceneScriptingFloatChannel, default=0, const [0,6]}
       ring_02_r_ctrl.Scale.Z {type=MovieSceneScriptingFloatChannel, default=0, const [0,6]}
       ring_03_r_ctrl.Location.X {type=MovieSceneScriptingFloatChannel, default=0, const [0,6]}
       ring_03_r_ctrl.Location.Y {type=MovieSceneScriptingFloatChannel, default=0, const [0,6]}
       ring_03_r_ctrl.Location.Z {type=MovieSceneScriptingFloatChannel, default=0, const [0,6]}
       ring_03_r_ctrl.Rotation.X {type=MovieSceneScriptingFloatChannel, default=0, const [0,6]}
       ring_03_r_ctrl.Rotation.Y {type=MovieSceneScriptingFloatChannel, default=0, const [0,6]}
       ring_03_r_ctrl.Rotation.Z {type=MovieSceneScriptingFloatChannel, default=0, keys [[0,63],[265,63],[274,0],[364,0],[370,63],[439,63]]}
       ring_03_r_ctrl.Scale.X {type=MovieSceneScriptingFloatChannel, default=0, const [0,6]}
       ring_03_r_ctrl.Scale.Y {type=MovieSceneScriptingFloatChannel, default=0, const [0,6]}
       ring_03_r_ctrl.Scale.Z {type=MovieSceneScriptingFloatChannel, default=0, const [0,6]}
       pinky_metacarpal_r_ctrl.Location.X {type=MovieSceneScriptingFloatChannel, default=0, const [0,6]}
       pinky_metacarpal_r_ctrl.Location.Y {type=MovieSceneScriptingFloatChannel, default=0, const [0,6]}
       pinky_metacarpal_r_ctrl.Location.Z {type=MovieSceneScriptingFloatChannel, default=0, const [0,6]}
       pinky_metacarpal_r_ctrl.Rotation.X {type=MovieSceneScriptingFloatChannel, default=0, const [0,6]}
       pinky_metacarpal_r_ctrl.Rotation.Y {type=MovieSceneScriptingFloatChannel, default=0, const [0,6]}
       pinky_metacarpal_r_ctrl.Rotation.Z {type=MovieSceneScriptingFloatChannel, default=0, const [0,6]}
       pinky_metacarpal_r_ctrl.Scale.X {type=MovieSceneScriptingFloatChannel, default=0, const [0,6]}
       pinky_metacarpal_r_ctrl.Scale.Y {type=MovieSceneScriptingFloatChannel, default=0, const [0,6]}
       pinky_metacarpal_r_ctrl.Scale.Z {type=MovieSceneScriptingFloatChannel, default=0, const [0,6]}
       pinky_01_r_ctrl.Location.X {type=MovieSceneScriptingFloatChannel, default=0, const [0,6]}
       pinky_01_r_ctrl.Location.Y {type=MovieSceneScriptingFloatChannel, default=0, const [0,6]}
       pinky_01_r_ctrl.Location.Z {type=MovieSceneScriptingFloatChannel, default=0, const [0,6]}
       pinky_01_r_ctrl.Rotation.X {type=MovieSceneScriptingFloatChannel, default=0, keys [[0,-9.4753],[265,-9.4753],[274,0],[364,0],[370,-9.4753],[439,-9.4753]]}
       pinky_01_r_ctrl.Rotation.Y {type=MovieSceneScriptingFloatChannel, default=0, const [0,6]}
       pinky_01_r_ctrl.Rotation.Z {type=MovieSceneScriptingFloatChannel, default=0, const [0,6]}
       pinky_01_r_ctrl.Scale.X {type=MovieSceneScriptingFloatChannel, default=0, const [0,6]}
       pinky_01_r_ctrl.Scale.Y {type=MovieSceneScriptingFloatChannel, default=0, const [0,6]}
       pinky_01_r_ctrl.Scale.Z {type=MovieSceneScriptingFloatChannel, default=0, const [0,6]}
       pinky_02_r_ctrl.Location.X {type=MovieSceneScriptingFloatChannel, default=0, const [0,6]}
       pinky_02_r_ctrl.Location.Y {type=MovieSceneScriptingFloatChannel, default=0, const [0,6]}
       pinky_02_r_ctrl.Location.Z {type=MovieSceneScriptingFloatChannel, default=0, const [0,6]}
       pinky_02_r_ctrl.Rotation.X {type=MovieSceneScriptingFloatChannel, default=0, const [0,6]}
       pinky_02_r_ctrl.Rotation.Y {type=MovieSceneScriptingFloatChannel, default=0, thinned_from=6, keys [[0,0],[439,0]]}
       pinky_02_r_ctrl.Rotation.Z {type=MovieSceneScriptingFloatChannel, default=0, keys [[0,48.5],[265,48.5],[274,0],[364,0],[370,48.5],[439,48.5]]}
       pinky_02_r_ctrl.Scale.X {type=MovieSceneScriptingFloatChannel, default=0, const [0,6]}
       pinky_02_r_ctrl.Scale.Y {type=MovieSceneScriptingFloatChannel, default=0, const [0,6]}
       pinky_02_r_ctrl.Scale.Z {type=MovieSceneScriptingFloatChannel, default=0, const [0,6]}
       pinky_03_r_ctrl.Location.X {type=MovieSceneScriptingFloatChannel, default=0, const [0,6]}
       pinky_03_r_ctrl.Location.Y {type=MovieSceneScriptingFloatChannel, default=0, const [0,6]}
       pinky_03_r_ctrl.Location.Z {type=MovieSceneScriptingFloatChannel, default=0, const [0,6]}
       pinky_03_r_ctrl.Rotation.X {type=MovieSceneScriptingFloatChannel, default=0, const [0,6]}
       pinky_03_r_ctrl.Rotation.Y {type=MovieSceneScriptingFloatChannel, default=0, const [0,6]}
       pinky_03_r_ctrl.Rotation.Z {type=MovieSceneScriptingFloatChannel, default=0, keys [[0,59.5],[265,59.5],[274,0],[364,0],[370,59.5],[439,59.5]]}
       pinky_03_r_ctrl.Scale.X {type=MovieSceneScriptingFloatChannel, default=0, const [0,6]}
       pinky_03_r_ctrl.Scale.Y {type=MovieSceneScriptingFloatChannel, default=0, const [0,6]}
       pinky_03_r_ctrl.Scale.Z {type=MovieSceneScriptingFloatChannel, default=0, const [0,6]}
       Weight {type=MovieSceneScriptingFloatChannel, default=1}
      ]
     }
    ]
   }
   {
    type=MovieSceneControlRigParameterTrack
    display_name=Fortnite_Legacy_Face_CtrlRig
    sections [
     {
      type=MovieSceneControlRigParameterSection
      channels [
       Flipbook {type=MovieSceneScriptingBoolChannel, default=false, const [false,5]}
       Face_Ctrl_Master.Location.X {type=MovieSceneScriptingFloatChannel, default=-36, const [-36,5]}
       Face_Ctrl_Master.Location.Y {type=MovieSceneScriptingFloatChannel, default=0, const [0,5]}
       Face_Ctrl_Master.Location.Z {type=MovieSceneScriptingFloatChannel, default=0, const [0,5]}
       Face_Ctrl_Master.Rotation.X {type=MovieSceneScriptingFloatChannel, default=0, const [0,5]}
       Face_Ctrl_Master.Rotation.Y {type=MovieSceneScriptingFloatChannel, default=0, const [0,5]}
       Face_Ctrl_Master.Rotation.Z {type=MovieSceneScriptingFloatChannel, default=0, const [0,5]}
       Face_Ctrl_Master.Scale.X {type=MovieSceneScriptingFloatChannel, default=1, const [1,5]}
       Face_Ctrl_Master.Scale.Y {type=MovieSceneScriptingFloatChannel, default=1, const [1,5]}
       Face_Ctrl_Master.Scale.Z {type=MovieSceneScriptingFloatChannel, default=1, const [1,5]}
       gui_follow_head {type=MovieSceneScriptingFloatChannel, default=1, const [1,5]}
       Weight {type=MovieSceneScriptingFloatChannel, default=1}
      ]
     }
     {type=MovieSceneControlRigParameterSection, channels [Weight {type=MovieSceneScriptingFloatChannel, default=1}]}
    ]
   }
  ]
 }
 f4 {
  id=#2
  possessed_class=FortPlayerMannequin
  tracks [
   {
    type=MovieSceneSpawnTrack
    sections [{type=MovieSceneSpawnSection, channels [None {type=MovieSceneScriptingBoolChannel, default=true, const [true,1]}]}]
   }
   {
    type=MovieScene3DTransformTrack
    sections [{
      type=MovieScene3DTransformSection
      channels [
       Location.X {type=MovieSceneScriptingDoubleChannel, default=-72.425}
       Location.Y {type=MovieSceneScriptingDoubleChannel, default=0}
       Location.Z {type=MovieSceneScriptingDoubleChannel, default=0}
       Rotation.X {type=MovieSceneScriptingDoubleChannel, default=0}
       Rotation.Y {type=MovieSceneScriptingDoubleChannel, default=0}
       Rotation.Z {type=MovieSceneScriptingDoubleChannel, default=90}
       Scale.X {type=MovieSceneScriptingDoubleChannel, default=1}
       Scale.Y {type=MovieSceneScriptingDoubleChannel, default=1}
       Scale.Z {type=MovieSceneScriptingDoubleChannel, default=1}
      ]
     }]
   }
   {type=MovieSceneSkeletalAnimationTrack, sections [{type=MovieSceneSkeletalAnimationSection, end_frame=440, start_frame=0}]}
  ]
 }
 f2 {
  id=#3
  possessed_class=FortPlayerMannequin
  tracks [
   {type=MovieSceneSpawnTrack, sections [{type=MovieSceneSpawnSection, channels [None {type=MovieSceneScriptingBoolChannel, default=true}]}]}
   {
    type=MovieScene3DTransformTrack
    sections [{
      type=MovieScene3DTransformSection
      channels [
       Location.X {type=MovieSceneScriptingDoubleChannel, default=88.6911}
       Location.Y {type=MovieSceneScriptingDoubleChannel, default=0}
       Location.Z {type=MovieSceneScriptingDoubleChannel, default=0}
       Rotation.X {type=MovieSceneScriptingDoubleChannel, default=0}
       Rotation.Y {type=MovieSceneScriptingDoubleChannel, default=0}
       Rotation.Z {type=MovieSceneScriptingDoubleChannel, default=-90}
       Scale.X {type=MovieSceneScriptingDoubleChannel, default=1}
       Scale.Y {type=MovieSceneScriptingDoubleChannel, default=1}
       Scale.Z {type=MovieSceneScriptingDoubleChannel, default=1}
      ]
     }]
   }
   {type=MovieSceneSkeletalAnimationTrack, sections [{type=MovieSceneSkeletalAnimationSection, end_frame=440, start_frame=0}]}
  ]
 }
 f3 {
  id=#4
  possessed_class=FortPlayerMannequin
  tracks [
   {
    type=MovieSceneSpawnTrack
    sections [{type=MovieSceneSpawnSection, channels [None {type=MovieSceneScriptingBoolChannel, default=true, const [true,1]}]}]
   }
   {
    type=MovieScene3DTransformTrack
    sections [{
      type=MovieScene3DTransformSection
      channels [
       Location.X {type=MovieSceneScriptingDoubleChannel, default=-86.9044, const [183.5305,1]}
       Location.Y {type=MovieSceneScriptingDoubleChannel, default=0, const [0,1]}
       Location.Z {type=MovieSceneScriptingDoubleChannel, default=0, const [0,1]}
       Rotation.X {type=MovieSceneScriptingDoubleChannel, default=0, const [0,1]}
       Rotation.Y {type=MovieSceneScriptingDoubleChannel, default=0, const [0,1]}
       Rotation.Z {type=MovieSceneScriptingDoubleChannel, default=90, const [180,1]}
       Scale.X {type=MovieSceneScriptingDoubleChannel, default=1, const [1,1]}
       Scale.Y {type=MovieSceneScriptingDoubleChannel, default=1, const [1,1]}
       Scale.Z {type=MovieSceneScriptingDoubleChannel, default=1, const [1,1]}
      ]
     }]
   }
   {type=MovieSceneSkeletalAnimationTrack, sections [{type=MovieSceneSkeletalAnimationSection, end_frame=440, start_frame=0}]}
  ]
 }
]
display_rate {denominator=1, numerator=30}
tick_resolution {denominator=1, numerator=24000}
tracks []
